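# Level Sequence that animates a PistolCanary boss sequence with left, front and right tentacles, plus a cut tentacle.
# Staggered visibility keys reveal the front, right, then left tentacle; the cut mesh appears just after the front one hides, then moves downward.
# Overlapping skeletal animations combine with weighted Control Rig layers and stacked transform sections for placement.
# Two subsequences play later, and an extra creature mesh actor carries an empty animation track.
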LevelSequence /PistolCanary/Anim/TripleTentacle/PistolCanary_spa1000_anm1
bindings [
 TripleTentacleLeft {
  id=#1
  possessed_class=PistolCanary_TentacleBoss_C
  tracks [
   {
    type=MovieSceneVisibilityTrack
    sections [{
      type=MovieSceneVisibilitySection
      channels [None {type=MovieSceneScriptingBoolChannel, default=false, keys [[2110.5,false],[2164,true],[3783,false]]}]
     }]
   }
   {
    type=MovieSceneSkeletalAnimationTrack
    sections [
     {type=MovieSceneSkeletalAnimationSection, end_frame=3144, start_frame=1797}
     {type=MovieSceneSkeletalAnimationSection, end_frame=2506, start_frame=1997}
     {type=MovieSceneSkeletalAnimationSection, end_frame=2615, start_frame=2444}
     {type=MovieSceneSkeletalAnimationSection, end_frame=2761, start_frame=2588}
     {type=MovieSceneSkeletalAnimationSection, end_frame=3067, start_frame=2716}
     {type=MovieSceneSkeletalAnimationSection, end_frame=3155, start_frame=3022}
     {type=MovieSceneSkeletalAnimationSection, end_frame=3320, start_frame=3094}
     {type=MovieSceneSkeletalAnimationSection, end_frame=3429, start_frame=3320}
     {type=MovieSceneSkeletalAnimationSection, end_frame=3617, start_frame=3376}
     {type=MovieSceneSkeletalAnimationSection, end_frame=3719, start_frame=3551}
     {type=MovieSceneSkeletalAnimationSection, end_frame=3780, start_frame=3660}
    ]
   }
   {type=MovieSceneSpawnTrack, sections [{type=MovieSceneSpawnSection, channels [None {type=MovieSceneScriptingBoolChannel, default=true}]}]}
   {
    type=MovieSceneControlRigParameterTrack
    display_name=Tentacle_PistolCanary_Creature_CtrlRig
    sections [
     {
      type=MovieSceneControlRigParameterSection
      channels [
       control rig options.Location.X {type=MovieSceneScriptingFloatChannel, default=0, const [0,6]}
       control rig options.Location.Y {type=MovieSceneScriptingFloatChannel, default=0, const [0,6]}
       control rig options.Location.Z {type=MovieSceneScriptingFloatChannel, default=0, const [0,6]}
       control rig options.Rotation.X {type=MovieSceneScriptingFloatChannel, default=0, const [0,6]}
       control rig options.Rotation.Y {type=MovieSceneScriptingFloatChannel, default=0, const [0,6]}
       control rig options.Rotation.Z {type=MovieSceneScriptingFloatChannel, default=0, const [0,6]}
       control rig options.Scale.X {type=MovieSceneScriptingFloatChannel, default=1, const [1,6]}
       control rig options.Scale.Y {type=MovieSceneScriptingFloatChannel, default=1, const [1,6]}
       control rig options.Scale.Z {type=MovieSceneScriptingFloatChannel, default=1, const [1,6]}
      ]
     }
     {
      type=MovieSceneControlRigParameterSection
      channels [
       control rig options.Location.X {type=MovieSceneScriptingFloatChannel, default=0}
       control rig options.Location.Y {type=MovieSceneScriptingFloatChannel, default=0}
       control rig options.Location.Z {type=MovieSceneScriptingFloatChannel, default=0}
       control rig options.Rotation.X {type=MovieSceneScriptingFloatChannel, default=0}
       control rig options.Rotation.Y {type=MovieSceneScriptingFloatChannel, default=0}
       control rig options.Rotation.Z {type=MovieSceneScriptingFloatChannel, default=0}
       control rig options.Scale.X {type=MovieSceneScriptingFloatChannel, default=0}
       control rig options.Scale.Y {type=MovieSceneScriptingFloatChannel, default=0}
       control rig options.Scale.Z {type=MovieSceneScriptingFloatChannel, default=0}
       Weight {type=MovieSceneScriptingFloatChannel, default=0.35, keys [[3055,0.35],[3057,0]]}
      ]
     }
     {type=MovieSceneControlRigParameterSection, channels [Weight {type=MovieSceneScriptingFloatChannel, default=1, const [1,2]}]}
    ]
   }
   {
    type=MovieScene3DTransformTrack
    sections [
     {
      type=MovieScene3DTransformSection
      channels [
       Weight {type=MovieSceneScriptingFloatChannel, default=1}
       Scale.X {type=MovieSceneScriptingDoubleChannel, default=1}
       Scale.Y {type=MovieSceneScriptingDoubleChannel, default=1}
       Scale.Z {type=MovieSceneScriptingDoubleChannel, default=1}
      ]
     }
     {
      type=MovieScene3DTransformSection
      channels [
       Location.X {type=MovieSceneScriptingDoubleChannel, default=14436}
       Location.Y {type=MovieSceneScriptingDoubleChannel, default=76088}
       Location.Z {type=MovieSceneScriptingDoubleChannel, default=3072}
       Rotation.X {type=MovieSceneScriptingDoubleChannel, default=0}
       Rotation.Y {type=MovieSceneScriptingDoubleChannel, default=0}
       Rotation.Z {type=MovieSceneScriptingDoubleChannel, default=0}
       Scale.X {type=MovieSceneScriptingDoubleChannel, default=0}
       Scale.Y {type=MovieSceneScriptingDoubleChannel, default=0}
       Scale.Z {type=MovieSceneScriptingDoubleChannel, default=0}
      ]
     }
     {
      type=MovieScene3DTransformSection
      channels [
       Weight {type=MovieSceneScriptingFloatChannel, default=1}
       Location.X {type=MovieSceneScriptingDoubleChannel, default=-1036.5809}
       Location.Y {type=MovieSceneScriptingDoubleChannel, default=-7497.0932}
       Location.Z {type=MovieSceneScriptingDoubleChannel, default=1941.8927}
       Rotation.X {type=MovieSceneScriptingDoubleChannel, default=0}
       Rotation.Y {type=MovieSceneScriptingDoubleChannel, default=0}
       Rotation.Z {type=MovieSceneScriptingDoubleChannel, default=47.4855}
       Scale.X {type=MovieSceneScriptingDoubleChannel, default=0}
       Scale.Y {type=MovieSceneScriptingDoubleChannel, default=0}
       Scale.Z {type=MovieSceneScriptingDoubleChannel, default=0}
      ]
     }
    ]
   }
  ]
 }
 TripleTentacleFront {
  id=#2
  possessed_class=PistolCanary_TentacleBoss_C
  tracks [
   {
    type=MovieSceneVisibilityTrack
    sections [{
      type=MovieSceneVisibilitySection
      channels [None {type=MovieSceneScriptingBoolChannel, default=false, keys [[1830.9987,false],[1884.4987,true],[3089,false]]}]
     }]
   }
   {type=MovieSceneSpawnTrack, sections [{type=MovieSceneSpawnSection, channels [None {type=MovieSceneScriptingBoolChannel, default=true}]}]}
   {
    type=MovieSceneControlRigParameterTrack
    display_name=Tentacle_PistolCanary_Creature_CtrlRig
    sections [
     {
      type=MovieSceneControlRigParameterSection
      channels [
       control rig options.Location.X {type=MovieSceneScriptingFloatChannel, default=0, const [0,3]}
       control rig options.Location.Y {type=MovieSceneScriptingFloatChannel, default=0, const [0,3]}
       control rig options.Location.Z {type=MovieSceneScriptingFloatChannel, default=0, const [0,3]}
       control rig options.Rotation.X {type=MovieSceneScriptingFloatChannel, default=0, const [0,3]}
       control rig options.Rotation.Y {type=MovieSceneScriptingFloatChannel, default=0, const [0,3]}
       control rig options.Rotation.Z {type=MovieSceneScriptingFloatChannel, default=0, const [0,3]}
       control rig options.Scale.X {type=MovieSceneScriptingFloatChannel, default=1, const [1,3]}
       control rig options.Scale.Y {type=MovieSceneScriptingFloatChannel, default=1, const [1,3]}
       control rig options.Scale.Z {type=MovieSceneScriptingFloatChannel, default=1, const [1,3]}
       Weight {type=MovieSceneScriptingFloatChannel, default=1, const [1,3]}
      ]
     }
     {
      type=MovieSceneControlRigParameterSection
      channels [
       control rig options.Location.X {type=MovieSceneScriptingFloatChannel, default=0}
       control rig options.Location.Y {type=MovieSceneScriptingFloatChannel, default=0}
       control rig options.Location.Z {type=MovieSceneScriptingFloatChannel, default=0}
       control rig options.Rotation.X {type=MovieSceneScriptingFloatChannel, default=0}
       control rig options.Rotation.Y {type=MovieSceneScriptingFloatChannel, default=0}
       control rig options.Rotation.Z {type=MovieSceneScriptingFloatChannel, default=0}
       control rig options.Scale.X {type=MovieSceneScriptingFloatChannel, default=0}
       control rig options.Scale.Y {type=MovieSceneScriptingFloatChannel, default=0}
       control rig options.Scale.Z {type=MovieSceneScriptingFloatChannel, default=0}
       Weight {type=MovieSceneScriptingFloatChannel, default=0.3, keys [[2784,0.3],[3001,0.2],[3003,0]]}
      ]
     }
     {type=MovieSceneControlRigParameterSection, channels [Weight {type=MovieSceneScriptingFloatChannel, default=1, const [1,2]}]}
    ]
   }
   {
    type=MovieScene3DTransformTrack
    sections [
     {
      type=MovieScene3DTransformSection
      channels [
       Weight {type=MovieSceneScriptingFloatChannel, default=1}
       Rotation.X {type=MovieSceneScriptingDoubleChannel, default=0, const [0,1]}
       Rotation.Y {type=MovieSceneScriptingDoubleChannel, default=0, const [0,1]}
       Rotation.Z {type=MovieSceneScriptingDoubleChannel, default=-90, const [0,1]}
       Scale.X {type=MovieSceneScriptingDoubleChannel, default=1}
       Scale.Y {type=MovieSceneScriptingDoubleChannel, default=1}
       Scale.Z {type=MovieSceneScriptingDoubleChannel, default=1}
      ]
     }
     {
      type=MovieScene3DTransformSection
      channels [
       Location.X {type=MovieSceneScriptingDoubleChannel, default=0, const [14436,1]}
       Location.Y {type=MovieSceneScriptingDoubleChannel, default=0, const [76088,1]}
       Location.Z {type=MovieSceneScriptingDoubleChannel, default=0, const [3072,1]}
       Rotation.X {type=MovieSceneScriptingDoubleChannel, default=0, const [0,1]}
       Rotation.Y {type=MovieSceneScriptingDoubleChannel, default=0, const [0,1]}
       Rotation.Z {type=MovieSceneScriptingDoubleChannel, default=0, const [0,1]}
       Scale.X {type=MovieSceneScriptingDoubleChannel, default=0, const [0,1]}
       Scale.Y {type=MovieSceneScriptingDoubleChannel, default=0, const [0,1]}
       Scale.Z {type=MovieSceneScriptingDoubleChannel, default=0, const [0,1]}
      ]
     }
     {type=MovieScene3DTransformSection}
     {
      type=MovieScene3DTransformSection
      channels [
       Weight {type=MovieSceneScriptingFloatChannel, default=1, const [1,2]}
       Location.X {type=MovieSceneScriptingDoubleChannel, default=28.0001, const [28.0001,1]}
       Location.Y {type=MovieSceneScriptingDoubleChannel, default=-5598.6317, const [-5598.6318,1]}
       Location.Z {type=MovieSceneScriptingDoubleChannel, default=986.8001, const [986.8001,1]}
       Rotation.X {type=MovieSceneScriptingDoubleChannel, default=0, const [0,1]}
       Rotation.Y {type=MovieSceneScriptingDoubleChannel, default=0, const [0,1]}
       Rotation.Z {type=MovieSceneScriptingDoubleChannel, default=0, const [0,1]}
       Scale.X {type=MovieSceneScriptingDoubleChannel, default=0, const [0,1]}
       Scale.Y {type=MovieSceneScriptingDoubleChannel, default=0, const [0,1]}
       Scale.Z {type=MovieSceneScriptingDoubleChannel, default=0, const [0,1]}
      ]
     }
    ]
   }
  ]
 }
 TripleTentacleRight {
  id=#3
  possessed_class=PistolCanary_TentacleBoss_C
  tracks [
   {
    type=MovieSceneVisibilityTrack
    sections [{
      type=MovieSceneVisibilitySection
      channels [None {type=MovieSceneScriptingBoolChannel, default=false, keys [[1944,false],[1997.5,true],[3920.5,false]]}]
     }]
   }
   {
    type=MovieSceneSkeletalAnimationTrack
    sections [
     {type=MovieSceneSkeletalAnimationSection, end_frame=3137, start_frame=1789}
     {type=MovieSceneSkeletalAnimationSection, end_frame=2498, start_frame=2006}
     {type=MovieSceneSkeletalAnimationSection, end_frame=2607, start_frame=2436}
     {type=MovieSceneSkeletalAnimationSection, end_frame=2753, start_frame=2580}
     {type=MovieSceneSkeletalAnimationSection, end_frame=3060, start_frame=2708}
     {type=MovieSceneSkeletalAnimationSection, end_frame=3123, start_frame=2990}
     {type=MovieSceneSkeletalAnimationSection, end_frame=3317, start_frame=3086}
     {type=MovieSceneSkeletalAnimationSection, end_frame=3850, start_frame=3312}
     {type=MovieSceneSkeletalAnimationSection, end_frame=3543, start_frame=3434}
     {type=MovieSceneSkeletalAnimationSection, end_frame=3933, start_frame=3824}
    ]
   }
   {type=MovieSceneSpawnTrack, sections [{type=MovieSceneSpawnSection, channels [None {type=MovieSceneScriptingBoolChannel, default=true}]}]}
   {
    type=MovieSceneControlRigParameterTrack
    display_name=Tentacle_PistolCanary_Creature_CtrlRig
    sections [
     {
      type=MovieSceneControlRigParameterSection
      channels [
       control rig options.Location.X {type=MovieSceneScriptingFloatChannel, default=0, const [0,7]}
       control rig options.Location.Y {type=MovieSceneScriptingFloatChannel, default=0, const [0,7]}
       control rig options.Location.Z {type=MovieSceneScriptingFloatChannel, default=0, const [0,7]}
       control rig options.Rotation.X {type=MovieSceneScriptingFloatChannel, default=0, const [0,7]}
       control rig options.Rotation.Y {type=MovieSceneScriptingFloatChannel, default=0, const [0,7]}
       control rig options.Rotation.Z {type=MovieSceneScriptingFloatChannel, default=0, const [0,7]}
       control rig options.Scale.X {type=MovieSceneScriptingFloatChannel, default=1, const [1,7]}
       control rig options.Scale.Y {type=MovieSceneScriptingFloatChannel, default=1, const [1,7]}
       control rig options.Scale.Z {type=MovieSceneScriptingFloatChannel, default=1, const [1,7]}
      ]
     }
     {
      type=MovieSceneControlRigParameterSection
      channels [
       control rig options.Location.X {type=MovieSceneScriptingFloatChannel, default=0}
       control rig options.Location.Y {type=MovieSceneScriptingFloatChannel, default=0}
       control rig options.Location.Z {type=MovieSceneScriptingFloatChannel, default=0}
       control rig options.Rotation.X {type=MovieSceneScriptingFloatChannel, default=0}
       control rig options.Rotation.Y {type=MovieSceneScriptingFloatChannel, default=0}
       control rig options.Rotation.Z {type=MovieSceneScriptingFloatChannel, default=0}
       control rig options.Scale.X {type=MovieSceneScriptingFloatChannel, default=0}
       control rig options.Scale.Y {type=MovieSceneScriptingFloatChannel, default=0}
       control rig options.Scale.Z {type=MovieSceneScriptingFloatChannel, default=0}
       Weight {type=MovieSceneScriptingFloatChannel, default=0.35, keys [[3058,0.35],[3060,0]]}
      ]
     }
     {type=MovieSceneControlRigParameterSection, channels [Weight {type=MovieSceneScriptingFloatChannel, default=1, const [1,2]}]}
    ]
   }
   {
    type=MovieScene3DTransformTrack
    sections [
     {
      type=MovieScene3DTransformSection
      channels [
       Weight {type=MovieSceneScriptingFloatChannel, default=1}
       Scale.X {type=MovieSceneScriptingDoubleChannel, default=1}
       Scale.Y {type=MovieSceneScriptingDoubleChannel, default=1}
       Scale.Z {type=MovieSceneScriptingDoubleChannel, default=1}
      ]
     }
     {
      type=MovieScene3DTransformSection
      channels [
       Location.X {type=MovieSceneScriptingDoubleChannel, default=14436}
       Location.Y {type=MovieSceneScriptingDoubleChannel, default=76088}
       Location.Z {type=MovieSceneScriptingDoubleChannel, default=3072}
       Rotation.X {type=MovieSceneScriptingDoubleChannel, default=0}
       Rotation.Y {type=MovieSceneScriptingDoubleChannel, default=0}
       Rotation.Z {type=MovieSceneScriptingDoubleChannel, default=0}
       Scale.X {type=MovieSceneScriptingDoubleChannel, default=0}
       Scale.Y {type=MovieSceneScriptingDoubleChannel, default=0}
       Scale.Z {type=MovieSceneScriptingDoubleChannel, default=0}
      ]
     }
     {
      type=MovieScene3DTransformSection
      channels [
       Weight {type=MovieSceneScriptingFloatChannel, default=1}
       Location.X {type=MovieSceneScriptingDoubleChannel, default=1325.322}
       Location.Y {type=MovieSceneScriptingDoubleChannel, default=-9363.4971}
       Location.Z {type=MovieSceneScriptingDoubleChannel, default=2297.8594}
       Rotation.X {type=MovieSceneScriptingDoubleChannel, default=18.745}
       Rotation.Y {type=MovieSceneScriptingDoubleChannel, default=-7.1019}
       Rotation.Z {type=MovieSceneScriptingDoubleChannel, default=-69.9822}
       Scale.X {type=MovieSceneScriptingDoubleChannel, default=0}
       Scale.Y {type=MovieSceneScriptingDoubleChannel, default=0}
       Scale.Z {type=MovieSceneScriptingDoubleChannel, default=0}
      ]
     }
    ]
   }
  ]
 }
 SkeletalMesh {
  id=#4
  parent_id=#2
  possessed_class=SkeletalMeshComponent
  tracks [
   {
    type=MovieSceneSkeletalAnimationTrack
    sections [
     {type=MovieSceneSkeletalAnimationSection, end_frame=3311, start_frame=1788}
     {type=MovieSceneSkeletalAnimationSection, end_frame=2497, start_frame=1880}
     {type=MovieSceneSkeletalAnimationSection, end_frame=2606, start_frame=2435}
     {type=MovieSceneSkeletalAnimationSection, end_frame=2752, start_frame=2579}
     {type=MovieSceneSkeletalAnimationSection, end_frame=2981, start_frame=2707}
     {type=MovieSceneSkeletalAnimationSection, end_frame=3085, start_frame=2952}
    ]
   }
   {
    type=MovieSceneVisibilityTrack
    sections [{type=MovieSceneVisibilitySection, channels [None {type=MovieSceneScriptingBoolChannel, default=true}]}]
   }
  ]
 }
 PistolCanary_TentacleCut {
  id=#5
  possessed_class=PistolCanary_TentacleCut_C
  tracks [
   {
    type=MovieSceneVisibilityTrack
    sections [{type=MovieSceneVisibilitySection, channels [None {type=MovieSceneScriptingBoolChannel, default=true}]}]
   }
   {type=MovieSceneSpawnTrack, sections [{type=MovieSceneSpawnSection, channels [None {type=MovieSceneScriptingBoolChannel, default=true}]}]}
   {
    type=MovieScene3DTransformTrack
    sections [{
      type=MovieScene3DTransformSection
      channels [
       Location.X {type=MovieSceneScriptingDoubleChannel, default=1653.1514, const [14436,1]}
       Location.Y {type=MovieSceneScriptingDoubleChannel, default=-1737.0126, const [76088,1]}
       Location.Z {type=MovieSceneScriptingDoubleChannel, default=-16523.5561, const [3072,1]}
       Rotation.X {type=MovieSceneScriptingDoubleChannel, default=0, const [0,1]}
       Rotation.Y {type=MovieSceneScriptingDoubleChannel, default=0, const [0,1]}
       Rotation.Z {type=MovieSceneScriptingDoubleChannel, default=0, const [0,1]}
       Scale.X {type=MovieSceneScriptingDoubleChannel, default=1, const [0,1]}
       Scale.Y {type=MovieSceneScriptingDoubleChannel, default=1, const [0,1]}
       Scale.Z {type=MovieSceneScriptingDoubleChannel, default=1, const [0,1]}
      ]
     }]
   }
  ]
 }
 Tentacle Pistol Canary Creature {
  id=#6
  possessed_class=SkeletalMeshActor
  tracks [
   {type=MovieSceneSpawnTrack, sections [{type=MovieSceneSpawnSection, channels [None {type=MovieSceneScriptingBoolChannel, default=true}]}]}
   {
    type=MovieScene3DTransformTrack
    sections [{
      type=MovieScene3DTransformSection
      channels [
       Location.X {type=MovieSceneScriptingDoubleChannel, default=0}
       Location.Y {type=MovieSceneScriptingDoubleChannel, default=0}
       Location.Z {type=MovieSceneScriptingDoubleChannel, default=0}
       Rotation.X {type=MovieSceneScriptingDoubleChannel, default=0}
       Rotation.Y {type=MovieSceneScriptingDoubleChannel, default=0}
       Rotation.Z {type=MovieSceneScriptingDoubleChannel, default=0}
       Scale.X {type=MovieSceneScriptingDoubleChannel, default=1}
       Scale.Y {type=MovieSceneScriptingDoubleChannel, default=1}
       Scale.Z {type=MovieSceneScriptingDoubleChannel, default=1}
      ]
     }]
   }
   {type=MovieSceneSkeletalAnimationTrack, sections []}
   {
    type=MovieSceneControlRigParameterTrack
    display_name=Tentacle_PistolCanary_Creature_CtrlRig
    sections [
     {
      type=MovieSceneControlRigParameterSection
      channels [
       control rig options.Location.X {type=MovieSceneScriptingFloatChannel, default=0, const [0,3]}
       control rig options.Location.Y {type=MovieSceneScriptingFloatChannel, default=0, const [0,3]}
       control rig options.Location.Z {type=MovieSceneScriptingFloatChannel, default=0, const [0,3]}
       control rig options.Rotation.X {type=MovieSceneScriptingFloatChannel, default=0, const [0,3]}
       control rig options.Rotation.Y {type=MovieSceneScriptingFloatChannel, default=0, const [0,3]}
       control rig options.Rotation.Z {type=MovieSceneScriptingFloatChannel, default=0, const [0,3]}
       control rig options.Scale.X {type=MovieSceneScriptingFloatChannel, default=1, const [1,3]}
       control rig options.Scale.Y {type=MovieSceneScriptingFloatChannel, default=1, const [1,3]}
       control rig options.Scale.Z {type=MovieSceneScriptingFloatChannel, default=1, const [1,3]}
       Weight {type=MovieSceneScriptingFloatChannel, default=1, const [1,1]}
      ]
     }
     {
      type=MovieSceneControlRigParameterSection
      channels [
       control rig options.Location.X {type=MovieSceneScriptingFloatChannel, default=0}
       control rig options.Location.Y {type=MovieSceneScriptingFloatChannel, default=0}
       control rig options.Location.Z {type=MovieSceneScriptingFloatChannel, default=0}
       control rig options.Rotation.X {type=MovieSceneScriptingFloatChannel, default=0}
       control rig options.Rotation.Y {type=MovieSceneScriptingFloatChannel, default=0}
       control rig options.Rotation.Z {type=MovieSceneScriptingFloatChannel, default=0}
       control rig options.Scale.X {type=MovieSceneScriptingFloatChannel, default=0}
       control rig options.Scale.Y {type=MovieSceneScriptingFloatChannel, default=0}
       control rig options.Scale.Z {type=MovieSceneScriptingFloatChannel, default=0}
       Weight {type=MovieSceneScriptingFloatChannel, default=1}
      ]
     }
    ]
   }
  ]
 }
 TentacleCut_PistolCanary_Creature {
  id=#7
  parent_id=#5
  possessed_class=SkeletalMeshComponent
  tracks [
   {
    type=MovieSceneSkeletalAnimationTrack
    sections [
     {type=MovieSceneSkeletalAnimationSection, end_frame=3099, start_frame=1802}
     {type=MovieSceneSkeletalAnimationSection, end_frame=2511, start_frame=1894}
     {type=MovieSceneSkeletalAnimationSection, end_frame=2620, start_frame=2449}
     {type=MovieSceneSkeletalAnimationSection, end_frame=2766, start_frame=2593}
     {type=MovieSceneSkeletalAnimationSection, end_frame=2995, start_frame=2721}
     {type=MovieSceneSkeletalAnimationSection, end_frame=3099, start_frame=2966}
     {type=MovieSceneSkeletalAnimationSection, end_frame=3150, start_frame=3087}
     {type=MovieSceneSkeletalAnimationSection, end_frame=3207, start_frame=3150}
     {type=MovieSceneSkeletalAnimationSection, end_frame=3327, start_frame=3202}
     {type=MovieSceneSkeletalAnimationSection, end_frame=3429, start_frame=3316}
     {type=MovieSceneSkeletalAnimationSection, end_frame=3461, start_frame=3404}
     {type=MovieSceneSkeletalAnimationSection, end_frame=3570, start_frame=3445}
     {type=MovieSceneSkeletalAnimationSection, end_frame=3649, start_frame=3536}
    ]
   }
   {
    type=MovieScene3DTransformTrack
    sections [
     {
      type=MovieScene3DTransformSection
      channels [
       Location.X {type=MovieSceneScriptingDoubleChannel, default=1218, thinned_from=6, keys [[2163.0462,28],[3153,28],[3368,493.3554],[3505.5,1609]]}
       Location.Y {
        type=MovieSceneScriptingDoubleChannel
        default=-4104
        thinned_from=7
        keys [[2163.0462,-4488],[3153,-4488],[3368,-5334.2344],[3462,-4104],[3505.5,-7363]]
       }
       Location.Z {
        type=MovieSceneScriptingDoubleChannel
        default=-4842
        thinned_from=7
        keys [[1918.9987,-13622.5107],[2163.0462,2677.4897],[3153,2677.4897],[3368,464.1846],[3505.5,-4842]]
       }
       Rotation.X {type=MovieSceneScriptingDoubleChannel, default=0, keys [[2163.0462,0],[2621.5,0],[2697,-8.3572],[3099.5,0],[3153,0]]}
       Rotation.Y {type=MovieSceneScriptingDoubleChannel, default=0, keys [[2163.0462,0],[2621.5,0],[2697,29.5339],[3099.5,0],[3153,0]]}
       Rotation.Z {type=MovieSceneScriptingDoubleChannel, default=-90, keys [[2163.0462,0],[2621.5,0],[2697,-84.002],[2947,-84.002],[3099.5,0],[3153,0]]}
       Scale.X {type=MovieSceneScriptingDoubleChannel, default=1, const [1,3]}
       Scale.Y {type=MovieSceneScriptingDoubleChannel, default=1, const [1,3]}
       Scale.Z {type=MovieSceneScriptingDoubleChannel, default=1, const [1,3]}
      ]
     }
     {
      type=MovieScene3DTransformSection
      channels [
       Location.X {type=MovieSceneScriptingDoubleChannel, default=-569.5}
       Location.Y {type=MovieSceneScriptingDoubleChannel, default=896.6}
       Location.Z {type=MovieSceneScriptingDoubleChannel, default=-1200.4}
       Rotation.X {type=MovieSceneScriptingDoubleChannel, default=0}
       Rotation.Y {type=MovieSceneScriptingDoubleChannel, default=0}
       Rotation.Z {type=MovieSceneScriptingDoubleChannel, default=0}
       Scale.X {type=MovieSceneScriptingDoubleChannel, default=0}
       Scale.Y {type=MovieSceneScriptingDoubleChannel, default=0}
       Scale.Z {type=MovieSceneScriptingDoubleChannel, default=0}
      ]
     }
     {
      type=MovieScene3DTransformSection
      channels [
       Location.X {type=MovieSceneScriptingDoubleChannel, default=0}
       Location.Y {type=MovieSceneScriptingDoubleChannel, default=0}
       Location.Z {type=MovieSceneScriptingDoubleChannel, default=0}
       Rotation.X {type=MovieSceneScriptingDoubleChannel, default=0}
       Rotation.Y {type=MovieSceneScriptingDoubleChannel, default=0}
       Rotation.Z {type=MovieSceneScriptingDoubleChannel, default=0}
       Scale.X {type=MovieSceneScriptingDoubleChannel, default=0}
       Scale.Y {type=MovieSceneScriptingDoubleChannel, default=0}
       Scale.Z {type=MovieSceneScriptingDoubleChannel, default=0}
      ]
     }
    ]
   }
   {
    type=MovieSceneVisibilityTrack
    sections [{
      type=MovieSceneVisibilitySection
      channels [None {type=MovieSceneScriptingBoolChannel, default=false, keys [[2082,false],[3045.5,false],[3099,true],[3649,false]]}]
     }]
   }
  ]
 }
]
display_rate {denominator=1, numerator=30}
tick_resolution {denominator=1, numerator=24000}
tracks [{
  type=MovieSceneSubTrack
  display_name=Subsequences
  sections [{type=MovieSceneSubSection, end_frame=3205, start_frame=3061}, {type=MovieSceneSubSection, end_frame=4673, start_frame=3653}]
 }]
