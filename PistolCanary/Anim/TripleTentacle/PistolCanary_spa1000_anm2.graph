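# Level Sequence animating the PistolCanary triple-tentacle boss scene, with left, front and right tentacles appearing at staggered times.
# When the front tentacle hides, the animated PistolCanary_TentacleCut swaps in, and a severed-tentacle static mesh is also placed.
# Overlapping skeletal animation clips drive left, right and cut tentacles, with partially weighted Control Rig sections on left and right fading out mid-sequence.
# Constant transform sections position the actors, and two subsequences cover later ranges.
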LevelSequence /PistolCanary/Anim/TripleTentacle/PistolCanary_spa1000_anm2
bindings [
 TripleTentacleLeft {
  id=#1
  possessed_class=PistolCanary_TentacleBoss_C
  tracks [
   {
    type=MovieScene3DTransformTrack
    sections [
     {
      type=MovieScene3DTransformSection
      channels [
       Location.X {type=MovieSceneScriptingDoubleChannel, default=-1036.5809, const [-1036.5809,1]}
       Location.Y {type=MovieSceneScriptingDoubleChannel, default=-7497.0932, const [-7497.0932,1]}
       Location.Z {type=MovieSceneScriptingDoubleChannel, default=1941.8927, const [1941.8927,1]}
       Rotation.X {type=MovieSceneScriptingDoubleChannel, default=0, const [0,1]}
       Rotation.Y {type=MovieSceneScriptingDoubleChannel, default=0, const [0,1]}
       Rotation.Z {type=MovieSceneScriptingDoubleChannel, default=47.4855, const [47.4855,1]}
       Scale.X {type=MovieSceneScriptingDoubleChannel, default=1, const [1,1]}
       Scale.Y {type=MovieSceneScriptingDoubleChannel, default=1, const [1,1]}
       Scale.Z {type=MovieSceneScriptingDoubleChannel, default=1, const [1,1]}
      ]
     }
     {
      type=MovieScene3DTransformSection
      channels [
       Location.X {type=MovieSceneScriptingDoubleChannel, default=14436, const [14436,1]}
       Location.Y {type=MovieSceneScriptingDoubleChannel, default=76088, const [76088,1]}
       Location.Z {type=MovieSceneScriptingDoubleChannel, default=3072, const [3072,1]}
       Rotation.X {type=MovieSceneScriptingDoubleChannel, default=0, const [0,1]}
       Rotation.Y {type=MovieSceneScriptingDoubleChannel, default=0, const [0,1]}
       Rotation.Z {type=MovieSceneScriptingDoubleChannel, default=0, const [0,1]}
       Scale.X {type=MovieSceneScriptingDoubleChannel, default=0, const [0,1]}
       Scale.Y {type=MovieSceneScriptingDoubleChannel, default=0, const [0,1]}
       Scale.Z {type=MovieSceneScriptingDoubleChannel, default=0, const [0,1]}
      ]
     }
    ]
   }
   {
    type=MovieSceneVisibilityTrack
    sections [{
      type=MovieSceneVisibilitySection
      channels [None {type=MovieSceneScriptingBoolChannel, default=false, keys [[2113.5,false],[2167,true],[3786,false]]}]
     }]
   }
   {
    type=MovieSceneSkeletalAnimationTrack
    sections [
     {type=MovieSceneSkeletalAnimationSection, end_frame=3147, start_frame=1800}
     {type=MovieSceneSkeletalAnimationSection, end_frame=2509, start_frame=2000}
     {type=MovieSceneSkeletalAnimationSection, end_frame=2618, start_frame=2447}
     {type=MovieSceneSkeletalAnimationSection, end_frame=2764, start_frame=2591}
     {type=MovieSceneSkeletalAnimationSection, end_frame=3070, start_frame=2719}
     {type=MovieSceneSkeletalAnimationSection, end_frame=3158, start_frame=3025}
     {type=MovieSceneSkeletalAnimationSection, end_frame=3323, start_frame=3097}
     {type=MovieSceneSkeletalAnimationSection, end_frame=3432, start_frame=3323}
     {type=MovieSceneSkeletalAnimationSection, end_frame=3620, start_frame=3379}
     {type=MovieSceneSkeletalAnimationSection, end_frame=3722, start_frame=3554}
     {type=MovieSceneSkeletalAnimationSection, end_frame=3783, start_frame=3663}
    ]
   }
   {type=MovieSceneSpawnTrack, sections [{type=MovieSceneSpawnSection, channels [None {type=MovieSceneScriptingBoolChannel, default=true}]}]}
   {
    type=MovieSceneControlRigParameterTrack
    display_name=Tentacle_PistolCanary_Creature_CtrlRig
    sections [
     {
      type=MovieSceneControlRigParameterSection
      channels [
       control rig options.Location.X {type=MovieSceneScriptingFloatChannel, default=0}
       control rig options.Location.Y {type=MovieSceneScriptingFloatChannel, default=0}
       control rig options.Location.Z {type=MovieSceneScriptingFloatChannel, default=0}
       control rig options.Rotation.X {type=MovieSceneScriptingFloatChannel, default=0}
       control rig options.Rotation.Y {type=MovieSceneScriptingFloatChannel, default=0}
       control rig options.Rotation.Z {type=MovieSceneScriptingFloatChannel, default=0}
       control rig options.Scale.X {type=MovieSceneScriptingFloatChannel, default=0}
       control rig options.Scale.Y {type=MovieSceneScriptingFloatChannel, default=0}
       control rig options.Scale.Z {type=MovieSceneScriptingFloatChannel, default=0}
       Weight {type=MovieSceneScriptingFloatChannel, default=0.35, keys [[3115,0.35],[3125,0]]}
      ]
     }
     {
      type=MovieSceneControlRigParameterSection
      channels [
       control rig options.Location.X {type=MovieSceneScriptingFloatChannel, default=0, const [0,5]}
       control rig options.Location.Y {type=MovieSceneScriptingFloatChannel, default=0, const [0,5]}
       control rig options.Location.Z {type=MovieSceneScriptingFloatChannel, default=0, const [0,5]}
       control rig options.Rotation.X {type=MovieSceneScriptingFloatChannel, default=0, const [0,5]}
       control rig options.Rotation.Y {type=MovieSceneScriptingFloatChannel, default=0, const [0,5]}
       control rig options.Rotation.Z {type=MovieSceneScriptingFloatChannel, default=0, const [0,5]}
       control rig options.Scale.X {type=MovieSceneScriptingFloatChannel, default=1, const [1,5]}
       control rig options.Scale.Y {type=MovieSceneScriptingFloatChannel, default=1, const [1,5]}
       control rig options.Scale.Z {type=MovieSceneScriptingFloatChannel, default=1, const [1,5]}
      ]
     }
    ]
   }
  ]
 }
 TripleTentacleFront {
  id=#2
  possessed_class=PistolCanary_TentacleBoss_C
  tracks [
   {
    type=MovieScene3DTransformTrack
    sections [
     {
      type=MovieScene3DTransformSection
      channels [
       Location.X {type=MovieSceneScriptingDoubleChannel, default=28.0002, const [28.0002,1]}
       Location.Y {type=MovieSceneScriptingDoubleChannel, default=-5598.6317, const [-5598.6317,1]}
       Location.Z {type=MovieSceneScriptingDoubleChannel, default=986.8008, const [986.8008,1]}
       Rotation.X {type=MovieSceneScriptingDoubleChannel, default=0, const [0,1]}
       Rotation.Y {type=MovieSceneScriptingDoubleChannel, default=0, const [0,1]}
       Rotation.Z {type=MovieSceneScriptingDoubleChannel, default=0, const [0,1]}
       Scale.X {type=MovieSceneScriptingDoubleChannel, default=1}
       Scale.Y {type=MovieSceneScriptingDoubleChannel, default=1}
       Scale.Z {type=MovieSceneScriptingDoubleChannel, default=1}
      ]
     }
     {
      type=MovieScene3DTransformSection
      channels [
       Location.X {type=MovieSceneScriptingDoubleChannel, default=0, const [14436,1]}
       Location.Y {type=MovieSceneScriptingDoubleChannel, default=0, const [76088,1]}
       Location.Z {type=MovieSceneScriptingDoubleChannel, default=0, const [3072,1]}
       Rotation.X {type=MovieSceneScriptingDoubleChannel, default=0, const [0,1]}
       Rotation.Y {type=MovieSceneScriptingDoubleChannel, default=0, const [0,1]}
       Rotation.Z {type=MovieSceneScriptingDoubleChannel, default=0, const [0,1]}
       Scale.X {type=MovieSceneScriptingDoubleChannel, default=0, const [0,1]}
       Scale.Y {type=MovieSceneScriptingDoubleChannel, default=0, const [0,1]}
       Scale.Z {type=MovieSceneScriptingDoubleChannel, default=0, const [0,1]}
      ]
     }
    ]
   }
   {
    type=MovieSceneVisibilityTrack
    sections [{
      type=MovieSceneVisibilitySection
      channels [None {type=MovieSceneScriptingBoolChannel, default=false, keys [[1838.9987,false],[1892.4987,true],[3091,false]]}]
     }]
   }
   {type=MovieSceneSpawnTrack, sections [{type=MovieSceneSpawnSection, channels [None {type=MovieSceneScriptingBoolChannel, default=true}]}]}
   {
    type=MovieSceneControlRigParameterTrack
    display_name=Tentacle_PistolCanary_Creature_CtrlRig
    sections [{
      type=MovieSceneControlRigParameterSection
      channels [
       control rig options.Location.X {type=MovieSceneScriptingFloatChannel, default=0, const [0,6]}
       control rig options.Location.Y {type=MovieSceneScriptingFloatChannel, default=0, const [0,6]}
       control rig options.Location.Z {type=MovieSceneScriptingFloatChannel, default=0, const [0,6]}
       control rig options.Rotation.X {type=MovieSceneScriptingFloatChannel, default=0, const [0,6]}
       control rig options.Rotation.Y {type=MovieSceneScriptingFloatChannel, default=0, const [0,6]}
       control rig options.Rotation.Z {type=MovieSceneScriptingFloatChannel, default=0, const [0,6]}
       control rig options.Scale.X {type=MovieSceneScriptingFloatChannel, default=1, const [1,6]}
       control rig options.Scale.Y {type=MovieSceneScriptingFloatChannel, default=1, const [1,6]}
       control rig options.Scale.Z {type=MovieSceneScriptingFloatChannel, default=1, const [1,6]}
      ]
     }]
   }
  ]
 }
 TripleTentacleRight {
  id=#3
  possessed_class=PistolCanary_TentacleBoss_C
  tracks [
   {
    type=MovieScene3DTransformTrack
    sections [
     {
      type=MovieScene3DTransformSection
      channels [
       Location.X {type=MovieSceneScriptingDoubleChannel, default=1325.322, const [1325.322,1]}
       Location.Y {type=MovieSceneScriptingDoubleChannel, default=-9363.4971, const [-9363.4971,1]}
       Location.Z {type=MovieSceneScriptingDoubleChannel, default=2297.8594, const [2297.8594,1]}
       Rotation.X {type=MovieSceneScriptingDoubleChannel, default=18.745, const [18.745,1]}
       Rotation.Y {type=MovieSceneScriptingDoubleChannel, default=-7.1019, const [-7.1019,1]}
       Rotation.Z {type=MovieSceneScriptingDoubleChannel, default=-69.9822, const [-69.9822,1]}
       Scale.X {type=MovieSceneScriptingDoubleChannel, default=1, const [1,2]}
       Scale.Y {type=MovieSceneScriptingDoubleChannel, default=1, const [1,2]}
       Scale.Z {type=MovieSceneScriptingDoubleChannel, default=1, const [1,2]}
      ]
     }
     {
      type=MovieScene3DTransformSection
      channels [
       Location.X {type=MovieSceneScriptingDoubleChannel, default=0, const [14436,1]}
       Location.Y {type=MovieSceneScriptingDoubleChannel, default=0, const [76088,1]}
       Location.Z {type=MovieSceneScriptingDoubleChannel, default=0, const [3072,1]}
       Rotation.X {type=MovieSceneScriptingDoubleChannel, default=0, const [0,1]}
       Rotation.Y {type=MovieSceneScriptingDoubleChannel, default=0, const [0,1]}
       Rotation.Z {type=MovieSceneScriptingDoubleChannel, default=0, const [0,1]}
       Scale.X {type=MovieSceneScriptingDoubleChannel, default=0, const [0,1]}
       Scale.Y {type=MovieSceneScriptingDoubleChannel, default=0, const [0,1]}
       Scale.Z {type=MovieSceneScriptingDoubleChannel, default=0, const [0,1]}
      ]
     }
    ]
   }
   {
    type=MovieSceneVisibilityTrack
    sections [{
      type=MovieSceneVisibilitySection
      channels [None {type=MovieSceneScriptingBoolChannel, default=false, keys [[1955,false],[2008.5,true],[3931.5,false]]}]
     }]
   }
   {
    type=MovieSceneSkeletalAnimationTrack
    sections [
     {type=MovieSceneSkeletalAnimationSection, end_frame=3148, start_frame=1800}
     {type=MovieSceneSkeletalAnimationSection, end_frame=2509, start_frame=2017}
     {type=MovieSceneSkeletalAnimationSection, end_frame=2618, start_frame=2447}
     {type=MovieSceneSkeletalAnimationSection, end_frame=2764, start_frame=2591}
     {type=MovieSceneSkeletalAnimationSection, end_frame=3071, start_frame=2719}
     {type=MovieSceneSkeletalAnimationSection, end_frame=3134, start_frame=3001}
     {type=MovieSceneSkeletalAnimationSection, end_frame=3328, start_frame=3097}
     {type=MovieSceneSkeletalAnimationSection, end_frame=3861, start_frame=3323}
     {type=MovieSceneSkeletalAnimationSection, end_frame=3554, start_frame=3445}
     {type=MovieSceneSkeletalAnimationSection, end_frame=3944, start_frame=3835}
    ]
   }
   {type=MovieSceneSpawnTrack, sections [{type=MovieSceneSpawnSection, channels [None {type=MovieSceneScriptingBoolChannel, default=true}]}]}
   {
    type=MovieSceneControlRigParameterTrack
    display_name=Tentacle_PistolCanary_Creature_CtrlRig
    sections [
     {
      type=MovieSceneControlRigParameterSection
      channels [
       control rig options.Location.X {type=MovieSceneScriptingFloatChannel, default=0}
       control rig options.Location.Y {type=MovieSceneScriptingFloatChannel, default=0}
       control rig options.Location.Z {type=MovieSceneScriptingFloatChannel, default=0}
       control rig options.Rotation.X {type=MovieSceneScriptingFloatChannel, default=0}
       control rig options.Rotation.Y {type=MovieSceneScriptingFloatChannel, default=0}
       control rig options.Rotation.Z {type=MovieSceneScriptingFloatChannel, default=0}
       control rig options.Scale.X {type=MovieSceneScriptingFloatChannel, default=0}
       control rig options.Scale.Y {type=MovieSceneScriptingFloatChannel, default=0}
       control rig options.Scale.Z {type=MovieSceneScriptingFloatChannel, default=0}
       Weight {type=MovieSceneScriptingFloatChannel, default=0.35, keys [[3124,0.35],[3126,0]]}
      ]
     }
     {
      type=MovieSceneControlRigParameterSection
      channels [
       control rig options.Location.X {type=MovieSceneScriptingFloatChannel, default=0, const [0,2]}
       control rig options.Location.Y {type=MovieSceneScriptingFloatChannel, default=0, const [0,2]}
       control rig options.Location.Z {type=MovieSceneScriptingFloatChannel, default=0, const [0,2]}
       control rig options.Rotation.X {type=MovieSceneScriptingFloatChannel, default=0, const [0,2]}
       control rig options.Rotation.Y {type=MovieSceneScriptingFloatChannel, default=0, const [0,2]}
       control rig options.Rotation.Z {type=MovieSceneScriptingFloatChannel, default=0, const [0,2]}
       control rig options.Scale.X {type=MovieSceneScriptingFloatChannel, default=1, const [1,2]}
       control rig options.Scale.Y {type=MovieSceneScriptingFloatChannel, default=1, const [1,2]}
       control rig options.Scale.Z {type=MovieSceneScriptingFloatChannel, default=1, const [1,2]}
      ]
     }
    ]
   }
  ]
 }
 SM_PistolCanary_SeveredTentacle_Fresh2 {
  id=#4
  possessed_class=FortStaticMeshActor
  tracks [
   {type=MovieSceneSpawnTrack, sections [{type=MovieSceneSpawnSection, channels [None {type=MovieSceneScriptingBoolChannel, default=true}]}]}
   {
    type=MovieScene3DTransformTrack
    sections [{
      type=MovieScene3DTransformSection
      channels [
       Location.X {type=MovieSceneScriptingDoubleChannel, default=27141.5554, const [27141.5547,1]}
       Location.Y {type=MovieSceneScriptingDoubleChannel, default=78512.1126, const [78512.1094,1]}
       Location.Z {type=MovieSceneScriptingDoubleChannel, default=3676.1629, const [3676.1628,1]}
       Rotation.X {type=MovieSceneScriptingDoubleChannel, default=0, const [0,1]}
       Rotation.Y {type=MovieSceneScriptingDoubleChannel, default=-90, const [-90,1]}
       Rotation.Z {type=MovieSceneScriptingDoubleChannel, default=0, const [0,1]}
       Scale.X {type=MovieSceneScriptingDoubleChannel, default=1, const [1,1]}
       Scale.Y {type=MovieSceneScriptingDoubleChannel, default=1, const [1,1]}
       Scale.Z {type=MovieSceneScriptingDoubleChannel, default=1, const [1,1]}
      ]
     }]
   }
  ]
 }
 SkeletalMesh {
  id=#5
  parent_id=#2
  possessed_class=SkeletalMeshComponent
  tracks [{
    type=MovieSceneVisibilityTrack
    sections [{type=MovieSceneVisibilitySection, channels [None {type=MovieSceneScriptingBoolChannel, default=true}]}]
   }]
 }
 PistolCanary_TentacleCut {
  id=#6
  possessed_class=PistolCanary_TentacleCut_C
  tracks [
   {
    type=MovieSceneVisibilityTrack
    sections [{type=MovieSceneVisibilitySection, channels [None {type=MovieSceneScriptingBoolChannel, default=true}]}]
   }
   {type=MovieSceneSpawnTrack, sections [{type=MovieSceneSpawnSection, channels [None {type=MovieSceneScriptingBoolChannel, default=true}]}]}
   {
    type=MovieScene3DTransformTrack
    sections [
     {
      type=MovieScene3DTransformSection
      channels [
       Location.X {type=MovieSceneScriptingDoubleChannel, default=28.0002, const [28.0002,1]}
       Location.Y {type=MovieSceneScriptingDoubleChannel, default=-5598.6317, const [-5598.6317,1]}
       Location.Z {type=MovieSceneScriptingDoubleChannel, default=986.8008, const [986.8008,1]}
       Rotation.X {type=MovieSceneScriptingDoubleChannel, default=0, const [0,1]}
       Rotation.Y {type=MovieSceneScriptingDoubleChannel, default=0, const [0,1]}
       Rotation.Z {type=MovieSceneScriptingDoubleChannel, default=0, const [0,1]}
       Scale.X {type=MovieSceneScriptingDoubleChannel, default=1, const [1,1]}
       Scale.Y {type=MovieSceneScriptingDoubleChannel, default=1, const [1,1]}
       Scale.Z {type=MovieSceneScriptingDoubleChannel, default=1, const [1,1]}
      ]
     }
     {
      type=MovieScene3DTransformSection
      channels [
       Location.X {type=MovieSceneScriptingDoubleChannel, default=0, const [14436,1]}
       Location.Y {type=MovieSceneScriptingDoubleChannel, default=0, const [76088,1]}
       Location.Z {type=MovieSceneScriptingDoubleChannel, default=0, const [4702,1]}
       Rotation.X {type=MovieSceneScriptingDoubleChannel, default=0, const [0,1]}
       Rotation.Y {type=MovieSceneScriptingDoubleChannel, default=0, const [0,1]}
       Rotation.Z {type=MovieSceneScriptingDoubleChannel, default=0, const [0,1]}
       Scale.X {type=MovieSceneScriptingDoubleChannel, default=0, const [0,1]}
       Scale.Y {type=MovieSceneScriptingDoubleChannel, default=0, const [0,1]}
       Scale.Z {type=MovieSceneScriptingDoubleChannel, default=0, const [0,1]}
      ]
     }
    ]
   }
  ]
 }
 TentacleCut_PistolCanary_Creature {
  id=#7
  parent_id=#6
  possessed_class=SkeletalMeshComponent
  tracks [
   {
    type=MovieSceneSkeletalAnimationTrack
    sections [
     {type=MovieSceneSkeletalAnimationSection, end_frame=3097, start_frame=1800}
     {type=MovieSceneSkeletalAnimationSection, end_frame=2509, start_frame=1892}
     {type=MovieSceneSkeletalAnimationSection, end_frame=2618, start_frame=2447}
     {type=MovieSceneSkeletalAnimationSection, end_frame=2764, start_frame=2591}
     {type=MovieSceneSkeletalAnimationSection, end_frame=2993, start_frame=2719}
     {type=MovieSceneSkeletalAnimationSection, end_frame=3097, start_frame=2964}
     {type=MovieSceneSkeletalAnimationSection, end_frame=3148, start_frame=3085}
     {type=MovieSceneSkeletalAnimationSection, end_frame=3205, start_frame=3148}
     {type=MovieSceneSkeletalAnimationSection, end_frame=3325, start_frame=3200}
     {type=MovieSceneSkeletalAnimationSection, end_frame=3427, start_frame=3314}
     {type=MovieSceneSkeletalAnimationSection, end_frame=3459, start_frame=3402}
     {type=MovieSceneSkeletalAnimationSection, end_frame=3568, start_frame=3443}
     {type=MovieSceneSkeletalAnimationSection, end_frame=3647, start_frame=3534}
    ]
   }
   {
    type=MovieSceneVisibilityTrack
    sections [{
      type=MovieSceneVisibilitySection
      channels [None {type=MovieSceneScriptingBoolChannel, default=false, keys [[3043.5,false],[3091,true],[3647,false]]}]
     }]
   }
   {
    type=MovieSceneControlRigParameterTrack
    display_name=TentacleCut_PistolCanary_Creature_CtrlRig
    sections [{
      type=MovieSceneControlRigParameterSection
      channels [
       control rig options.Location.X {type=MovieSceneScriptingFloatChannel, default=0, const [0,7]}
       control rig options.Location.Y {type=MovieSceneScriptingFloatChannel, default=0, const [0,7]}
       control rig options.Location.Z {type=MovieSceneScriptingFloatChannel, default=0, const [0,7]}
       control rig options.Rotation.X {type=MovieSceneScriptingFloatChannel, default=0, const [0,7]}
       control rig options.Rotation.Y {type=MovieSceneScriptingFloatChannel, default=0, const [0,7]}
       control rig options.Rotation.Z {type=MovieSceneScriptingFloatChannel, default=0, const [0,7]}
       control rig options.Scale.X {type=MovieSceneScriptingFloatChannel, default=1, const [1,7]}
       control rig options.Scale.Y {type=MovieSceneScriptingFloatChannel, default=1, const [1,7]}
       control rig options.Scale.Z {type=MovieSceneScriptingFloatChannel, default=1, const [1,7]}
       Weight {type=MovieSceneScriptingFloatChannel, default=1, const [1,2]}
      ]
     }]
   }
  ]
 }
]
display_rate {denominator=1, numerator=30}
tick_resolution {denominator=1, numerator=24000}
tracks [{
  type=MovieSceneSubTrack
  display_name=Subsequences
  sections [{type=MovieSceneSubSection, end_frame=3205, start_frame=3061}, {type=MovieSceneSubSection, end_frame=4611, start_frame=3591}]
 }]
