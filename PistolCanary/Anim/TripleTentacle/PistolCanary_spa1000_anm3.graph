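# Level Sequence animating a three-tentacle PistolCanary boss scene, with left, front and right tentacles appearing at staggered times.
# Tentacles layer overlapping skeletal animations and offset transform sections, with Control Rig tracks on left and right.
# Front tentacle swaps by visibility to a matching TentacleCut actor, which drops away and hides.
# An attached laser mesh flashes via scale keys late on, alongside two subsequences and a closing camera cut.
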LevelSequence /PistolCanary/Anim/TripleTentacle/PistolCanary_spa1000_anm3
bindings [
 SkeletalMesh {id=#1, parent_id=#2, possessed_class=SkeletalMeshComponent, tracks []}
 TripleTentacleLeft {
  id=#2
  possessed_class=PistolCanary_TentacleBoss_C
  tracks [
   {
    type=MovieScene3DTransformTrack
    sections [
     {
      type=MovieScene3DTransformSection
      channels [
       Location.X {type=MovieSceneScriptingDoubleChannel, default=-1036.5809, const [-1036.5809,1]}
       Location.Y {type=MovieSceneScriptingDoubleChannel, default=-7497.0932, const [-7497.0932,1]}
       Location.Z {type=MovieSceneScriptingDoubleChannel, default=1941.8927, const [1941.8927,1]}
       Rotation.X {type=MovieSceneScriptingDoubleChannel, default=0, const [0,1]}
       Rotation.Y {type=MovieSceneScriptingDoubleChannel, default=0, const [0,1]}
       Rotation.Z {type=MovieSceneScriptingDoubleChannel, default=47.4855, const [47.4855,1]}
       Scale.X {type=MovieSceneScriptingDoubleChannel, default=1}
       Scale.Y {type=MovieSceneScriptingDoubleChannel, default=1}
       Scale.Z {type=MovieSceneScriptingDoubleChannel, default=1}
      ]
     }
     {
      type=MovieScene3DTransformSection
      channels [
       Location.X {type=MovieSceneScriptingDoubleChannel, default=14436, const [14436,2]}
       Location.Y {type=MovieSceneScriptingDoubleChannel, default=76088, const [76088,2]}
       Location.Z {type=MovieSceneScriptingDoubleChannel, default=3072, const [3072,2]}
       Rotation.X {type=MovieSceneScriptingDoubleChannel, default=0, const [0,2]}
       Rotation.Y {type=MovieSceneScriptingDoubleChannel, default=0, const [0,2]}
       Rotation.Z {type=MovieSceneScriptingDoubleChannel, default=0, keys [[2230,0],[3957,0]]}
       Scale.X {type=MovieSceneScriptingDoubleChannel, default=0, const [0,2]}
       Scale.Y {type=MovieSceneScriptingDoubleChannel, default=0, const [0,2]}
       Scale.Z {type=MovieSceneScriptingDoubleChannel, default=0, const [0,2]}
      ]
     }
    ]
   }
   {
    type=MovieSceneVisibilityTrack
    sections [{
      type=MovieSceneVisibilitySection
      channels [None {type=MovieSceneScriptingBoolChannel, default=false, keys [[2113.5,false],[2167,true],[3796,false]]}]
     }]
   }
   {
    type=MovieSceneSkeletalAnimationTrack
    sections [
     {type=MovieSceneSkeletalAnimationSection, end_frame=3147, start_frame=1800}
     {type=MovieSceneSkeletalAnimationSection, end_frame=2509, start_frame=2000}
     {type=MovieSceneSkeletalAnimationSection, end_frame=2618, start_frame=2447}
     {type=MovieSceneSkeletalAnimationSection, end_frame=2764, start_frame=2591}
     {type=MovieSceneSkeletalAnimationSection, end_frame=3070, start_frame=2719}
     {type=MovieSceneSkeletalAnimationSection, end_frame=3158, start_frame=3025}
     {type=MovieSceneSkeletalAnimationSection, end_frame=3323, start_frame=3097}
     {type=MovieSceneSkeletalAnimationSection, end_frame=3432, start_frame=3323}
     {type=MovieSceneSkeletalAnimationSection, end_frame=3620, start_frame=3379}
     {type=MovieSceneSkeletalAnimationSection, end_frame=3722, start_frame=3554}
     {type=MovieSceneSkeletalAnimationSection, end_frame=3783, start_frame=3663}
    ]
   }
   {type=MovieSceneSpawnTrack, sections [{type=MovieSceneSpawnSection, channels [None {type=MovieSceneScriptingBoolChannel, default=true}]}]}
   {
    type=MovieSceneControlRigParameterTrack
    display_name=Tentacle_PistolCanary_Creature_CtrlRig
    sections [
     {
      type=MovieSceneControlRigParameterSection
      channels [
       control rig options.Location.X {type=MovieSceneScriptingFloatChannel, default=0, const [0,405]}
       control rig options.Location.Y {type=MovieSceneScriptingFloatChannel, default=0, const [0,405]}
       control rig options.Location.Z {type=MovieSceneScriptingFloatChannel, default=0, const [0,405]}
       control rig options.Rotation.X {type=MovieSceneScriptingFloatChannel, default=0, const [0,405]}
       control rig options.Rotation.Y {type=MovieSceneScriptingFloatChannel, default=0, const [0,405]}
       control rig options.Rotation.Z {type=MovieSceneScriptingFloatChannel, default=0, const [0,405]}
       control rig options.Scale.X {type=MovieSceneScriptingFloatChannel, default=1, const [1,405]}
       control rig options.Scale.Y {type=MovieSceneScriptingFloatChannel, default=1, const [1,405]}
       control rig options.Scale.Z {type=MovieSceneScriptingFloatChannel, default=1, const [1,405]}
       Weight {type=MovieSceneScriptingFloatChannel, default=1, const [1,1]}
      ]
     }
     {type=MovieSceneControlRigParameterSection, channels [Weight {type=MovieSceneScriptingFloatChannel, default=1}]}
    ]
   }
  ]
 }
 Laser_Tentacle {
  id=#3
  possessed_class=FortStaticMeshActor
  tracks [
   {
    type=MovieScene3DTransformTrack
    sections [{
      type=MovieScene3DTransformSection
      channels [
       Location.X {
        type=MovieSceneScriptingDoubleChannel
        default=2985.3436
        thinned_from=8
        keys [[3792,2985.3436],[3808,2985.3436],[3809,3721.9219],[3815,4265.8911],[3823,2985.3436],[3824,4371.3975],[3831,4371.3975]]
       }
       Location.Y {
        type=MovieSceneScriptingDoubleChannel
        default=2080.2008
        thinned_from=8
        keys [[3792,2080.2008],[3808,2080.2008],[3809,2593.4519],[3815,2972.4919],[3823,2080.2008],[3824,3046.0093],[3831,3046.0093]]
       }
       Location.Z {
        type=MovieSceneScriptingDoubleChannel
        default=-1788.3217
        thinned_from=8
        keys [[3792,-1788.3217],[3808,-1788.3217],[3809,-2229.5571],[3815,-2555.4128],[3823,-1788.3217],[3824,-2618.615],[3831,-2618.615]]
       }
       Rotation.X {type=MovieSceneScriptingDoubleChannel, default=0, thinned_from=7, keys [[3792,0],[3828,0]]}
       Rotation.Y {type=MovieSceneScriptingDoubleChannel, default=63.8266, thinned_from=7, keys [[3792,63.8266],[3828,63.8266]]}
       Rotation.Z {type=MovieSceneScriptingDoubleChannel, default=34.869, thinned_from=7, keys [[3792,34.869],[3828,34.869]]}
       Scale.X {
        type=MovieSceneScriptingDoubleChannel
        default=8.5328
        keys [[3791,0],[3792,8.5328],[3797,8.5328],[3798,0],[3808,0],[3809,8.5328],[3815,8.5328],[3818,0],[3823,0],[3827,8.5328],[3828,8.5328],[3831,0]]
       }
       Scale.Y {type=MovieSceneScriptingDoubleChannel, default=9.5336, thinned_from=7, keys [[3792,9.5336],[3828,9.5336]]}
       Scale.Z {
        type=MovieSceneScriptingDoubleChannel
        default=72.3307
        thinned_from=7
        keys [[3792,72.3307],[3808,72.3307],[3809,92.3661],[3815,100.9153],[3823,103.4248],[3828,103.4248]]
       }
      ]
     }]
   }
   {type=MovieScene3DAttachTrack, sections [{type=MovieScene3DAttachSection, end_frame=3953, start_frame=3792}]}
   {type=MovieSceneSpawnTrack, sections [{type=MovieSceneSpawnSection, channels [None {type=MovieSceneScriptingBoolChannel, default=true}]}]}
  ]
 }
 TripleTentacleFront {
  id=#4
  possessed_class=PistolCanary_TentacleBoss_C
  tracks [
   {
    type=MovieScene3DTransformTrack
    sections [
     {
      type=MovieScene3DTransformSection
      channels [
       Location.X {type=MovieSceneScriptingDoubleChannel, default=28, const [28,3]}
       Location.Y {type=MovieSceneScriptingDoubleChannel, default=-3888, const [-4488,3]}
       Location.Z {
        type=MovieSceneScriptingDoubleChannel
        default=-250.5103
        thinned_from=4
        keys [[1916.9987,-13622.5107],[2161.0462,2677.4897],[2695,2677.4897]]
       }
       Rotation.X {type=MovieSceneScriptingDoubleChannel, default=0, keys [[2161.0462,0],[2619.5,0],[2695,-8.3572],[3097.5,0]]}
       Rotation.Y {type=MovieSceneScriptingDoubleChannel, default=0, keys [[2161.0462,0],[2619.5,0],[2695,29.5339],[3097.5,0]]}
       Rotation.Z {type=MovieSceneScriptingDoubleChannel, default=-90, keys [[2161.0462,0],[2619.5,0],[2695,-84.002],[2945,-84.002],[3097.5,0]]}
       Scale.X {type=MovieSceneScriptingDoubleChannel, default=1, const [1,2]}
       Scale.Y {type=MovieSceneScriptingDoubleChannel, default=1, const [1,2]}
       Scale.Z {type=MovieSceneScriptingDoubleChannel, default=1, const [1,2]}
      ]
     }
     {
      type=MovieScene3DTransformSection
      channels [
       Location.X {type=MovieSceneScriptingDoubleChannel, default=0, const [14436,1]}
       Location.Y {type=MovieSceneScriptingDoubleChannel, default=0, const [76088,1]}
       Location.Z {type=MovieSceneScriptingDoubleChannel, default=0, const [3072,1]}
       Rotation.X {type=MovieSceneScriptingDoubleChannel, default=0, const [0,1]}
       Rotation.Y {type=MovieSceneScriptingDoubleChannel, default=0, const [0,1]}
       Rotation.Z {type=MovieSceneScriptingDoubleChannel, default=0, const [0,1]}
       Scale.X {type=MovieSceneScriptingDoubleChannel, default=0, const [0,1]}
       Scale.Y {type=MovieSceneScriptingDoubleChannel, default=0, const [0,1]}
       Scale.Z {type=MovieSceneScriptingDoubleChannel, default=0, const [0,1]}
      ]
     }
    ]
   }
   {
    type=MovieSceneVisibilityTrack
    sections [{
      type=MovieSceneVisibilitySection
      channels [None {type=MovieSceneScriptingBoolChannel, default=false, keys [[1838.9987,false],[1892.4987,true],[3097,false]]}]
     }]
   }
   {type=MovieSceneSpawnTrack, sections [{type=MovieSceneSpawnSection, channels [None {type=MovieSceneScriptingBoolChannel, default=true}]}]}
  ]
 }
 TripleTentacleRight {
  id=#5
  possessed_class=PistolCanary_TentacleBoss_C
  tracks [
   {
    type=MovieScene3DTransformTrack
    sections [
     {
      type=MovieScene3DTransformSection
      channels [
       Location.X {type=MovieSceneScriptingDoubleChannel, default=1325.322, const [1325.322,1]}
       Location.Y {type=MovieSceneScriptingDoubleChannel, default=-9363.4971, const [-9363.4971,1]}
       Location.Z {type=MovieSceneScriptingDoubleChannel, default=2297.8594, const [2297.8594,1]}
       Rotation.X {type=MovieSceneScriptingDoubleChannel, default=18.745, const [18.745,1]}
       Rotation.Y {type=MovieSceneScriptingDoubleChannel, default=-7.1019, const [-7.1019,1]}
       Rotation.Z {type=MovieSceneScriptingDoubleChannel, default=-69.9822, const [-69.9822,1]}
       Scale.X {type=MovieSceneScriptingDoubleChannel, default=1}
       Scale.Y {type=MovieSceneScriptingDoubleChannel, default=1}
       Scale.Z {type=MovieSceneScriptingDoubleChannel, default=1}
      ]
     }
     {
      type=MovieScene3DTransformSection
      channels [
       Location.X {type=MovieSceneScriptingDoubleChannel, default=0, const [14436,1]}
       Location.Y {type=MovieSceneScriptingDoubleChannel, default=0, const [76088,1]}
       Location.Z {type=MovieSceneScriptingDoubleChannel, default=0, const [3072,1]}
       Rotation.X {type=MovieSceneScriptingDoubleChannel, default=0, const [0,1]}
       Rotation.Y {type=MovieSceneScriptingDoubleChannel, default=0, const [0,1]}
       Rotation.Z {type=MovieSceneScriptingDoubleChannel, default=0, const [0,1]}
       Scale.X {type=MovieSceneScriptingDoubleChannel, default=0, const [0,1]}
       Scale.Y {type=MovieSceneScriptingDoubleChannel, default=0, const [0,1]}
       Scale.Z {type=MovieSceneScriptingDoubleChannel, default=0, const [0,1]}
      ]
     }
    ]
   }
   {
    type=MovieSceneVisibilityTrack
    sections [{
      type=MovieSceneVisibilitySection
      channels [None {type=MovieSceneScriptingBoolChannel, default=false, keys [[1955,false],[2027,true],[3958,false]]}]
     }]
   }
   {
    type=MovieSceneSkeletalAnimationTrack
    sections [
     {type=MovieSceneSkeletalAnimationSection, end_frame=3148, start_frame=1800}
     {type=MovieSceneSkeletalAnimationSection, end_frame=2509, start_frame=2017}
     {type=MovieSceneSkeletalAnimationSection, end_frame=2618, start_frame=2447}
     {type=MovieSceneSkeletalAnimationSection, end_frame=2764, start_frame=2591}
     {type=MovieSceneSkeletalAnimationSection, end_frame=3071, start_frame=2719}
     {type=MovieSceneSkeletalAnimationSection, end_frame=3134, start_frame=3001}
     {type=MovieSceneSkeletalAnimationSection, end_frame=3328, start_frame=3097}
     {type=MovieSceneSkeletalAnimationSection, end_frame=3861, start_frame=3323}
     {type=MovieSceneSkeletalAnimationSection, end_frame=3554, start_frame=3445}
     {type=MovieSceneSkeletalAnimationSection, end_frame=3944, start_frame=3835}
    ]
   }
   {type=MovieSceneSpawnTrack, sections [{type=MovieSceneSpawnSection, channels [None {type=MovieSceneScriptingBoolChannel, default=true}]}]}
   {
    type=MovieSceneControlRigParameterTrack
    display_name=Tentacle_PistolCanary_Creature_CtrlRig
    sections [
     {
      type=MovieSceneControlRigParameterSection
      channels [
       control rig options.Location.X {type=MovieSceneScriptingFloatChannel, default=0, const [0,7]}
       control rig options.Location.Y {type=MovieSceneScriptingFloatChannel, default=0, const [0,7]}
       control rig options.Location.Z {type=MovieSceneScriptingFloatChannel, default=0, const [0,7]}
       control rig options.Rotation.X {type=MovieSceneScriptingFloatChannel, default=0, const [0,7]}
       control rig options.Rotation.Y {type=MovieSceneScriptingFloatChannel, default=0, const [0,7]}
       control rig options.Rotation.Z {type=MovieSceneScriptingFloatChannel, default=0, const [0,7]}
       control rig options.Scale.X {type=MovieSceneScriptingFloatChannel, default=1, const [1,7]}
       control rig options.Scale.Y {type=MovieSceneScriptingFloatChannel, default=1, const [1,7]}
       control rig options.Scale.Z {type=MovieSceneScriptingFloatChannel, default=1, const [1,7]}
       Weight {type=MovieSceneScriptingFloatChannel, default=1, const [1,1]}
      ]
     }
     {type=MovieSceneControlRigParameterSection, channels [Weight {type=MovieSceneScriptingFloatChannel, default=1, const [1,1]}]}
     {type=MovieSceneControlRigParameterSection, channels [Weight {type=MovieSceneScriptingFloatChannel, default=1, const [1,1]}]}
     {type=MovieSceneControlRigParameterSection, channels [Weight {type=MovieSceneScriptingFloatChannel, default=1, const [1,1]}]}
     {type=MovieSceneControlRigParameterSection, channels [Weight {type=MovieSceneScriptingFloatChannel, default=1, const [1,1]}]}
     {type=MovieSceneControlRigParameterSection, channels [Weight {type=MovieSceneScriptingFloatChannel, default=1, const [1,1]}]}
    ]
   }
  ]
 }
 SkeletalMesh {
  id=#6
  parent_id=#4
  possessed_class=SkeletalMeshComponent
  tracks [
   {
    type=MovieSceneSkeletalAnimationTrack
    sections [
     {type=MovieSceneSkeletalAnimationSection, end_frame=3323, start_frame=1800}
     {type=MovieSceneSkeletalAnimationSection, end_frame=2509, start_frame=1892}
     {type=MovieSceneSkeletalAnimationSection, end_frame=2618, start_frame=2447}
     {type=MovieSceneSkeletalAnimationSection, end_frame=2764, start_frame=2591}
     {type=MovieSceneSkeletalAnimationSection, end_frame=2993, start_frame=2719}
     {type=MovieSceneSkeletalAnimationSection, end_frame=3097, start_frame=2964}
    ]
   }
   {
    type=MovieSceneVisibilityTrack
    sections [{type=MovieSceneVisibilitySection, channels [None {type=MovieSceneScriptingBoolChannel, default=true}]}]
   }
  ]
 }
 PistolCanary_TentacleCut {
  id=#7
  possessed_class=PistolCanary_TentacleCut_C
  tracks [
   {
    type=MovieSceneVisibilityTrack
    sections [{type=MovieSceneVisibilitySection, channels [None {type=MovieSceneScriptingBoolChannel, default=true}]}]
   }
   {type=MovieSceneSpawnTrack, sections [{type=MovieSceneSpawnSection, channels [None {type=MovieSceneScriptingBoolChannel, default=true}]}]}
   {
    type=MovieScene3DTransformTrack
    sections [{
      type=MovieScene3DTransformSection
      channels [
       Location.X {type=MovieSceneScriptingDoubleChannel, default=1653.1514, const [14436,1]}
       Location.Y {type=MovieSceneScriptingDoubleChannel, default=-1737.0126, const [76088,1]}
       Location.Z {type=MovieSceneScriptingDoubleChannel, default=-16523.5561, const [3072,1]}
       Rotation.X {type=MovieSceneScriptingDoubleChannel, default=0, const [0,1]}
       Rotation.Y {type=MovieSceneScriptingDoubleChannel, default=0, const [0,1]}
       Rotation.Z {type=MovieSceneScriptingDoubleChannel, default=0, const [0,1]}
       Scale.X {type=MovieSceneScriptingDoubleChannel, default=1, const [0,1]}
       Scale.Y {type=MovieSceneScriptingDoubleChannel, default=1, const [0,1]}
       Scale.Z {type=MovieSceneScriptingDoubleChannel, default=1, const [0,1]}
      ]
     }]
   }
  ]
 }
 TentacleCut_PistolCanary_Creature {
  id=#8
  parent_id=#7
  possessed_class=SkeletalMeshComponent
  tracks [
   {
    type=MovieSceneSkeletalAnimationTrack
    sections [
     {type=MovieSceneSkeletalAnimationSection, end_frame=3097, start_frame=1800}
     {type=MovieSceneSkeletalAnimationSection, end_frame=2509, start_frame=1892}
     {type=MovieSceneSkeletalAnimationSection, end_frame=2618, start_frame=2447}
     {type=MovieSceneSkeletalAnimationSection, end_frame=2764, start_frame=2591}
     {type=MovieSceneSkeletalAnimationSection, end_frame=2993, start_frame=2719}
     {type=MovieSceneSkeletalAnimationSection, end_frame=3097, start_frame=2964}
     {type=MovieSceneSkeletalAnimationSection, end_frame=3148, start_frame=3085}
     {type=MovieSceneSkeletalAnimationSection, end_frame=3205, start_frame=3148}
     {type=MovieSceneSkeletalAnimationSection, end_frame=3325, start_frame=3200}
     {type=MovieSceneSkeletalAnimationSection, end_frame=3427, start_frame=3314}
     {type=MovieSceneSkeletalAnimationSection, end_frame=3459, start_frame=3402}
     {type=MovieSceneSkeletalAnimationSection, end_frame=3568, start_frame=3443}
     {type=MovieSceneSkeletalAnimationSection, end_frame=3647, start_frame=3534}
    ]
   }
   {
    type=MovieScene3DTransformTrack
    sections [
     {
      type=MovieScene3DTransformSection
      channels [
       Location.X {type=MovieSceneScriptingDoubleChannel, default=1218, thinned_from=6, keys [[2161.0462,28],[3151,28],[3366,493.3554],[3503.5,1609]]}
       Location.Y {
        type=MovieSceneScriptingDoubleChannel
        default=-4104
        thinned_from=7
        keys [[2161.0462,-4488],[3151,-4488],[3366,-5334.2344],[3460,-4104],[3503.5,-7363]]
       }
       Location.Z {
        type=MovieSceneScriptingDoubleChannel
        default=-4842
        thinned_from=7
        keys [[1916.9987,-13622.5107],[2161.0462,2677.4897],[3151,2677.4897],[3366,464.1846],[3503.5,-4842]]
       }
       Rotation.X {type=MovieSceneScriptingDoubleChannel, default=0, keys [[2161.0462,0],[2619.5,0],[2695,-8.3572],[3097.5,0],[3151,0]]}
       Rotation.Y {type=MovieSceneScriptingDoubleChannel, default=0, keys [[2161.0462,0],[2619.5,0],[2695,29.5339],[3097.5,0],[3151,0]]}
       Rotation.Z {type=MovieSceneScriptingDoubleChannel, default=-90, keys [[2161.0462,0],[2619.5,0],[2695,-84.002],[2945,-84.002],[3097.5,0],[3151,0]]}
       Scale.X {type=MovieSceneScriptingDoubleChannel, default=1, const [1,3]}
       Scale.Y {type=MovieSceneScriptingDoubleChannel, default=1, const [1,3]}
       Scale.Z {type=MovieSceneScriptingDoubleChannel, default=1, const [1,3]}
      ]
     }
     {
      type=MovieScene3DTransformSection
      channels [
       Location.X {type=MovieSceneScriptingDoubleChannel, default=-569.5}
       Location.Y {type=MovieSceneScriptingDoubleChannel, default=896.6}
       Location.Z {type=MovieSceneScriptingDoubleChannel, default=-1200.4}
       Rotation.X {type=MovieSceneScriptingDoubleChannel, default=0}
       Rotation.Y {type=MovieSceneScriptingDoubleChannel, default=0}
       Rotation.Z {type=MovieSceneScriptingDoubleChannel, default=0}
       Scale.X {type=MovieSceneScriptingDoubleChannel, default=0}
       Scale.Y {type=MovieSceneScriptingDoubleChannel, default=0}
       Scale.Z {type=MovieSceneScriptingDoubleChannel, default=0}
      ]
     }
     {
      type=MovieScene3DTransformSection
      channels [
       Location.X {type=MovieSceneScriptingDoubleChannel, default=0}
       Location.Y {type=MovieSceneScriptingDoubleChannel, default=0}
       Location.Z {type=MovieSceneScriptingDoubleChannel, default=0}
       Rotation.X {type=MovieSceneScriptingDoubleChannel, default=0}
       Rotation.Y {type=MovieSceneScriptingDoubleChannel, default=0}
       Rotation.Z {type=MovieSceneScriptingDoubleChannel, default=0}
       Scale.X {type=MovieSceneScriptingDoubleChannel, default=0}
       Scale.Y {type=MovieSceneScriptingDoubleChannel, default=0}
       Scale.Z {type=MovieSceneScriptingDoubleChannel, default=0}
      ]
     }
    ]
   }
   {
    type=MovieSceneVisibilityTrack
    sections [{
      type=MovieSceneVisibilitySection
      channels [None {type=MovieSceneScriptingBoolChannel, default=false, keys [[3043.5,false],[3097,true],[3647,false]]}]
     }]
   }
  ]
 }
 SkeletalMesh {id=#9, parent_id=#5, possessed_class=SkeletalMeshComponent, tracks []}
]
display_rate {denominator=1, numerator=30}
tick_resolution {denominator=1, numerator=24000}
tracks [
 {
  type=MovieSceneSubTrack
  display_name=Subsequences
  sections [{type=MovieSceneSubSection, end_frame=3205, start_frame=3061}, {type=MovieSceneSubSection, end_frame=4673, start_frame=3653}]
 }
 {type=MovieSceneCameraCutTrack, sections [{type=MovieSceneCameraCutSection, end_frame=4057, start_frame=3654}]}
]
