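# Cinematic Level Sequence for the PistolCanary TripleTentacle animation set, animating a spawnable superhero NPC bound as ConTrail.
# Transform and visibility tracks place the character, briefly hide it and scale it up to five times size before restoring it.
# A skeletal animation section runs alongside a Fortnite biped Control Rig track with densely keyed FK and IK body controls.
# The rig's root carries a long, twisting dive-and-climb move, then returns to zero.
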LevelSequence /PistolCanary/Anim/TripleTentacle/PistolCanary_spa1000b_anm1
bindings [
 ConTrail {
  id=#1
  possessed_class=PistolCanary_CinematicNPC_Superhero_Main_C
  tracks [
   {
    type=MovieScene3DTransformTrack
    sections [
     {
      type=MovieScene3DTransformSection
      end_frame=3770
      channels [
       Location.X {type=MovieSceneScriptingDoubleChannel, default=-52714, const [24199.1484,1]}
       Location.Y {type=MovieSceneScriptingDoubleChannel, default=25000, const [10733.0879,1]}
       Location.Z {type=MovieSceneScriptingDoubleChannel, default=8097.1553, const [25858.6641,1]}
       Rotation.X {type=MovieSceneScriptingDoubleChannel, default=0, const [30.0169,1]}
       Rotation.Y {type=MovieSceneScriptingDoubleChannel, default=0, const [0,1]}
       Rotation.Z {type=MovieSceneScriptingDoubleChannel, default=0, const [0,1]}
       Scale.X {type=MovieSceneScriptingDoubleChannel, default=5, keys [[1800,1],[1814,5],[1943,4],[1944,1]]}
       Scale.Y {type=MovieSceneScriptingDoubleChannel, default=5, keys [[1800,1],[1814,5],[1943,4],[1944,1]]}
       Scale.Z {type=MovieSceneScriptingDoubleChannel, default=5, keys [[1800,1],[1814,5],[1943,4],[1944,1]]}
      ]
     }
     {
      type=MovieScene3DTransformSection
      channels [
       Location.X {type=MovieSceneScriptingDoubleChannel, default=0, const [14436,1]}
       Location.Y {type=MovieSceneScriptingDoubleChannel, default=0, const [76088,1]}
       Location.Z {type=MovieSceneScriptingDoubleChannel, default=0, const [3072,1]}
       Rotation.X {type=MovieSceneScriptingDoubleChannel, default=0, const [0,1]}
       Rotation.Y {type=MovieSceneScriptingDoubleChannel, default=0, const [0,1]}
       Rotation.Z {type=MovieSceneScriptingDoubleChannel, default=0, const [0,1]}
       Scale.X {type=MovieSceneScriptingDoubleChannel, default=0, const [0,1]}
       Scale.Y {type=MovieSceneScriptingDoubleChannel, default=0, const [0,1]}
       Scale.Z {type=MovieSceneScriptingDoubleChannel, default=0, const [0,1]}
      ]
     }
    ]
   }
   {
    type=MovieSceneVisibilityTrack
    sections [{
      type=MovieSceneVisibilitySection
      channels [None {type=MovieSceneScriptingBoolChannel, default=true, keys [[1802,false],[1813,true],[1917,true]]}]
     }]
   }
   {type=MovieSceneSpawnTrack, sections [{type=MovieSceneSpawnSection, channels [None {type=MovieSceneScriptingBoolChannel, default=true}]}]}
   {type=MovieSceneSkeletalAnimationTrack, sections [{type=MovieSceneSkeletalAnimationSection, end_frame=1962, start_frame=1800}]}
   {
    type=MovieSceneControlRigParameterTrack
    display_name=Fortnite_Biped_CtrlRig
    sections [{
      type=MovieSceneControlRigParameterSection
      channels [
       leg_l_stretch_switch {type=MovieSceneScriptingBoolChannel, default=false, const [false,2]}
       leg_r_stretch_switch {type=MovieSceneScriptingBoolChannel, default=false, const [false,2]}
       arm_l_fk_local {type=MovieSceneScriptingBoolChannel, default=true, const [true,2]}
       Auto Interpolate_3 {type=MovieSceneScriptingBoolChannel, default=false, const [false,2]}
       Auto Interpolate {type=MovieSceneScriptingBoolChannel, default=false, const [false,2]}
       arm_l_stretch_switch {type=MovieSceneScriptingBoolChannel, default=false, const [false,2]}
       Keep Aligned {type=MovieSceneScriptingBoolChannel, default=false, const [false,2]}
       arm_r_fk_local {type=MovieSceneScriptingBoolChannel, default=true, const [true,2]}
       Auto Interpolate_4 {type=MovieSceneScriptingBoolChannel, default=false, const [false,2]}
       Auto Interpolate_2 {type=MovieSceneScriptingBoolChannel, default=false, const [false,2]}
       arm_r_stretch_switch {type=MovieSceneScriptingBoolChannel, default=false, const [false,2]}
       Keep Aligned_2 {type=MovieSceneScriptingBoolChannel, default=false, const [false,2]}
       head_local {type=MovieSceneScriptingBoolChannel, default=true, const [true,2]}
       Stretch {type=MovieSceneScriptingBoolChannel, default=false, const [false,2]}
       head_stretch {type=MovieSceneScriptingBoolChannel, default=false, const [false,2]}
       spine_stretch_switch {type=MovieSceneScriptingBoolChannel, default=false, const [false,2]}
       arm_l_fk_ik_switch {type=MovieSceneScriptingBoolChannel, default=false, const [false,2]}
       leg_l_fk_ik_switch {type=MovieSceneScriptingBoolChannel, default=true, const [true,2]}
       arm_r_fk_ik_switch {type=MovieSceneScriptingBoolChannel, default=false, const [false,1]}
       leg_r_fk_ik_switch {type=MovieSceneScriptingBoolChannel, default=true, const [false,1]}
       ShowBodyControls {type=MovieSceneScriptingBoolChannel, default=true, const [true,2]}
       spine_fk_ik_switch {type=MovieSceneScriptingBoolChannel, default=false, const [false,2]}
       neck_fk_ik_switch {type=MovieSceneScriptingBoolChannel, default=false, const [false,2]}
       space {type=MovieSceneScriptingByteChannel, default=1, const [1,2]}
       global_ctrl.Location.X {type=MovieSceneScriptingFloatChannel, default=0, const [0,2]}
       global_ctrl.Location.Y {type=MovieSceneScriptingFloatChannel, default=0, const [0,2]}
       global_ctrl.Location.Z {type=MovieSceneScriptingFloatChannel, default=0, const [0,2]}
       global_ctrl.Rotation.X {type=MovieSceneScriptingFloatChannel, default=0, const [0,2]}
       global_ctrl.Rotation.Y {type=MovieSceneScriptingFloatChannel, default=0, const [0,2]}
       global_ctrl.Rotation.Z {type=MovieSceneScriptingFloatChannel, default=0, const [0,2]}
       root_ctrl.Location.X {
        type=MovieSceneScriptingFloatChannel
        default=0
        thinned_from=146
        keys [[1800,8597.5938],[1819,8597.5938],[1820,8402.8447],[1821,7848.5571],[1822,6979.6748],[1823,5841.1392],[1824,4477.8921],[1825,2934.8772],[1826,1257.0353],[1827,-510.6908],[1828,-2323.3591],[1829,-4243.0332],[1831,-8246.293],[1832,-10197.9941],[1833,-12029.0957],[1834,-13674.1133],[1835,-15067.5635],[1836,-16143.9609],[1837,-16837.8203],[1838,-17083.6602],[1944,-17083.6602],[1962,0]]
       }
       root_ctrl.Location.Y {
        type=MovieSceneScriptingFloatChannel
        default=35.9044
        thinned_from=146
        keys [[1800,-623.8912],[1801,-625.8849],[1802,-630.879],[1805,-649.8969],[1806,-655.0887],[1807,-658.761],[1808,-660.154],[1849,-659.8716],[1854,-656.3411],[1857,-660.154],[1902,-660.1131],[1903,-658.576],[1905,-651.8715],[1906,-650.3499],[1907,-651.8817],[1909,-658.6221],[1910,-660.154],[1944,-659.8947],[1962,0]]
       }
       root_ctrl.Location.Z {
        type=MovieSceneScriptingFloatChannel
        default=0
        thinned_from=146
        keys [[1800,89.1277],[1801,88.9013],[1802,88.2426],[1803,87.1827],[1804,85.7522],[1805,83.9821],[1806,81.9033],[1807,79.5467],[1808,76.943],[1809,74.1233],[1810,71.1183],[1812,64.676],[1815,54.3952],[1816,-82.0019],[1817,-441.6493],[1818,-960.6511],[1819,-1575.111],[1820,-2221.1331],[1821,-2834.8213],[1822,-3352.2793],[1823,-3709.6111],[1824,-3842.9209],[1833,-3842.9209],[1834,-3543.5967],[1835,-2771.6548],[1836,-1716.1428],[1837,-566.1072],[1838,489.4047],[1839,1261.3464],[1840,1560.6707],[1944,1560.6707],[1962,0]]
       }
       root_ctrl.Rotation.X {
        type=MovieSceneScriptingFloatChannel
        default=0
        thinned_from=146
        keys [[1800,92.3974],[1802,92.3717],[1803,93.1711],[1804,94.9296],[1805,97.7846],[1806,101.7826],[1807,106.7985],[1808,112.6308],[1810,125.7431],[1811,130.078],[1813,134.9253],[1814,136.837],[1815,138.1342],[1816,138.6121],[1827,135.2593],[1828,134.1632],[1829,132.1265],[1830,129.5093],[1831,126.3909],[1833,119.3192],[1834,115.4091],[1835,111.0689],[1837,101.6717],[1838,90.1299],[1839,80.9853],[1840,81.5337],[1845,86.7367],[1849,91.3234],[1850,93.5678],[1851,97.7382],[1852,103.2831],[1853,109.6662],[1855,123.095],[1856,129.3491],[1857,134.8241],[1858,139.2481],[1859,142.3854],[1860,143.8115],[1861,143.5091],[1862,141.6293],[1863,137.6102],[1864,132.5214],[1865,126.3707],[1866,119.431],[1867,112.0679],[1869,96.4043],[1870,88.1489],[1871,79.3795],[1873,60.9205],[1874,52.8129],[1875,46.8212],[1876,43.9713],[1877,43.1124],[1878,42.722],[1880,42.8274],[1886,43.9478],[1887,44.3863],[1891,47.5176],[1902,57.797],[1903,58.8408],[1904,60.2062],[1905,62.0202],[1906,64.1954],[1908,68.6994],[1909,70.6189],[1910,72.0884],[1912,74.1804],[1915,76.3459],[1925,80.8897],[1929,83.7146],[1935,89.54],[1938,91.8923],[1942,94.5381],[1943,95.0127],[1944,94.4374],[1962,0]]
       }
       root_ctrl.Rotation.Y {
        type=MovieSceneScriptingFloatChannel
        default=0
        thinned_from=146
        keys [[1800,0.4941],[1801,1.1517],[1802,2.9957],[1803,5.8331],[1804,9.3482],[1807,21.1114],[1808,23.3298],[1809,23.8211],[1810,24.1989],[1811,24.35],[1812,24.0126],[1813,23.2079],[1814,22.2475],[1815,21.4429],[1816,21.1054],[1817,21.2105],[1820,21.8168],[1823,22.6414],[1830,24.976],[1833,26.0148],[1835,26.5892],[1836,26.8239],[1839,27.2623],[1840,27.5265],[1841,27.8587],[1844,28.6066],[1846,28.9677],[1848,29.1874],[1849,29.2405],[1850,29.0334],[1851,28.3492],[1852,27.1955],[1853,25.6171],[1854,23.6921],[1856,19.3183],[1857,17.0963],[1858,15.0253],[1859,13.2556],[1860,12.1481],[1861,11.6617],[1864,10.6166],[1865,10.1833],[1866,9.6163],[1867,8.4784],[1868,6.5526],[1869,4.2328],[1870,-0.6995],[1871,-6.4418],[1872,-11.1532],[1873,-15.4972],[1874,-19.1065],[1875,-21.6138],[1876,-22.6518],[1877,-22.8176],[1878,-22.893],[1886,-22.8677],[1887,-22.8267],[1893,-22.1965],[1898,-21.373],[1903,-20.3734],[1905,-20.0626],[1906,-19.9886],[1907,-19.8548],[1908,-19.5788],[1915,-17.3755],[1918,-16.5574],[1924,-15.1739],[1928,-14.3942],[1930,-14.0609],[1933,-13.6589],[1935,-13.4489],[1937,-13.3261],[1942,-13.3053],[1943,-13.3442],[1944,-11.8762],[1962,0]]
       }
       root_ctrl.Rotation.Z {
        type=MovieSceneScriptingFloatChannel
        default=0
        thinned_from=146
        keys [[1800,-0.9873],[1801,0.2912],[1802,3.9175],[1803,9.7097],[1804,17.076],[1805,26.1038],[1806,36.4217],[1807,47.7213],[1808,59.8444],[1809,72.8608],[1810,85.1992],[1811,93.5748],[1812,98.8035],[1813,103.6083],[1814,107.5652],[1815,110.2502],[1816,111.2394],[1837,110.9604],[1839,111.341],[1840,111.8499],[1844,115.3872],[1849,121.2489],[1850,123.3765],[1851,127.0852],[1852,131.9129],[1853,137.4306],[1856,155.1265],[1857,160.5438],[1858,165.3038],[1859,169.2354],[1860,171.8208],[1866,180.7963],[1867,183.6406],[1868,187.361],[1869,189.2113],[1870,197.3578],[1871,207.6746],[1872,222.1603],[1873,235.6263],[1874,247.053],[1875,255.4208],[1876,259.71],[1905,297.5877],[1934,318.8424],[1943,320.8629],[1944,324.8624],[1962,0]]
       }
       root_ctrl.Scale.X {type=MovieSceneScriptingFloatChannel, default=1, const [1,2]}
       root_ctrl.Scale.Y {type=MovieSceneScriptingFloatChannel, default=1, const [1,2]}
       root_ctrl.Scale.Z {type=MovieSceneScriptingFloatChannel, default=1, const [1,2]}
       body_offset_ctrl.Location.X {type=MovieSceneScriptingFloatChannel, default=0, const [0,2]}
       body_offset_ctrl.Location.Y {
        type=MovieSceneScriptingFloatChannel
        default=35.9044
        thinned_from=14
        keys [[1800,35.9044],[1804,15.8504],[1808,-0.3584],[1848,-0.3584],[1854,3.4545],[1857,-0.3584],[1899,-0.3584],[1902,-0.3175],[1906,9.4458],[1910,-0.3584],[1962,0]]
       }
       body_offset_ctrl.Location.Z {type=MovieSceneScriptingFloatChannel, default=0, const [0,2]}
       body_offset_ctrl.Rotation.X {type=MovieSceneScriptingFloatChannel, default=0, const [0,2]}
       body_offset_ctrl.Rotation.Y {type=MovieSceneScriptingFloatChannel, default=0, const [0,2]}
       body_offset_ctrl.Rotation.Z {type=MovieSceneScriptingFloatChannel, default=0, const [0,2]}
       body_offset_ctrl.Scale.X {type=MovieSceneScriptingFloatChannel, default=1, const [1,2]}
       body_offset_ctrl.Scale.Y {type=MovieSceneScriptingFloatChannel, default=1, const [1,2]}
       body_offset_ctrl.Scale.Z {type=MovieSceneScriptingFloatChannel, default=1, const [1,2]}
       thigh_l_fk_ctrl.Location.X {type=MovieSceneScriptingFloatChannel, default=0, thinned_from=113, keys [[1800,0],[1962,0]]}
       thigh_l_fk_ctrl.Location.Y {type=MovieSceneScriptingFloatChannel, default=0, thinned_from=82, keys [[1800,0],[1962,0]]}
       thigh_l_fk_ctrl.Location.Z {type=MovieSceneScriptingFloatChannel, default=0, thinned_from=64, keys [[1800,0],[1962,0]]}
       thigh_l_fk_ctrl.Rotation.X {
        type=MovieSceneScriptingFloatChannel
        default=-1.3606
        keys [[1800,-1.3606],[1810,-0.9754],[1822,-1.7114],[1825,8.4511],[1835,-3.4343],[1837,-127.1069],[1847,-2.9454],[1853,-2.5741],[1856,-1.3804],[1860,-0.9758],[1872,-1.8788],[1884,-3.4343],[1899,-2.8884],[1902,-2.9207],[1904,-2.7082],[1907,-1.2023],[1912,-0.9756],[1924,-1.7949],[1937,-3.431],[1952,-2.8907],[1961,-3.0443],[1962,0]]
       }
       thigh_l_fk_ctrl.Rotation.Y {
        type=MovieSceneScriptingFloatChannel
        default=-5.6427
        keys [[1800,-5.6427],[1817,-5.1715],[1825,1.326],[1837,-14.2115],[1839,-8.8161],[1852,-7.9845],[1859,-5.3604],[1866,-5.1715],[1888,-8.8161],[1903,-8.1116],[1906,-6.2536],[1910,-5.4006],[1918,-5.1748],[1941,-8.8115],[1955,-8.191],[1961,-8.329],[1962,0]]
       }
       thigh_l_fk_ctrl.Rotation.Z {
        type=MovieSceneScriptingFloatChannel
        default=0.9627
        keys [[1800,0.9627],[1802,-1.3788],[1804,-3.5274],[1808,-6.7342],[1817,-8.0891],[1825,-52.9525],[1835,-1.3128],[1837,7.8726],[1850,-3.5585],[1852,-3.4408],[1859,-7.5242],[1866,-8.0891],[1884,-1.3128],[1899,-3.5685],[1904,-3.1391],[1911,-7.3764],[1918,-8.0888],[1937,-1.327],[1952,-3.5616],[1957,-3.31],[1959,-3.1905],[1960,-3.1474],[1961,-3.1232],[1962,0]]
       }
       thigh_l_fk_ctrl.Scale.X {type=MovieSceneScriptingFloatChannel, default=1, const [1,4]}
       thigh_l_fk_ctrl.Scale.Y {type=MovieSceneScriptingFloatChannel, default=1, const [1,4]}
       thigh_l_fk_ctrl.Scale.Z {type=MovieSceneScriptingFloatChannel, default=1, const [1,4]}
       calf_l_fk_ctrl.Location.X {type=MovieSceneScriptingFloatChannel, default=0, thinned_from=111, keys [[1800,0],[1962,0]]}
       calf_l_fk_ctrl.Location.Y {type=MovieSceneScriptingFloatChannel, default=0, thinned_from=118, keys [[1800,0],[1962,0]]}
       calf_l_fk_ctrl.Location.Z {type=MovieSceneScriptingFloatChannel, default=0, thinned_from=151, keys [[1800,0],[1962,0]]}
       calf_l_fk_ctrl.Rotation.X {
        type=MovieSceneScriptingFloatChannel
        default=-0.1037
        keys [[1800,-0.1037],[1803,0.1443],[1811,0.9113],[1825,-0.0437],[1826,-0.3172],[1837,0.1887],[1841,0.0638],[1851,0.0334],[1854,0.1621],[1855,0.3341],[1860,0.9113],[1875,-0.3172],[1890,0.0638],[1903,0.0187],[1905,0.0892],[1908,0.5769],[1913,0.9034],[1928,-0.3147],[1943,0.0627],[1944,0.0603],[1961,-0.0036],[1962,0]]
       }
       calf_l_fk_ctrl.Rotation.Y {
        type=MovieSceneScriptingFloatChannel
        default=-0.2814
        keys [[1800,-0.2814],[1804,-0.3823],[1809,-0.4104],[1813,-0.3711],[1817,-0.1275],[1825,1.2242],[1826,0.3985],[1837,1.2104],[1841,-0.1191],[1851,-0.0646],[1855,-0.2564],[1860,-0.41],[1875,0.3985],[1890,-0.1191],[1902,-0.0405],[1904,-0.0734],[1906,-0.2907],[1908,-0.4007],[1911,-0.4102],[1919,-0.0877],[1928,0.3952],[1943,-0.1186],[1944,-0.1168],[1962,0]]
       }
       calf_l_fk_ctrl.Rotation.Z {
        type=MovieSceneScriptingFloatChannel
        default=3.5461
        keys [[1800,3.5461],[1811,12.8412],[1825,-4.2092],[1826,2.9504],[1837,6.6972],[1843,7.9209],[1852,5.7223],[1860,12.8412],[1875,2.9504],[1892,7.9209],[1904,5.3095],[1908,9.7879],[1913,12.7786],[1928,2.9751],[1944,7.896],[1957,5.3184],[1961,5.5307],[1962,0]]
       }
       calf_l_fk_ctrl.Scale.X {type=MovieSceneScriptingFloatChannel, default=1, const [1,4]}
       calf_l_fk_ctrl.Scale.Y {type=MovieSceneScriptingFloatChannel, default=1, const [1,4]}
       calf_l_fk_ctrl.Scale.Z {type=MovieSceneScriptingFloatChannel, default=1, const [1,4]}
       foot_l_fk_ctrl.Location.X {type=MovieSceneScriptingFloatChannel, default=0, thinned_from=27, keys [[1800,0],[1962,0]]}
       foot_l_fk_ctrl.Location.Y {
        type=MovieSceneScriptingFloatChannel
        default=0
        thinned_from=119
        keys [[1800,0],[1824,0],[1825,0.0398],[1827,0],[1836,0],[1837,0.0398],[1838,0],[1962,0]]
       }
       foot_l_fk_ctrl.Location.Z {
        type=MovieSceneScriptingFloatChannel
        default=0
        thinned_from=38
        keys [[1800,0],[1817,0],[1825,-0.8885],[1826,0],[1832,0],[1837,-0.8885],[1857,0],[1962,0]]
       }
       foot_l_fk_ctrl.Rotation.X {
        type=MovieSceneScriptingFloatChannel
        default=-5.8803
        thinned_from=21
        keys [[1800,-5.8803],[1803,-2.8881],[1806,-2.4885],[1813,-2.0077],[1825,-8.4993],[1827,-0.8174],[1837,138.1638],[1847,-1.1426],[1849,-1.141],[1858,-2.3154],[1876,-0.8174],[1896,-1.1426],[1901,-1.1354],[1908,-2.4574],[1929,-0.8189],[1949,-1.1426],[1955,-1.1317],[1958,-1.1244],[1961,-1.1207],[1962,0]]
       }
       foot_l_fk_ctrl.Rotation.Y {
        type=MovieSceneScriptingFloatChannel
        default=2.0453
        thinned_from=20
        keys [[1800,2.0453],[1802,-2.596],[1804,-5.5473],[1825,1.4829],[1827,-4.0262],[1837,-39.6553],[1847,-4.988],[1851,-4.9214],[1857,-5.4284],[1876,-4.0262],[1896,-4.988],[1904,-4.7724],[1908,-5.5084],[1929,-4.0306],[1949,-4.9876],[1961,-4.9386],[1962,0]]
       }
       foot_l_fk_ctrl.Rotation.Z {
        type=MovieSceneScriptingFloatChannel
        default=46.8455
        thinned_from=20
        keys [[1800,46.8455],[1802,59.7571],[1804,65.6207],[1825,96.8779],[1827,40.1424],[1837,-44.9347],[1847,56.6776],[1850,56.5452],[1857,62.9472],[1876,40.1424],[1896,56.6776],[1903,56.19],[1908,64.6673],[1929,40.2015],[1949,56.6714],[1961,55.774],[1962,0]]
       }
       foot_l_fk_ctrl.Scale.X {type=MovieSceneScriptingFloatChannel, default=1, const [1,4]}
       foot_l_fk_ctrl.Scale.Y {type=MovieSceneScriptingFloatChannel, default=1, const [1,4]}
       foot_l_fk_ctrl.Scale.Z {type=MovieSceneScriptingFloatChannel, default=1, const [1,4]}
       ball_l_fk_ctrl.Location.X {type=MovieSceneScriptingFloatChannel, default=0, thinned_from=4, keys [[1800,0],[1825,-0.0002],[1962,0]]}
       ball_l_fk_ctrl.Location.Y {type=MovieSceneScriptingFloatChannel, default=0, thinned_from=4, keys [[1800,0],[1825,0.0003],[1962,0]]}
       ball_l_fk_ctrl.Location.Z {type=MovieSceneScriptingFloatChannel, default=0, thinned_from=4, keys [[1800,0],[1962,0]]}
       ball_l_fk_ctrl.Rotation.X {type=MovieSceneScriptingFloatChannel, default=0, thinned_from=4, keys [[1800,0],[1962,0]]}
       ball_l_fk_ctrl.Rotation.Y {type=MovieSceneScriptingFloatChannel, default=0, thinned_from=4, keys [[1800,0],[1962,0]]}
       ball_l_fk_ctrl.Rotation.Z {type=MovieSceneScriptingFloatChannel, default=0, thinned_from=4, keys [[1800,0],[1962,0]]}
       ball_l_fk_ctrl.Scale.X {type=MovieSceneScriptingFloatChannel, default=1, const [1,4]}
       ball_l_fk_ctrl.Scale.Y {type=MovieSceneScriptingFloatChannel, default=1, const [1,4]}
       ball_l_fk_ctrl.Scale.Z {type=MovieSceneScriptingFloatChannel, default=1, const [1,4]}
       foot_l_ik_ctrl.Location.X {
        type=MovieSceneScriptingFloatChannel
        default=-8.9493
        thinned_from=76
        keys [[1800,-8.9492],[1801,-10.0879],[1802,-12.1875],[1803,-14.375],[1804,-16.9492],[1805,-19.3125],[1806,-20.2715],[1807,-19.1953],[1808,-17.1152],[1811,-9.2598],[1812,-8.3711],[1815,-110.7812],[1817,-530.0215],[1818,-741.1992],[1820,-1166.1094],[1825,-5720.7026],[1826,-2456.2666],[1837,-4814.0659],[1839,-5250.4688],[1842,-5974.2134],[1846,-6828.0825],[1849,-7217.0483],[1855,-7721.3706],[1858,-7943.0825],[1860,-8062.5337],[1862,-8163.5059],[1864,-8247.0439],[1869,-8384.0488],[1875,-8495.3828],[1876,-8488.9268],[1885,-8307.9033],[1891,-8084.8232],[1895,-7897.3013],[1903,-7441.5093],[1909,-7016.3965],[1916,-6551.1948],[1929,-5599.7324],[1933,-5316.3887],[1937,-5046.2896],[1941,-4794.8027],[1943,-4677.6069],[1944,-18446.4551],[1962,-18435.6914]]
       }
       foot_l_ik_ctrl.Location.Y {
        type=MovieSceneScriptingFloatChannel
        default=-63.7971
        thinned_from=76
        keys [[1800,-63.7969],[1801,-61.6055],[1802,-56.168],[1804,-41.9971],[1805,-36.3262],[1806,-33.0928],[1807,-32.2598],[1808,-32.6865],[1811,-31.1543],[1812,-29.0049],[1815,-21.7363],[1817,9.3213],[1818,33.6064],[1820,93.5732],[1825,-1601.0342],[1826,285.457],[1827,308.1113],[1836,138.7638],[1837,56.5265],[1839,-157.0612],[1841,-431.0022],[1842,-575.5255],[1846,-1133.1553],[1849,-1443.7183],[1850,-1518.647],[1851,-1589.002],[1854,-1783.9482],[1855,-1845.8872],[1857,-1959.2031],[1858,-2010.1489],[1864,-2286.502],[1866,-2392.8711],[1867,-2454.3789],[1869,-2723.8359],[1875,-4822.5312],[1876,-5057.7734],[1885,-5537.4766],[1888,-5637.8379],[1891,-5721.0137],[1895,-5807.4023],[1898,-5855.2578],[1905,-5920.582],[1919,-5901.5195],[1933,-5824.9053],[1937,-5820.4941],[1943,-5849.6079],[1944,-23195.4062],[1961,-23197.0547],[1962,-23117.334]]
       }
       foot_l_ik_ctrl.Location.Z {
        type=MovieSceneScriptingFloatChannel
        default=89.0205
        thinned_from=76
        keys [[1800,89.0215],[1801,90.4609],[1802,90.4004],[1803,89.9219],[1804,89.0293],[1806,86.7227],[1807,86.2461],[1808,86.5762],[1811,93.8301],[1812,97.0957],[1815,3.3301],[1817,-433.1328],[1818,-669.0684],[1820,-1164.7129],[1825,-2501.1086],[1826,-2693.6865],[1827,-2934.3281],[1836,-4450.02],[1837,-4511.0112],[1839,-4563.9355],[1842,-4560.2129],[1846,-4581.9209],[1849,-4679.8857],[1851,-4797.3174],[1853,-4948.1504],[1855,-5135.7578],[1860,-5666.3701],[1862,-5865.9941],[1864,-6052.2661],[1866,-6215.1108],[1867,-6286.9521],[1869,-6393.3833],[1875,-6367.7217],[1876,-6317.3174],[1885,-5695.0029],[1891,-5206.9512],[1895,-4853.416],[1898,-4574.1289],[1903,-4081.2109],[1908,-3544.793],[1913,-3027.0215],[1916,-2702.3184],[1922,-2024.2754],[1929,-1212.832],[1933,-746.3242],[1937,-282.2266],[1941,176.1641],[1943,402.3516],[1944,1560.7969],[1961,1559.4062],[1962,1476.6484]]
       }
       foot_l_ik_ctrl.Rotation.X {
        type=MovieSceneScriptingFloatChannel
        default=143.1197
        thinned_from=76
        keys [[1800,143.1197],[1801,147.7076],[1802,156.4818],[1803,162.0676],[1804,163.961],[1805,166.7937],[1806,170.7401],[1807,175.8054],[1808,181.6519],[1811,197.6176],[1812,200.4592],[1815,199.0677],[1820,187.1098],[1825,179.5181],[1827,178.4638],[1836,158.8616],[1837,155.0803],[1839,135.3135],[1842,140.8227],[1846,146.0503],[1849,147.3569],[1850,148.9644],[1851,152.9369],[1853,168.4904],[1855,188.7507],[1857,204.4286],[1858,208.7528],[1860,211.3792],[1862,205.6351],[1864,192.5406],[1866,174.8126],[1869,145.7432],[1875,93.99],[1876,91.2773],[1885,102.3334],[1888,108.6235],[1890,112.5367],[1891,114.2956],[1895,119.1442],[1901,121.8283],[1903,123.3767],[1905,128.7885],[1907,136.9489],[1908,140.3625],[1909,142.3749],[1913,142.2879],[1916,138.3848],[1919,133.2498],[1921,130.061],[1922,128.7428],[1927,127.6489],[1929,130.3403],[1933,138.43],[1937,147.9723],[1941,156.4202],[1943,159.4608],[1944,151.8754],[1947,153.3808],[1951,152.7046],[1959,150.4361],[1961,150.3901],[1962,0]]
       }
       foot_l_ik_ctrl.Rotation.Y {
        type=MovieSceneScriptingFloatChannel
        default=-1.3876
        thinned_from=76
        keys [[1800,-1.3876],[1801,2.8257],[1802,8.0612],[1803,13.4775],[1804,18.335],[1806,26.6764],[1807,30.3091],[1808,32.9396],[1811,33.2046],[1812,31.66],[1815,28.6829],[1817,28.3025],[1818,28.2039],[1820,28.1993],[1825,29.3849],[1826,29.7752],[1827,30.2539],[1836,32.9914],[1837,32.7689],[1839,30.5455],[1842,32.3126],[1846,33.9626],[1849,34.422],[1850,34.4136],[1851,34.1884],[1853,33.2221],[1854,32.207],[1855,30.6536],[1857,26.2113],[1858,23.7593],[1860,20.2453],[1862,19.3407],[1864,18.6053],[1866,16.9705],[1867,15.5829],[1869,9.225],[1875,-23.8481],[1876,-25.0956],[1885,-25.3439],[1888,-24.344],[1891,-23.081],[1895,-21.4866],[1898,-20.5908],[1901,-19.7573],[1902,-19.4334],[1903,-19.067],[1905,-16.6017],[1906,-15.1432],[1907,-13.9955],[1908,-13.0357],[1909,-12.3768],[1913,-11.2588],[1916,-11.3022],[1919,-11.7418],[1922,-12.2625],[1927,-12.4861],[1929,-12.1105],[1933,-10.703],[1937,-9.0599],[1941,-7.7165],[1943,-7.293],[1944,6.0239],[1947,6.4098],[1948,6.4389],[1951,6.3616],[1955,6.1484],[1958,6.0256],[1960,5.9831],[1961,5.9744],[1962,0]]
       }
       foot_l_ik_ctrl.Rotation.Z {
        type=MovieSceneScriptingFloatChannel
        default=-8.4456
        thinned_from=76
        keys [[1800,-8.4456],[1801,-7.14],[1802,-2.6264],[1803,4.1605],[1804,12.2729],[1805,22.034],[1806,33.2772],[1807,45.7875],[1808,59.1491],[1811,96.2537],[1812,104.8957],[1815,113.7975],[1825,108.9056],[1827,108.2788],[1836,103.6556],[1839,101.5375],[1841,102.9663],[1842,103.9226],[1846,108.8378],[1849,112.9149],[1850,115.283],[1851,119.5083],[1853,132.3643],[1854,140.2489],[1855,148.4178],[1857,163.4113],[1858,169.2737],[1860,176.4149],[1862,178.342],[1864,179.3553],[1867,180.0633],[1869,185.0533],[1875,246.6021],[1876,252.7808],[1885,-96.3141],[1891,-88.9056],[1898,-79.5424],[1903,-73.5029],[1905,-71.4235],[1913,-61.3213],[1916,-58.0913],[1919,-55.42],[1921,-53.9666],[1922,-53.3346],[1927,-50.9654],[1929,-50.2747],[1941,-46.6184],[1943,-46.1354],[1944,-8.4751],[1947,-8.0796],[1948,-7.9745],[1951,-7.8198],[1953,-7.8442],[1959,-8.0301],[1961,-8.0545],[1962,0]]
       }
       foot_l_ik_ctrl.Scale.X {type=MovieSceneScriptingFloatChannel, default=1, thinned_from=76, keys [[1800,1],[1962,1]]}
       foot_l_ik_ctrl.Scale.Y {type=MovieSceneScriptingFloatChannel, default=1, thinned_from=76, keys [[1800,1],[1962,1]]}
       foot_l_ik_ctrl.Scale.Z {type=MovieSceneScriptingFloatChannel, default=1, thinned_from=76, keys [[1800,1],[1962,1]]}
       ball_l_ik_ctrl.Location.X {type=MovieSceneScriptingFloatChannel, default=0, thinned_from=4, keys [[1800,0],[1962,0]]}
       ball_l_ik_ctrl.Location.Y {type=MovieSceneScriptingFloatChannel, default=0, thinned_from=4, keys [[1800,0],[1837,0.0002],[1962,0]]}
       ball_l_ik_ctrl.Location.Z {type=MovieSceneScriptingFloatChannel, default=0, thinned_from=4, keys [[1800,0],[1837,-0.0002],[1962,0]]}
       ball_l_ik_ctrl.Rotation.X {type=MovieSceneScriptingFloatChannel, default=0, thinned_from=4, keys [[1800,0],[1962,0]]}
       ball_l_ik_ctrl.Rotation.Y {type=MovieSceneScriptingFloatChannel, default=0, thinned_from=4, keys [[1800,0],[1962,0]]}
       ball_l_ik_ctrl.Rotation.Z {type=MovieSceneScriptingFloatChannel, default=0, thinned_from=4, keys [[1800,0],[1962,0]]}
       ball_l_ik_ctrl.Scale.X {type=MovieSceneScriptingFloatChannel, default=1, const [1,4]}
       ball_l_ik_ctrl.Scale.Y {type=MovieSceneScriptingFloatChannel, default=1, const [1,4]}
       ball_l_ik_ctrl.Scale.Z {type=MovieSceneScriptingFloatChannel, default=1, const [1,4]}
       foot_bk1_l_ctrl.Location.X {type=MovieSceneScriptingFloatChannel, default=0, const [0,2]}
       foot_bk1_l_ctrl.Location.Y {type=MovieSceneScriptingFloatChannel, default=0, const [0,2]}
       foot_bk1_l_ctrl.Location.Z {type=MovieSceneScriptingFloatChannel, default=0, const [0,2]}
       foot_bk1_l_ctrl.Rotation.X {type=MovieSceneScriptingFloatChannel, default=0, const [0,2]}
       foot_bk1_l_ctrl.Rotation.Y {type=MovieSceneScriptingFloatChannel, default=0, const [0,2]}
       foot_bk1_l_ctrl.Rotation.Z {type=MovieSceneScriptingFloatChannel, default=0, const [0,2]}
       foot_bk1_l_ctrl.Scale.X {type=MovieSceneScriptingFloatChannel, default=1, const [1,2]}
       foot_bk1_l_ctrl.Scale.Y {type=MovieSceneScriptingFloatChannel, default=1, const [1,2]}
       foot_bk1_l_ctrl.Scale.Z {type=MovieSceneScriptingFloatChannel, default=1, const [1,2]}
       foot_roll_l_ctrl.Location.X {type=MovieSceneScriptingFloatChannel, default=0, const [0,2]}
       foot_roll_l_ctrl.Location.Y {type=MovieSceneScriptingFloatChannel, default=0, const [0,2]}
       foot_roll_l_ctrl.Location.Z {type=MovieSceneScriptingFloatChannel, default=0, const [0,2]}
       foot_roll_l_ctrl.Rotation.X {type=MovieSceneScriptingFloatChannel, default=0, const [0,2]}
       foot_roll_l_ctrl.Rotation.Y {type=MovieSceneScriptingFloatChannel, default=0, const [0,2]}
       foot_roll_l_ctrl.Rotation.Z {type=MovieSceneScriptingFloatChannel, default=0, const [0,2]}
       foot_roll_l_ctrl.Scale.X {type=MovieSceneScriptingFloatChannel, default=1, const [1,2]}
       foot_roll_l_ctrl.Scale.Y {type=MovieSceneScriptingFloatChannel, default=1, const [1,2]}
       foot_roll_l_ctrl.Scale.Z {type=MovieSceneScriptingFloatChannel, default=1, const [1,2]}
       roll_blend_l_ctrl {type=MovieSceneScriptingFloatChannel, default=1, const [1,2]}
       tip_l_ctrl.Location.X {type=MovieSceneScriptingFloatChannel, default=0, const [0,2]}
       tip_l_ctrl.Location.Y {type=MovieSceneScriptingFloatChannel, default=0, const [0,2]}
       tip_l_ctrl.Location.Z {type=MovieSceneScriptingFloatChannel, default=0, const [0,2]}
       tip_l_ctrl.Rotation.X {type=MovieSceneScriptingFloatChannel, default=0, const [0,2]}
       tip_l_ctrl.Rotation.Y {type=MovieSceneScriptingFloatChannel, default=0, const [0,2]}
       tip_l_ctrl.Rotation.Z {type=MovieSceneScriptingFloatChannel, default=0, const [0,2]}
       tip_l_ctrl.Scale.X {type=MovieSceneScriptingFloatChannel, default=1, const [1,2]}
       tip_l_ctrl.Scale.Y {type=MovieSceneScriptingFloatChannel, default=1, const [1,2]}
       tip_l_ctrl.Scale.Z {type=MovieSceneScriptingFloatChannel, default=1, const [1,2]}
       heel_l_ctrl.Location.X {type=MovieSceneScriptingFloatChannel, default=0, const [0,2]}
       heel_l_ctrl.Location.Y {type=MovieSceneScriptingFloatChannel, default=0, const [0,2]}
       heel_l_ctrl.Location.Z {type=MovieSceneScriptingFloatChannel, default=0, const [0,2]}
       heel_l_ctrl.Rotation.X {type=MovieSceneScriptingFloatChannel, default=0, const [0,2]}
       heel_l_ctrl.Rotation.Y {type=MovieSceneScriptingFloatChannel, default=0, const [0,2]}
       heel_l_ctrl.Rotation.Z {type=MovieSceneScriptingFloatChannel, default=0, const [0,2]}
       heel_l_ctrl.Scale.X {type=MovieSceneScriptingFloatChannel, default=1, const [1,2]}
       heel_l_ctrl.Scale.Y {type=MovieSceneScriptingFloatChannel, default=1, const [1,2]}
       heel_l_ctrl.Scale.Z {type=MovieSceneScriptingFloatChannel, default=1, const [1,2]}
       leg_l_softness {type=MovieSceneScriptingFloatChannel, default=0, const [0,2]}
       leg_l_pv_ik_ctrl.Location.X {
        type=MovieSceneScriptingFloatChannel
        default=0.953
        thinned_from=53
        keys [[1800,0.9531],[1801,-1.0898],[1802,-4.3574],[1803,-6.9844],[1804,-7.8008],[1805,-8.0488],[1809,-7.9336],[1811,-8.5],[1812,-12.3125],[1825,-4.4132],[1826,-4.3408],[1827,-4.376],[1837,-868.8515],[1839,-6.2183],[1843,-7.7754],[1846,-6.6396],[1850,-6.7949],[1851,-6.7666],[1852,-6.7197],[1853,-7.0576],[1856,-7.853],[1858,-8.5498],[1859,-8.6587],[1860,-8.5874],[1861,-8.5601],[1875,-0.6113],[1876,-4.375],[1888,-5.7461],[1892,-6.6172],[1895,-7.1641],[1902,-6.5801],[1903,-6.5312],[1904,-6.502],[1906,-7.5449],[1907,-8.041],[1908,-8.2598],[1910,-8.5039],[1911,-8.418],[1913,-8.4785],[1914,-8.3789],[1927,-4.8574],[1928,-4.6914],[1941,-5.6367],[1944,1892.6406],[1947,189.5742],[1948,-7.0234],[1955,-6.7656],[1958,-6.6797],[1961,-6.6758],[1962,0]]
       }
       leg_l_pv_ik_ctrl.Location.Y {
        type=MovieSceneScriptingFloatChannel
        default=1.3105
        thinned_from=53
        keys [[1800,1.3105],[1801,1.375],[1802,1.5527],[1803,1.8213],[1804,2.1494],[1805,2.5166],[1809,4.5273],[1811,4.1035],[1812,3.7109],[1825,-1.8628],[1826,-2.4951],[1827,1.0132],[1837,-617.4167],[1839,2.3604],[1843,2.1687],[1846,2.0068],[1850,1.958],[1851,1.8691],[1852,1.4712],[1853,1.7866],[1856,3.165],[1858,3.7441],[1859,3.9463],[1860,4.1035],[1861,4.4277],[1875,4.3809],[1876,1.0137],[1888,1.8281],[1892,2.4609],[1895,2.4668],[1902,1.6875],[1903,1.666],[1904,1.6133],[1906,2.1914],[1907,2.5215],[1908,2.8809],[1910,3.707],[1911,4.1074],[1913,4.0957],[1914,3.8418],[1927,1.1885],[1928,1.1211],[1941,3.0771],[1944,-744.498],[1947,-86.8516],[1948,2.3721],[1955,1.709],[1957,1.6553],[1958,1.6602],[1960,1.7197],[1961,1.7305],[1962,0.002]]
       }
       leg_l_pv_ik_ctrl.Location.Z {
        type=MovieSceneScriptingFloatChannel
        default=1.4261
        thinned_from=53
        keys [[1800,1.4258],[1801,1.3906],[1802,1.4727],[1803,2.0645],[1804,2.0098],[1805,2.4004],[1809,2.8555],[1811,4.2578],[1812,7.6992],[1825,-1.9781],[1826,-0.6504],[1827,0.7227],[1837,1086.2396],[1839,2.1744],[1843,4.1455],[1846,2.5869],[1850,1.6699],[1851,1.5908],[1852,1.207],[1853,1.6777],[1856,3.1299],[1858,3.8369],[1859,3.374],[1860,4.2578],[1861,4.6562],[1875,-3.0439],[1876,0.7217],[1888,3.5869],[1892,2.2842],[1895,2.1426],[1902,1.498],[1903,1.4424],[1904,1.2988],[1906,2.0488],[1907,2.5781],[1908,3.0342],[1910,3.8672],[1911,4.2144],[1913,4.2407],[1914,4.0947],[1927,0.988],[1928,1.0259],[1941,-0.073],[1944,-19828.7402],[1947,-2302.2825],[1948,2.1835],[1955,1.476],[1957,1.4175],[1958,1.4246],[1960,1.4844],[1961,1.4955],[1962,-0.002]]
       }
       leg_l_pv_ik_ctrl.Rotation.X {
        type=MovieSceneScriptingFloatChannel
        default=-1.9605
        thinned_from=53
        keys [[1800,-1.9611],[1801,-1.9283],[1802,-2.0367],[1803,-2.4181],[1804,-2.7584],[1805,-3.2822],[1809,-4.8495],[1811,-5.7723],[1812,-5.0645],[1825,2.7124],[1826,-6.1728],[1827,-1.067],[1837,-34.7487],[1839,-3.0358],[1843,-5.2983],[1846,-3.719],[1850,-2.3586],[1851,-2.2532],[1852,-1.7124],[1853,-2.3666],[1856,-4.248],[1858,-5.2165],[1859,-4.9081],[1860,-5.7724],[1861,-6.2869],[1875,1.4465],[1876,-1.0667],[1888,-3.9451],[1892,-3.1421],[1895,-3.0485],[1902,-2.1231],[1903,-2.0499],[1904,-1.853],[1906,-2.8657],[1907,-3.5771],[1908,-4.1795],[1911,-5.8064],[1913,-5.7491],[1914,-5.4493],[1927,-0.8839],[1941,-2.2099],[1944,-3.9567],[1947,2.0949],[1948,-3.0491],[1955,-2.0977],[1957,-2.02],[1958,-2.0297],[1960,-2.1091],[1961,-2.1263],[1962,-0.0002]]
       }
       leg_l_pv_ik_ctrl.Rotation.Y {
        type=MovieSceneScriptingFloatChannel
        default=0.2077
        thinned_from=53
        keys [[1800,0.2077],[1801,0.1093],[1802,-0.0893],[1803,0.1646],[1804,-0.2852],[1805,-0.3065],[1809,-1.432],[1811,-0.2733],[1812,2.3364],[1825,-1.5058],[1826,0.9247],[1827,-0.4702],[1837,30.8052],[1839,-0.3359],[1843,-1.1122],[1846,-0.7886],[1850,-0.3962],[1851,-0.3934],[1852,-0.2487],[1853,-0.3735],[1856,0.0037],[1858,-0.294],[1859,-0.3867],[1860,-0.274],[1861,-0.3335],[1875,-3.0132],[1876,-0.4706],[1888,-0.9221],[1892,0.1581],[1895,-0.2124],[1902,-0.4071],[1903,-0.3987],[1904,-0.3793],[1906,-0.3345],[1907,-0.3235],[1908,-0.3118],[1910,-0.3111],[1911,-0.389],[1913,-0.2727],[1914,-0.2179],[1927,0.0392],[1928,-0.0812],[1941,-1.411],[1944,62.5479],[1947,33.9084],[1948,-0.3601],[1955,-0.402],[1957,-0.4112],[1959,-0.4181],[1960,-0.42],[1961,-0.4199],[1962,0.0006]]
       }
       leg_l_pv_ik_ctrl.Rotation.Z {
        type=MovieSceneScriptingFloatChannel
        default=-1.1042
        thinned_from=53
        keys [[1800,-1.1042],[1801,1.5814],[1802,5.8489],[1803,8.7784],[1804,10.2072],[1805,10.4145],[1809,10.562],[1811,10.7254],[1812,10.6481],[1825,4.8462],[1826,5.4154],[1827,5.5909],[1837,9.8192],[1839,7.9874],[1843,8.7571],[1846,9.0184],[1850,8.8149],[1851,8.7876],[1852,8.9319],[1853,9.1972],[1856,10.4474],[1858,10.8536],[1859,10.8557],[1860,10.6731],[1861,10.4477],[1875,4.6077],[1876,5.5909],[1888,7.9464],[1892,8.8445],[1895,9.0586],[1902,8.5736],[1903,8.5235],[1904,8.5092],[1906,9.7944],[1907,10.3953],[1908,10.6353],[1910,10.7352],[1911,10.6357],[1913,10.5014],[1914,10.3198],[1927,5.6724],[1928,5.6319],[1941,8.0587],[1944,12.2291],[1947,12.6323],[1948,9.0427],[1955,8.7533],[1958,8.6856],[1961,8.6692],[1962,0]]
       }
       leg_l_pv_ik_ctrl.Scale.X {
        type=MovieSceneScriptingFloatChannel
        default=1
        thinned_from=53
        keys [[1800,1],[1802,1],[1803,1.0024],[1804,1],[1809,1.0047],[1811,1],[1812,0.9843],[1825,1],[1941,1],[1944,3.683],[1947,1.3108],[1948,1],[1962,1]]
       }
       leg_l_pv_ik_ctrl.Scale.Y {
        type=MovieSceneScriptingFloatChannel
        default=1
        thinned_from=53
        keys [[1800,1],[1802,1],[1803,1.0024],[1804,1],[1809,1.0047],[1811,1],[1812,0.9843],[1825,1],[1941,1],[1944,3.683],[1947,1.3108],[1948,1],[1962,1]]
       }
       leg_l_pv_ik_ctrl.Scale.Z {
        type=MovieSceneScriptingFloatChannel
        default=1
        thinned_from=53
        keys [[1800,1],[1802,1],[1803,1.0024],[1804,1],[1809,1.0047],[1811,1],[1812,0.9843],[1825,1],[1941,1],[1944,3.683],[1947,1.3108],[1948,1],[1962,1]]
       }
       thigh_r_fk_ctrl.Location.X {type=MovieSceneScriptingFloatChannel, default=0, thinned_from=142, keys [[1800,0],[1962,0]]}
       thigh_r_fk_ctrl.Location.Y {type=MovieSceneScriptingFloatChannel, default=0, thinned_from=140, keys [[1800,0],[1962,0]]}
       thigh_r_fk_ctrl.Location.Z {type=MovieSceneScriptingFloatChannel, default=0, thinned_from=144, keys [[1800,0],[1962,0]]}
       thigh_r_fk_ctrl.Rotation.X {
        type=MovieSceneScriptingFloatChannel
        default=17.8861
        thinned_from=18
        keys [[1800,17.8861],[1808,15.5209],[1821,17.5387],[1825,0],[1840,15.9926],[1853,16.46],[1858,15.5739],[1870,17.5387],[1889,15.9926],[1905,16.61],[1910,15.5379],[1923,17.5329],[1942,15.9955],[1956,16.4247],[1961,16.3084],[1962,0]]
       }
       thigh_r_fk_ctrl.Rotation.Y {
        type=MovieSceneScriptingFloatChannel
        default=-1.9598
        keys [[1800,-1.9598],[1807,-1.2855],[1815,-0.814],[1825,0],[1837,-1.0377],[1850,-0.9023],[1856,-1.2175],[1864,-0.814],[1886,-1.0377],[1902,-0.8875],[1907,-1.3992],[1912,-0.9844],[1917,-0.8162],[1938,-1.0375],[1954,-0.887],[1961,-0.9446],[1962,0]]
       }
       thigh_r_fk_ctrl.Rotation.Z {
        type=MovieSceneScriptingFloatChannel
        default=-0.2294
        keys [[1800,-0.2294],[1808,0.6338],[1821,-9.5851],[1825,0],[1840,0.7246],[1852,-2.4964],[1858,0.3808],[1862,-2.7378],[1870,-9.5851],[1889,0.7246],[1903,-3.004],[1906,-0.3619],[1909,0.6335],[1923,-9.544],[1942,0.6985],[1956,-3.1334],[1959,-2.9819],[1960,-2.8857],[1961,-2.8232],[1962,0]]
       }
       thigh_r_fk_ctrl.Scale.X {type=MovieSceneScriptingFloatChannel, default=1, const [1,3]}
       thigh_r_fk_ctrl.Scale.Y {type=MovieSceneScriptingFloatChannel, default=1, const [1,3]}
       thigh_r_fk_ctrl.Scale.Z {type=MovieSceneScriptingFloatChannel, default=1, const [1,3]}
       calf_r_fk_ctrl.Location.X {type=MovieSceneScriptingFloatChannel, default=0, thinned_from=82, keys [[1800,0],[1962,0]]}
       calf_r_fk_ctrl.Location.Y {type=MovieSceneScriptingFloatChannel, default=0, thinned_from=99, keys [[1800,0],[1962,0]]}
       calf_r_fk_ctrl.Location.Z {type=MovieSceneScriptingFloatChannel, default=0, thinned_from=120, keys [[1800,0],[1962,0]]}
       calf_r_fk_ctrl.Rotation.X {
        type=MovieSceneScriptingFloatChannel
        default=0.2767
        keys [[1800,0.2767],[1802,0.3678],[1805,0.448],[1810,0.4804],[1814,0.4633],[1819,0.2954],[1825,0],[1829,-0.0164],[1843,0.035],[1850,0.0181],[1853,0.0841],[1854,0.1631],[1856,0.3703],[1859,0.4785],[1863,0.4633],[1868,0.2954],[1878,-0.0164],[1892,0.035],[1902,0.0046],[1904,0.0537],[1906,0.319],[1908,0.4617],[1911,0.4801],[1915,0.4678],[1920,0.3169],[1930,-0.0162],[1945,0.0348],[1961,-0.0135],[1962,0]]
       }
       calf_r_fk_ctrl.Rotation.Y {
        type=MovieSceneScriptingFloatChannel
        default=-0.6756
        keys [[1800,-0.6756],[1806,2.1405],[1813,4.2597],[1825,0],[1829,-2.0791],[1844,0.5781],[1852,0.3137],[1857,2.8165],[1862,4.2597],[1878,-2.0791],[1893,0.5781],[1904,0.1676],[1909,2.8389],[1914,4.2306],[1931,-2.0624],[1946,0.5749],[1961,-0.0063],[1962,0]]
       }
       calf_r_fk_ctrl.Rotation.Z {
        type=MovieSceneScriptingFloatChannel
        default=11.2582
        keys [[1800,11.2582],[1817,17.006],[1825,0],[1832,10.2115],[1847,11.6734],[1851,11.5656],[1866,17.006],[1881,10.2115],[1896,11.6734],[1903,11.3968],[1908,13.8039],[1918,16.985],[1934,10.2205],[1949,11.6697],[1955,11.415],[1958,11.2501],[1960,11.1773],[1961,11.1607],[1962,0]]
       }
       calf_r_fk_ctrl.Scale.X {type=MovieSceneScriptingFloatChannel, default=1, const [1,3]}
       calf_r_fk_ctrl.Scale.Y {type=MovieSceneScriptingFloatChannel, default=1, const [1,3]}
       calf_r_fk_ctrl.Scale.Z {type=MovieSceneScriptingFloatChannel, default=1, const [1,3]}
       foot_r_fk_ctrl.Location.X {type=MovieSceneScriptingFloatChannel, default=0, thinned_from=104, keys [[1800,0],[1962,0]]}
       foot_r_fk_ctrl.Location.Y {type=MovieSceneScriptingFloatChannel, default=0, thinned_from=140, keys [[1800,0],[1962,0]]}
       foot_r_fk_ctrl.Location.Z {type=MovieSceneScriptingFloatChannel, default=0, thinned_from=112, keys [[1800,0],[1962,0]]}
       foot_r_fk_ctrl.Rotation.X {
        type=MovieSceneScriptingFloatChannel
        default=-1.1619
        keys [[1800,-1.1619],[1802,0.2686],[1804,1.3464],[1808,2.2566],[1825,0],[1831,-2.5243],[1844,-0.5644],[1851,0.3868],[1854,0.884],[1858,2.1783],[1880,-2.5243],[1901,0.4052],[1903,0.3674],[1907,1.7522],[1910,2.2427],[1933,-2.5147],[1954,0.4009],[1962,0]]
       }
       foot_r_fk_ctrl.Rotation.Y {
        type=MovieSceneScriptingFloatChannel
        default=5.8737
        thinned_from=19
        keys [[1800,5.8737],[1801,7.0988],[1804,10.2306],[1808,12.0734],[1825,0],[1831,7.4589],[1851,10.7985],[1852,10.7593],[1858,12.0182],[1880,7.4589],[1901,10.8847],[1905,10.397],[1910,12.0594],[1932,7.47],[1954,10.8809],[1961,10.7215],[1962,0]]
       }
       foot_r_fk_ctrl.Rotation.Z {
        type=MovieSceneScriptingFloatChannel
        default=50.7477
        thinned_from=17
        keys [[1800,50.7477],[1808,56.4929],[1825,0],[1831,30.4057],[1845,42.561],[1851,46.8693],[1858,56.0991],[1880,30.4057],[1901,46.7976],[1902,46.7624],[1909,56.4883],[1933,30.4576],[1954,46.7771],[1960,45.9742],[1961,45.9008],[1962,0]]
       }
       foot_r_fk_ctrl.Scale.X {type=MovieSceneScriptingFloatChannel, default=1, const [1,3]}
       foot_r_fk_ctrl.Scale.Y {type=MovieSceneScriptingFloatChannel, default=1, const [1,3]}
       foot_r_fk_ctrl.Scale.Z {type=MovieSceneScriptingFloatChannel, default=1, const [1,3]}
       ball_r_fk_ctrl.Location.X {type=MovieSceneScriptingFloatChannel, default=0, const [0,3]}
       ball_r_fk_ctrl.Location.Y {type=MovieSceneScriptingFloatChannel, default=0, const [0,3]}
       ball_r_fk_ctrl.Location.Z {type=MovieSceneScriptingFloatChannel, default=0, const [0,3]}
       ball_r_fk_ctrl.Rotation.X {type=MovieSceneScriptingFloatChannel, default=0, const [0,3]}
       ball_r_fk_ctrl.Rotation.Y {type=MovieSceneScriptingFloatChannel, default=0, const [0,3]}
       ball_r_fk_ctrl.Rotation.Z {type=MovieSceneScriptingFloatChannel, default=0, const [0,3]}
       ball_r_fk_ctrl.Scale.X {type=MovieSceneScriptingFloatChannel, default=1, const [1,3]}
       ball_r_fk_ctrl.Scale.Y {type=MovieSceneScriptingFloatChannel, default=1, const [1,3]}
       ball_r_fk_ctrl.Scale.Z {type=MovieSceneScriptingFloatChannel, default=1, const [1,3]}
       foot_r_ik_ctrl.Location.X {
        type=MovieSceneScriptingFloatChannel
        default=-5.8997
        thinned_from=77
        keys [[1800,-5.9004],[1801,-4.4785],[1802,-1.6484],[1804,4.6211],[1805,7.1934],[1806,7.9434],[1808,2.1699],[1809,-2.7559],[1810,-8.5117],[1812,-17.373],[1817,496.8379],[1818,707.4766],[1822,1559.3047],[1829,3068.4346],[1838,4989.1953],[1842,5937.7173],[1849,7178.9756],[1856,7756.5537],[1859,7955.8516],[1862,8112.6445],[1866,8261.1523],[1871,8398.1797],[1874,8459.5195],[1878,8435.0605],[1887,8215.3301],[1891,8059.4697],[1899,7658.5488],[1903,7418.0508],[1920,6242.645],[1931,5436.1782],[1940,4834.4087],[1943,4656.0225],[1944,18430.168],[1962,18435.6914]]
       }
       foot_r_ik_ctrl.Location.Y {
        type=MovieSceneScriptingFloatChannel
        default=-62.6941
        thinned_from=77
        keys [[1800,-62.6934],[1801,-60.7559],[1802,-56.127],[1804,-45.46],[1805,-42.4258],[1806,-42.4336],[1808,-48.8242],[1809,-52.2285],[1810,-53.9102],[1812,-52.3545],[1817,-10.9727],[1818,13.9238],[1822,144.1328],[1825,243.1597],[1827,293.5913],[1828,311.3433],[1829,322.6763],[1838,-54.0633],[1841,-444.9714],[1842,-589.536],[1849,-1457.0752],[1850,-1531.7856],[1851,-1601.8506],[1854,-1796.958],[1856,-1919.105],[1857,-1974.1323],[1859,-2071.8369],[1862,-2204.9346],[1866,-2396.2148],[1867,-2456.0977],[1871,-3338.9512],[1874,-4488.1367],[1876,-5043.9648],[1877,-5163.4062],[1878,-5216.5547],[1887,-5595.0801],[1891,-5710.5195],[1899,-5857.248],[1905,-5910.8555],[1920,-5883.4482],[1931,-5822.894],[1941,-5825.4238],[1943,-5840.7739],[1944,-23194.5371],[1961,-23196.4727],[1962,-23117.334]]
       }
       foot_r_ik_ctrl.Location.Z {
        type=MovieSceneScriptingFloatChannel
        default=91.4909
        thinned_from=77
        keys [[1800,91.4902],[1801,94.6621],[1802,93.7676],[1804,92.709],[1805,91.2676],[1806,89.6348],[1808,88.0547],[1809,88.7148],[1810,90.7871],[1812,96.877],[1817,-436.918],[1818,-673.4492],[1822,-1682.9902],[1825,-2451.1758],[1827,-2939.7129],[1828,-3170.8721],[1829,-3390.7051],[1838,-4550.5879],[1841,-4564.3184],[1842,-4562.0654],[1849,-4682.9307],[1851,-4800.6201],[1853,-4950.1562],[1854,-5038.043],[1859,-5561.0996],[1862,-5862.6577],[1866,-6213.0044],[1867,-6285.1465],[1871,-6431.7466],[1874,-6397.6904],[1876,-6309.5508],[1878,-6185.4854],[1887,-5529.3359],[1891,-5197.4082],[1899,-4469.5293],[1903,-4072.2871],[1908,-3528.9824],[1914,-2901.7949],[1920,-2241.8184],[1924,-1786.7383],[1940,69.8008],[1941,184.0234],[1943,410.6992],[1944,1566.0312],[1961,1562.6523],[1962,1476.6484]]
       }
       foot_r_ik_ctrl.Rotation.X {
        type=MovieSceneScriptingFloatChannel
        default=152.2009
        thinned_from=77
        keys [[1800,152.2009],[1801,152.8859],[1802,154.7313],[1804,161.1889],[1805,165.6169],[1806,170.6765],[1809,186.4949],[1810,191.4268],[1812,198.8056],[1817,194.5845],[1822,183.1463],[1825,177.2291],[1827,174.0125],[1829,169.7766],[1838,130.3987],[1841,127.0169],[1842,128.6881],[1849,143.5139],[1850,146.9834],[1851,152.7001],[1853,169.1422],[1854,178.5281],[1856,195.879],[1857,203.0157],[1858,208.3829],[1859,211.562],[1862,208.0974],[1866,180.0851],[1867,170.8822],[1871,129.8374],[1874,93.6319],[1876,80.4946],[1877,78.64],[1878,77.8711],[1887,88.1939],[1891,97.4805],[1898,111.4931],[1899,112.6578],[1902,115.0186],[1903,115.9889],[1905,122.6016],[1906,128.5391],[1907,135.0752],[1908,140.8145],[1909,145.3366],[1910,147.9967],[1911,149.0913],[1914,148.6236],[1915,147.5275],[1918,142.485],[1924,130.0849],[1926,126.9679],[1929,124.7896],[1930,124.9547],[1931,125.6561],[1940,142.1019],[1943,148.1151],[1944,141.2142],[1951,148.8561],[1953,148.8866],[1959,146.9024],[1961,146.7077],[1962,0]]
       }
       foot_r_ik_ctrl.Rotation.Y {
        type=MovieSceneScriptingFloatChannel
        default=16.6896
        thinned_from=77
        keys [[1800,16.6896],[1801,15.4251],[1802,12.1103],[1804,2.2338],[1805,-3.2223],[1806,-8.7616],[1808,-18.1488],[1809,-21.0919],[1810,-22.6219],[1812,-22.3508],[1817,-17.6943],[1818,-16.9754],[1822,-14.4552],[1825,-13.593],[1827,-13.6964],[1828,-13.8388],[1829,-14.0351],[1838,-14.8482],[1841,-15.6311],[1842,-16.0092],[1849,-17.8784],[1850,-17.8587],[1851,-17.4973],[1853,-16.0672],[1854,-15.5161],[1857,-16.075],[1858,-15.7606],[1859,-15.1235],[1862,-11.9527],[1866,-2.4024],[1867,0.2905],[1871,19.7113],[1874,30.8061],[1876,31.2965],[1877,31.202],[1878,31.2381],[1887,33.5267],[1890,33.9418],[1891,34.0336],[1898,33.7078],[1901,33.1847],[1903,33.0358],[1905,33.1519],[1906,33.0454],[1907,32.2793],[1908,31.2844],[1910,29.9628],[1911,29.4266],[1914,28.8856],[1918,28.9995],[1924,28.5865],[1926,28.273],[1931,27.2652],[1940,25.0016],[1941,24.7952],[1943,24.4589],[1944,10.8943],[1951,10.3346],[1952,10.3236],[1953,10.34],[1957,10.5285],[1961,10.5263],[1962,0]]
       }
       foot_r_ik_ctrl.Rotation.Z {
        type=MovieSceneScriptingFloatChannel
        default=2.4513
        thinned_from=77
        keys [[1800,2.4513],[1801,3.8372],[1802,1.4918],[1804,-9.2878],[1805,-17.2517],[1806,-26.5497],[1808,-47.7044],[1809,-58.8657],[1810,-69.6677],[1812,-87.5323],[1817,262.2607],[1829,257.8497],[1841,250.1457],[1842,249.8109],[1849,244.4602],[1850,242.6485],[1851,239.6229],[1853,231.6529],[1856,217.8875],[1858,207.9504],[1859,203.9275],[1862,198.6495],[1866,190.3138],[1867,187.1729],[1871,151.7966],[1874,106.9799],[1876,88.3967],[1877,86.1532],[1878,84.8283],[1887,76.777],[1891,73.6701],[1898,67.9106],[1899,66.9172],[1902,63.6059],[1903,62.5886],[1905,61.6098],[1906,62.7903],[1907,64.4421],[1908,64.9958],[1909,65.209],[1910,65.1566],[1911,64.6945],[1914,61.4074],[1915,59.9502],[1920,51.8859],[1924,45.8882],[1926,43.5353],[1929,41.2612],[1930,40.9211],[1931,40.8092],[1940,44.2129],[1943,45.1816],[1944,5.4705],[1951,6.7437],[1952,6.7152],[1953,6.6408],[1956,6.3128],[1958,6.1274],[1959,6.062],[1960,6.0158],[1961,5.9922],[1962,0]]
       }
       foot_r_ik_ctrl.Scale.X {type=MovieSceneScriptingFloatChannel, default=1, thinned_from=77, keys [[1800,1],[1962,1]]}
       foot_r_ik_ctrl.Scale.Y {type=MovieSceneScriptingFloatChannel, default=1, thinned_from=77, keys [[1800,1],[1962,1]]}
       foot_r_ik_ctrl.Scale.Z {type=MovieSceneScriptingFloatChannel, default=1, thinned_from=77, keys [[1800,1],[1962,1]]}
       ball_r_ik_ctrl.Location.X {type=MovieSceneScriptingFloatChannel, default=0, const [0,2]}
       ball_r_ik_ctrl.Location.Y {type=MovieSceneScriptingFloatChannel, default=0, const [0,2]}
       ball_r_ik_ctrl.Location.Z {type=MovieSceneScriptingFloatChannel, default=0, const [0,2]}
       ball_r_ik_ctrl.Rotation.X {type=MovieSceneScriptingFloatChannel, default=0, const [0,2]}
       ball_r_ik_ctrl.Rotation.Y {type=MovieSceneScriptingFloatChannel, default=0, const [0,2]}
       ball_r_ik_ctrl.Rotation.Z {type=MovieSceneScriptingFloatChannel, default=0, const [0,2]}
       ball_r_ik_ctrl.Scale.X {type=MovieSceneScriptingFloatChannel, default=1, const [1,2]}
       ball_r_ik_ctrl.Scale.Y {type=MovieSceneScriptingFloatChannel, default=1, const [1,2]}
       ball_r_ik_ctrl.Scale.Z {type=MovieSceneScriptingFloatChannel, default=1, const [1,2]}
       foot_bk1_r_ctrl.Location.X {type=MovieSceneScriptingFloatChannel, default=0, const [0,2]}
       foot_bk1_r_ctrl.Location.Y {type=MovieSceneScriptingFloatChannel, default=0, const [0,2]}
       foot_bk1_r_ctrl.Location.Z {type=MovieSceneScriptingFloatChannel, default=0, const [0,2]}
       foot_bk1_r_ctrl.Rotation.X {type=MovieSceneScriptingFloatChannel, default=0, const [0,2]}
       foot_bk1_r_ctrl.Rotation.Y {type=MovieSceneScriptingFloatChannel, default=0, const [0,2]}
       foot_bk1_r_ctrl.Rotation.Z {type=MovieSceneScriptingFloatChannel, default=0, const [0,2]}
       foot_bk1_r_ctrl.Scale.X {type=MovieSceneScriptingFloatChannel, default=1, const [1,2]}
       foot_bk1_r_ctrl.Scale.Y {type=MovieSceneScriptingFloatChannel, default=1, const [1,2]}
       foot_bk1_r_ctrl.Scale.Z {type=MovieSceneScriptingFloatChannel, default=1, const [1,2]}
       foot_roll_r_ctrl.Location.X {type=MovieSceneScriptingFloatChannel, default=0, const [0,2]}
       foot_roll_r_ctrl.Location.Y {type=MovieSceneScriptingFloatChannel, default=0, const [0,2]}
       foot_roll_r_ctrl.Location.Z {type=MovieSceneScriptingFloatChannel, default=0, const [0,2]}
       foot_roll_r_ctrl.Rotation.X {type=MovieSceneScriptingFloatChannel, default=0, const [0,2]}
       foot_roll_r_ctrl.Rotation.Y {type=MovieSceneScriptingFloatChannel, default=0, const [0,2]}
       foot_roll_r_ctrl.Rotation.Z {type=MovieSceneScriptingFloatChannel, default=0, const [0,2]}
       foot_roll_r_ctrl.Scale.X {type=MovieSceneScriptingFloatChannel, default=1, const [1,2]}
       foot_roll_r_ctrl.Scale.Y {type=MovieSceneScriptingFloatChannel, default=1, const [1,2]}
       foot_roll_r_ctrl.Scale.Z {type=MovieSceneScriptingFloatChannel, default=1, const [1,2]}
       roll_blend_r_ctrl {type=MovieSceneScriptingFloatChannel, default=1, const [1,2]}
       tip_r_ctrl.Location.X {type=MovieSceneScriptingFloatChannel, default=0, const [0,2]}
       tip_r_ctrl.Location.Y {type=MovieSceneScriptingFloatChannel, default=0, const [0,2]}
       tip_r_ctrl.Location.Z {type=MovieSceneScriptingFloatChannel, default=0, const [0,2]}
       tip_r_ctrl.Rotation.X {type=MovieSceneScriptingFloatChannel, default=0, const [0,2]}
       tip_r_ctrl.Rotation.Y {type=MovieSceneScriptingFloatChannel, default=0, const [0,2]}
       tip_r_ctrl.Rotation.Z {type=MovieSceneScriptingFloatChannel, default=0, const [0,2]}
       tip_r_ctrl.Scale.X {type=MovieSceneScriptingFloatChannel, default=1, const [1,2]}
       tip_r_ctrl.Scale.Y {type=MovieSceneScriptingFloatChannel, default=1, const [1,2]}
       tip_r_ctrl.Scale.Z {type=MovieSceneScriptingFloatChannel, default=1, const [1,2]}
       heel_r_ctrl.Location.X {type=MovieSceneScriptingFloatChannel, default=0, const [0,2]}
       heel_r_ctrl.Location.Y {type=MovieSceneScriptingFloatChannel, default=0, const [0,2]}
       heel_r_ctrl.Location.Z {type=MovieSceneScriptingFloatChannel, default=0, const [0,2]}
       heel_r_ctrl.Rotation.X {type=MovieSceneScriptingFloatChannel, default=0, const [0,2]}
       heel_r_ctrl.Rotation.Y {type=MovieSceneScriptingFloatChannel, default=0, const [0,2]}
       heel_r_ctrl.Rotation.Z {type=MovieSceneScriptingFloatChannel, default=0, const [0,2]}
       heel_r_ctrl.Scale.X {type=MovieSceneScriptingFloatChannel, default=1, const [1,2]}
       heel_r_ctrl.Scale.Y {type=MovieSceneScriptingFloatChannel, default=1, const [1,2]}
       heel_r_ctrl.Scale.Z {type=MovieSceneScriptingFloatChannel, default=1, const [1,2]}
       leg_r_softness {type=MovieSceneScriptingFloatChannel, default=0, const [0,2]}
       leg_r_pv_ik_ctrl.Location.X {
        type=MovieSceneScriptingFloatChannel
        default=0.3547
        thinned_from=71
        keys [[1800,0.3555],[1801,0.7402],[1804,1.9922],[1806,2.4902],[1808,2.7168],[1809,2.0781],[1813,4.2539],[1816,-6.1621],[1817,1.877],[1819,2.8672],[1825,4.6035],[1829,2.7705],[1830,-1.8662],[1832,-22.6138],[1837,-0.7217],[1841,2.2979],[1844,0.3975],[1846,0.3086],[1848,2.0791],[1850,2.8701],[1851,2.8857],[1852,2.9121],[1853,2.791],[1855,2.748],[1857,2.7148],[1858,2.6641],[1862,2.3037],[1865,1.981],[1866,1.8662],[1868,0.9673],[1874,-2.3262],[1878,2.7705],[1879,0.1475],[1881,-0.9463],[1887,1.502],[1890,2.2969],[1893,1.9082],[1895,1.6543],[1897,1.9219],[1901,2.9473],[1902,2.9609],[1903,2.8672],[1905,2.4551],[1906,2.3613],[1908,2.6191],[1909,2.7109],[1910,2.6895],[1911,2.5352],[1915,2.2305],[1918,1.877],[1920,1.7676],[1926,2.6387],[1931,3.5078],[1932,3.957],[1934,4.5586],[1940,1.3633],[1942,-0.332],[1946,714.9688],[1948,2.5352],[1949,2.5859],[1950,2.6523],[1954,2.8672],[1955,2.9141],[1956,2.9453],[1958,2.9766],[1961,2.9727],[1962,0]]
       }
       leg_r_pv_ik_ctrl.Location.Y {
        type=MovieSceneScriptingFloatChannel
        default=4.102
        thinned_from=71
        keys [[1800,4.1016],[1801,4.1191],[1804,4.3574],[1806,4.624],[1808,4.9346],[1809,5.7402],[1813,6.5713],[1816,-89.1338],[1817,5.9277],[1819,3.6997],[1825,-0.0786],[1829,3.8438],[1830,5.4739],[1832,-10.7573],[1837,13.776],[1841,3.9541],[1844,3.5505],[1846,3.0977],[1848,3.9995],[1850,4.0479],[1851,4.042],[1852,3.75],[1853,4.1333],[1855,4.0557],[1856,4.5303],[1857,4.7051],[1858,4.8906],[1859,5.3096],[1862,6.0967],[1865,8.125],[1866,5.9277],[1868,1.8867],[1874,7.4082],[1878,3.8438],[1879,4.168],[1881,4.6133],[1887,2.6172],[1890,3.9531],[1893,3.6406],[1895,3.4492],[1897,3.5547],[1901,4],[1903,3.9824],[1905,4.1953],[1908,4.8125],[1910,5.1406],[1915,5.8086],[1918,5.9189],[1920,5.8213],[1926,5.041],[1931,4.2139],[1932,3.7998],[1934,3.1641],[1940,4.4131],[1942,10.7959],[1946,-298.9111],[1948,4.0742],[1949,4.0801],[1950,4.0723],[1954,4.0088],[1955,3.9805],[1958,3.9248],[1961,3.8965],[1962,0.001]]
       }
       leg_r_pv_ik_ctrl.Location.Z {
        type=MovieSceneScriptingFloatChannel
        default=4.416
        thinned_from=71
        keys [[1800,4.416],[1801,4.4648],[1804,5.0273],[1806,5.541],[1808,5.9648],[1809,5.3164],[1813,11.6367],[1816,74.6133],[1817,6.9336],[1819,3.5742],[1825,3.6699],[1829,4.5137],[1830,19.8789],[1832,25.0342],[1837,20.368],[1841,4.5715],[1844,3.7729],[1846,4.2373],[1848,4.5781],[1850,4.79],[1851,4.7803],[1852,4.4141],[1853,4.8428],[1855,4.9033],[1856,5.5664],[1857,5.8594],[1858,6.1162],[1859,5.916],[1862,6.2246],[1865,10.0625],[1866,6.9336],[1868,24.6992],[1874,1.1914],[1878,4.5137],[1879,2.7295],[1881,-1.9395],[1887,6.7461],[1890,4.5713],[1893,4.1182],[1895,3.6104],[1897,4.2305],[1901,4.7368],[1903,4.7119],[1905,4.958],[1908,5.666],[1910,6.0388],[1911,6.1619],[1915,6.7584],[1918,6.9271],[1920,6.833],[1926,5.5698],[1931,5.6436],[1932,6.2749],[1934,8.0044],[1940,0.978],[1942,14.1045],[1946,-7818.8789],[1948,4.752],[1949,4.7881],[1950,4.791],[1954,4.7656],[1955,4.7236],[1958,4.6592],[1961,4.623],[1962,-0.002]]
       }
       leg_r_pv_ik_ctrl.Rotation.X {
        type=MovieSceneScriptingFloatChannel
        default=-5.8995
        thinned_from=71
        keys [[1800,-5.8994],[1801,-6.0073],[1804,-6.935],[1806,-7.5083],[1808,-8.0489],[1809,-7.7776],[1813,-6.7998],[1816,167.6106],[1817,-9.2079],[1819,-5.8916],[1825,-13.6029],[1829,-5.9184],[1830,-42.327],[1832,-77.4745],[1837,-24.6826],[1841,-6.095],[1844,-5.3612],[1846,-6.2054],[1848,-6.2637],[1850,-6.4253],[1851,-6.409],[1852,-5.9417],[1853,-6.5059],[1855,-6.6695],[1856,-7.5554],[1857,-7.9352],[1858,-8.2229],[1859,-8.139],[1862,-8.3825],[1865,-13.0001],[1866,-9.2077],[1868,330.2256],[1874,-5.0432],[1878,-5.9186],[1879,-4.8114],[1881,-1.6327],[1887,-7.5946],[1890,-6.0953],[1893,-5.6375],[1895,-5.0945],[1897,-5.8385],[1901,-6.3484],[1902,-6.3236],[1903,-6.3132],[1905,353.4051],[1909,352.0974],[1910,-8.1543],[1911,-8.387],[1915,-9.0129],[1918,-9.2485],[1920,-9.0914],[1926,-7.0508],[1931,-5.452],[1932,-4.3844],[1934,-4.0121],[1940,-4.9253],[1942,-17.7718],[1946,15.9106],[1948,-6.3942],[1950,-6.4244],[1954,-6.3643],[1958,-6.2396],[1960,-6.1979],[1961,-6.1895],[1962,-0.0001]]
       }
       leg_r_pv_ik_ctrl.Rotation.Y {
        type=MovieSceneScriptingFloatChannel
        default=0.419
        keys [[1800,0.4197],[1801,0.3546],[1804,-0.3424],[1806,-0.8906],[1808,-1.3453],[1809,-0.5108],[1813,-5.3285],[1816,15.6181],[1817,-1.4767],[1819,0.467],[1825,-2.6395],[1829,1.2575],[1830,-9.0252],[1832,-1.2941],[1837,-6.2684],[1841,0.0318],[1844,-3.3653],[1846,-3.0064],[1848,-1.0968],[1850,0.0212],[1851,0.039],[1852,0.0132],[1853,-0.0502],[1855,-0.5762],[1856,-0.9486],[1857,-1.2261],[1858,-1.4501],[1859,-1.5362],[1862,-1.8585],[1865,-1.4135],[1866,-1.4764],[1868,1.181],[1874,4.0091],[1878,1.2573],[1879,4.8074],[1881,4.0883],[1887,1.1747],[1890,0.032],[1893,-1.034],[1895,-1.9959],[1897,-1.0746],[1901,0.0979],[1902,0.1148],[1903,0.1222],[1905,-0.1498],[1906,-0.4219],[1908,-1.0109],[1909,-1.2383],[1910,-1.4182],[1911,-1.4711],[1915,-2.006],[1918,-1.6698],[1920,-1.1858],[1926,0.0888],[1931,0.2018],[1932,-1.2133],[1934,-3.1924],[1940,2.2475],[1942,-3.6716],[1946,-46.0426],[1948,-0.0288],[1949,-0.0283],[1950,-0.0378],[1954,0.0677],[1955,0.1019],[1956,0.1298],[1958,0.1791],[1959,0.1972],[1960,0.2091],[1961,0.212],[1962,-0.0005]]
       }
       leg_r_pv_ik_ctrl.Rotation.Z {
        type=MovieSceneScriptingFloatChannel
        default=-0.1017
        thinned_from=71
        keys [[1800,-0.1017],[1801,-0.7458],[1804,-2.8647],[1806,-3.8046],[1808,-4.4268],[1809,-4.5942],[1813,-2.8913],[1816,3.1959],[1817,-3.4472],[1819,-3.419],[1825,-1.6253],[1829,-2.4252],[1830,3.9109],[1832,14.1714],[1837,-1.7113],[1841,-3.0121],[1844,-3.544],[1846,-3.5836],[1848,-3.6541],[1850,-3.6747],[1851,-3.6454],[1852,-3.7141],[1853,-3.5614],[1855,-3.8733],[1856,-4.0369],[1857,-4.2658],[1858,-4.4419],[1859,-4.5795],[1862,-4.4337],[1865,-3.3493],[1866,-3.5504],[1868,-359.5603],[1874,-1.6866],[1878,-2.4233],[1879,-2.5042],[1881,-2.896],[1887,-2.5783],[1890,-2.9504],[1895,-3.7149],[1897,-3.7059],[1901,-3.7301],[1902,-3.6506],[1903,-3.4758],[1905,-3.2214],[1906,-3.4102],[1908,-4.1197],[1909,-4.3112],[1910,-4.4587],[1911,-4.5077],[1915,-4.1512],[1918,-3.5554],[1920,-3.131],[1926,-2.431],[1931,-2.5175],[1932,-2.6234],[1934,-2.7864],[1940,-2.8],[1942,-1.6615],[1946,345.763],[1948,-3.3923],[1949,-3.4971],[1950,-3.5832],[1954,-3.7253],[1955,-3.7217],[1959,-3.6599],[1961,-3.6435],[1962,0]]
       }
       leg_r_pv_ik_ctrl.Scale.X {
        type=MovieSceneScriptingFloatChannel
        default=1
        thinned_from=71
        keys [[1800,1],[1808,1],[1809,1.0047],[1813,0.9709],[1816,1],[1942,1],[1946,2.0523],[1948,1],[1962,1]]
       }
       leg_r_pv_ik_ctrl.Scale.Y {
        type=MovieSceneScriptingFloatChannel
        default=1
        thinned_from=71
        keys [[1800,1],[1808,1],[1809,1.0047],[1813,0.9709],[1816,1],[1942,1],[1946,2.0523],[1948,1],[1962,1]]
       }
       leg_r_pv_ik_ctrl.Scale.Z {
        type=MovieSceneScriptingFloatChannel
        default=1
        thinned_from=71
        keys [[1800,1],[1808,1],[1809,1.0047],[1813,0.9709],[1816,1],[1942,1],[1946,2.0523],[1948,1],[1962,1]]
       }
       upperarm_l_fk_ctrl.Location.X {type=MovieSceneScriptingFloatChannel, default=0, thinned_from=143, keys [[1800,0],[1962,0]]}
       upperarm_l_fk_ctrl.Location.Y {type=MovieSceneScriptingFloatChannel, default=0, thinned_from=141, keys [[1800,0],[1962,0]]}
       upperarm_l_fk_ctrl.Location.Z {type=MovieSceneScriptingFloatChannel, default=0, thinned_from=147, keys [[1800,0],[1962,0]]}
       upperarm_l_fk_ctrl.Rotation.X {
        type=MovieSceneScriptingFloatChannel
        default=94.9421
        thinned_from=18
        keys [[1800,94.9421],[1803,94.0076],[1804,94.0511],[1824,89.307],[1849,92.6114],[1850,92.6098],[1856,93.3662],[1873,89.307],[1898,92.6114],[1903,92.5002],[1907,93.8225],[1926,89.3137],[1950,92.6093],[1961,92.254],[1962,0]]
       }
       upperarm_l_fk_ctrl.Rotation.Y {
        type=MovieSceneScriptingFloatChannel
        default=25.179
        thinned_from=18
        keys [[1800,25.179],[1802,23.2385],[1808,21.0766],[1817,21.9634],[1818,21.9609],[1838,23.1788],[1858,21.1786],[1866,21.9634],[1867,21.9609],[1887,23.1789],[1909,21.0835],[1939,23.1769],[1956,22.5263],[1961,22.4467],[1962,0]]
       }
       upperarm_l_fk_ctrl.Rotation.Z {
        type=MovieSceneScriptingFloatChannel
        default=-173.0651
        keys [[1800,-173.0651],[1803,-174.8604],[1805,-174.5795],[1824,-186.8655],[1849,-177.8934],[1851,-177.9159],[1857,-175.8506],[1873,-186.8655],[1898,-177.8934],[1903,-178.2303],[1907,-175.0371],[1926,-186.8564],[1950,-177.9024],[1961,-178.9435],[1962,0]]
       }
       upperarm_l_fk_ctrl.Scale.X {type=MovieSceneScriptingFloatChannel, default=1, const [1,2]}
       upperarm_l_fk_ctrl.Scale.Y {type=MovieSceneScriptingFloatChannel, default=1, const [1,2]}
       upperarm_l_fk_ctrl.Scale.Z {type=MovieSceneScriptingFloatChannel, default=1, const [1,2]}
       lowerarm_l_fk_ctrl.Location.X {type=MovieSceneScriptingFloatChannel, default=0, thinned_from=36, keys [[1800,0],[1962,0]]}
       lowerarm_l_fk_ctrl.Location.Y {type=MovieSceneScriptingFloatChannel, default=0, thinned_from=34, keys [[1800,0],[1962,0]]}
       lowerarm_l_fk_ctrl.Location.Z {type=MovieSceneScriptingFloatChannel, default=0, thinned_from=36, keys [[1800,0],[1962,0]]}
       lowerarm_l_fk_ctrl.Rotation.X {
        type=MovieSceneScriptingFloatChannel
        default=0.0442
        thinned_from=17
        keys [[1800,0.0442],[1801,0.0593],[1803,0.0888],[1807,0.1286],[1852,0.1271],[1856,0.1289],[1901,0.1289],[1905,0.1162],[1908,0.1256],[1961,0.1256],[1962,0]]
       }
       lowerarm_l_fk_ctrl.Rotation.Y {
        type=MovieSceneScriptingFloatChannel
        default=1.3721
        thinned_from=13
        keys [[1800,1.3721],[1807,1.4391],[1856,1.4392],[1905,1.4343],[1908,1.4386],[1961,1.4386],[1962,0]]
       }
       lowerarm_l_fk_ctrl.Rotation.Z {
        type=MovieSceneScriptingFloatChannel
        default=-17.611
        thinned_from=13
        keys [[1800,-17.611],[1807,-18.5886],[1856,-18.589],[1905,-18.5175],[1908,-18.5821],[1961,-18.5821],[1962,0]]
       }
       lowerarm_l_fk_ctrl.Scale.X {type=MovieSceneScriptingFloatChannel, default=1, const [1,2]}
       lowerarm_l_fk_ctrl.Scale.Y {type=MovieSceneScriptingFloatChannel, default=1, const [1,2]}
       lowerarm_l_fk_ctrl.Scale.Z {type=MovieSceneScriptingFloatChannel, default=1, const [1,2]}
       hand_l_fk_ctrl.Location.X {type=MovieSceneScriptingFloatChannel, default=0, thinned_from=16, keys [[1800,0],[1962,0]]}
       hand_l_fk_ctrl.Location.Y {type=MovieSceneScriptingFloatChannel, default=0, thinned_from=14, keys [[1800,0],[1962,0]]}
       hand_l_fk_ctrl.Location.Z {type=MovieSceneScriptingFloatChannel, default=0, thinned_from=15, keys [[1800,0],[1962,0]]}
       hand_l_fk_ctrl.Rotation.X {type=MovieSceneScriptingFloatChannel, default=-9.2151, thinned_from=12, keys [[1800,-9.2151],[1805,-9.2438],[1961,-9.2438],[1962,0]]}
       hand_l_fk_ctrl.Rotation.Y {type=MovieSceneScriptingFloatChannel, default=-4.2623, thinned_from=12, keys [[1800,-4.2623],[1803,-4.2792],[1961,-4.2821],[1962,0]]}
       hand_l_fk_ctrl.Rotation.Z {type=MovieSceneScriptingFloatChannel, default=-29.5018, thinned_from=12, keys [[1800,-29.5018],[1805,-29.5938],[1961,-29.5938],[1962,0]]}
       hand_l_fk_ctrl.Scale.X {type=MovieSceneScriptingFloatChannel, default=1, const [1,2]}
       hand_l_fk_ctrl.Scale.Y {type=MovieSceneScriptingFloatChannel, default=1, const [1,2]}
       hand_l_fk_ctrl.Scale.Z {type=MovieSceneScriptingFloatChannel, default=1, const [1,2]}
       thumb_01_l_ctrl.Location.X {type=MovieSceneScriptingFloatChannel, default=-48.0575, thinned_from=12, keys [[1800,-48.0575],[1801,0],[1961,0],[1962,64.6193]]}
       thumb_01_l_ctrl.Location.Y {type=MovieSceneScriptingFloatChannel, default=30.9177, thinned_from=12, keys [[1800,30.9177],[1801,0],[1961,0],[1962,73.1395]]}
       thumb_01_l_ctrl.Location.Z {type=MovieSceneScriptingFloatChannel, default=-17.1129, thinned_from=12, keys [[1800,-17.1129],[1801,0],[1961,0.0001],[1962,25.5677]]}
       thumb_01_l_ctrl.Rotation.X {
        type=MovieSceneScriptingFloatChannel
        default=-1.9522
        thinned_from=12
        keys [[1800,-1.9522],[1801,-1.5659],[1804,-1.5584],[1961,-1.5584],[1962,104.1537]]
       }
       thumb_01_l_ctrl.Rotation.Y {
        type=MovieSceneScriptingFloatChannel
        default=-0.7185
        thinned_from=12
        keys [[1800,-0.7185],[1801,0.8844],[1804,0.8742],[1961,0.8742],[1962,21.071]]
       }
       thumb_01_l_ctrl.Rotation.Z {
        type=MovieSceneScriptingFloatChannel
        default=-12.6815
        thinned_from=12
        keys [[1800,-12.6815],[1801,-13.9346],[1961,-13.9221],[1962,23.0017]]
       }
       thumb_01_l_ctrl.Scale.X {type=MovieSceneScriptingFloatChannel, default=1, const [1,2]}
       thumb_01_l_ctrl.Scale.Y {type=MovieSceneScriptingFloatChannel, default=1, const [1,2]}
       thumb_01_l_ctrl.Scale.Z {type=MovieSceneScriptingFloatChannel, default=1, const [1,2]}
       thumb_02_l_ctrl.Location.X {type=MovieSceneScriptingFloatChannel, default=0, const [0,2]}
       thumb_02_l_ctrl.Location.Y {type=MovieSceneScriptingFloatChannel, default=0, const [0,2]}
       thumb_02_l_ctrl.Location.Z {type=MovieSceneScriptingFloatChannel, default=0, const [0,2]}
       thumb_02_l_ctrl.Rotation.X {type=MovieSceneScriptingFloatChannel, default=0, const [0,2]}
       thumb_02_l_ctrl.Rotation.Y {type=MovieSceneScriptingFloatChannel, default=0, const [0,2]}
       thumb_02_l_ctrl.Rotation.Z {type=MovieSceneScriptingFloatChannel, default=0, const [0,2]}
       thumb_02_l_ctrl.Scale.X {type=MovieSceneScriptingFloatChannel, default=1, const [1,2]}
       thumb_02_l_ctrl.Scale.Y {type=MovieSceneScriptingFloatChannel, default=1, const [1,2]}
       thumb_02_l_ctrl.Scale.Z {type=MovieSceneScriptingFloatChannel, default=1, const [1,2]}
       thumb_03_l_ctrl.Location.X {type=MovieSceneScriptingFloatChannel, default=0, const [0,2]}
       thumb_03_l_ctrl.Location.Y {type=MovieSceneScriptingFloatChannel, default=0, const [0,2]}
       thumb_03_l_ctrl.Location.Z {type=MovieSceneScriptingFloatChannel, default=0, const [0,2]}
       thumb_03_l_ctrl.Rotation.X {type=MovieSceneScriptingFloatChannel, default=8.1645, const [8.1645,2]}
       thumb_03_l_ctrl.Rotation.Y {type=MovieSceneScriptingFloatChannel, default=-9.9358, const [-9.9358,2]}
       thumb_03_l_ctrl.Rotation.Z {type=MovieSceneScriptingFloatChannel, default=-50.7089, const [-50.7089,2]}
       thumb_03_l_ctrl.Scale.X {type=MovieSceneScriptingFloatChannel, default=1, const [1,2]}
       thumb_03_l_ctrl.Scale.Y {type=MovieSceneScriptingFloatChannel, default=1, const [1,2]}
       thumb_03_l_ctrl.Scale.Z {type=MovieSceneScriptingFloatChannel, default=1, const [1,2]}
       index_metacarpal_l_ctrl.Location.X {type=MovieSceneScriptingFloatChannel, default=-55.1158, thinned_from=11, keys [[1800,-55.1158],[1801,0],[1961,0],[1962,27.0764]]}
       index_metacarpal_l_ctrl.Location.Y {type=MovieSceneScriptingFloatChannel, default=22.365, thinned_from=11, keys [[1800,22.365],[1801,0],[1961,0],[1962,-8.4613]]}
       index_metacarpal_l_ctrl.Location.Z {type=MovieSceneScriptingFloatChannel, default=-4.4257, thinned_from=11, keys [[1800,-4.4257],[1801,0],[1961,0],[1962,-96.6419]]}
       index_metacarpal_l_ctrl.Rotation.X {type=MovieSceneScriptingFloatChannel, default=-0.654, thinned_from=11, keys [[1800,-0.654],[1801,0],[1961,0],[1962,10.5572]]}
       index_metacarpal_l_ctrl.Rotation.Y {type=MovieSceneScriptingFloatChannel, default=-0.7479, thinned_from=11, keys [[1800,-0.7479],[1801,0],[1961,0],[1962,-78.2884]]}
       index_metacarpal_l_ctrl.Rotation.Z {type=MovieSceneScriptingFloatChannel, default=-1.7943, thinned_from=11, keys [[1800,-1.7943],[1801,0],[1961,0],[1962,67.0062]]}
       index_metacarpal_l_ctrl.Scale.X {type=MovieSceneScriptingFloatChannel, default=1, const [1,2]}
       index_metacarpal_l_ctrl.Scale.Y {type=MovieSceneScriptingFloatChannel, default=1, const [1,2]}
       index_metacarpal_l_ctrl.Scale.Z {type=MovieSceneScriptingFloatChannel, default=1, const [1,2]}
       index_01_l_ctrl.Location.X {type=MovieSceneScriptingFloatChannel, default=0.1105, const [0.1105,2]}
       index_01_l_ctrl.Location.Y {type=MovieSceneScriptingFloatChannel, default=-0.4345, const [-0.4345,2]}
       index_01_l_ctrl.Location.Z {type=MovieSceneScriptingFloatChannel, default=0.5473, const [0.5473,2]}
       index_01_l_ctrl.Rotation.X {type=MovieSceneScriptingFloatChannel, default=2.6836, const [2.6836,2]}
       index_01_l_ctrl.Rotation.Y {type=MovieSceneScriptingFloatChannel, default=-1.2982, const [-1.2982,2]}
       index_01_l_ctrl.Rotation.Z {type=MovieSceneScriptingFloatChannel, default=-68.0656, const [-68.0656,2]}
       index_01_l_ctrl.Scale.X {type=MovieSceneScriptingFloatChannel, default=1, const [1,2]}
       index_01_l_ctrl.Scale.Y {type=MovieSceneScriptingFloatChannel, default=1, const [1,2]}
       index_01_l_ctrl.Scale.Z {type=MovieSceneScriptingFloatChannel, default=1, const [1,2]}
       index_02_l_ctrl.Location.X {type=MovieSceneScriptingFloatChannel, default=0, thinned_from=3, keys [[1800,0],[1962,0]]}
       index_02_l_ctrl.Location.Y {type=MovieSceneScriptingFloatChannel, default=0, thinned_from=9, keys [[1800,0],[1962,0]]}
       index_02_l_ctrl.Location.Z {type=MovieSceneScriptingFloatChannel, default=0, thinned_from=3, keys [[1800,0],[1962,0]]}
       index_02_l_ctrl.Rotation.X {
        type=MovieSceneScriptingFloatChannel
        default=0
        thinned_from=24
        keys [[1800,0],[1801,-0.0001],[1802,-0.0004],[1803,-0.0004],[1804,0],[1962,0]]
       }
       index_02_l_ctrl.Rotation.Y {
        type=MovieSceneScriptingFloatChannel
        default=0
        thinned_from=33
        keys [[1800,0],[1801,0.0001],[1802,0.0005],[1803,0.0005],[1804,0.0001],[1962,0]]
       }
       index_02_l_ctrl.Rotation.Z {
        type=MovieSceneScriptingFloatChannel
        default=-60.8298
        thinned_from=31
        keys [[1800,-60.8298],[1801,-60.6903],[1802,-60.8298],[1962,-60.8298]]
       }
       index_02_l_ctrl.Scale.X {type=MovieSceneScriptingFloatChannel, default=1, const [1,2]}
       index_02_l_ctrl.Scale.Y {type=MovieSceneScriptingFloatChannel, default=1, const [1,2]}
       index_02_l_ctrl.Scale.Z {type=MovieSceneScriptingFloatChannel, default=1, const [1,2]}
       index_03_l_ctrl.Location.X {type=MovieSceneScriptingFloatChannel, default=0, const [0,2]}
       index_03_l_ctrl.Location.Y {type=MovieSceneScriptingFloatChannel, default=0, const [0,2]}
       index_03_l_ctrl.Location.Z {type=MovieSceneScriptingFloatChannel, default=0, const [0,2]}
       index_03_l_ctrl.Rotation.X {type=MovieSceneScriptingFloatChannel, default=0.0236, const [0.0236,2]}
       index_03_l_ctrl.Rotation.Y {type=MovieSceneScriptingFloatChannel, default=-0.109, const [-0.109,2]}
       index_03_l_ctrl.Rotation.Z {type=MovieSceneScriptingFloatChannel, default=-78.806, const [-78.806,2]}
       index_03_l_ctrl.Scale.X {type=MovieSceneScriptingFloatChannel, default=1, const [1,2]}
       index_03_l_ctrl.Scale.Y {type=MovieSceneScriptingFloatChannel, default=1, const [1,2]}
       index_03_l_ctrl.Scale.Z {type=MovieSceneScriptingFloatChannel, default=1, const [1,2]}
       middle_metacarpal_l_ctrl.Location.X {type=MovieSceneScriptingFloatChannel, default=0, const [0,2]}
       middle_metacarpal_l_ctrl.Location.Y {type=MovieSceneScriptingFloatChannel, default=0, const [0,2]}
       middle_metacarpal_l_ctrl.Location.Z {type=MovieSceneScriptingFloatChannel, default=0, const [0,2]}
       middle_metacarpal_l_ctrl.Rotation.X {type=MovieSceneScriptingFloatChannel, default=0, const [0,2]}
       middle_metacarpal_l_ctrl.Rotation.Y {type=MovieSceneScriptingFloatChannel, default=0, const [0,2]}
       middle_metacarpal_l_ctrl.Rotation.Z {type=MovieSceneScriptingFloatChannel, default=0, const [0,2]}
       middle_metacarpal_l_ctrl.Scale.X {type=MovieSceneScriptingFloatChannel, default=1, const [1,2]}
       middle_metacarpal_l_ctrl.Scale.Y {type=MovieSceneScriptingFloatChannel, default=1, const [1,2]}
       middle_metacarpal_l_ctrl.Scale.Z {type=MovieSceneScriptingFloatChannel, default=1, const [1,2]}
       middle_01_l_ctrl.Location.X {type=MovieSceneScriptingFloatChannel, default=0.0387, thinned_from=11, keys [[1800,0.0387],[1961,0.0387],[1962,0]]}
       middle_01_l_ctrl.Location.Y {type=MovieSceneScriptingFloatChannel, default=-0.3373, thinned_from=11, keys [[1800,-0.3373],[1961,-0.3373],[1962,0]]}
       middle_01_l_ctrl.Location.Z {type=MovieSceneScriptingFloatChannel, default=0.1567, thinned_from=11, keys [[1800,0.1567],[1961,0.1567],[1962,0]]}
       middle_01_l_ctrl.Rotation.X {type=MovieSceneScriptingFloatChannel, default=-2.1526, thinned_from=12, keys [[1800,-2.1526],[1805,-2.1555],[1961,-2.1555],[1962,0]]}
       middle_01_l_ctrl.Rotation.Y {type=MovieSceneScriptingFloatChannel, default=-5.7326, thinned_from=12, keys [[1800,-5.7326],[1961,-5.7372],[1962,0]]}
       middle_01_l_ctrl.Rotation.Z {type=MovieSceneScriptingFloatChannel, default=-77.1561, thinned_from=12, keys [[1800,-77.1561],[1805,-77.2631],[1961,-77.2631],[1962,0]]}
       middle_01_l_ctrl.Scale.X {type=MovieSceneScriptingFloatChannel, default=1, const [1,2]}
       middle_01_l_ctrl.Scale.Y {type=MovieSceneScriptingFloatChannel, default=1, const [1,2]}
       middle_01_l_ctrl.Scale.Z {type=MovieSceneScriptingFloatChannel, default=1, const [1,2]}
       middle_02_l_ctrl.Location.X {type=MovieSceneScriptingFloatChannel, default=15.5144, thinned_from=12, keys [[1800,15.5144],[1801,-0.0452],[1961,-0.0452],[1962,17.3383]]}
       middle_02_l_ctrl.Location.Y {type=MovieSceneScriptingFloatChannel, default=-57.4903, keys [[1800,-57.4903],[1801,0.0507],[1802,0.0507],[1962,2.8617]]}
       middle_02_l_ctrl.Location.Z {type=MovieSceneScriptingFloatChannel, default=-2.5677, thinned_from=11, keys [[1800,-2.5677],[1802,0.1146],[1961,0.1146],[1962,-107.8384]]}
       middle_02_l_ctrl.Rotation.X {type=MovieSceneScriptingFloatChannel, default=-2.9532, thinned_from=12, keys [[1800,-2.9532],[1801,-3.3978],[1961,-3.3958],[1962,-10.0543]]}
       middle_02_l_ctrl.Rotation.Y {type=MovieSceneScriptingFloatChannel, default=-1.4046, thinned_from=11, keys [[1800,-1.4046],[1802,-1.8595],[1961,-1.8595],[1962,75.3949]]}
       middle_02_l_ctrl.Rotation.Z {
        type=MovieSceneScriptingFloatChannel
        default=-60.1103
        thinned_from=12
        keys [[1800,-60.1103],[1801,-57.9078],[1802,-58.032],[1961,-58.032],[1962,-70.0817]]
       }
       middle_02_l_ctrl.Scale.X {type=MovieSceneScriptingFloatChannel, default=1, const [1,2]}
       middle_02_l_ctrl.Scale.Y {type=MovieSceneScriptingFloatChannel, default=1, const [1,2]}
       middle_02_l_ctrl.Scale.Z {type=MovieSceneScriptingFloatChannel, default=1, const [1,2]}
       middle_03_l_ctrl.Location.X {type=MovieSceneScriptingFloatChannel, default=0.0206, const [0.0206,2]}
       middle_03_l_ctrl.Location.Y {type=MovieSceneScriptingFloatChannel, default=-0.0583, const [-0.0583,2]}
       middle_03_l_ctrl.Location.Z {type=MovieSceneScriptingFloatChannel, default=0.1324, const [0.1324,2]}
       middle_03_l_ctrl.Rotation.X {type=MovieSceneScriptingFloatChannel, default=-1.8232, const [-1.8232,2]}
       middle_03_l_ctrl.Rotation.Y {type=MovieSceneScriptingFloatChannel, default=1.4852, const [1.4852,2]}
       middle_03_l_ctrl.Rotation.Z {type=MovieSceneScriptingFloatChannel, default=-64.6207, const [-64.6207,2]}
       middle_03_l_ctrl.Scale.X {type=MovieSceneScriptingFloatChannel, default=1, const [1,2]}
       middle_03_l_ctrl.Scale.Y {type=MovieSceneScriptingFloatChannel, default=1, const [1,2]}
       middle_03_l_ctrl.Scale.Z {type=MovieSceneScriptingFloatChannel, default=1, const [1,2]}
       ring_metacarpal_l_ctrl.Location.X {type=MovieSceneScriptingFloatChannel, default=-53.8812, thinned_from=10, keys [[1800,-53.8812],[1801,0],[1961,0],[1962,10.0837]]}
       ring_metacarpal_l_ctrl.Location.Y {type=MovieSceneScriptingFloatChannel, default=25.276, thinned_from=11, keys [[1800,25.276],[1801,0],[1961,0],[1962,-13.5375]]}
       ring_metacarpal_l_ctrl.Location.Z {type=MovieSceneScriptingFloatChannel, default=2.9233, thinned_from=11, keys [[1800,2.9233],[1801,0],[1961,0],[1962,-103.0535]]}
       ring_metacarpal_l_ctrl.Rotation.X {type=MovieSceneScriptingFloatChannel, default=-0.37, thinned_from=11, keys [[1800,-0.37],[1801,0],[1961,0],[1962,-37.1383]]}
       ring_metacarpal_l_ctrl.Rotation.Y {type=MovieSceneScriptingFloatChannel, default=-0.5558, thinned_from=11, keys [[1800,-0.5558],[1801,0],[1961,0],[1962,-73.1088]]}
       ring_metacarpal_l_ctrl.Rotation.Z {type=MovieSceneScriptingFloatChannel, default=-1.9414, thinned_from=11, keys [[1800,-1.9414],[1801,0],[1961,0],[1962,105.4957]]}
       ring_metacarpal_l_ctrl.Scale.X {type=MovieSceneScriptingFloatChannel, default=1, const [1,2]}
       ring_metacarpal_l_ctrl.Scale.Y {type=MovieSceneScriptingFloatChannel, default=1, const [1,2]}
       ring_metacarpal_l_ctrl.Scale.Z {type=MovieSceneScriptingFloatChannel, default=1, const [1,2]}
       ring_01_l_ctrl.Location.X {type=MovieSceneScriptingFloatChannel, default=0.1701, thinned_from=11, keys [[1800,0.1701],[1962,0.1701]]}
       ring_01_l_ctrl.Location.Y {type=MovieSceneScriptingFloatChannel, default=-0.6802, thinned_from=11, keys [[1800,-0.6802],[1962,-0.6802]]}
       ring_01_l_ctrl.Location.Z {type=MovieSceneScriptingFloatChannel, default=-0.0628, thinned_from=11, keys [[1800,-0.0628],[1962,-0.0628]]}
       ring_01_l_ctrl.Rotation.X {type=MovieSceneScriptingFloatChannel, default=-3.7319, const [-3.7319,2]}
       ring_01_l_ctrl.Rotation.Y {type=MovieSceneScriptingFloatChannel, default=-3.7906, const [-3.7906,2]}
       ring_01_l_ctrl.Rotation.Z {type=MovieSceneScriptingFloatChannel, default=-84.7111, const [-84.7111,2]}
       ring_01_l_ctrl.Scale.X {type=MovieSceneScriptingFloatChannel, default=1, const [1,2]}
       ring_01_l_ctrl.Scale.Y {type=MovieSceneScriptingFloatChannel, default=1, const [1,2]}
       ring_01_l_ctrl.Scale.Z {type=MovieSceneScriptingFloatChannel, default=1, const [1,2]}
       ring_02_l_ctrl.Location.X {type=MovieSceneScriptingFloatChannel, default=0, const [0,2]}
       ring_02_l_ctrl.Location.Y {type=MovieSceneScriptingFloatChannel, default=0, const [0,2]}
       ring_02_l_ctrl.Location.Z {type=MovieSceneScriptingFloatChannel, default=0, const [0,2]}
       ring_02_l_ctrl.Rotation.X {type=MovieSceneScriptingFloatChannel, default=0, const [0,2]}
       ring_02_l_ctrl.Rotation.Y {type=MovieSceneScriptingFloatChannel, default=0, const [0,2]}
       ring_02_l_ctrl.Rotation.Z {type=MovieSceneScriptingFloatChannel, default=-66.6687, const [-66.6687,2]}
       ring_02_l_ctrl.Scale.X {type=MovieSceneScriptingFloatChannel, default=1, const [1,2]}
       ring_02_l_ctrl.Scale.Y {type=MovieSceneScriptingFloatChannel, default=1, const [1,2]}
       ring_02_l_ctrl.Scale.Z {type=MovieSceneScriptingFloatChannel, default=1, const [1,2]}
       ring_03_l_ctrl.Location.X {type=MovieSceneScriptingFloatChannel, default=0, const [0,2]}
       ring_03_l_ctrl.Location.Y {type=MovieSceneScriptingFloatChannel, default=0, const [0,2]}
       ring_03_l_ctrl.Location.Z {type=MovieSceneScriptingFloatChannel, default=0, const [0,2]}
       ring_03_l_ctrl.Rotation.X {type=MovieSceneScriptingFloatChannel, default=0, const [0,2]}
       ring_03_l_ctrl.Rotation.Y {type=MovieSceneScriptingFloatChannel, default=0, const [0,2]}
       ring_03_l_ctrl.Rotation.Z {type=MovieSceneScriptingFloatChannel, default=-73.2004, const [-73.2004,2]}
       ring_03_l_ctrl.Scale.X {type=MovieSceneScriptingFloatChannel, default=1, const [1,2]}
       ring_03_l_ctrl.Scale.Y {type=MovieSceneScriptingFloatChannel, default=1, const [1,2]}
       ring_03_l_ctrl.Scale.Z {type=MovieSceneScriptingFloatChannel, default=1, const [1,2]}
       pinky_metacarpal_l_ctrl.Location.X {type=MovieSceneScriptingFloatChannel, default=-52.6219, thinned_from=10, keys [[1800,-52.6219],[1801,0],[1961,0],[1962,9.0146]]}
       pinky_metacarpal_l_ctrl.Location.Y {type=MovieSceneScriptingFloatChannel, default=27.7538, thinned_from=11, keys [[1800,27.7538],[1801,0],[1961,0],[1962,-14.498]]}
       pinky_metacarpal_l_ctrl.Location.Z {type=MovieSceneScriptingFloatChannel, default=3.2335, thinned_from=11, keys [[1800,3.2335],[1801,0],[1961,0],[1962,-104.4191]]}
       pinky_metacarpal_l_ctrl.Rotation.X {type=MovieSceneScriptingFloatChannel, default=-0.3585, thinned_from=11, keys [[1800,-0.3585],[1801,0],[1961,0],[1962,-47.2019]]}
       pinky_metacarpal_l_ctrl.Rotation.Y {type=MovieSceneScriptingFloatChannel, default=-0.4898, thinned_from=11, keys [[1800,-0.4898],[1801,0],[1961,0],[1962,-71.6176]]}
       pinky_metacarpal_l_ctrl.Rotation.Z {type=MovieSceneScriptingFloatChannel, default=-1.9615, thinned_from=11, keys [[1800,-1.9615],[1801,0],[1961,0],[1962,111.4913]]}
       pinky_metacarpal_l_ctrl.Scale.X {type=MovieSceneScriptingFloatChannel, default=1, const [1,2]}
       pinky_metacarpal_l_ctrl.Scale.Y {type=MovieSceneScriptingFloatChannel, default=1, const [1,2]}
       pinky_metacarpal_l_ctrl.Scale.Z {type=MovieSceneScriptingFloatChannel, default=1, const [1,2]}
       pinky_01_l_ctrl.Location.X {type=MovieSceneScriptingFloatChannel, default=0.3698, thinned_from=36, keys [[1800,0.3698],[1962,0.3698]]}
       pinky_01_l_ctrl.Location.Y {type=MovieSceneScriptingFloatChannel, default=-1.3951, thinned_from=36, keys [[1800,-1.3951],[1962,-1.3951]]}
       pinky_01_l_ctrl.Location.Z {type=MovieSceneScriptingFloatChannel, default=-0.5172, thinned_from=36, keys [[1800,-0.5172],[1962,-0.5172]]}
       pinky_01_l_ctrl.Rotation.X {type=MovieSceneScriptingFloatChannel, default=-6.1921, const [-6.1921,2]}
       pinky_01_l_ctrl.Rotation.Y {type=MovieSceneScriptingFloatChannel, default=-5.193, const [-5.193,2]}
       pinky_01_l_ctrl.Rotation.Z {type=MovieSceneScriptingFloatChannel, default=-96.8843, const [-96.8843,2]}
       pinky_01_l_ctrl.Scale.X {type=MovieSceneScriptingFloatChannel, default=1, const [1,2]}
       pinky_01_l_ctrl.Scale.Y {type=MovieSceneScriptingFloatChannel, default=1, const [1,2]}
       pinky_01_l_ctrl.Scale.Z {type=MovieSceneScriptingFloatChannel, default=1, const [1,2]}
       pinky_02_l_ctrl.Location.X {type=MovieSceneScriptingFloatChannel, default=0, thinned_from=3, keys [[1800,0],[1962,0]]}
       pinky_02_l_ctrl.Location.Y {type=MovieSceneScriptingFloatChannel, default=0, thinned_from=3, keys [[1800,0],[1962,0]]}
       pinky_02_l_ctrl.Location.Z {type=MovieSceneScriptingFloatChannel, default=0, thinned_from=3, keys [[1800,0],[1962,0]]}
       pinky_02_l_ctrl.Rotation.X {type=MovieSceneScriptingFloatChannel, default=0, keys [[1800,0],[1802,0.0006],[1962,0.0006]]}
       pinky_02_l_ctrl.Rotation.Y {type=MovieSceneScriptingFloatChannel, default=0, thinned_from=4, keys [[1800,0],[1802,0.0003],[1962,0.0003]]}
       pinky_02_l_ctrl.Rotation.Z {type=MovieSceneScriptingFloatChannel, default=-63.0936, keys [[1800,-63.0936],[1801,-62.9683],[1802,-63.0936],[1962,-63.0936]]}
       pinky_02_l_ctrl.Scale.X {type=MovieSceneScriptingFloatChannel, default=1, const [1,2]}
       pinky_02_l_ctrl.Scale.Y {type=MovieSceneScriptingFloatChannel, default=1, const [1,2]}
       pinky_02_l_ctrl.Scale.Z {type=MovieSceneScriptingFloatChannel, default=1, const [1,2]}
       pinky_03_l_ctrl.Location.X {type=MovieSceneScriptingFloatChannel, default=0, const [0,2]}
       pinky_03_l_ctrl.Location.Y {type=MovieSceneScriptingFloatChannel, default=0, const [0,2]}
       pinky_03_l_ctrl.Location.Z {type=MovieSceneScriptingFloatChannel, default=0, const [0,2]}
       pinky_03_l_ctrl.Rotation.X {type=MovieSceneScriptingFloatChannel, default=0, const [0,2]}
       pinky_03_l_ctrl.Rotation.Y {type=MovieSceneScriptingFloatChannel, default=0, const [0,2]}
       pinky_03_l_ctrl.Rotation.Z {type=MovieSceneScriptingFloatChannel, default=-117.2953, const [-117.2953,2]}
       pinky_03_l_ctrl.Scale.X {type=MovieSceneScriptingFloatChannel, default=1, const [1,2]}
       pinky_03_l_ctrl.Scale.Y {type=MovieSceneScriptingFloatChannel, default=1, const [1,2]}
       pinky_03_l_ctrl.Scale.Z {type=MovieSceneScriptingFloatChannel, default=1, const [1,2]}
       hand_l_ik_ctrl.Location.X {
        type=MovieSceneScriptingFloatChannel
        default=42.3628
        keys [[1800,42.3628],[1807,60.9253],[1818,38.0047],[1830,28.5997],[1848,33.0928],[1850,32.9443],[1854,41.4501],[1856,55.3213],[1858,58.8605],[1868,36.8621],[1879,28.5997],[1897,33.0928],[1902,32.3789],[1905,40.3177],[1907,56.2247],[1909,60.7816],[1920,37.433],[1932,28.6124],[1949,33.0878],[1955,32.2687],[1958,31.744],[1960,31.5125],[1961,31.4608],[1962,0]]
       }
       hand_l_ik_ctrl.Location.Y {
        type=MovieSceneScriptingFloatChannel
        default=15.7537
        thinned_from=23
        keys [[1800,15.7537],[1802,20.1308],[1805,23.5851],[1811,27.6881],[1821,20.7397],[1831,16.6326],[1845,19.8116],[1852,20.7277],[1860,27.6879],[1870,20.7396],[1880,16.6324],[1901,20.6525],[1903,20.597],[1913,27.6512],[1922,21.1397],[1932,16.6429],[1954,20.6593],[1961,20.5647],[1962,0]]
       }
       hand_l_ik_ctrl.Location.Z {
        type=MovieSceneScriptingFloatChannel
        default=116.0902
        keys [[1800,116.0902],[1806,132.9363],[1816,81.3174],[1837,97.1111],[1850,95.1458],[1853,101.6412],[1856,123.6824],[1865,81.3174],[1886,97.1111],[1902,94.8119],[1908,131.3706],[1918,81.4565],[1939,97.1085],[1955,94.7858],[1961,94.9616],[1962,0]]
       }
       hand_l_ik_ctrl.Rotation.X {
        type=MovieSceneScriptingFloatChannel
        default=-51.1453
        thinned_from=29
        keys [[1800,-51.1453],[1801,-51.3988],[1808,-31.8439],[1815,-3.4283],[1829,-35.9685],[1836,-48.3402],[1843,-49.6202],[1845,-49.6718],[1854,-47.0043],[1856,-38.53],[1864,-3.429],[1878,-35.968],[1885,-48.3393],[1892,-49.6193],[1894,-49.6709],[1902,-48.7366],[1904,-48.9515],[1917,-3.5877],[1930,-34.6462],[1937,-47.8852],[1944,-49.6345],[1953,-48.945],[1959,-48.3467],[1961,-48.2816],[1962,0]]
       }
       hand_l_ik_ctrl.Rotation.Y {
        type=MovieSceneScriptingFloatChannel
        default=80.6031
        thinned_from=11
        keys [[1800,80.6031],[1811,84.3481],[1837,82.0491],[1860,84.3479],[1886,82.0489],[1904,82.7311],[1913,84.3484],[1939,82.0532],[1961,82.9067],[1962,0]]
       }
       hand_l_ik_ctrl.Rotation.Z {
        type=MovieSceneScriptingFloatChannel
        default=-114.7221
        thinned_from=24
        keys [[1800,-114.7221],[1809,-89.2851],[1815,-65.0317],[1828,-95.6291],[1842,-112.2266],[1845,-112.2289],[1854,-109.3013],[1864,-65.033],[1877,-95.6292],[1891,-112.2262],[1894,-112.2285],[1902,-111.1403],[1904,-111.3257],[1917,-65.1957],[1930,-97.0336],[1944,-112.2382],[1961,-110.6132],[1962,0]]
       }
       hand_l_ik_ctrl.Scale.X {type=MovieSceneScriptingFloatChannel, default=1, const [1,2]}
       hand_l_ik_ctrl.Scale.Y {type=MovieSceneScriptingFloatChannel, default=1, const [1,2]}
       hand_l_ik_ctrl.Scale.Z {type=MovieSceneScriptingFloatChannel, default=1, const [1,2]}
       Softness {type=MovieSceneScriptingFloatChannel, default=0, const [0,2]}
       hand_l_ikorient_ctrl.Rotation.Roll {type=MovieSceneScriptingFloatChannel, default=0, const [0,2]}
       hand_l_ikorient_ctrl.Rotation.Pitch {type=MovieSceneScriptingFloatChannel, default=0, const [0,2]}
       hand_l_ikorient_ctrl.Rotation.Yaw {type=MovieSceneScriptingFloatChannel, default=0, const [0,2]}
       arm_l_pv_ik_ctrl.Location.X {
        type=MovieSceneScriptingFloatChannel
        default=5.4218
        keys [[1800,5.4218],[1801,3.7743],[1803,1.2134],[1804,0.4123],[1807,-0.1922],[1816,2.0696],[1821,3.7303],[1826,4.0558],[1841,2.6268],[1849,1.7772],[1853,1.2376],[1857,-0.0051],[1866,2.4736],[1875,4.0558],[1891,2.5077],[1899,1.6904],[1903,1.4812],[1908,-0.159],[1918,2.273],[1923,3.7986],[1928,4.0531],[1945,2.3343],[1953,1.5766],[1957,1.3572],[1959,1.3113],[1960,1.2918],[1961,1.2819],[1962,0]]
       }
       arm_l_pv_ik_ctrl.Location.Y {
        type=MovieSceneScriptingFloatChannel
        default=140.9263
        thinned_from=21
        keys [[1800,140.9263],[1806,164.6851],[1812,129.81],[1819,109.3354],[1843,119.6056],[1850,118.9689],[1857,155.1317],[1862,123.183],[1868,109.3354],[1892,119.6056],[1902,118.4573],[1908,163.4787],[1914,126.4967],[1920,109.3525],[1944,119.6025],[1957,118.2948],[1961,118.2382],[1962,0]]
       }
       arm_l_pv_ik_ctrl.Location.Z {
        type=MovieSceneScriptingFloatChannel
        default=-1.8166
        keys [[1800,-1.8166],[1801,-4.5308],[1803,-7.1494],[1812,-21.1882],[1822,-7.299],[1832,1.6019],[1849,-3.8424],[1851,-3.7937],[1853,-4.3587],[1858,-18.0864],[1861,-21.1882],[1871,-7.2989],[1881,1.6019],[1898,-3.8424],[1903,-3.3718],[1913,-21.1269],[1923,-8.113],[1933,1.5847],[1951,-3.8369],[1959,-2.7576],[1961,-2.5967],[1962,0]]
       }
       arm_l_pv_ik_ctrl.Rotation.X {
        type=MovieSceneScriptingFloatChannel
        default=7.4963
        thinned_from=21
        keys [[1800,7.4963],[1803,6.4262],[1807,5.5312],[1815,4.5108],[1841,6.2592],[1843,6.2558],[1853,6.1439],[1864,4.5108],[1890,6.2592],[1892,6.2558],[1904,6.1306],[1917,4.5107],[1942,6.2589],[1952,6.1908],[1957,6.1348],[1961,6.121],[1962,0]]
       }
       arm_l_pv_ik_ctrl.Rotation.Y {
        type=MovieSceneScriptingFloatChannel
        default=70.0096
        thinned_from=15
        keys [[1800,70.0096],[1815,62.7171],[1838,68.5225],[1864,62.7171],[1887,68.5225],[1902,67.9931],[1904,68.0168],[1917,62.7332],[1940,68.5201],[1961,67.8304],[1962,0]]
       }
       arm_l_pv_ik_ctrl.Rotation.Z {
        type=MovieSceneScriptingFloatChannel
        default=94.3187
        thinned_from=11
        keys [[1800,94.3187],[1804,95.523],[1821,93.9818],[1854,95.705],[1870,93.9818],[1902,95.6793],[1905,95.654],[1922,93.9823],[1961,95.7469],[1962,0]]
       }
       arm_l_pv_ik_ctrl.Scale.X {type=MovieSceneScriptingFloatChannel, default=1, const [1,2]}
       arm_l_pv_ik_ctrl.Scale.Y {type=MovieSceneScriptingFloatChannel, default=1, const [1,2]}
       arm_l_pv_ik_ctrl.Scale.Z {type=MovieSceneScriptingFloatChannel, default=1, const [1,2]}
       upperarm_r_fk_ctrl.Location.X {type=MovieSceneScriptingFloatChannel, default=0, thinned_from=142, keys [[1800,0],[1962,0]]}
       upperarm_r_fk_ctrl.Location.Y {type=MovieSceneScriptingFloatChannel, default=0, thinned_from=143, keys [[1800,0],[1962,0]]}
       upperarm_r_fk_ctrl.Location.Z {type=MovieSceneScriptingFloatChannel, default=0, thinned_from=130, keys [[1800,0],[1962,0]]}
       upperarm_r_fk_ctrl.Rotation.X {
        type=MovieSceneScriptingFloatChannel
        default=98.0828
        thinned_from=18
        keys [[1800,98.0828],[1803,97.2549],[1804,97.3081],[1825,92.067],[1848,94.9113],[1850,94.8931],[1857,96.7793],[1874,92.067],[1897,94.9113],[1902,94.7791],[1908,97.1323],[1927,92.0728],[1950,94.9091],[1961,94.6044],[1962,0]]
       }
       upperarm_r_fk_ctrl.Rotation.Y {
        type=MovieSceneScriptingFloatChannel
        default=35.1406
        keys [[1800,35.1406],[1820,28.6606],[1841,25.7051],[1849,25.9573],[1857,32.4548],[1869,28.6606],[1890,25.7051],[1899,26.0112],[1903,26.4869],[1908,33.4319],[1925,27.7131],[1943,25.7065],[1961,26.365],[1962,0]]
       }
       upperarm_r_fk_ctrl.Rotation.Z {
        type=MovieSceneScriptingFloatChannel
        default=-173.0015
        thinned_from=18
        keys [[1800,-173.0015],[1802,-174.2719],[1805,-173.5491],[1826,-187.1528],[1848,-180.1299],[1850,-180.1531],[1857,-174.8208],[1875,-187.1528],[1897,-180.1299],[1902,-180.413],[1908,-173.8589],[1927,-187.123],[1950,-180.131],[1961,-180.8126],[1962,-360]]
       }
       upperarm_r_fk_ctrl.Scale.X {type=MovieSceneScriptingFloatChannel, default=1, const [1,2]}
       upperarm_r_fk_ctrl.Scale.Y {type=MovieSceneScriptingFloatChannel, default=1, const [1,2]}
       upperarm_r_fk_ctrl.Scale.Z {type=MovieSceneScriptingFloatChannel, default=1, const [1,2]}
       lowerarm_r_fk_ctrl.Location.X {type=MovieSceneScriptingFloatChannel, default=0, thinned_from=24, keys [[1800,0],[1962,0]]}
       lowerarm_r_fk_ctrl.Location.Y {type=MovieSceneScriptingFloatChannel, default=0, thinned_from=22, keys [[1800,0],[1962,0]]}
       lowerarm_r_fk_ctrl.Location.Z {type=MovieSceneScriptingFloatChannel, default=0, thinned_from=23, keys [[1800,0],[1962,0]]}
       lowerarm_r_fk_ctrl.Rotation.X {
        type=MovieSceneScriptingFloatChannel
        default=0.1326
        thinned_from=12
        keys [[1800,0.1326],[1805,0.1299],[1807,0.1311],[1904,0.1307],[1909,0.1311],[1961,0.1311],[1962,0]]
       }
       lowerarm_r_fk_ctrl.Rotation.Y {type=MovieSceneScriptingFloatChannel, default=1.4143, thinned_from=12, keys [[1800,1.4143],[1805,1.4353],[1961,1.4359],[1962,0]]}
       lowerarm_r_fk_ctrl.Rotation.Z {
        type=MovieSceneScriptingFloatChannel
        default=-18.2675
        thinned_from=12
        keys [[1800,-18.2675],[1805,-18.5572],[1807,-18.5781],[1961,-18.5781],[1962,0]]
       }
       lowerarm_r_fk_ctrl.Scale.X {type=MovieSceneScriptingFloatChannel, default=1, const [1,2]}
       lowerarm_r_fk_ctrl.Scale.Y {type=MovieSceneScriptingFloatChannel, default=1, const [1,2]}
       lowerarm_r_fk_ctrl.Scale.Z {type=MovieSceneScriptingFloatChannel, default=1, const [1,2]}
       hand_r_fk_ctrl.Location.X {type=MovieSceneScriptingFloatChannel, default=0, thinned_from=14, keys [[1800,0],[1962,0]]}
       hand_r_fk_ctrl.Location.Y {type=MovieSceneScriptingFloatChannel, default=0, thinned_from=16, keys [[1800,0],[1904,0],[1905,0.0003],[1908,0.0005],[1909,0],[1962,0]]}
       hand_r_fk_ctrl.Location.Z {
        type=MovieSceneScriptingFloatChannel
        default=0
        thinned_from=17
        keys [[1800,0],[1904,0],[1905,0.0045],[1906,0.0045],[1907,-0.0012],[1908,-0.0012],[1909,0],[1962,0]]
       }
       hand_r_fk_ctrl.Rotation.X {type=MovieSceneScriptingFloatChannel, default=-9.0022, thinned_from=12, keys [[1800,-9.0022],[1805,-9.121],[1961,-9.121],[1962,0]]}
       hand_r_fk_ctrl.Rotation.Y {
        type=MovieSceneScriptingFloatChannel
        default=-4.1823
        thinned_from=12
        keys [[1800,-4.1823],[1805,-4.3528],[1905,-4.343],[1907,-4.3555],[1961,-4.3528],[1962,0]]
       }
       hand_r_fk_ctrl.Rotation.Z {type=MovieSceneScriptingFloatChannel, default=-30.3401, thinned_from=12, keys [[1800,-30.3401],[1805,-30.3815],[1961,-30.3815],[1962,0]]}
       hand_r_fk_ctrl.Scale.X {type=MovieSceneScriptingFloatChannel, default=1, const [1,2]}
       hand_r_fk_ctrl.Scale.Y {type=MovieSceneScriptingFloatChannel, default=1, const [1,2]}
       hand_r_fk_ctrl.Scale.Z {type=MovieSceneScriptingFloatChannel, default=1, const [1,2]}
       thumb_01_r_ctrl.Location.X {type=MovieSceneScriptingFloatChannel, default=-51.5958, thinned_from=11, keys [[1800,-51.5958],[1801,0],[1961,0],[1962,60.0617]]}
       thumb_01_r_ctrl.Location.Y {type=MovieSceneScriptingFloatChannel, default=28.2255, thinned_from=11, keys [[1800,28.2255],[1801,0],[1961,0],[1962,66.0438]]}
       thumb_01_r_ctrl.Location.Z {type=MovieSceneScriptingFloatChannel, default=-19.5666, thinned_from=11, keys [[1800,-19.5666],[1801,0],[1961,0],[1962,31.2541]]}
       thumb_01_r_ctrl.Rotation.X {
        type=MovieSceneScriptingFloatChannel
        default=0.2215
        thinned_from=12
        keys [[1800,0.2215],[1801,-1.5889],[1804,-1.5584],[1961,-1.5584],[1962,104.2378]]
       }
       thumb_01_r_ctrl.Rotation.Y {
        type=MovieSceneScriptingFloatChannel
        default=-1.5988
        thinned_from=12
        keys [[1800,-1.5988],[1801,0.8687],[1804,0.8708],[1961,0.8708],[1962,23.4175]]
       }
       thumb_01_r_ctrl.Rotation.Z {
        type=MovieSceneScriptingFloatChannel
        default=-4.9235
        thinned_from=12
        keys [[1800,-4.9235],[1801,-13.9046],[1804,-13.9228],[1961,-13.9228],[1962,22.0451]]
       }
       thumb_01_r_ctrl.Scale.X {type=MovieSceneScriptingFloatChannel, default=1, const [1,2]}
       thumb_01_r_ctrl.Scale.Y {type=MovieSceneScriptingFloatChannel, default=1, const [1,2]}
       thumb_01_r_ctrl.Scale.Z {type=MovieSceneScriptingFloatChannel, default=1, const [1,2]}
       thumb_02_r_ctrl.Location.X {type=MovieSceneScriptingFloatChannel, default=0, const [0,2]}
       thumb_02_r_ctrl.Location.Y {type=MovieSceneScriptingFloatChannel, default=0, const [0,2]}
       thumb_02_r_ctrl.Location.Z {type=MovieSceneScriptingFloatChannel, default=0, const [0,2]}
       thumb_02_r_ctrl.Rotation.X {type=MovieSceneScriptingFloatChannel, default=0, const [0,2]}
       thumb_02_r_ctrl.Rotation.Y {type=MovieSceneScriptingFloatChannel, default=0, const [0,2]}
       thumb_02_r_ctrl.Rotation.Z {type=MovieSceneScriptingFloatChannel, default=0, const [0,2]}
       thumb_02_r_ctrl.Scale.X {type=MovieSceneScriptingFloatChannel, default=1, const [1,2]}
       thumb_02_r_ctrl.Scale.Y {type=MovieSceneScriptingFloatChannel, default=1, const [1,2]}
       thumb_02_r_ctrl.Scale.Z {type=MovieSceneScriptingFloatChannel, default=1, const [1,2]}
       thumb_03_r_ctrl.Location.X {type=MovieSceneScriptingFloatChannel, default=0, const [0,2]}
       thumb_03_r_ctrl.Location.Y {type=MovieSceneScriptingFloatChannel, default=0, const [0,2]}
       thumb_03_r_ctrl.Location.Z {type=MovieSceneScriptingFloatChannel, default=0, const [0,2]}
       thumb_03_r_ctrl.Rotation.X {type=MovieSceneScriptingFloatChannel, default=8.1645, const [8.1645,2]}
       thumb_03_r_ctrl.Rotation.Y {type=MovieSceneScriptingFloatChannel, default=-9.9357, const [-9.9357,2]}
       thumb_03_r_ctrl.Rotation.Z {type=MovieSceneScriptingFloatChannel, default=-50.7089, const [-50.7089,2]}
       thumb_03_r_ctrl.Scale.X {type=MovieSceneScriptingFloatChannel, default=1, const [1,2]}
       thumb_03_r_ctrl.Scale.Y {type=MovieSceneScriptingFloatChannel, default=1, const [1,2]}
       thumb_03_r_ctrl.Scale.Z {type=MovieSceneScriptingFloatChannel, default=1, const [1,2]}
       index_metacarpal_r_ctrl.Location.X {type=MovieSceneScriptingFloatChannel, default=-57.9099, thinned_from=11, keys [[1800,-57.9099],[1801,0],[1961,0],[1962,30.6731]]}
       index_metacarpal_r_ctrl.Location.Y {type=MovieSceneScriptingFloatChannel, default=22.2408, thinned_from=8, keys [[1800,22.2408],[1801,0],[1961,0],[1962,-2.5522]]}
       index_metacarpal_r_ctrl.Location.Z {type=MovieSceneScriptingFloatChannel, default=-0.1767, thinned_from=11, keys [[1800,-0.1767],[1801,0],[1961,0],[1962,-89.1937]]}
       index_metacarpal_r_ctrl.Rotation.X {type=MovieSceneScriptingFloatChannel, default=-3.4597, thinned_from=11, keys [[1800,-3.4597],[1801,0],[1961,0],[1962,15.9284]]}
       index_metacarpal_r_ctrl.Rotation.Y {type=MovieSceneScriptingFloatChannel, default=-8.2829, thinned_from=11, keys [[1800,-8.2829],[1801,0],[1961,0],[1962,-76.8355]]}
       index_metacarpal_r_ctrl.Rotation.Z {type=MovieSceneScriptingFloatChannel, default=-2.9061, thinned_from=11, keys [[1800,-2.9061],[1801,0],[1961,0],[1962,63.5058]]}
       index_metacarpal_r_ctrl.Scale.X {type=MovieSceneScriptingFloatChannel, default=1, const [1,2]}
       index_metacarpal_r_ctrl.Scale.Y {type=MovieSceneScriptingFloatChannel, default=1, const [1,2]}
       index_metacarpal_r_ctrl.Scale.Z {type=MovieSceneScriptingFloatChannel, default=1, const [1,2]}
       index_01_r_ctrl.Location.X {type=MovieSceneScriptingFloatChannel, default=0.1105, const [0.1105,2]}
       index_01_r_ctrl.Location.Y {type=MovieSceneScriptingFloatChannel, default=-0.4345, const [-0.4345,2]}
       index_01_r_ctrl.Location.Z {type=MovieSceneScriptingFloatChannel, default=0.5473, const [0.5473,2]}
       index_01_r_ctrl.Rotation.X {type=MovieSceneScriptingFloatChannel, default=2.6836, const [2.6836,2]}
       index_01_r_ctrl.Rotation.Y {type=MovieSceneScriptingFloatChannel, default=-1.2982, const [-1.2982,2]}
       index_01_r_ctrl.Rotation.Z {type=MovieSceneScriptingFloatChannel, default=-68.0656, const [-68.0656,2]}
       index_01_r_ctrl.Scale.X {type=MovieSceneScriptingFloatChannel, default=1, const [1,2]}
       index_01_r_ctrl.Scale.Y {type=MovieSceneScriptingFloatChannel, default=1, const [1,2]}
       index_01_r_ctrl.Scale.Z {type=MovieSceneScriptingFloatChannel, default=1, const [1,2]}
       index_02_r_ctrl.Location.X {type=MovieSceneScriptingFloatChannel, default=0, const [0,2]}
       index_02_r_ctrl.Location.Y {type=MovieSceneScriptingFloatChannel, default=0, const [0,2]}
       index_02_r_ctrl.Location.Z {type=MovieSceneScriptingFloatChannel, default=0, const [0,2]}
       index_02_r_ctrl.Rotation.X {type=MovieSceneScriptingFloatChannel, default=0, const [0,2]}
       index_02_r_ctrl.Rotation.Y {type=MovieSceneScriptingFloatChannel, default=0, const [0,2]}
       index_02_r_ctrl.Rotation.Z {type=MovieSceneScriptingFloatChannel, default=-60.8298, const [-60.8298,2]}
       index_02_r_ctrl.Scale.X {type=MovieSceneScriptingFloatChannel, default=1, const [1,2]}
       index_02_r_ctrl.Scale.Y {type=MovieSceneScriptingFloatChannel, default=1, const [1,2]}
       index_02_r_ctrl.Scale.Z {type=MovieSceneScriptingFloatChannel, default=1, const [1,2]}
       index_03_r_ctrl.Location.X {type=MovieSceneScriptingFloatChannel, default=0, const [0,2]}
       index_03_r_ctrl.Location.Y {type=MovieSceneScriptingFloatChannel, default=0, const [0,2]}
       index_03_r_ctrl.Location.Z {type=MovieSceneScriptingFloatChannel, default=0, const [0,2]}
       index_03_r_ctrl.Rotation.X {type=MovieSceneScriptingFloatChannel, default=0, const [0,2]}
       index_03_r_ctrl.Rotation.Y {type=MovieSceneScriptingFloatChannel, default=0, const [0,2]}
       index_03_r_ctrl.Rotation.Z {type=MovieSceneScriptingFloatChannel, default=-78.806, const [-78.806,2]}
       index_03_r_ctrl.Scale.X {type=MovieSceneScriptingFloatChannel, default=1, const [1,2]}
       index_03_r_ctrl.Scale.Y {type=MovieSceneScriptingFloatChannel, default=1, const [1,2]}
       index_03_r_ctrl.Scale.Z {type=MovieSceneScriptingFloatChannel, default=1, const [1,2]}
       middle_metacarpal_r_ctrl.Location.X {type=MovieSceneScriptingFloatChannel, default=0, const [0,2]}
       middle_metacarpal_r_ctrl.Location.Y {type=MovieSceneScriptingFloatChannel, default=0, const [0,2]}
       middle_metacarpal_r_ctrl.Location.Z {type=MovieSceneScriptingFloatChannel, default=0, const [0,2]}
       middle_metacarpal_r_ctrl.Rotation.X {type=MovieSceneScriptingFloatChannel, default=-0.0209, const [-0.0209,2]}
       middle_metacarpal_r_ctrl.Rotation.Y {type=MovieSceneScriptingFloatChannel, default=0.0017, const [0.0017,2]}
       middle_metacarpal_r_ctrl.Rotation.Z {type=MovieSceneScriptingFloatChannel, default=-0.0067, const [-0.0067,2]}
       middle_metacarpal_r_ctrl.Scale.X {type=MovieSceneScriptingFloatChannel, default=1, const [1,2]}
       middle_metacarpal_r_ctrl.Scale.Y {type=MovieSceneScriptingFloatChannel, default=1, const [1,2]}
       middle_metacarpal_r_ctrl.Scale.Z {type=MovieSceneScriptingFloatChannel, default=1, const [1,2]}
       middle_01_r_ctrl.Location.X {type=MovieSceneScriptingFloatChannel, default=0.0387, thinned_from=10, keys [[1800,0.0387],[1961,0.0387],[1962,0]]}
       middle_01_r_ctrl.Location.Y {type=MovieSceneScriptingFloatChannel, default=-0.3373, thinned_from=10, keys [[1800,-0.3373],[1961,-0.3373],[1962,0]]}
       middle_01_r_ctrl.Location.Z {type=MovieSceneScriptingFloatChannel, default=0.1567, thinned_from=10, keys [[1800,0.1567],[1961,0.1567],[1962,0]]}
       middle_01_r_ctrl.Rotation.X {type=MovieSceneScriptingFloatChannel, default=-2.1499, thinned_from=10, keys [[1800,-2.1499],[1961,-2.1499],[1962,0]]}
       middle_01_r_ctrl.Rotation.Y {type=MovieSceneScriptingFloatChannel, default=-5.7277, thinned_from=10, keys [[1800,-5.7277],[1961,-5.7277],[1962,0]]}
       middle_01_r_ctrl.Rotation.Z {type=MovieSceneScriptingFloatChannel, default=-77.2709, thinned_from=10, keys [[1800,-77.2709],[1961,-77.2709],[1962,0]]}
       middle_01_r_ctrl.Scale.X {type=MovieSceneScriptingFloatChannel, default=1, const [1,2]}
       middle_01_r_ctrl.Scale.Y {type=MovieSceneScriptingFloatChannel, default=1, const [1,2]}
       middle_01_r_ctrl.Scale.Z {type=MovieSceneScriptingFloatChannel, default=1, const [1,2]}
       middle_02_r_ctrl.Location.X {type=MovieSceneScriptingFloatChannel, default=17.1301, thinned_from=11, keys [[1800,17.1301],[1801,0],[1961,0],[1962,11.7803]]}
       middle_02_r_ctrl.Location.Y {type=MovieSceneScriptingFloatChannel, default=-59.368, thinned_from=7, keys [[1800,-59.368],[1801,0],[1961,0],[1962,4.348]]}
       middle_02_r_ctrl.Location.Z {type=MovieSceneScriptingFloatChannel, default=3.448, thinned_from=11, keys [[1800,3.448],[1801,0],[1961,0],[1962,-101.8015]]}
       middle_02_r_ctrl.Rotation.X {type=MovieSceneScriptingFloatChannel, default=2.226, thinned_from=11, keys [[1800,2.226],[1801,0.0063],[1961,0.0063],[1962,-1.2728]]}
       middle_02_r_ctrl.Rotation.Y {type=MovieSceneScriptingFloatChannel, default=8.1164, thinned_from=11, keys [[1800,8.1164],[1801,-0.0122],[1961,-0.0122],[1962,76.0877]]}
       middle_02_r_ctrl.Rotation.Z {
        type=MovieSceneScriptingFloatChannel
        default=-62.2864
        thinned_from=11
        keys [[1800,-62.2864],[1801,-58.0462],[1961,-58.0462],[1962,-59.7705]]
       }
       middle_02_r_ctrl.Scale.X {type=MovieSceneScriptingFloatChannel, default=1, const [1,2]}
       middle_02_r_ctrl.Scale.Y {type=MovieSceneScriptingFloatChannel, default=1, const [1,2]}
       middle_02_r_ctrl.Scale.Z {type=MovieSceneScriptingFloatChannel, default=1, const [1,2]}
       middle_03_r_ctrl.Location.X {type=MovieSceneScriptingFloatChannel, default=0, const [0,2]}
       middle_03_r_ctrl.Location.Y {type=MovieSceneScriptingFloatChannel, default=0, const [0,2]}
       middle_03_r_ctrl.Location.Z {type=MovieSceneScriptingFloatChannel, default=0, const [0,2]}
       middle_03_r_ctrl.Rotation.X {type=MovieSceneScriptingFloatChannel, default=-0.007, const [-0.007,2]}
       middle_03_r_ctrl.Rotation.Y {type=MovieSceneScriptingFloatChannel, default=-0.0086, const [-0.0086,2]}
       middle_03_r_ctrl.Rotation.Z {type=MovieSceneScriptingFloatChannel, default=-64.5776, const [-64.5776,2]}
       middle_03_r_ctrl.Scale.X {type=MovieSceneScriptingFloatChannel, default=1, const [1,2]}
       middle_03_r_ctrl.Scale.Y {type=MovieSceneScriptingFloatChannel, default=1, const [1,2]}
       middle_03_r_ctrl.Scale.Z {type=MovieSceneScriptingFloatChannel, default=1, const [1,2]}
       ring_metacarpal_r_ctrl.Location.X {type=MovieSceneScriptingFloatChannel, default=-55.4122, thinned_from=11, keys [[1800,-55.4122],[1801,0],[1961,0],[1962,15.0741]]}
       ring_metacarpal_r_ctrl.Location.Y {type=MovieSceneScriptingFloatChannel, default=25.3742, thinned_from=11, keys [[1800,25.3742],[1801,0],[1961,0],[1962,-7.2411]]}
       ring_metacarpal_r_ctrl.Location.Z {type=MovieSceneScriptingFloatChannel, default=7.376, thinned_from=11, keys [[1800,7.376],[1801,0],[1961,0],[1962,-96.7528]]}
       ring_metacarpal_r_ctrl.Rotation.X {type=MovieSceneScriptingFloatChannel, default=-3.1709, thinned_from=11, keys [[1800,-3.1709],[1801,0],[1961,0],[1962,-31.1788]]}
       ring_metacarpal_r_ctrl.Rotation.Y {type=MovieSceneScriptingFloatChannel, default=-7.7925, thinned_from=11, keys [[1800,-7.7925],[1801,0],[1961,0],[1962,-72.4594]]}
       ring_metacarpal_r_ctrl.Rotation.Z {type=MovieSceneScriptingFloatChannel, default=-4.2287, thinned_from=11, keys [[1800,-4.2287],[1801,0],[1961,0],[1962,101.5233]]}
       ring_metacarpal_r_ctrl.Scale.X {type=MovieSceneScriptingFloatChannel, default=1, const [1,2]}
       ring_metacarpal_r_ctrl.Scale.Y {type=MovieSceneScriptingFloatChannel, default=1, const [1,2]}
       ring_metacarpal_r_ctrl.Scale.Z {type=MovieSceneScriptingFloatChannel, default=1, const [1,2]}
       ring_01_r_ctrl.Location.X {type=MovieSceneScriptingFloatChannel, default=0.1701, const [0.1701,2]}
       ring_01_r_ctrl.Location.Y {type=MovieSceneScriptingFloatChannel, default=-0.6802, const [-0.6802,2]}
       ring_01_r_ctrl.Location.Z {type=MovieSceneScriptingFloatChannel, default=-0.0628, const [-0.0628,2]}
       ring_01_r_ctrl.Rotation.X {type=MovieSceneScriptingFloatChannel, default=-3.7319, const [-3.7319,2]}
       ring_01_r_ctrl.Rotation.Y {type=MovieSceneScriptingFloatChannel, default=-3.7907, const [-3.7907,2]}
       ring_01_r_ctrl.Rotation.Z {type=MovieSceneScriptingFloatChannel, default=-84.7111, const [-84.7111,2]}
       ring_01_r_ctrl.Scale.X {type=MovieSceneScriptingFloatChannel, default=1, const [1,2]}
       ring_01_r_ctrl.Scale.Y {type=MovieSceneScriptingFloatChannel, default=1, const [1,2]}
       ring_01_r_ctrl.Scale.Z {type=MovieSceneScriptingFloatChannel, default=1, const [1,2]}
       ring_02_r_ctrl.Location.X {type=MovieSceneScriptingFloatChannel, default=0, const [0,2]}
       ring_02_r_ctrl.Location.Y {type=MovieSceneScriptingFloatChannel, default=0, const [0,2]}
       ring_02_r_ctrl.Location.Z {type=MovieSceneScriptingFloatChannel, default=0, const [0,2]}
       ring_02_r_ctrl.Rotation.X {type=MovieSceneScriptingFloatChannel, default=0, const [0,2]}
       ring_02_r_ctrl.Rotation.Y {type=MovieSceneScriptingFloatChannel, default=0, const [0,2]}
       ring_02_r_ctrl.Rotation.Z {type=MovieSceneScriptingFloatChannel, default=-66.6687, const [-66.6687,2]}
       ring_02_r_ctrl.Scale.X {type=MovieSceneScriptingFloatChannel, default=1, const [1,2]}
       ring_02_r_ctrl.Scale.Y {type=MovieSceneScriptingFloatChannel, default=1, const [1,2]}
       ring_02_r_ctrl.Scale.Z {type=MovieSceneScriptingFloatChannel, default=1, const [1,2]}
       ring_03_r_ctrl.Location.X {type=MovieSceneScriptingFloatChannel, default=0, const [0,2]}
       ring_03_r_ctrl.Location.Y {type=MovieSceneScriptingFloatChannel, default=0, const [0,2]}
       ring_03_r_ctrl.Location.Z {type=MovieSceneScriptingFloatChannel, default=0, const [0,2]}
       ring_03_r_ctrl.Rotation.X {type=MovieSceneScriptingFloatChannel, default=0, const [0,2]}
       ring_03_r_ctrl.Rotation.Y {type=MovieSceneScriptingFloatChannel, default=0, const [0,2]}
       ring_03_r_ctrl.Rotation.Z {type=MovieSceneScriptingFloatChannel, default=-73.2004, const [-73.2004,2]}
       ring_03_r_ctrl.Scale.X {type=MovieSceneScriptingFloatChannel, default=1, const [1,2]}
       ring_03_r_ctrl.Scale.Y {type=MovieSceneScriptingFloatChannel, default=1, const [1,2]}
       ring_03_r_ctrl.Scale.Z {type=MovieSceneScriptingFloatChannel, default=1, const [1,2]}
       pinky_metacarpal_r_ctrl.Location.X {type=MovieSceneScriptingFloatChannel, default=-53.8335, thinned_from=16, keys [[1800,-53.8335],[1801,0],[1961,0],[1962,14.3359]]}
       pinky_metacarpal_r_ctrl.Location.Y {type=MovieSceneScriptingFloatChannel, default=27.9076, thinned_from=11, keys [[1800,27.9076],[1801,0],[1961,0],[1962,-8.2517]]}
       pinky_metacarpal_r_ctrl.Location.Z {type=MovieSceneScriptingFloatChannel, default=7.6874, thinned_from=13, keys [[1800,7.6874],[1802,0],[1961,0],[1962,-98.3072]]}
       pinky_metacarpal_r_ctrl.Rotation.X {type=MovieSceneScriptingFloatChannel, default=-3.4214, thinned_from=12, keys [[1800,-3.4214],[1802,0.004],[1804,0],[1961,0],[1962,-41.5162]]}
       pinky_metacarpal_r_ctrl.Rotation.Y {
        type=MovieSceneScriptingFloatChannel
        default=-7.5331
        thinned_from=13
        keys [[1800,-7.5331],[1801,-0.0254],[1802,0.0222],[1804,0],[1961,0],[1962,-71.199]]
       }
       pinky_metacarpal_r_ctrl.Rotation.Z {
        type=MovieSceneScriptingFloatChannel
        default=-4.4745
        thinned_from=12
        keys [[1800,-4.4745],[1802,0.0012],[1804,0],[1961,0],[1962,107.8429]]
       }
       pinky_metacarpal_r_ctrl.Scale.X {type=MovieSceneScriptingFloatChannel, default=1, const [1,2]}
       pinky_metacarpal_r_ctrl.Scale.Y {type=MovieSceneScriptingFloatChannel, default=1, const [1,2]}
       pinky_metacarpal_r_ctrl.Scale.Z {type=MovieSceneScriptingFloatChannel, default=1, const [1,2]}
       pinky_01_r_ctrl.Location.X {type=MovieSceneScriptingFloatChannel, default=0.3698, const [0.3698,2]}
       pinky_01_r_ctrl.Location.Y {type=MovieSceneScriptingFloatChannel, default=-1.3951, const [-1.3951,2]}
       pinky_01_r_ctrl.Location.Z {type=MovieSceneScriptingFloatChannel, default=-0.5172, const [-0.5172,2]}
       pinky_01_r_ctrl.Rotation.X {type=MovieSceneScriptingFloatChannel, default=-6.1921, const [-6.1921,2]}
       pinky_01_r_ctrl.Rotation.Y {type=MovieSceneScriptingFloatChannel, default=-5.1929, const [-5.1929,2]}
       pinky_01_r_ctrl.Rotation.Z {type=MovieSceneScriptingFloatChannel, default=-96.8843, const [-96.8843,2]}
       pinky_01_r_ctrl.Scale.X {type=MovieSceneScriptingFloatChannel, default=1, const [1,2]}
       pinky_01_r_ctrl.Scale.Y {type=MovieSceneScriptingFloatChannel, default=1, const [1,2]}
       pinky_01_r_ctrl.Scale.Z {type=MovieSceneScriptingFloatChannel, default=1, const [1,2]}
       pinky_02_r_ctrl.Location.X {type=MovieSceneScriptingFloatChannel, default=0, const [0,2]}
       pinky_02_r_ctrl.Location.Y {type=MovieSceneScriptingFloatChannel, default=0, const [0,2]}
       pinky_02_r_ctrl.Location.Z {type=MovieSceneScriptingFloatChannel, default=0, const [0,2]}
       pinky_02_r_ctrl.Rotation.X {type=MovieSceneScriptingFloatChannel, default=0, const [0,2]}
       pinky_02_r_ctrl.Rotation.Y {type=MovieSceneScriptingFloatChannel, default=0, const [0,2]}
       pinky_02_r_ctrl.Rotation.Z {type=MovieSceneScriptingFloatChannel, default=-63.0936, const [-63.0936,2]}
       pinky_02_r_ctrl.Scale.X {type=MovieSceneScriptingFloatChannel, default=1, const [1,2]}
       pinky_02_r_ctrl.Scale.Y {type=MovieSceneScriptingFloatChannel, default=1, const [1,2]}
       pinky_02_r_ctrl.Scale.Z {type=MovieSceneScriptingFloatChannel, default=1, const [1,2]}
       pinky_03_r_ctrl.Location.X {type=MovieSceneScriptingFloatChannel, default=0, const [0,2]}
       pinky_03_r_ctrl.Location.Y {type=MovieSceneScriptingFloatChannel, default=0, const [0,2]}
       pinky_03_r_ctrl.Location.Z {type=MovieSceneScriptingFloatChannel, default=0, const [0,2]}
       pinky_03_r_ctrl.Rotation.X {type=MovieSceneScriptingFloatChannel, default=0, const [0,2]}
       pinky_03_r_ctrl.Rotation.Y {type=MovieSceneScriptingFloatChannel, default=0, const [0,2]}
       pinky_03_r_ctrl.Rotation.Z {type=MovieSceneScriptingFloatChannel, default=-117.2953, const [-117.2953,2]}
       pinky_03_r_ctrl.Scale.X {type=MovieSceneScriptingFloatChannel, default=1, const [1,2]}
       pinky_03_r_ctrl.Scale.Y {type=MovieSceneScriptingFloatChannel, default=1, const [1,2]}
       pinky_03_r_ctrl.Scale.Z {type=MovieSceneScriptingFloatChannel, default=1, const [1,2]}
       hand_r_ik_ctrl.Location.X {
        type=MovieSceneScriptingFloatChannel
        default=47.1085
        thinned_from=23
        keys [[1800,47.1085],[1802,53.9275],[1807,65.0337],[1818,39.8079],[1829,31.8149],[1848,37.2168],[1850,37.1696],[1854,45.9473],[1856,59.4624],[1858,62.4989],[1868,38.6076],[1878,31.8149],[1897,37.2168],[1902,36.6819],[1909,64.7672],[1919,40.4228],[1930,31.839],[1950,37.2088],[1956,36.3906],[1959,35.8809],[1961,35.7108],[1962,0]]
       }
       hand_r_ik_ctrl.Location.Y {
        type=MovieSceneScriptingFloatChannel
        default=8.4887
        keys [[1800,8.4887],[1815,18.7953],[1836,14.7986],[1847,15.5195],[1851,15.2156],[1855,12.8945],[1859,16.6394],[1864,18.7953],[1885,14.7986],[1896,15.5195],[1901,15.2833],[1903,14.9194],[1906,11.3141],[1911,16.0983],[1916,18.7146],[1938,14.7994],[1949,15.5164],[1955,15.1683],[1958,14.9045],[1960,14.7772],[1961,14.7479],[1962,0]]
       }
       hand_r_ik_ctrl.Location.Z {
        type=MovieSceneScriptingFloatChannel
        default=110.9436
        keys [[1800,110.9436],[1806,127.9802],[1816,75.9537],[1837,90.562],[1850,87.9959],[1853,94.9057],[1856,118.4838],[1865,75.9537],[1886,90.562],[1902,87.6127],[1908,126.4663],[1918,76.003],[1939,90.5466],[1955,87.5836],[1961,87.8088],[1962,0]]
       }
       hand_r_ik_ctrl.Rotation.X {
        type=MovieSceneScriptingFloatChannel
        default=-28.9394
        keys [[1800,-28.9394],[1816,-15.2382],[1838,-41.0977],[1851,-34.7268],[1858,-21.3383],[1865,-15.2383],[1887,-41.0977],[1902,-34.28],[1908,-24.1147],[1918,-15.3412],[1940,-41.0955],[1956,-34.1842],[1961,-34.7021],[1962,0]]
       }
       hand_r_ik_ctrl.Rotation.Y {
        type=MovieSceneScriptingFloatChannel
        default=75.1674
        thinned_from=17
        keys [[1800,75.1674],[1831,83.0685],[1837,82.9804],[1848,83.1912],[1857,78.7405],[1880,83.0685],[1886,82.9804],[1897,83.1912],[1907,77.7616],[1932,83.069],[1939,82.9834],[1950,83.191],[1961,82.9035],[1962,0]]
       }
       hand_r_ik_ctrl.Rotation.Z {
        type=MovieSceneScriptingFloatChannel
        default=-98.4306
        keys [[1800,-98.4306],[1816,-80.4233],[1838,-105.3587],[1851,-99.1071],[1865,-80.4234],[1887,-105.3587],[1906,-94.8174],[1918,-80.4608],[1940,-105.3545],[1956,-98.5059],[1961,-99.0779],[1962,0]]
       }
       hand_r_ik_ctrl.Scale.X {type=MovieSceneScriptingFloatChannel, default=1, const [1,2]}
       hand_r_ik_ctrl.Scale.Y {type=MovieSceneScriptingFloatChannel, default=1, const [1,2]}
       hand_r_ik_ctrl.Scale.Z {type=MovieSceneScriptingFloatChannel, default=1, const [1,2]}
       Softness_2 {type=MovieSceneScriptingFloatChannel, default=0, const [0,2]}
       hand_r_ikorient_ctrl.Rotation.Roll {type=MovieSceneScriptingFloatChannel, default=0, const [0,2]}
       hand_r_ikorient_ctrl.Rotation.Pitch {type=MovieSceneScriptingFloatChannel, default=0, const [0,2]}
       hand_r_ikorient_ctrl.Rotation.Yaw {type=MovieSceneScriptingFloatChannel, default=0, const [0,2]}
       arm_r_pv_ik_ctrl.Location.X {
        type=MovieSceneScriptingFloatChannel
        default=12.1878
        thinned_from=19
        keys [[1800,12.1878],[1803,12.8598],[1814,9.5065],[1826,7.1208],[1841,8.1048],[1848,8.53],[1852,9.0662],[1856,11.1533],[1875,7.1208],[1891,8.1735],[1899,8.6264],[1903,8.8824],[1907,12.0242],[1917,9.0601],[1928,7.1227],[1957,8.786],[1961,8.8014],[1962,0]]
       }
       arm_r_pv_ik_ctrl.Location.Y {
        type=MovieSceneScriptingFloatChannel
        default=135.3175
        thinned_from=21
        keys [[1800,135.3175],[1806,159.1205],[1812,124.6002],[1819,103.5897],[1840,115.1703],[1850,114.0686],[1857,149.8553],[1862,117.8993],[1868,103.5897],[1889,115.1703],[1902,113.4742],[1908,157.9813],[1914,121.2516],[1920,103.6121],[1942,115.1672],[1957,113.2447],[1961,113.1536],[1962,0]]
       }
       arm_r_pv_ik_ctrl.Location.Z {
        type=MovieSceneScriptingFloatChannel
        default=-4.5422
        keys [[1800,-4.5422],[1801,-6.8925],[1803,-8.7523],[1812,-21.6322],[1821,-11.2488],[1831,-2.3991],[1849,-7.427],[1852,-7.2907],[1861,-21.6322],[1870,-11.2488],[1880,-2.3991],[1898,-7.427],[1904,-6.8198],[1909,-15.8534],[1914,-21.5719],[1923,-10.5406],[1933,-2.4052],[1951,-7.4252],[1956,-6.7966],[1959,-6.2487],[1960,-6.1258],[1961,-6.0579],[1962,0]]
       }
       arm_r_pv_ik_ctrl.Rotation.X {
        type=MovieSceneScriptingFloatChannel
        default=12.545
        keys [[1800,12.545],[1808,9.8828],[1817,8.5899],[1839,9.8983],[1850,9.6468],[1855,10.3773],[1866,8.5899],[1888,9.8984],[1902,9.5925],[1906,10.8152],[1919,8.591],[1940,9.8981],[1956,9.5854],[1961,9.6012],[1962,0]]
       }
       arm_r_pv_ik_ctrl.Rotation.Y {
        type=MovieSceneScriptingFloatChannel
        default=69.1049
        keys [[1800,69.1049],[1817,63.9324],[1838,66.4145],[1850,65.7415],[1855,66.4178],[1866,63.9324],[1887,66.4145],[1902,65.6581],[1906,67.0151],[1919,63.9402],[1940,66.4107],[1954,65.6568],[1961,65.8067],[1962,0]]
       }
       arm_r_pv_ik_ctrl.Rotation.Z {
        type=MovieSceneScriptingFloatChannel
        default=91.0037
        thinned_from=12
        keys [[1800,91.0037],[1839,98.0723],[1858,92.345],[1888,98.0723],[1908,91.9282],[1941,98.0702],[1961,97.2972],[1962,0]]
       }
       arm_r_pv_ik_ctrl.Scale.X {type=MovieSceneScriptingFloatChannel, default=1, const [1,2]}
       arm_r_pv_ik_ctrl.Scale.Y {type=MovieSceneScriptingFloatChannel, default=1, const [1,2]}
       arm_r_pv_ik_ctrl.Scale.Z {type=MovieSceneScriptingFloatChannel, default=1, const [1,2]}
       body_ctrl.Location.X {
        type=MovieSceneScriptingFloatChannel
        default=0.3578
        thinned_from=15
        keys [[1819,8597.5938],[1828,-2323.3591],[1829,-4243.0332],[1838,-17083.6602],[1948,-17083.6602],[1950,-18435.752],[1962,-18435.6914]]
       }
       body_ctrl.Location.Y {
        type=MovieSceneScriptingFloatChannel
        default=26.9212
        thinned_from=14
        keys [[1828,-658.5575],[1948,-658.5575],[1950,-23105.6582],[1962,-23117.3359]]
       }
       body_ctrl.Location.Z {
        type=MovieSceneScriptingFloatChannel
        default=-14.382
        thinned_from=17
        keys [[1800,-14.3828],[1815,-49.1152],[1824,-3946.4314],[1833,-3946.4314],[1840,1457.1602],[1961,1458.6992],[1962,1476.6484]]
       }
       body_ctrl.Rotation.X {
        type=MovieSceneScriptingFloatChannel
        default=92.3974
        thinned_from=85
        keys [[1800,92.3974],[1802,92.3717],[1804,94.9296],[1805,97.7846],[1806,101.7826],[1808,112.6308],[1811,130.078],[1816,138.6121],[1827,135.2593],[1828,134.1632],[1829,132.1265],[1833,119.3192],[1837,101.6717],[1839,80.9853],[1841,82.5868],[1845,86.7367],[1849,91.3234],[1850,93.5678],[1851,97.7382],[1852,103.2831],[1853,109.6662],[1855,123.095],[1856,129.3491],[1857,134.8241],[1858,139.2481],[1860,143.8115],[1861,143.5091],[1864,132.5214],[1865,126.3707],[1866,119.431],[1869,96.4043],[1870,88.1489],[1871,79.3795],[1876,43.9713],[1878,42.722],[1886,43.9478],[1889,45.9153],[1890,46.6971],[1894,50.2291],[1902,57.797],[1903,58.8408],[1904,60.2062],[1905,62.0202],[1906,64.1954],[1908,68.6994],[1909,70.6189],[1910,72.0884],[1913,75.0029],[1916,76.8983],[1924,80.3582],[1929,83.7146],[1938,91.8923],[1942,94.5381],[1943,95.0127],[1948,89.4813],[1950,89.8055],[1952,89.9233],[1961,89.2847],[1962,0]]
       }
       body_ctrl.Rotation.Y {
        type=MovieSceneScriptingFloatChannel
        default=0.4941
        thinned_from=85
        keys [[1800,0.4941],[1801,1.1517],[1802,2.9957],[1804,9.3482],[1806,17.1896],[1808,23.3298],[1811,24.35],[1816,21.1054],[1817,21.2105],[1820,21.8168],[1822,22.3531],[1829,24.6249],[1833,26.0148],[1837,26.9957],[1839,27.2623],[1841,27.8587],[1844,28.6066],[1845,28.8031],[1847,29.0958],[1849,29.2405],[1850,29.0334],[1851,28.3492],[1852,27.1955],[1853,25.6171],[1855,21.5243],[1857,17.0963],[1858,15.0253],[1860,12.1481],[1861,11.6617],[1864,10.6166],[1865,10.1833],[1866,9.6163],[1869,4.2328],[1870,-0.6995],[1871,-6.4418],[1876,-22.6518],[1878,-22.893],[1886,-22.8677],[1890,-22.5439],[1893,-22.1965],[1896,-21.7295],[1903,-20.3734],[1905,-20.0626],[1906,-19.9886],[1907,-19.8548],[1908,-19.5788],[1910,-18.9187],[1913,-17.9848],[1916,-17.0902],[1919,-16.3111],[1924,-15.1739],[1929,-14.2198],[1930,-14.0609],[1938,-13.3083],[1942,-13.3053],[1943,-13.3442],[1948,0.7752],[1950,0.7792],[1952,0.7807],[1961,0.7708],[1962,0]]
       }
       body_ctrl.Rotation.Z {
        type=MovieSceneScriptingFloatChannel
        default=-0.9873
        thinned_from=85
        keys [[1800,-0.9873],[1801,0.2912],[1802,3.9175],[1804,17.076],[1805,26.1038],[1806,36.4217],[1808,59.8444],[1811,93.5748],[1816,111.2394],[1837,110.9604],[1839,111.341],[1841,112.627],[1844,115.3872],[1849,121.2489],[1850,123.3765],[1851,127.0852],[1852,131.9129],[1853,137.4306],[1856,155.1265],[1857,160.5438],[1858,165.3038],[1860,171.8208],[1866,180.7963],[1869,189.2113],[1870,-162.6422],[1876,-100.29],[1878,-96.7955],[1886,-85.5843],[1890,-80.2201],[1896,-72.6016],[1902,-65.5728],[1905,-62.4123],[1908,-60.1624],[1910,-58.2985],[1913,-55.2565],[1916,-52.3628],[1918,-50.658],[1921,-48.4748],[1924,-46.4922],[1929,-43.5138],[1930,-42.9847],[1938,-39.8434],[1942,-39.1865],[1943,-39.1371],[1948,-0.9848],[1950,-1.0049],[1955,-1.0601],[1957,-1.0783],[1959,-1.0908],[1960,-1.0943],[1961,-1.0955],[1962,0]]
       }
       body_ctrl.Scale.X {type=MovieSceneScriptingFloatChannel, default=1, thinned_from=85, keys [[1800,1],[1962,1]]}
       body_ctrl.Scale.Y {type=MovieSceneScriptingFloatChannel, default=1, thinned_from=85, keys [[1800,1],[1962,1]]}
       body_ctrl.Scale.Z {type=MovieSceneScriptingFloatChannel, default=1, thinned_from=85, keys [[1800,1],[1962,1]]}
       hips_ctrl.Location.X {type=MovieSceneScriptingFloatChannel, default=0, thinned_from=118, keys [[1800,0],[1962,0]]}
       hips_ctrl.Location.Y {type=MovieSceneScriptingFloatChannel, default=0, thinned_from=144, keys [[1800,0],[1962,0]]}
       hips_ctrl.Location.Z {type=MovieSceneScriptingFloatChannel, default=0, thinned_from=129, keys [[1800,0],[1962,0]]}
       hips_ctrl.Rotation.X {
        type=MovieSceneScriptingFloatChannel
        default=-1.1744
        thinned_from=30
        keys [[1800,-1.1744],[1801,2.2804],[1802,2.5917],[1805,0.5474],[1808,1.2093],[1810,0.9558],[1826,5.4026],[1837,3.5718],[1849,3.0189],[1851,3.019],[1855,1.4957],[1859,0.962],[1875,5.4026],[1886,3.5718],[1898,3.0189],[1901,3.0207],[1904,2.9239],[1907,0.7485],[1910,1.1246],[1912,0.9602],[1920,3.5784],[1928,5.3929],[1939,3.5012],[1951,3.0192],[1961,3.0218],[1962,0]]
       }
       hips_ctrl.Rotation.Y {
        type=MovieSceneScriptingFloatChannel
        default=0.31
        thinned_from=27
        keys [[1800,0.31],[1805,1.2921],[1810,1.9922],[1821,2.4768],[1840,1.6545],[1844,1.6742],[1849,1.738],[1851,1.738],[1854,1.6035],[1862,2.2064],[1870,2.4768],[1889,1.6545],[1893,1.6742],[1902,1.7815],[1906,1.3386],[1910,1.8666],[1914,2.1734],[1922,2.4744],[1942,1.6546],[1953,1.7688],[1955,1.7866],[1957,1.7866],[1959,1.8],[1961,1.8],[1962,0]]
       }
       hips_ctrl.Rotation.Z {
        type=MovieSceneScriptingFloatChannel
        default=-0.0064
        thinned_from=36
        keys [[1800,-0.0064],[1801,0.0177],[1803,0.0311],[1805,0.0123],[1808,0.0385],[1810,0.0332],[1825,0.2225],[1834,0.1317],[1843,0.089],[1845,0.0891],[1851,0.0916],[1856,0.0428],[1857,0.0431],[1859,0.0334],[1874,0.2225],[1883,0.1317],[1892,0.089],[1894,0.0891],[1902,0.094],[1904,0.0844],[1907,0.0193],[1910,0.0366],[1911,0.034],[1916,0.0711],[1920,0.153],[1926,0.2221],[1935,0.137],[1945,0.089],[1955,0.0943],[1956,0.0943],[1958,0.095],[1961,0.095],[1962,0]]
       }
       hips_ctrl.Scale.X {type=MovieSceneScriptingFloatChannel, default=1, const [1,2]}
       hips_ctrl.Scale.Y {type=MovieSceneScriptingFloatChannel, default=1, const [1,2]}
       hips_ctrl.Scale.Z {type=MovieSceneScriptingFloatChannel, default=1, const [1,2]}
       hips_tan_ctrl.Location.X {
        type=MovieSceneScriptingFloatChannel
        default=-0.0213
        thinned_from=21
        keys [[1800,-0.0213],[1804,-0.0224],[1812,-0.0175],[1821,-0.015],[1841,-0.0211],[1851,-0.0204],[1855,-0.0213],[1862,-0.017],[1870,-0.015],[1890,-0.0211],[1902,-0.0201],[1907,-0.022],[1914,-0.0172],[1922,-0.015],[1942,-0.0211],[1961,-0.02],[1962,0]]
       }
       hips_tan_ctrl.Location.Y {
        type=MovieSceneScriptingFloatChannel
        default=-0.2584
        thinned_from=20
        keys [[1800,-0.2584],[1801,-0.2896],[1805,-0.2149],[1810,-0.1665],[1820,-0.0842],[1841,-0.1837],[1850,-0.1727],[1856,-0.1903],[1869,-0.0842],[1890,-0.1837],[1902,-0.1684],[1906,-0.2089],[1922,-0.0847],[1943,-0.1836],[1952,-0.1716],[1957,-0.1663],[1960,-0.1653],[1961,-0.1652],[1962,0.008]]
       }
       hips_tan_ctrl.Location.Z {
        type=MovieSceneScriptingFloatChannel
        default=2.6533
        keys [[1800,2.6533],[1805,1.5294],[1810,0.6168],[1821,-0.4293],[1840,0.6036],[1850,0.4343],[1856,0.9434],[1863,0.0424],[1870,-0.4293],[1889,0.6036],[1902,0.3769],[1907,1.3297],[1909,1.0012],[1914,0.246],[1922,-0.4284],[1942,0.6017],[1957,0.351],[1960,0.341],[1961,0.3402],[1962,-0.0073]]
       }
       hips_tan_ctrl.Rotation.X {type=MovieSceneScriptingFloatChannel, default=0, const [0,2]}
       hips_tan_ctrl.Rotation.Y {type=MovieSceneScriptingFloatChannel, default=0, const [0,2]}
       hips_tan_ctrl.Rotation.Z {type=MovieSceneScriptingFloatChannel, default=0, const [0,2]}
       hips_tan_ctrl.Scale.X {type=MovieSceneScriptingFloatChannel, default=1, const [1,2]}
       hips_tan_ctrl.Scale.Y {type=MovieSceneScriptingFloatChannel, default=1, const [1,2]}
       hips_tan_ctrl.Scale.Z {type=MovieSceneScriptingFloatChannel, default=1, const [1,2]}
       spine_01_ctrl.Location.X {
        type=MovieSceneScriptingFloatChannel
        default=-0.0104
        keys [[1800,-0.0104],[1802,-0.006],[1804,-0.0021],[1809,0.0064],[1819,0.0103],[1839,0.0029],[1850,0.0042],[1854,0.0028],[1861,0.0087],[1868,0.0103],[1888,0.0029],[1902,0.0046],[1906,0.0003],[1909,0.0045],[1913,0.0084],[1920,0.0103],[1941,0.0029],[1961,0.0048],[1962,0]]
       }
       spine_01_ctrl.Location.Y {
        type=MovieSceneScriptingFloatChannel
        default=-0.2331
        thinned_from=22
        keys [[1800,-0.2331],[1801,-0.2817],[1805,-0.2004],[1810,-0.1668],[1819,-0.1169],[1841,-0.1965],[1850,-0.1877],[1853,-0.1916],[1855,-0.1879],[1856,-0.1886],[1868,-0.1169],[1890,-0.1965],[1902,-0.1844],[1906,-0.2039],[1920,-0.1171],[1942,-0.1964],[1952,-0.1868],[1957,-0.1827],[1961,-0.1817],[1962,0]]
       }
       spine_01_ctrl.Location.Z {
        type=MovieSceneScriptingFloatChannel
        default=2.1577
        keys [[1800,2.1577],[1805,1.2539],[1810,0.519],[1821,-0.33],[1840,0.5087],[1850,0.3725],[1856,0.7825],[1863,0.0537],[1870,-0.33],[1889,0.5087],[1902,0.3261],[1907,1.0932],[1909,0.8292],[1914,0.2189],[1922,-0.3293],[1942,0.5073],[1957,0.3052],[1960,0.297],[1961,0.2964],[1962,0]]
       }
       spine_01_ctrl.Rotation.X {
        type=MovieSceneScriptingFloatChannel
        default=-1.1565
        thinned_from=26
        keys [[1800,-1.1565],[1801,-1.8219],[1802,-1.8451],[1805,-1.6778],[1807,-1.9468],[1810,-2.0784],[1827,-2.5677],[1838,-2.4778],[1850,-2.544],[1856,-2.0646],[1857,-2.096],[1859,-2.0799],[1876,-2.5677],[1887,-2.4778],[1902,-2.5566],[1904,-2.45],[1907,-1.7859],[1909,-2.001],[1912,-2.0835],[1929,-2.5676],[1939,-2.4781],[1956,-2.5592],[1961,-2.5554],[1962,0]]
       }
       spine_01_ctrl.Rotation.Y {
        type=MovieSceneScriptingFloatChannel
        default=0.1187
        thinned_from=20
        keys [[1800,0.1187],[1801,0.0757],[1803,0.0148],[1807,-0.0923],[1815,-0.1568],[1837,-0.0883],[1850,-0.1103],[1855,-0.0829],[1859,-0.1381],[1864,-0.1568],[1886,-0.0883],[1902,-0.1146],[1906,-0.047],[1911,-0.133],[1916,-0.1567],[1939,-0.0884],[1959,-0.1163],[1961,-0.1162],[1962,0]]
       }
       spine_01_ctrl.Rotation.Z {
        type=MovieSceneScriptingFloatChannel
        default=-0.7959
        thinned_from=19
        keys [[1800,-0.7959],[1803,-1.0936],[1807,-1.2799],[1821,-0.93],[1843,-1.2965],[1853,-1.2621],[1856,-1.2795],[1870,-0.93],[1892,-1.2965],[1905,-1.2195],[1909,-1.2783],[1923,-0.9312],[1945,-1.2965],[1953,-1.2856],[1961,-1.2743],[1962,0]]
       }
       spine_01_ctrl.Scale.X {type=MovieSceneScriptingFloatChannel, default=1, const [1,2]}
       spine_01_ctrl.Scale.Y {type=MovieSceneScriptingFloatChannel, default=1, const [1,2]}
       spine_01_ctrl.Scale.Z {type=MovieSceneScriptingFloatChannel, default=1, const [1,2]}
       spine_02_ctrl.Location.X {
        type=MovieSceneScriptingFloatChannel
        default=-0.0214
        keys [[1800,-0.0214],[1802,-0.0189],[1810,-0.0145],[1824,-0.017],[1850,-0.0149],[1855,-0.0158],[1859,-0.0145],[1873,-0.017],[1902,-0.0149],[1906,-0.0166],[1912,-0.0145],[1925,-0.017],[1955,-0.0149],[1961,-0.0149],[1962,0]]
       }
       spine_02_ctrl.Location.Y {
        type=MovieSceneScriptingFloatChannel
        default=-0.2342
        thinned_from=20
        keys [[1800,-0.2342],[1801,-0.2528],[1805,-0.2011],[1808,-0.1823],[1820,-0.087],[1842,-0.1754],[1850,-0.1669],[1856,-0.1827],[1869,-0.087],[1891,-0.1754],[1902,-0.1635],[1906,-0.1963],[1922,-0.0871],[1943,-0.1753],[1952,-0.166],[1957,-0.1617],[1961,-0.1608],[1962,0]]
       }
       spine_02_ctrl.Location.Z {
        type=MovieSceneScriptingFloatChannel
        default=1.7891
        keys [[1800,1.7891],[1805,1.042],[1810,0.4352],[1821,-0.2632],[1840,0.4296],[1850,0.3171],[1856,0.6537],[1863,0.0511],[1870,-0.2632],[1889,0.4296],[1902,0.2787],[1907,0.9094],[1909,0.6917],[1914,0.1874],[1922,-0.2628],[1942,0.4285],[1957,0.2615],[1960,0.2547],[1961,0.2543],[1962,0]]
       }
       spine_02_ctrl.Rotation.X {
        type=MovieSceneScriptingFloatChannel
        default=-1.518
        thinned_from=19
        keys [[1800,-1.518],[1801,-1.9017],[1802,-1.9595],[1810,-2.4539],[1821,-2.1696],[1850,-2.6788],[1855,-2.3757],[1859,-2.4546],[1870,-2.1696],[1901,-2.685],[1904,-2.58],[1907,-2.1805],[1912,-2.4511],[1922,-2.1706],[1955,-2.6853],[1961,-2.6778],[1962,0]]
       }
       spine_02_ctrl.Rotation.Y {
        type=MovieSceneScriptingFloatChannel
        default=0.3265
        keys [[1800,0.3265],[1804,0.2977],[1805,0.2995],[1808,0.2907],[1821,0.4205],[1843,0.3054],[1851,0.3093],[1857,0.2925],[1870,0.4205],[1892,0.3054],[1903,0.3114],[1910,0.2917],[1923,0.4201],[1945,0.3054],[1961,0.3137],[1962,0]]
       }
       spine_02_ctrl.Rotation.Z {
        type=MovieSceneScriptingFloatChannel
        default=-0.6535
        thinned_from=18
        keys [[1800,-0.6535],[1806,-1.0316],[1822,-0.6959],[1843,-1.02],[1853,-0.9955],[1856,-1.0243],[1871,-0.6959],[1892,-1.02],[1905,-0.9653],[1908,-1.0303],[1923,-0.6965],[1945,-1.0199],[1961,-1.0002],[1962,0]]
       }
       spine_02_ctrl.Scale.X {type=MovieSceneScriptingFloatChannel, default=1, const [1,2]}
       spine_02_ctrl.Scale.Y {type=MovieSceneScriptingFloatChannel, default=1, const [1,2]}
       spine_02_ctrl.Scale.Z {type=MovieSceneScriptingFloatChannel, default=1, const [1,2]}
       spine_03_ctrl.Location.X {
        type=MovieSceneScriptingFloatChannel
        default=-0.0639
        keys [[1800,-0.0639],[1807,-0.0726],[1808,-0.0725],[1821,-0.0826],[1843,-0.0717],[1850,-0.072],[1853,-0.0718],[1856,-0.0726],[1857,-0.0725],[1870,-0.0826],[1892,-0.0717],[1902,-0.0722],[1905,-0.0711],[1908,-0.0726],[1909,-0.0726],[1923,-0.0826],[1945,-0.0717],[1961,-0.0724],[1962,0]]
       }
       spine_03_ctrl.Location.Y {
        type=MovieSceneScriptingFloatChannel
        default=-0.4307
        thinned_from=19
        keys [[1800,-0.4307],[1801,-0.4529],[1811,-0.3442],[1821,-0.1819],[1842,-0.3654],[1851,-0.3517],[1856,-0.3868],[1870,-0.1819],[1891,-0.3654],[1902,-0.3459],[1907,-0.4031],[1923,-0.1824],[1944,-0.3653],[1953,-0.3484],[1957,-0.3428],[1961,-0.3409],[1962,0]]
       }
       spine_03_ctrl.Location.Z {
        type=MovieSceneScriptingFloatChannel
        default=2.398
        keys [[1800,2.398],[1805,1.386],[1810,0.5661],[1821,-0.3688],[1840,0.5583],[1850,0.4062],[1856,0.8607],[1863,0.0508],[1870,-0.3688],[1889,0.5583],[1902,0.3546],[1907,1.2068],[1909,0.9119],[1914,0.2333],[1922,-0.3683],[1942,0.5567],[1957,0.3314],[1960,0.3225],[1961,0.3218],[1962,0]]
       }
       spine_03_ctrl.Rotation.X {
        type=MovieSceneScriptingFloatChannel
        default=-1.2275
        thinned_from=23
        keys [[1800,-1.2275],[1802,-2.0968],[1805,-3.4244],[1810,-4.3794],[1823,-3.1699],[1836,-4.0797],[1849,-4.4544],[1855,-4.0518],[1859,-4.3795],[1872,-3.1699],[1886,-4.1529],[1900,-4.4567],[1903,-4.3705],[1906,-3.5253],[1911,-4.375],[1924,-3.1716],[1938,-4.1163],[1953,-4.4564],[1961,-4.4354],[1962,0]]
       }
       spine_03_ctrl.Rotation.Y {
        type=MovieSceneScriptingFloatChannel
        default=1.1249
        keys [[1800,1.1249],[1802,0.9195],[1804,0.8486],[1821,1.3059],[1842,0.9249],[1851,0.9526],[1856,0.8959],[1870,1.3059],[1891,0.9249],[1903,0.963],[1907,0.8641],[1923,1.3046],[1944,0.925],[1961,0.9698],[1962,0]]
       }
       spine_03_ctrl.Rotation.Z {
        type=MovieSceneScriptingFloatChannel
        default=-0.9008
        thinned_from=18
        keys [[1800,-0.9008],[1806,-1.301],[1822,-0.7668],[1843,-1.2257],[1852,-1.1987],[1856,-1.2659],[1871,-0.7668],[1892,-1.2257],[1904,-1.1797],[1908,-1.2907],[1924,-0.7672],[1945,-1.2254],[1957,-1.1982],[1961,-1.194],[1962,0]]
       }
       spine_03_ctrl.Scale.X {type=MovieSceneScriptingFloatChannel, default=1, const [1,2]}
       spine_03_ctrl.Scale.Y {type=MovieSceneScriptingFloatChannel, default=1, const [1,2]}
       spine_03_ctrl.Scale.Z {type=MovieSceneScriptingFloatChannel, default=1, const [1,2]}
       spine_04_ctrl.Location.X {
        type=MovieSceneScriptingFloatChannel
        default=-0.1295
        keys [[1800,-0.1295],[1805,-0.1766],[1810,-0.1918],[1821,-0.2184],[1843,-0.1854],[1850,-0.187],[1854,-0.1839],[1870,-0.2184],[1892,-0.1854],[1902,-0.188],[1906,-0.1769],[1923,-0.2184],[1944,-0.1854],[1961,-0.1889],[1962,0]]
       }
       spine_04_ctrl.Location.Y {
        type=MovieSceneScriptingFloatChannel
        default=-0.6993
        thinned_from=18
        keys [[1800,-0.6993],[1805,-0.7519],[1822,-0.3138],[1843,-0.6614],[1851,-0.639],[1857,-0.7189],[1871,-0.3138],[1892,-0.6614],[1902,-0.6309],[1908,-0.7414],[1924,-0.3148],[1944,-0.6613],[1953,-0.6353],[1957,-0.6256],[1961,-0.6223],[1962,0]]
       }
       spine_04_ctrl.Location.Z {
        type=MovieSceneScriptingFloatChannel
        default=2.9349
        keys [[1800,2.9349],[1805,1.6642],[1810,0.6485],[1820,-0.4835],[1840,0.6432],[1850,0.4542],[1856,1.0134],[1862,0.1614],[1869,-0.4835],[1889,0.6432],[1902,0.3907],[1907,1.4412],[1909,1.0751],[1914,0.24],[1922,-0.483],[1942,0.6408],[1957,0.3623],[1960,0.3513],[1961,0.3506],[1962,0]]
       }
       spine_04_ctrl.Rotation.X {
        type=MovieSceneScriptingFloatChannel
        default=-21.7796
        thinned_from=19
        keys [[1800,-21.7796],[1801,-20.959],[1808,-24.3349],[1823,-12.5586],[1843,-22.2763],[1852,-21.8598],[1858,-24.2086],[1872,-12.5586],[1892,-22.2763],[1904,-21.6661],[1909,-24.3221],[1925,-12.5719],[1945,-22.2734],[1957,-21.6972],[1961,-21.6139],[1962,0]]
       }
       spine_04_ctrl.Rotation.Y {
        type=MovieSceneScriptingFloatChannel
        default=8.2729
        keys [[1800,8.2729],[1805,9.8532],[1811,10.7433],[1822,11.3422],[1842,10.3605],[1850,10.4433],[1854,10.254],[1871,11.3422],[1891,10.3605],[1902,10.4813],[1906,9.9195],[1923,11.3416],[1944,10.362],[1961,10.5083],[1962,0]]
       }
       spine_04_ctrl.Rotation.Z {
        type=MovieSceneScriptingFloatChannel
        default=0.2979
        keys [[1800,0.2979],[1801,0.3049],[1806,-0.1677],[1812,0.0705],[1820,0.9671],[1825,1.1614],[1834,0.6276],[1843,0.2929],[1851,0.3406],[1858,-0.0659],[1866,0.6362],[1874,1.1614],[1883,0.6276],[1892,0.2929],[1903,0.3572],[1908,-0.1385],[1914,0.1152],[1921,0.9186],[1926,1.1592],[1935,0.6638],[1945,0.2939],[1961,0.3677],[1962,0]]
       }
       spine_04_ctrl.Scale.X {type=MovieSceneScriptingFloatChannel, default=1, const [1,2]}
       spine_04_ctrl.Scale.Y {type=MovieSceneScriptingFloatChannel, default=1, const [1,2]}
       spine_04_ctrl.Scale.Z {type=MovieSceneScriptingFloatChannel, default=1, const [1,2]}
       head_ctrl.Location.X {
        type=MovieSceneScriptingFloatChannel
        default=0.0028
        thinned_from=12
        keys [[1800,0.0028],[1801,0.0027],[1802,-0.0002],[1803,-0.0009],[1804,0],[1903,0.0001],[1905,-0.0002],[1906,0],[1962,0]]
       }
       head_ctrl.Location.Y {
        type=MovieSceneScriptingFloatChannel
        default=-0.3955
        thinned_from=11
        keys [[1800,-0.3955],[1801,-0.2059],[1803,0.0012],[1804,0],[1851,-0.0014],[1852,0.0001],[1903,-0.0042],[1905,0.0002],[1906,0],[1962,0]]
       }
       head_ctrl.Location.Z {
        type=MovieSceneScriptingFloatChannel
        default=-0.0165
        thinned_from=15
        keys [[1800,-0.0165],[1801,-0.0084],[1803,0.0006],[1804,0],[1850,-0.0001],[1891,0],[1903,-0.0003],[1905,0.0001],[1962,0]]
       }
       head_ctrl.Rotation.X {
        type=MovieSceneScriptingFloatChannel
        default=-72.1385
        thinned_from=17
        keys [[1800,-72.1385],[1820,-52.5409],[1842,-69.3384],[1852,-68.312],[1854,-68.3674],[1869,-52.5409],[1891,-69.3384],[1903,-67.89],[1906,-68.9693],[1921,-52.5855],[1944,-69.3373],[1953,-68.2699],[1957,-67.5321],[1961,-67.1252],[1962,0]]
       }
       head_ctrl.Rotation.Y {
        type=MovieSceneScriptingFloatChannel
        default=2.0821
        keys [[1800,2.0821],[1802,1.4581],[1808,2.3279],[1816,1.8579],[1820,1.9434],[1841,0.9126],[1842,0.9165],[1850,1.0223],[1854,1.4096],[1856,2.0636],[1858,2.2606],[1865,1.8579],[1869,1.9434],[1890,0.9126],[1891,0.9165],[1900,1.0358],[1904,1.137],[1909,2.3126],[1917,1.8665],[1922,1.9415],[1942,0.9127],[1961,1.092],[1962,0]]
       }
       head_ctrl.Rotation.Z {
        type=MovieSceneScriptingFloatChannel
        default=-1.2626
        keys [[1800,-1.2626],[1801,-0.791],[1803,-0.2903],[1805,-0.3239],[1809,-0.0704],[1814,-0.0055],[1824,0.1283],[1842,-0.0032],[1844,-0.0034],[1849,0],[1853,-0.0423],[1855,-0.1539],[1857,-0.0652],[1858,-0.0694],[1873,0.1283],[1891,-0.0032],[1893,-0.0034],[1901,0.0018],[1904,-0.035],[1907,-0.2662],[1911,-0.0661],[1916,0.0051],[1925,0.128],[1945,-0.0033],[1956,0.0023],[1957,0.0023],[1958,0.0025],[1959,0.0021],[1960,0.0028],[1962,0]]
       }
       head_ctrl.Scale.X {type=MovieSceneScriptingFloatChannel, default=1, const [1,2]}
       head_ctrl.Scale.Y {type=MovieSceneScriptingFloatChannel, default=1, const [1,2]}
       head_ctrl.Scale.Z {type=MovieSceneScriptingFloatChannel, default=1, const [1,2]}
       clavicle_l_ctrl.Location.X {type=MovieSceneScriptingFloatChannel, default=0, thinned_from=86, keys [[1800,0],[1962,0]]}
       clavicle_l_ctrl.Location.Y {type=MovieSceneScriptingFloatChannel, default=0, thinned_from=105, keys [[1800,0],[1962,0]]}
       clavicle_l_ctrl.Location.Z {type=MovieSceneScriptingFloatChannel, default=0, thinned_from=90, keys [[1800,0],[1962,0]]}
       clavicle_l_ctrl.Rotation.X {
        type=MovieSceneScriptingFloatChannel
        default=8.6331
        thinned_from=18
        keys [[1800,8.6331],[1832,17.4936],[1845,17.2681],[1851,16.9193],[1858,10.6497],[1881,17.4936],[1894,17.2681],[1901,17.0966],[1904,15.8485],[1906,11.3782],[1908,9.9499],[1934,17.4927],[1948,17.2278],[1958,17.0209],[1961,16.9959],[1962,0]]
       }
       clavicle_l_ctrl.Rotation.Y {
        type=MovieSceneScriptingFloatChannel
        default=37.1028
        thinned_from=11
        keys [[1800,37.1028],[1808,38.3292],[1832,38.1442],[1851,38.94],[1881,38.1442],[1902,39.0519],[1908,38.3136],[1933,38.1447],[1961,39.2331],[1962,0]]
       }
       clavicle_l_ctrl.Rotation.Z {
        type=MovieSceneScriptingFloatChannel
        default=18.7381
        keys [[1800,18.7381],[1804,28.7723],[1808,33.4718],[1832,21.7818],[1845,24.0781],[1851,25.5603],[1858,33.2521],[1881,21.7818],[1895,24.3465],[1902,25.9971],[1910,33.4422],[1934,21.7916],[1961,26.8367],[1962,0]]
       }
       clavicle_l_ctrl.Scale.X {type=MovieSceneScriptingFloatChannel, default=1, const [1,2]}
       clavicle_l_ctrl.Scale.Y {type=MovieSceneScriptingFloatChannel, default=1, const [1,2]}
       clavicle_l_ctrl.Scale.Z {type=MovieSceneScriptingFloatChannel, default=1, const [1,2]}
       clavicle_r_ctrl.Location.X {type=MovieSceneScriptingFloatChannel, default=0, thinned_from=74, keys [[1800,0],[1962,0]]}
       clavicle_r_ctrl.Location.Y {type=MovieSceneScriptingFloatChannel, default=0, thinned_from=123, keys [[1800,0],[1962,0]]}
       clavicle_r_ctrl.Location.Z {type=MovieSceneScriptingFloatChannel, default=0, thinned_from=58, keys [[1800,0],[1962,0]]}
       clavicle_r_ctrl.Rotation.X {
        type=MovieSceneScriptingFloatChannel
        default=28.2872
        keys [[1800,28.2872],[1808,31.4494],[1843,23.7847],[1851,24.7094],[1859,31.3168],[1892,23.7847],[1901,24.7633],[1910,31.4361],[1944,23.7952],[1961,25.5753],[1962,0]]
       }
       clavicle_r_ctrl.Rotation.Y {
        type=MovieSceneScriptingFloatChannel
        default=43.4419
        thinned_from=18
        keys [[1800,43.4419],[1808,47.4251],[1821,38.1351],[1843,40.9694],[1850,40.8715],[1858,47.122],[1870,38.1351],[1892,40.9694],[1902,40.788],[1909,47.3841],[1923,38.144],[1944,40.9609],[1961,40.6772],[1962,0]]
       }
       clavicle_r_ctrl.Rotation.Z {
        type=MovieSceneScriptingFloatChannel
        default=34.6316
        keys [[1800,34.6316],[1808,40.917],[1820,34.8518],[1843,26.693],[1851,28.1073],[1859,40.5725],[1867,36.0016],[1892,26.693],[1901,28.1331],[1910,40.8782],[1922,34.6261],[1944,26.7076],[1961,29.3276],[1962,0]]
       }
       clavicle_r_ctrl.Scale.X {type=MovieSceneScriptingFloatChannel, default=1, const [1,2]}
       clavicle_r_ctrl.Scale.Y {type=MovieSceneScriptingFloatChannel, default=1, const [1,2]}
       clavicle_r_ctrl.Scale.Z {type=MovieSceneScriptingFloatChannel, default=1, const [1,2]}
       neck_01_ctrl.Location.X {type=MovieSceneScriptingFloatChannel, default=0, thinned_from=118, keys [[1800,0],[1962,0]]}
       neck_01_ctrl.Location.Y {type=MovieSceneScriptingFloatChannel, default=0.0001, thinned_from=102, keys [[1800,0.0001],[1962,0]]}
       neck_01_ctrl.Location.Z {
        type=MovieSceneScriptingFloatChannel
        default=-0.0011
        thinned_from=71
        keys [[1800,-0.0011],[1803,-0.0011],[1804,0],[1902,0],[1905,-0.0002],[1906,0],[1962,0]]
       }
       neck_01_ctrl.Rotation.X {
        type=MovieSceneScriptingFloatChannel
        default=6.5135
        keys [[1800,6.5135],[1801,3.1779],[1803,0.7134],[1807,4.6548],[1812,7.706],[1821,2.1439],[1831,-2.1492],[1841,2.4157],[1851,4.6043],[1854,3.9897],[1861,7.706],[1870,2.1439],[1880,-2.1492],[1891,2.8172],[1902,4.8393],[1904,4.3261],[1906,2.7585],[1910,6.1013],[1914,7.6893],[1923,1.8216],[1932,-2.1171],[1945,3.2496],[1958,4.9555],[1960,4.9712],[1961,4.9712],[1962,0]]
       }
       neck_01_ctrl.Rotation.Y {
        type=MovieSceneScriptingFloatChannel
        default=-13.6472
        thinned_from=15
        keys [[1800,-13.6472],[1810,-16.8152],[1831,-15.1362],[1850,-16.1357],[1853,-16.0624],[1859,-16.8134],[1880,-15.1362],[1902,-16.175],[1905,-15.8563],[1912,-16.8084],[1932,-15.1372],[1958,-16.1938],[1961,-16.1961],[1962,0]]
       }
       neck_01_ctrl.Rotation.Z {
        type=MovieSceneScriptingFloatChannel
        default=7.2691
        thinned_from=12
        keys [[1800,7.2691],[1805,7.7448],[1811,7.5089],[1831,8.3672],[1860,7.5089],[1880,8.3672],[1912,7.5103],[1932,8.3657],[1958,7.8341],[1961,7.8341],[1962,0]]
       }
       neck_01_ctrl.Scale.X {type=MovieSceneScriptingFloatChannel, default=1, const [1,2]}
       neck_01_ctrl.Scale.Y {type=MovieSceneScriptingFloatChannel, default=1, const [1,2]}
       neck_01_ctrl.Scale.Z {type=MovieSceneScriptingFloatChannel, default=1, const [1,2]}
       neck_02_ctrl.Location.X {type=MovieSceneScriptingFloatChannel, default=0, const [0,2]}
       neck_02_ctrl.Location.Y {type=MovieSceneScriptingFloatChannel, default=0, const [0,2]}
       neck_02_ctrl.Location.Z {type=MovieSceneScriptingFloatChannel, default=0, const [0,2]}
       neck_02_ctrl.Rotation.X {type=MovieSceneScriptingFloatChannel, default=0, const [0,2]}
       neck_02_ctrl.Rotation.Y {type=MovieSceneScriptingFloatChannel, default=0, const [0,2]}
       neck_02_ctrl.Rotation.Z {type=MovieSceneScriptingFloatChannel, default=0, const [0,2]}
       neck_02_ctrl.Scale.X {type=MovieSceneScriptingFloatChannel, default=1, const [1,2]}
       neck_02_ctrl.Scale.Y {type=MovieSceneScriptingFloatChannel, default=1, const [1,2]}
       neck_02_ctrl.Scale.Z {type=MovieSceneScriptingFloatChannel, default=1, const [1,2]}
       head_ik_ctrl.Location.X {
        type=MovieSceneScriptingFloatChannel
        default=1.7326
        thinned_from=18
        keys [[1800,1.7326],[1803,2.1274],[1807,2.2391],[1817,2.1422],[1834,2.1722],[1852,2.1562],[1857,2.2291],[1866,2.1422],[1883,2.1722],[1904,2.1475],[1909,2.2373],[1919,2.1428],[1936,2.1721],[1961,2.1575],[1962,0]]
       }
       head_ik_ctrl.Location.Y {
        type=MovieSceneScriptingFloatChannel
        default=0.9809
        keys [[1800,0.9809],[1801,0.6087],[1803,0.4031],[1808,1.2318],[1813,1.5918],[1822,0.5853],[1831,-0.0572],[1841,0.6868],[1851,1.0565],[1854,0.9526],[1858,1.3682],[1862,1.5918],[1871,0.5853],[1880,-0.0572],[1891,0.7532],[1902,1.0979],[1904,1.0096],[1906,0.7442],[1910,1.303],[1914,1.5903],[1923,0.6404],[1933,-0.0532],[1940,0.4729],[1947,0.908],[1961,1.1225],[1962,0]]
       }
       head_ik_ctrl.Location.Z {
        type=MovieSceneScriptingFloatChannel
        default=0.2814
        thinned_from=24
        keys [[1800,0.2814],[1802,0.2462],[1810,0.565],[1821,0.2924],[1832,0.2317],[1841,0.3948],[1850,0.4657],[1854,0.4351],[1859,0.5647],[1870,0.2924],[1881,0.2317],[1891,0.411],[1902,0.4712],[1906,0.3743],[1909,0.5065],[1912,0.5625],[1922,0.2962],[1933,0.2325],[1945,0.4268],[1957,0.4723],[1961,0.4714],[1962,0]]
       }
       head_ik_ctrl.Rotation.X {
        type=MovieSceneScriptingFloatChannel
        default=-72.1385
        thinned_from=17
        keys [[1800,-72.1385],[1820,-52.5409],[1842,-69.3384],[1852,-68.312],[1854,-68.3674],[1869,-52.5409],[1891,-69.3384],[1903,-67.89],[1906,-68.9693],[1921,-52.5855],[1944,-69.3373],[1953,-68.2699],[1957,-67.5321],[1961,-67.1252],[1962,0]]
       }
       head_ik_ctrl.Rotation.Y {
        type=MovieSceneScriptingFloatChannel
        default=2.0821
        keys [[1800,2.0821],[1802,1.4581],[1808,2.3279],[1816,1.8579],[1820,1.9434],[1841,0.9126],[1849,1.0051],[1853,1.1924],[1855,1.7617],[1858,2.2606],[1865,1.8579],[1869,1.9434],[1890,0.9126],[1899,1.0206],[1904,1.137],[1909,2.3126],[1917,1.8665],[1922,1.9415],[1942,0.9127],[1961,1.092],[1962,0]]
       }
       head_ik_ctrl.Rotation.Z {
        type=MovieSceneScriptingFloatChannel
        default=-1.2626
        keys [[1800,-1.2626],[1801,-0.791],[1803,-0.2903],[1805,-0.3239],[1809,-0.0704],[1814,-0.0055],[1824,0.1283],[1842,-0.0032],[1843,-0.0032],[1844,-0.0034],[1849,0],[1853,-0.0423],[1855,-0.1539],[1857,-0.0652],[1858,-0.0694],[1873,0.1283],[1891,-0.0032],[1892,-0.0032],[1893,-0.0034],[1901,0.0018],[1904,-0.035],[1907,-0.2662],[1911,-0.0661],[1916,0.0051],[1925,0.128],[1945,-0.0033],[1956,0.0023],[1957,0.0023],[1958,0.0025],[1959,0.0021],[1960,0.0028],[1962,0]]
       }
       head_ik_ctrl.Scale.X {type=MovieSceneScriptingFloatChannel, default=1, const [1,2]}
       head_ik_ctrl.Scale.Y {type=MovieSceneScriptingFloatChannel, default=1, const [1,2]}
       head_ik_ctrl.Scale.Z {type=MovieSceneScriptingFloatChannel, default=1, const [1,2]}
       chest_ctrl.Location.X {
        type=MovieSceneScriptingFloatChannel
        default=-0.2478
        thinned_from=20
        keys [[1800,-0.2478],[1808,-0.1894],[1817,-0.1666],[1839,-0.2019],[1851,-0.1943],[1855,-0.2026],[1859,-0.1787],[1866,-0.1666],[1888,-0.2019],[1902,-0.1924],[1906,-0.2122],[1912,-0.1767],[1918,-0.1666],[1940,-0.2018],[1951,-0.1946],[1956,-0.1919],[1961,-0.1914],[1962,0]]
       }
       chest_ctrl.Location.Y {
        type=MovieSceneScriptingFloatChannel
        default=-1.5366
        thinned_from=22
        keys [[1800,-1.5366],[1801,-1.8938],[1806,-1.6255],[1808,-1.6797],[1820,-1.2267],[1842,-1.777],[1855,-1.6684],[1857,-1.6835],[1869,-1.2267],[1891,-1.777],[1902,-1.7297],[1904,-1.7456],[1907,-1.6375],[1909,-1.6743],[1921,-1.2287],[1944,-1.7767],[1957,-1.7207],[1961,-1.714],[1962,0]]
       }
       chest_ctrl.Location.Z {
        type=MovieSceneScriptingFloatChannel
        default=6.2551
        keys [[1800,6.2551],[1805,3.5529],[1810,1.3671],[1821,-1.101],[1840,1.3171],[1850,0.913],[1856,2.1425],[1863,0.0117],[1870,-1.101],[1889,1.3171],[1902,0.7768],[1907,3.0735],[1910,1.9025],[1914,0.492],[1922,-1.0985],[1942,1.3123],[1957,0.7158],[1960,0.6925],[1962,0]]
       }
       chest_ctrl.Rotation.X {
        type=MovieSceneScriptingFloatChannel
        default=-3.9106
        thinned_from=21
        keys [[1800,-3.9106],[1801,-5.4724],[1805,-7.1949],[1810,-8.9112],[1821,-7.7766],[1849,-9.6735],[1855,-8.4943],[1859,-8.9135],[1870,-7.7766],[1901,-9.6967],[1904,-9.1634],[1907,-7.6505],[1909,-8.6114],[1912,-8.899],[1922,-7.7831],[1954,-9.6974],[1961,-9.669],[1962,0]]
       }
       chest_ctrl.Rotation.Y {
        type=MovieSceneScriptingFloatChannel
        default=1.5135
        keys [[1800,1.5135],[1802,1.1679],[1804,1.0259],[1808,0.9338],[1821,1.4839],[1843,0.9732],[1853,1.0029],[1857,0.9397],[1870,1.4839],[1892,0.9732],[1905,1.0255],[1910,0.938],[1923,1.4826],[1945,0.9736],[1961,1.0095],[1962,0]]
       }
       chest_ctrl.Rotation.Z {
        type=MovieSceneScriptingFloatChannel
        default=-2.4084
        thinned_from=18
        keys [[1800,-2.4084],[1806,-3.6734],[1822,-2.515],[1843,-3.6328],[1853,-3.5497],[1856,-3.6457],[1871,-2.515],[1892,-3.6328],[1905,-3.4485],[1908,-3.6672],[1923,-2.5168],[1945,-3.6326],[1961,-3.565],[1962,0]]
       }
       chest_ctrl.Scale.X {type=MovieSceneScriptingFloatChannel, default=1, const [1,2]}
       chest_ctrl.Scale.Y {type=MovieSceneScriptingFloatChannel, default=1, const [1,2]}
       chest_ctrl.Scale.Z {type=MovieSceneScriptingFloatChannel, default=1, const [1,2]}
       chest_tan_ctrl.Location.X {
        type=MovieSceneScriptingFloatChannel
        default=-0.1524
        keys [[1800,-0.1524],[1801,-0.1194],[1802,-0.0956],[1804,-0.0683],[1808,-0.0569],[1821,-0.0639],[1843,-0.0559],[1854,-0.0589],[1857,-0.0569],[1870,-0.0639],[1892,-0.0559],[1905,-0.0639],[1910,-0.057],[1922,-0.0638],[1945,-0.0559],[1961,-0.0565],[1962,0]]
       }
       chest_tan_ctrl.Location.Y {
        type=MovieSceneScriptingFloatChannel
        default=0.3708
        thinned_from=20
        keys [[1800,0.3708],[1801,0.2971],[1804,0.0493],[1808,-0.1823],[1816,-0.2296],[1835,-0.1978],[1850,-0.2209],[1855,-0.1508],[1865,-0.2296],[1884,-0.1978],[1902,-0.2251],[1904,-0.1751],[1906,-0.0577],[1911,-0.2002],[1917,-0.2295],[1936,-0.1978],[1959,-0.2269],[1961,-0.2269],[1962,-0.0303]]
       }
       chest_tan_ctrl.Location.Z {
        type=MovieSceneScriptingFloatChannel
        default=-2.6956
        keys [[1800,-2.6956],[1805,-1.4954],[1810,-0.4984],[1821,0.6811],[1840,-0.4845],[1850,-0.2983],[1856,-0.8588],[1863,0.145],[1870,0.6811],[1889,-0.4845],[1902,-0.2345],[1907,-1.2785],[1909,-0.9233],[1914,-0.0838],[1922,0.6797],[1942,-0.4829],[1952,-0.28],[1957,-0.2057],[1960,-0.1944],[1961,-0.1936],[1962,0.265]]
       }
       chest_tan_ctrl.Rotation.X {
        type=MovieSceneScriptingFloatChannel
        default=0
        keys [[1800,0],[1801,0.1658],[1802,1.6493],[1805,2.9808],[1811,4.3533],[1824,3.1049],[1852,4.4559],[1856,4.0082],[1860,4.3556],[1873,3.1049],[1904,4.5159],[1908,3.4997],[1913,4.3428],[1925,3.1054],[1955,4.4327],[1961,4.4178],[1962,9.5911]]
       }
       chest_tan_ctrl.Rotation.Y {type=MovieSceneScriptingFloatChannel, default=0.4319, keys [[1800,0.4319],[1801,0],[1962,0]]}
       chest_tan_ctrl.Rotation.Z {
        type=MovieSceneScriptingFloatChannel
        default=0
        keys [[1800,0],[1801,0.7069],[1803,0.8146],[1807,1.233],[1809,1.2021],[1810,1.214],[1824,0.678],[1845,1.1509],[1850,1.1434],[1851,1.148],[1854,1.1211],[1857,1.1985],[1858,1.1944],[1859,1.2119],[1873,0.678],[1894,1.1509],[1902,1.1329],[1904,1.164],[1906,1.1004],[1909,1.2252],[1911,1.2081],[1912,1.2124],[1926,0.6783],[1946,1.1509],[1961,1.1169],[1962,3.346]]
       }
       chest_tan_ctrl.Scale.X {type=MovieSceneScriptingFloatChannel, default=1, const [1,2]}
       chest_tan_ctrl.Scale.Y {type=MovieSceneScriptingFloatChannel, default=1, const [1,2]}
       chest_tan_ctrl.Scale.Z {type=MovieSceneScriptingFloatChannel, default=1, const [1,2]}
       foot_l_ik_ctrl.Space {
        type=MovieSceneScriptingSpaceChannel
        const [
         "<Struct 'MovieSceneControlRigSpaceBaseKey' (0x85268b73ea0) {space_type: Parent, control_rig_element: {type: Null, name: \"leg_l_controls_space\"}}>"
         1
        ]
       }
       leg_l_pv_ik_ctrl.Space {
        type=MovieSceneScriptingSpaceChannel
        const [
         "<Struct 'MovieSceneControlRigSpaceBaseKey' (0x85268b73b80) {space_type: Parent, control_rig_element: {type: Null, name: \"leg_l_pv_footFollow\"}}>"
         1
        ]
       }
       foot_r_ik_ctrl.Space {
        type=MovieSceneScriptingSpaceChannel
        const [
         "<Struct 'MovieSceneControlRigSpaceBaseKey' (0x85268b73920) {space_type: Parent, control_rig_element: {type: Null, name: \"leg_r_controls_space\"}}>"
         1
        ]
       }
       leg_r_pv_ik_ctrl.Space {
        type=MovieSceneScriptingSpaceChannel
        const [
         "<Struct 'MovieSceneControlRigSpaceBaseKey' (0x85268b733a0) {space_type: Parent, control_rig_element: {type: Null, name: \"leg_r_pv_footFollow\"}}>"
         1
        ]
       }
       body_ctrl.Space {
        type=MovieSceneScriptingSpaceChannel
        const [
         "<Struct 'MovieSceneControlRigSpaceBaseKey' (0x85263a8d1d0) {space_type: Parent, control_rig_element: {type: Control, name: \"body_offset_ctrl\"}}>"
         1
        ]
       }
      ]
     }]
   }
  ]
 }
 Niagara {
  id=#2
  parent_id=#1
  possessed_class=NiagaraComponent
  tracks [{
    type=MovieSceneNiagaraFloatParameterTrack
    display_name=LifetimeScale
    sections [{type=MovieSceneFloatSection, channels [None {type=MovieSceneScriptingFloatChannel, default=1, keys [[1809.5,20],[1927,20],[1943.5,1]]}]}]
   }]
 }
]
display_rate {denominator=1, numerator=30}
tick_resolution {denominator=1, numerator=24000}
tracks []
