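# Level Sequence animation shot spa1000c for the PistolCanary TripleTentacle cinematic, featuring the Superman NPC and an eye-laser effect.
# The Superman Blueprint layers a skeletal animation section with a keyed Fortnite_Biped_CtrlRig track that poses the body and limbs.
# The Laser_EYE static mesh has spawn and attach tracks, with keyed transforms that stretch the beam briefly, then scale it to zero.
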LevelSequence /PistolCanary/Anim/TripleTentacle/PistolCanary_spa1000c_anm1
bindings [
 Laser_EYE {
  id=#1
  possessed_class=FortStaticMeshActor
  tracks [
   {
    type=MovieScene3DTransformTrack
    sections [{
      type=MovieScene3DTransformSection
      channels [
       Location.X {
        type=MovieSceneScriptingDoubleChannel
        default=209.1541
        keys [[1838,54.663],[1839,54.663],[1841,-123.179],[1848,-69.8099],[1852,213.0272],[1870,209.6237],[1875,295.6728]]
       }
       Location.Y {
        type=MovieSceneScriptingDoubleChannel
        default=532.8854
        keys [[1838,126.7431],[1839,126.7431],[1841,663.1003],[1848,581.6267],[1852,504.4407],[1870,529.4371],[1875,770.3594]]
       }
       Location.Z {
        type=MovieSceneScriptingDoubleChannel
        default=72.1663
        keys [[1838,17.3461],[1839,17.3461],[1841,-167.1088],[1848,-115.0915],[1852,-47.014],[1870,64.284],[1875,99.7686]]
       }
       Rotation.X {
        type=MovieSceneScriptingDoubleChannel
        default=69.5183
        keys [[1838,79.9475],[1839,79.9475],[1841,136.2652],[1848,122.724],[1852,106.405],[1870,72.1584],[1875,69.5183]]
       }
       Rotation.Y {
        type=MovieSceneScriptingDoubleChannel
        default=71.9246
        thinned_from=7
        keys [[1838,71.375],[1839,71.375],[1841,68.4067],[1848,69.8415],[1852,72.1217],[1875,71.9246]]
       }
       Rotation.Z {
        type=MovieSceneScriptingDoubleChannel
        default=-39.5675
        keys [[1838,-24.3903],[1839,-24.3903],[1841,57.5662],[1848,38.4045],[1852,-6.6087],[1870,-37.4004],[1875,-39.5675]]
       }
       Scale.X {
        type=MovieSceneScriptingDoubleChannel
        default=-0.0666
        thinned_from=7
        keys [[1838,1.6272],[1839,1.3384],[1841,-0.0666],[1875,-0.0666],[1876,0]]
       }
       Scale.Y {
        type=MovieSceneScriptingDoubleChannel
        default=-0.1288
        thinned_from=7
        keys [[1838,2.0206],[1839,1.6724],[1841,-0.1288],[1875,-0.1288],[1876,0]]
       }
       Scale.Z {
        type=MovieSceneScriptingDoubleChannel
        default=10.4166
        keys [[1838,1.898],[1839,1.898],[1841,12.8005],[1852,10.4166],[1870,10.4166],[1875,1.3418],[1876,0]]
       }
      ]
     }]
   }
   {type=MovieScene3DAttachTrack, sections [{type=MovieScene3DAttachSection, end_frame=1891, start_frame=1840}]}
   {type=MovieSceneSpawnTrack, sections [{type=MovieSceneSpawnSection, channels [None {type=MovieSceneScriptingBoolChannel, default=true}]}]}
  ]
 }
 Superman BP {
  id=#2
  possessed_class=PistolCanary_CinematicNPC_Superhero_Main_C
  tracks [
   {type=MovieSceneSkeletalAnimationTrack, sections [{type=MovieSceneSkeletalAnimationSection, end_frame=2820, start_frame=1800}]}
   {
    type=MovieSceneControlRigParameterTrack
    display_name=Fortnite_Biped_CtrlRig
    sections [{
      type=MovieSceneControlRigParameterSection
      channels [
       leg_l_stretch_switch {type=MovieSceneScriptingBoolChannel, default=false, const [false,263]}
       leg_r_stretch_switch {type=MovieSceneScriptingBoolChannel, default=false, const [false,263]}
       arm_l_fk_local {type=MovieSceneScriptingBoolChannel, default=true, const [true,31]}
       Auto Interpolate_3 {type=MovieSceneScriptingBoolChannel, default=true, const [true,15]}
       Auto Interpolate {type=MovieSceneScriptingBoolChannel, default=true, const [true,15]}
       arm_l_stretch_switch {type=MovieSceneScriptingBoolChannel, default=false, const [false,263]}
       Keep Aligned {type=MovieSceneScriptingBoolChannel, default=false, const [false,263]}
       arm_r_fk_local {type=MovieSceneScriptingBoolChannel, default=true, const [true,34]}
       Auto Interpolate_4 {type=MovieSceneScriptingBoolChannel, default=true, const [true,15]}
       Auto Interpolate_2 {type=MovieSceneScriptingBoolChannel, default=true, const [true,15]}
       arm_r_stretch_switch {type=MovieSceneScriptingBoolChannel, default=false, const [false,263]}
       Keep Aligned_2 {type=MovieSceneScriptingBoolChannel, default=false, const [false,263]}
       head_local {type=MovieSceneScriptingBoolChannel, default=true, const [true,22]}
       Stretch {type=MovieSceneScriptingBoolChannel, default=false, const [false,22]}
       head_stretch {type=MovieSceneScriptingBoolChannel, default=false, const [false,263]}
       spine_stretch_switch {type=MovieSceneScriptingBoolChannel, default=false, const [false,30]}
       arm_l_fk_ik_switch {type=MovieSceneScriptingBoolChannel, default=false, const [false,10]}
       leg_l_fk_ik_switch {type=MovieSceneScriptingBoolChannel, default=false, const [false,10]}
       arm_r_fk_ik_switch {type=MovieSceneScriptingBoolChannel, default=false, const [false,10]}
       leg_r_fk_ik_switch {type=MovieSceneScriptingBoolChannel, default=false, const [false,10]}
       ShowBodyControls {type=MovieSceneScriptingBoolChannel, default=true, const [true,10]}
       spine_fk_ik_switch {
        type=MovieSceneScriptingBoolChannel
        default=true
        keys [[1801,true],[1816,true],[1825,true],[1830,true],[1931,true],[1938,true],[1989,true],[2007,true],[2062,true],[2098,false],[2103,false]]
       }
       neck_fk_ik_switch {type=MovieSceneScriptingBoolChannel, default=false, const [false,10]}
       space {type=MovieSceneScriptingByteChannel, default=1, const [1,22]}
       global_ctrl.Location.X {type=MovieSceneScriptingFloatChannel, default=0, const [0,10]}
       global_ctrl.Location.Y {type=MovieSceneScriptingFloatChannel, default=0, const [0,10]}
       global_ctrl.Location.Z {type=MovieSceneScriptingFloatChannel, default=0, const [0,10]}
       global_ctrl.Rotation.X {type=MovieSceneScriptingFloatChannel, default=0, const [0,10]}
       global_ctrl.Rotation.Y {type=MovieSceneScriptingFloatChannel, default=0, const [0,10]}
       global_ctrl.Rotation.Z {type=MovieSceneScriptingFloatChannel, default=0, const [0,10]}
       root_ctrl.Location.X {type=MovieSceneScriptingFloatChannel, default=0, const [0,1]}
       root_ctrl.Location.Y {type=MovieSceneScriptingFloatChannel, default=-0.0001, const [-0.0001,1]}
       root_ctrl.Location.Z {type=MovieSceneScriptingFloatChannel, default=-0.0002, const [-118.3337,1]}
       root_ctrl.Rotation.X {type=MovieSceneScriptingFloatChannel, default=360, const [0,1]}
       root_ctrl.Rotation.Y {type=MovieSceneScriptingFloatChannel, default=0, const [0,1]}
       root_ctrl.Rotation.Z {type=MovieSceneScriptingFloatChannel, default=360, const [0,1]}
       root_ctrl.Scale.X {type=MovieSceneScriptingFloatChannel, default=1, const [1,1]}
       root_ctrl.Scale.Y {type=MovieSceneScriptingFloatChannel, default=1, const [1,1]}
       root_ctrl.Scale.Z {type=MovieSceneScriptingFloatChannel, default=1, const [1,1]}
       body_offset_ctrl.Location.X {
        type=MovieSceneScriptingFloatChannel
        default=0
        thinned_from=8
        keys [[1800,10.8314],[1837,10.7979],[1885,6.5996],[1938,6.579],[1989,-1459.1001],[2007,-2047.6892],[2062,-2047.6892]]
       }
       body_offset_ctrl.Location.Y {
        type=MovieSceneScriptingFloatChannel
        default=0
        thinned_from=8
        keys [[1800,-14.9575],[1837,-14.9113],[1885,-9.1137],[1938,-9.0851],[1989,-554.0756],[2007,-426.4153],[2062,-426.4153]]
       }
       body_offset_ctrl.Location.Z {
        type=MovieSceneScriptingFloatChannel
        default=0
        thinned_from=8
        keys [[1800,-777.8148],[1837,-775.413],[1885,-473.9283],[1938,-472.4445],[1989,-333.1681],[2007,-250.6467],[2062,-250.6467]]
       }
       body_offset_ctrl.Rotation.X {type=MovieSceneScriptingFloatChannel, default=0, thinned_from=8, keys [[1800,0],[1938,0],[1989,-0.9442],[2007,-1.3646],[2062,-1.3646]]}
       body_offset_ctrl.Rotation.Y {type=MovieSceneScriptingFloatChannel, default=0, thinned_from=8, keys [[1800,0],[1938,0],[1989,0.5533],[2007,0.5368],[2062,0.5368]]}
       body_offset_ctrl.Rotation.Z {type=MovieSceneScriptingFloatChannel, default=0, thinned_from=8, keys [[1800,0],[1938,0],[1989,-47.4437],[2007,-65.2369],[2062,-65.2369]]}
       body_offset_ctrl.Scale.X {type=MovieSceneScriptingFloatChannel, default=1, const [1,8]}
       body_offset_ctrl.Scale.Y {type=MovieSceneScriptingFloatChannel, default=1, const [1,8]}
       body_offset_ctrl.Scale.Z {type=MovieSceneScriptingFloatChannel, default=1, const [1,8]}
       thigh_l_fk_ctrl.Location.X {type=MovieSceneScriptingFloatChannel, default=0, const [0,34]}
       thigh_l_fk_ctrl.Location.Y {type=MovieSceneScriptingFloatChannel, default=0, const [0,34]}
       thigh_l_fk_ctrl.Location.Z {type=MovieSceneScriptingFloatChannel, default=0, const [0,34]}
       thigh_l_fk_ctrl.Rotation.X {
        type=MovieSceneScriptingFloatChannel
        default=15.8993
        thinned_from=36
        keys [[1801,15.8993],[1830,15.8993],[1832,17.959],[1837,20.313],[1848,24.4135],[1851,27.3935],[1878,15.8993],[1922,15.8993],[1931,25.0109],[1933,25.4108],[1935,24.1593],[1938,24.4135],[1942,15.8993],[1984,15.8993],[1994,10.4822],[2007,15.8993],[2098,15.8993]]
       }
       thigh_l_fk_ctrl.Rotation.Y {
        type=MovieSceneScriptingFloatChannel
        default=-9.3903
        thinned_from=36
        keys [[1801,-9.3903],[1816,-5.9561],[1825,5.2222],[1830,6.1971],[1832,10.4369],[1837,15.277],[1848,23.0895],[1851,28.5631],[1878,6.1971],[1922,6.1971],[1931,24.1427],[1933,24.9304],[1935,22.4655],[1938,23.0895],[1942,-9.3903],[1950,-9.3903],[1959,-7.7828],[1984,-9.3903],[2007,6.1971],[2098,6.1971]]
       }
       thigh_l_fk_ctrl.Rotation.Z {
        type=MovieSceneScriptingFloatChannel
        default=-0.2035
        thinned_from=46
        keys [[1801,-0.2035],[1816,-11.4583],[1819,13.8615],[1825,-25.9283],[1830,-58.3462],[1832,-50.8632],[1837,-21.6863],[1839,-39.0378],[1841,-42.7927],[1849,-84.8672],[1856,-51.2877],[1880,-51.2877],[1882,-58.0733],[1885,-98.2548],[1894,-51.2877],[1922,-51.2877],[1931,-83.4553],[1933,-84.8672],[1935,-80.4488],[1938,-76.1614],[1942,-0.2035],[1950,-0.2035],[1955,28.5697],[1959,30.3053],[1964,14.0966],[1972,0],[1984,-0.2035],[1998,-18.1287],[2012,-81.5345],[2014,-77.7173],[2017,-107.1917],[2018,-60.603],[2019,-25.4473],[2020,-35.3946],[2021,-18.3633],[2023,-0.9633],[2027,-6.6688],[2035,-38.7209],[2043,-76.6599],[2048,-78.6749],[2062,-63.567],[2098,-51.2877],[2103,-51.2877]]
       }
       thigh_l_fk_ctrl.Scale.X {type=MovieSceneScriptingFloatChannel, default=1, const [1,34]}
       thigh_l_fk_ctrl.Scale.Y {type=MovieSceneScriptingFloatChannel, default=1, const [1,34]}
       thigh_l_fk_ctrl.Scale.Z {type=MovieSceneScriptingFloatChannel, default=1, const [1,34]}
       calf_l_fk_ctrl.Location.X {type=MovieSceneScriptingFloatChannel, default=0, const [0,31]}
       calf_l_fk_ctrl.Location.Y {type=MovieSceneScriptingFloatChannel, default=0, const [0,31]}
       calf_l_fk_ctrl.Location.Z {type=MovieSceneScriptingFloatChannel, default=0, const [0,31]}
       calf_l_fk_ctrl.Rotation.X {type=MovieSceneScriptingFloatChannel, default=1.1747, const [1.1747,31]}
       calf_l_fk_ctrl.Rotation.Y {type=MovieSceneScriptingFloatChannel, default=-0.0192, const [-0.0192,31]}
       calf_l_fk_ctrl.Rotation.Z {
        type=MovieSceneScriptingFloatChannel
        default=-3.3246
        thinned_from=36
        keys [[1801,-3.3246],[1816,22.3057],[1825,105.7322],[1829,77.7102],[1833,104.422],[1838,113.0081],[1843,82.0259],[1846,72.6432],[1849,90.3804],[1853,79.2603],[1859,113.0081],[1884,113.0081],[1887,80.0368],[1888,32.3314],[1893,113.0081],[1922,113.0081],[1931,91.3318],[1933,90.3804],[1935,93.3577],[1938,101.6942],[1942,-3.3246],[1950,-3.3246],[1955,25.4487],[1960,-3.3246],[1964,10.9755],[1972,0],[1984,-3.3246],[1994,100.0122],[2007,113.0081],[2017,70.0468],[2021,113.0081],[2098,113.0081]]
       }
       calf_l_fk_ctrl.Scale.X {type=MovieSceneScriptingFloatChannel, default=1, const [1,31]}
       calf_l_fk_ctrl.Scale.Y {type=MovieSceneScriptingFloatChannel, default=1, const [1,31]}
       calf_l_fk_ctrl.Scale.Z {type=MovieSceneScriptingFloatChannel, default=1, const [1,31]}
       foot_l_fk_ctrl.Location.X {type=MovieSceneScriptingFloatChannel, default=0, const [0,15]}
       foot_l_fk_ctrl.Location.Y {type=MovieSceneScriptingFloatChannel, default=0, const [0,15]}
       foot_l_fk_ctrl.Location.Z {type=MovieSceneScriptingFloatChannel, default=0, const [0,15]}
       foot_l_fk_ctrl.Rotation.X {type=MovieSceneScriptingFloatChannel, default=-1.478, const [-1.478,15]}
       foot_l_fk_ctrl.Rotation.Y {type=MovieSceneScriptingFloatChannel, default=-2.2009, const [-2.2009,15]}
       foot_l_fk_ctrl.Rotation.Z {
        type=MovieSceneScriptingFloatChannel
        default=45.6659
        thinned_from=15
        keys [[1801,45.6659],[1816,40.2529],[1825,22.6334],[1830,21.0968],[1935,21.0968],[1942,45.6659],[1984,45.6659],[2007,21.0968],[2098,21.0968]]
       }
       foot_l_fk_ctrl.Scale.X {type=MovieSceneScriptingFloatChannel, default=1, const [1,15]}
       foot_l_fk_ctrl.Scale.Y {type=MovieSceneScriptingFloatChannel, default=1, const [1,15]}
       foot_l_fk_ctrl.Scale.Z {type=MovieSceneScriptingFloatChannel, default=1, const [1,15]}
       ball_l_fk_ctrl.Location.X {type=MovieSceneScriptingFloatChannel, default=0, const [0,15]}
       ball_l_fk_ctrl.Location.Y {type=MovieSceneScriptingFloatChannel, default=0, const [0,15]}
       ball_l_fk_ctrl.Location.Z {type=MovieSceneScriptingFloatChannel, default=0, const [0,15]}
       ball_l_fk_ctrl.Rotation.X {type=MovieSceneScriptingFloatChannel, default=0, const [0,15]}
       ball_l_fk_ctrl.Rotation.Y {type=MovieSceneScriptingFloatChannel, default=0, const [0,15]}
       ball_l_fk_ctrl.Rotation.Z {type=MovieSceneScriptingFloatChannel, default=0, const [0,15]}
       ball_l_fk_ctrl.Scale.X {type=MovieSceneScriptingFloatChannel, default=1, const [1,15]}
       ball_l_fk_ctrl.Scale.Y {type=MovieSceneScriptingFloatChannel, default=1, const [1,15]}
       ball_l_fk_ctrl.Scale.Z {type=MovieSceneScriptingFloatChannel, default=1, const [1,15]}
       foot_l_ik_ctrl.Location.X {
        type=MovieSceneScriptingFloatChannel
        default=0
        thinned_from=265
        keys [[1648,0],[1801,0],[1802,-16.9412],[1803,-34.5142],[1804,-52.5518],[1805,-70.905],[1809,-145.3074],[1810,-163.6722],[1811,-181.6926],[1812,-199.1921],[1813,-215.9741],[1814,-231.845],[1815,-246.6467],[1816,-260.2955],[1817,-272.8202],[1818,-284.3953],[1820,-306.2248],[1821,-317.6458],[1822,-330.3895],[1823,-345.2629],[1827,-411.8088],[1835,-551.4203],[1838,-600.12],[1850,-683.7155],[1853,-700.9842],[1863,-741.6907],[1866,-758.9789],[1868,-774.4854],[1870,-794.2222],[1872,-819.08],[1874,-849.9301],[1875,-867.8672],[1877,-908.8591],[1880,-978.5137],[1896,-1407.8916],[1900,-1504.0742],[1903,-1566.439],[1906,-1617.8901],[1935,-1962.4141],[1985,-2304.9512],[2041,-2436.8403],[2042,-2431.8928],[2044,-2406.4995],[2046,-2366.9658],[2048,-2313.0491],[2050,-2245.0381],[2052,-2164.791],[2054,-2074.031],[2057,-1921.9509],[2061,-1698.4673],[2062,-1678.1088],[2098,-11.4469],[2103,-11.4482]]
       }
       foot_l_ik_ctrl.Location.Y {
        type=MovieSceneScriptingFloatChannel
        default=0
        thinned_from=265
        keys [[1648,0],[1801,0],[1802,11.1332],[1803,22.8989],[1804,35.9454],[1805,50.7526],[1806,67.6112],[1807,86.615],[1808,107.6688],[1809,130.5105],[1810,154.745],[1811,179.8882],[1813,230.8204],[1814,255.6465],[1815,279.5413],[1816,302.2764],[1817,323.7631],[1818,344.0519],[1819,363.3196],[1831,583.8],[1838,719.4905],[1849,610.0591],[1851,588.5154],[1852,574.9734],[1853,556.4375],[1854,533.2976],[1855,505.9453],[1856,474.7695],[1857,440.1589],[1858,402.5],[1859,362.178],[1860,319.5764],[1861,275.0781],[1862,229.0637],[1863,181.9146],[1864,134.0103],[1866,37.4609],[1867,-10.4202],[1868,-57.5276],[1869,-103.4751],[1870,-147.8756],[1871,-190.3396],[1872,-230.476],[1873,-267.8925],[1874,-302.1959],[1875,-332.9924],[1876,-359.8875],[1877,-384.2603],[1878,-407.7499],[1879,-430.3522],[1881,-472.896],[1883,-511.9752],[1885,-547.6169],[1887,-579.8354],[1889,-608.6465],[1891,-634.0671],[1893,-656.1165],[1895,-674.8145],[1897,-690.1831],[1899,-702.2446],[1901,-711.0229],[1903,-716.5415],[1905,-718.8245],[1907,-717.8955],[1944,-645.1794],[1994,-503.4214],[2004,-464.7871],[2015,-410.6104],[2036,-288.394],[2040,-263.8574],[2041,-256.5884],[2042,-241.6387],[2043,-218.3701],[2044,-187.7705],[2045,-151.2793],[2046,-109.9204],[2047,-63.8755],[2048,-13.0498],[2049,42.3193],[2050,101.7705],[2051,164.8433],[2052,231.0762],[2053,300.0073],[2054,371.1763],[2055,444.1206],[2060,819.3594],[2061,893.291],[2062,1320.4249],[2098,-1.9605],[2103,-1.9594]]
       }
       foot_l_ik_ctrl.Location.Z {
        type=MovieSceneScriptingFloatChannel
        default=0
        thinned_from=265
        keys [[1648,0],[1801,0],[1802,114.8301],[1803,232.8555],[1804,353.8223],[1805,477.4805],[1806,603.5918],[1807,731.9238],[1808,862.2695],[1810,1128.2754],[1813,1538.502],[1824,3088.8438],[1828,3626.0684],[1831,4006.1357],[1834,4360.2402],[1837,4682.7344],[1840,4969.3442],[1843,5213.6807],[1845,5349.8193],[1847,5462.5269],[1849,5550.062],[1893,6518.3271],[1906,6636.5171],[2042,6417.1929],[2058,6707.0664],[2061,6727.5737],[2062,5354.4941],[2098,-64.2167],[2103,-64.217]]
       }
       foot_l_ik_ctrl.Rotation.X {
        type=MovieSceneScriptingFloatChannel
        default=0
        thinned_from=265
        keys [[1648,0],[1801,0],[1802,0.0613],[1803,0.2308],[1804,0.4848],[1805,0.797],[1806,1.1405],[1807,1.489],[1808,1.8183],[1809,2.1071],[1810,2.3385],[1811,2.5001],[1812,2.5849],[1813,-357.409],[1820,-358.9225],[1821,0.7528],[1822,0.4013],[1823,0],[1824,-0.6119],[1825,-1.5587],[1826,-2.8178],[1827,-4.3653],[1828,-6.1755],[1829,-8.2212],[1830,-10.474],[1831,-12.905],[1832,344.5148],[1852,284.49],[1855,273.2982],[1868,214.8127],[1871,202.7805],[1873,195.8062],[1875,189.8993],[1877,185.2564],[1878,183.4701],[1879,182.0732],[1895,170.8593],[1904,167.64],[1926,170.6744],[1981,190.4786],[1991,191.1617],[1994,189.7767],[2004,180.6994],[2007,179.0847],[2040,178.8531],[2041,180.6054],[2042,192.919],[2043,213.1892],[2044,236.8699],[2045,259.4147],[2046,276.2764],[2047,282.9075],[2050,281.8029],[2054,277.4879],[2061,265.386],[2062,245.058],[2098,-278.203],[2103,-278.2039]]
       }
       foot_l_ik_ctrl.Rotation.Y {
        type=MovieSceneScriptingFloatChannel
        default=0
        thinned_from=265
        keys [[1648,0],[1801,0],[1802,-0.0076],[1803,-0.031],[1804,-0.0728],[1805,-0.1364],[1806,-0.2253],[1807,-0.3409],[1808,-0.4827],[1809,-0.6473],[1810,-0.8289],[1812,-1.2078],[1813,-1.3839],[1814,-1.5355],[1815,-1.6509],[1816,-1.7186],[1817,-1.7277],[1818,-1.6676],[1819,-1.5281],[1820,-1.2993],[1821,-0.9724],[1822,-0.5401],[1823,0],[1824,0.5808],[1825,1.1356],[1826,1.662],[1827,2.1578],[1828,2.6218],[1829,3.054],[1830,3.4567],[1831,3.8341],[1832,4.1926],[1834,4.8883],[1835,5.2462],[1836,5.6256],[1837,6.0372],[1838,6.4906],[1839,7.021],[1840,7.6419],[1841,8.3303],[1842,9.0629],[1844,10.568],[1845,11.294],[1846,11.9714],[1847,12.5769],[1848,13.0874],[1849,13.4796],[1850,13.7304],[1851,13.8165],[1852,13.7513],[1853,13.5691],[1854,13.2786],[1855,12.8882],[1856,12.4062],[1857,11.8411],[1858,11.2014],[1859,10.4955],[1860,9.7318],[1861,8.9188],[1862,8.0651],[1863,7.179],[1864,6.2691],[1867,3.4828],[1868,2.564],[1869,1.6645],[1870,0.7932],[1871,-0.0412],[1872,-0.83],[1873,-1.5643],[1874,-2.2354],[1875,-2.8344],[1876,-3.3525],[1877,-3.781],[1878,-4.1109],[1879,-4.3335],[1880,-4.4401],[1881,-4.4771],[1882,-4.4978],[1883,-4.503],[1884,-4.4934],[1885,-4.4699],[1886,-4.4332],[1887,-4.3843],[1888,-4.3238],[1889,-4.2527],[1890,-4.1718],[1891,-4.0818],[1893,-3.8783],[1895,-3.6489],[1898,-3.2714],[1903,-2.6125],[1905,-2.3595],[1906,-2.2389],[1907,-2.1235],[1909,-1.9011],[1911,-1.67],[1913,-1.4304],[1915,-1.184],[1917,-0.9321],[1919,-0.6763],[1924,-0.0291],[1925,0.1002],[1926,0.2289],[1927,0.3569],[1928,0.484],[1929,0.6101],[1931,0.8582],[1933,1.0997],[1935,1.333],[1937,1.5566],[1939,1.7689],[1941,1.9684],[1943,2.1534],[1945,2.3225],[1958,3.3195],[1964,3.7382],[1969,4.0488],[1974,4.3132],[1978,4.4842],[1982,4.6131],[1985,4.6788],[1988,4.7155],[1990,4.7227],[1991,4.6602],[1992,4.4822],[1993,4.2026],[1994,3.8355],[1995,3.3949],[1996,2.8947],[1997,2.349],[1998,1.7717],[1999,1.177],[2000,0.5787],[2001,-0.009],[2002,-0.5722],[2003,-1.0969],[2004,-1.5691],[2005,-1.9748],[2006,-2.2999],[2007,-2.5304],[2008,-2.6524],[2009,-2.6683],[2040,-2.6683],[2041,-2.7203],[2042,-3.0858],[2043,-3.6883],[2044,-4.394],[2045,-5.0716],[2046,-5.5923],[2047,-5.8288],[2054,-6.1768],[2057,-6.2885],[2060,-6.3563],[2061,-6.3667],[2062,-0.5348],[2098,-4.3398],[2103,-4.3419]]
       }
       foot_l_ik_ctrl.Rotation.Z {
        type=MovieSceneScriptingFloatChannel
        default=0
        thinned_from=265
        keys [[1648,0],[1801,0],[1802,0.1897],[1803,0.7233],[1804,1.5475],[1805,2.6097],[1806,3.8569],[1807,5.2358],[1808,6.6923],[1809,8.1718],[1810,9.6193],[1811,10.9789],[1812,12.1946],[1813,13.2098],[1814,13.968],[1815,14.4127],[1816,14.4875],[1817,14.1364],[1818,13.3037],[1819,11.9345],[1820,9.9739],[1821,7.3677],[1822,4.0615],[1823,0],[1824,-4.9055],[1825,-10.5977],[1826,-16.9137],[1827,-23.6906],[1828,-30.7651],[1830,-45.1534],[1831,-52.1398],[1832,-58.7695],[1833,295.1217],[1834,289.6976],[1835,285.1219],[1836,281.5586],[1837,279.1716],[1838,278.1248],[1853,272.8715],[1857,268.6382],[1862,260.1809],[1874,233.0723],[1875,-129.216],[1879,-137.8319],[1893,-163.7696],[1900,-174.4706],[1906,-181.4302],[1964,-220.433],[1988,-227.0722],[2041,-226.7498],[2042,-225.1844],[2045,-216.7234],[2046,-214.5758],[2047,-213.7308],[2061,-218.7608],[2062,-193.8979],[2098,-21.6226],[2103,-21.6224]]
       }
       foot_l_ik_ctrl.Scale.X {type=MovieSceneScriptingFloatChannel, default=1, const [1,264]}
       foot_l_ik_ctrl.Scale.Y {type=MovieSceneScriptingFloatChannel, default=1, const [1,264]}
       foot_l_ik_ctrl.Scale.Z {type=MovieSceneScriptingFloatChannel, default=1, const [1,264]}
       ball_l_ik_ctrl.Location.X {type=MovieSceneScriptingFloatChannel, default=0, thinned_from=264, keys [[1648,0],[2098,0]]}
       ball_l_ik_ctrl.Location.Y {type=MovieSceneScriptingFloatChannel, default=0, thinned_from=265, keys [[1648,0],[2061,0],[2062,-0.0001],[2103,0]]}
       ball_l_ik_ctrl.Location.Z {type=MovieSceneScriptingFloatChannel, default=0, thinned_from=264, keys [[1648,0],[2098,0]]}
       ball_l_ik_ctrl.Rotation.X {type=MovieSceneScriptingFloatChannel, default=0, thinned_from=264, keys [[1648,0],[2098,0]]}
       ball_l_ik_ctrl.Rotation.Y {type=MovieSceneScriptingFloatChannel, default=0, thinned_from=264, keys [[1648,0],[2098,0]]}
       ball_l_ik_ctrl.Rotation.Z {type=MovieSceneScriptingFloatChannel, default=0, thinned_from=264, keys [[1648,0],[2098,0]]}
       ball_l_ik_ctrl.Scale.X {type=MovieSceneScriptingFloatChannel, default=1, const [1,264]}
       ball_l_ik_ctrl.Scale.Y {type=MovieSceneScriptingFloatChannel, default=1, const [1,264]}
       ball_l_ik_ctrl.Scale.Z {type=MovieSceneScriptingFloatChannel, default=1, const [1,264]}
       foot_bk1_l_ctrl.Location.X {type=MovieSceneScriptingFloatChannel, default=0, const [0,263]}
       foot_bk1_l_ctrl.Location.Y {type=MovieSceneScriptingFloatChannel, default=0, const [0,263]}
       foot_bk1_l_ctrl.Location.Z {type=MovieSceneScriptingFloatChannel, default=0, const [0,263]}
       foot_bk1_l_ctrl.Rotation.X {type=MovieSceneScriptingFloatChannel, default=0, const [0,263]}
       foot_bk1_l_ctrl.Rotation.Y {type=MovieSceneScriptingFloatChannel, default=0, const [0,263]}
       foot_bk1_l_ctrl.Rotation.Z {type=MovieSceneScriptingFloatChannel, default=0, const [0,263]}
       foot_bk1_l_ctrl.Scale.X {type=MovieSceneScriptingFloatChannel, default=1, const [1,263]}
       foot_bk1_l_ctrl.Scale.Y {type=MovieSceneScriptingFloatChannel, default=1, const [1,263]}
       foot_bk1_l_ctrl.Scale.Z {type=MovieSceneScriptingFloatChannel, default=1, const [1,263]}
       foot_roll_l_ctrl.Location.X {type=MovieSceneScriptingFloatChannel, default=0, const [0,263]}
       foot_roll_l_ctrl.Location.Y {type=MovieSceneScriptingFloatChannel, default=0, const [0,263]}
       foot_roll_l_ctrl.Location.Z {type=MovieSceneScriptingFloatChannel, default=0, const [0,263]}
       foot_roll_l_ctrl.Rotation.X {type=MovieSceneScriptingFloatChannel, default=0, const [0,263]}
       foot_roll_l_ctrl.Rotation.Y {type=MovieSceneScriptingFloatChannel, default=0, const [0,263]}
       foot_roll_l_ctrl.Rotation.Z {type=MovieSceneScriptingFloatChannel, default=0, const [0,263]}
       foot_roll_l_ctrl.Scale.X {type=MovieSceneScriptingFloatChannel, default=1, const [1,263]}
       foot_roll_l_ctrl.Scale.Y {type=MovieSceneScriptingFloatChannel, default=1, const [1,263]}
       foot_roll_l_ctrl.Scale.Z {type=MovieSceneScriptingFloatChannel, default=1, const [1,263]}
       roll_blend_l_ctrl {type=MovieSceneScriptingFloatChannel, default=1, const [1,263]}
       tip_l_ctrl.Location.X {type=MovieSceneScriptingFloatChannel, default=0, const [0,263]}
       tip_l_ctrl.Location.Y {type=MovieSceneScriptingFloatChannel, default=0, const [0,263]}
       tip_l_ctrl.Location.Z {type=MovieSceneScriptingFloatChannel, default=0, const [0,263]}
       tip_l_ctrl.Rotation.X {type=MovieSceneScriptingFloatChannel, default=0, const [0,263]}
       tip_l_ctrl.Rotation.Y {type=MovieSceneScriptingFloatChannel, default=0, const [0,263]}
       tip_l_ctrl.Rotation.Z {type=MovieSceneScriptingFloatChannel, default=0, const [0,263]}
       tip_l_ctrl.Scale.X {type=MovieSceneScriptingFloatChannel, default=1, const [1,263]}
       tip_l_ctrl.Scale.Y {type=MovieSceneScriptingFloatChannel, default=1, const [1,263]}
       tip_l_ctrl.Scale.Z {type=MovieSceneScriptingFloatChannel, default=1, const [1,263]}
       heel_l_ctrl.Location.X {type=MovieSceneScriptingFloatChannel, default=0, const [0,263]}
       heel_l_ctrl.Location.Y {type=MovieSceneScriptingFloatChannel, default=0, const [0,263]}
       heel_l_ctrl.Location.Z {type=MovieSceneScriptingFloatChannel, default=0, const [0,263]}
       heel_l_ctrl.Rotation.X {type=MovieSceneScriptingFloatChannel, default=0, const [0,263]}
       heel_l_ctrl.Rotation.Y {type=MovieSceneScriptingFloatChannel, default=0, const [0,263]}
       heel_l_ctrl.Rotation.Z {type=MovieSceneScriptingFloatChannel, default=0, const [0,263]}
       heel_l_ctrl.Scale.X {type=MovieSceneScriptingFloatChannel, default=1, const [1,263]}
       heel_l_ctrl.Scale.Y {type=MovieSceneScriptingFloatChannel, default=1, const [1,263]}
       heel_l_ctrl.Scale.Z {type=MovieSceneScriptingFloatChannel, default=1, const [1,263]}
       leg_l_softness {type=MovieSceneScriptingFloatChannel, default=0, const [0,263]}
       leg_l_pv_ik_ctrl.Location.X {
        type=MovieSceneScriptingFloatChannel
        default=0
        thinned_from=265
        keys [[1648,-0.0405],[1801,0],[1969,0],[1970,0.0005],[1971,0],[2061,0],[2062,-1654.0098],[2098,10.1392],[2103,10.139]]
       }
       leg_l_pv_ik_ctrl.Location.Y {
        type=MovieSceneScriptingFloatChannel
        default=0
        thinned_from=265
        keys [[1648,0.1141],[1801,0],[1811,0],[1812,0.0001],[1835,0],[1836,-0.0002],[1837,0],[1845,0],[1846,-0.0002],[1847,0],[1849,0],[1850,-0.0002],[1851,0],[1853,0],[1854,-0.0002],[1855,0],[1858,0],[1859,-0.0002],[1860,0],[1876,0],[1878,0.0001],[1880,0],[1881,0.0001],[1916,0],[1917,-0.0002],[1918,0],[1919,0],[1920,-0.0002],[1921,0],[2061,0],[2062,1321.527],[2098,16.9834],[2103,16.9833]]
       }
       leg_l_pv_ik_ctrl.Location.Z {
        type=MovieSceneScriptingFloatChannel
        default=0
        thinned_from=265
        keys [[1648,0.1056],[1801,0.0002],[1802,-0.001],[1803,0.0007],[1804,-0.0002],[1805,-0.0002],[1806,0],[1807,-0.001],[1808,-0.0012],[1809,0.0006],[1810,0.0006],[1811,0.0002],[1812,0.0005],[1814,-0.0006],[1815,0.0004],[1816,-0.0001],[1817,0.0001],[1818,-0.001],[1819,0.0007],[1820,-0.001],[1821,0.001],[1822,-0.0009],[1823,0.0007],[1825,-0.0003],[1826,-0.0004],[1827,0],[1828,-0.0003],[1829,0.0001],[1830,0],[1831,0.0004],[1833,-0.0005],[1834,0.0002],[1835,0.0002],[1836,-0.0002],[1837,0.0002],[1838,0.0004],[1839,-0.0001],[1848,0],[1849,0.0001],[1850,0],[1855,0.0001],[1857,-0.0001],[1859,0.0002],[1860,0],[1862,-0.0001],[1863,0.0002],[1864,0],[1865,0],[1866,0.0002],[1867,0],[1868,0],[1869,0.0002],[1870,0.0002],[1871,-0.0002],[1872,0.0002],[1873,0],[1874,0.0002],[1875,-0.0002],[1876,0],[1877,-0.0002],[1878,-0.0002],[1879,0],[1884,0],[1885,0.0005],[1886,0],[1889,0],[1890,0.0005],[1891,0.0005],[1892,0],[1893,0],[1894,0.0005],[1895,0],[1896,0],[1897,0.0005],[1898,0],[1899,0.0005],[1900,0],[1903,0],[1904,0.0005],[1905,0],[1913,0],[1914,0.0005],[1915,0],[1925,0],[1926,0.0005],[1927,0],[1928,0],[1929,0.0005],[1930,0],[1939,0],[1940,-0.0005],[1942,0.0005],[1943,0],[1947,0],[1948,0.0005],[1949,0],[1951,0],[1952,0.0005],[1953,0],[1960,0],[1961,-0.0005],[1962,0],[1963,-0.0005],[1964,0],[1965,0],[1966,0.0005],[1967,0],[1968,0.0005],[1969,0],[1979,0],[1980,0.0005],[1981,0.0005],[1982,0],[1983,0],[1984,-0.0005],[1985,0],[1989,0],[1990,0.0005],[1991,0],[1995,0],[1996,-0.0005],[1997,0],[1998,-0.0005],[1999,0],[2000,0],[2001,0.0005],[2002,0],[2004,0],[2005,0.0005],[2006,0],[2009,0],[2010,0.0005],[2011,-0.0005],[2012,0],[2017,0],[2018,0.0005],[2019,0],[2021,0],[2022,0.0005],[2023,0],[2039,0],[2040,0.0005],[2041,0],[2045,0],[2046,-0.0005],[2047,0],[2048,-0.0005],[2049,0],[2054,0],[2055,-0.0002],[2056,0.0002],[2058,0.0002],[2059,0],[2061,-0.0001],[2062,5250.8232],[2098,-54.9273],[2103,-54.9264]]
       }
       leg_l_pv_ik_ctrl.Rotation.X {
        type=MovieSceneScriptingFloatChannel
        default=0
        thinned_from=265
        keys [[1648,-0.2794],[1801,0],[1824,0],[1825,0.0001],[1826,0],[1827,0.0001],[1829,-0.0001],[1830,0.0001],[1831,-0.0001],[1833,-0.0001],[1835,0.0001],[1836,-0.0001],[1838,0.0002],[1839,0],[1840,0.0001],[1843,-0.0002],[1844,0],[1845,-0.0002],[1851,-0.0002],[1853,0.0001],[1855,-0.0001],[1857,0.0002],[1859,-0.0002],[1860,0.0001],[1862,0.0001],[1863,-0.0001],[1871,0.0001],[1876,0],[1878,0.0001],[1880,-0.0001],[1907,0.0001],[1908,-0.0001],[1913,0],[1914,-0.0001],[1916,0.0002],[1917,-0.0002],[1919,0],[1920,-0.0001],[1921,0.0001],[1923,0],[1924,0.0001],[1936,-0.0001],[1937,0.0001],[1939,0],[1940,0.0001],[1942,0],[1943,0.0001],[1944,-0.0001],[1945,0.0001],[1948,0],[1949,0.0003],[1952,-0.0002],[1954,0.0001],[1957,0.0002],[1958,0],[1961,0.0002],[1962,-0.0001],[1964,0.0002],[1969,-0.0001],[1974,0.0001],[1975,-0.0002],[1977,0.0001],[1978,-0.0001],[1985,0.0002],[1986,0],[1989,0],[1990,-0.0002],[1992,0.0001],[1993,-0.0001],[1996,0.0003],[1997,0.0001],[1998,0.0003],[2000,-0.0001],[2001,-0.0002],[2002,0.0002],[2006,-0.0001],[2009,0.0002],[2014,0.0001],[2015,0.0002],[2016,-0.0001],[2017,-0.0002],[2018,0.0001],[2019,-0.0001],[2020,0.0002],[2021,-0.0002],[2022,-0.0003],[2023,0.0001],[2024,-0.0001],[2025,0.0001],[2029,0],[2030,0.0002],[2033,-0.0001],[2034,0],[2036,-0.0001],[2038,0],[2040,-0.0002],[2041,0],[2042,-0.0001],[2043,0.0001],[2049,0.0001],[2050,-0.0001],[2052,0.0001],[2053,0],[2055,0.0002],[2056,-0.0001],[2058,0],[2059,-0.0002],[2060,0.0001],[2061,0.0002],[2062,111.5672],[2098,-53.0996],[2103,-53.1009]]
       }
       leg_l_pv_ik_ctrl.Rotation.Y {
        type=MovieSceneScriptingFloatChannel
        default=0
        thinned_from=265
        keys [[1648,-1.2115],[1801,0],[1814,0.0001],[1819,0],[1820,0.0001],[1821,-0.0001],[1822,0],[1823,-0.0001],[1832,0],[1833,0.0002],[1835,-0.0002],[1836,0.0001],[1837,-0.0002],[1839,-0.0001],[1840,0.0002],[1841,-0.0001],[1843,0],[1845,-0.0001],[1846,0.0001],[1847,-0.0001],[1849,0.0001],[1853,0],[1854,-0.0002],[1856,0],[1857,-0.0001],[1860,0],[1862,0],[1863,0.0001],[1866,0.0001],[1867,-0.0001],[1884,0],[1900,0],[1901,-0.0002],[1903,0],[1908,0],[1909,-0.0003],[1913,-0.0001],[1914,0.0001],[1916,0.0001],[1917,-0.0002],[1922,0],[1923,-0.0001],[1924,0.0001],[1925,-0.0001],[1926,0],[1927,-0.0002],[1934,0.0001],[1935,-0.0002],[1937,0.0001],[1939,-0.0002],[1940,0.0002],[1942,-0.0001],[1943,0.0001],[1944,-0.0002],[1945,0.0001],[1946,-0.0001],[1947,0.0001],[1951,-0.0001],[1952,0],[1953,-0.0001],[1954,0.0001],[1955,-0.0001],[1960,0.0003],[1962,-0.0001],[1964,0.0003],[1965,0],[1966,0.0001],[1970,-0.0003],[1971,0.0001],[1972,-0.0002],[1974,0.0001],[1976,-0.0002],[1977,0.0002],[1979,-0.0002],[1984,0.0002],[1985,-0.0001],[1986,0.0001],[1995,-0.0001],[1996,0.0002],[1997,-0.0001],[1998,0],[1999,-0.0001],[2000,0.0001],[2003,0.0002],[2005,0.0001],[2006,-0.0002],[2007,0.0002],[2008,0.0002],[2009,0],[2010,0.0002],[2011,-0.0001],[2013,0.0001],[2015,0],[2016,0.0001],[2019,-0.0002],[2020,0],[2022,-0.0001],[2024,0.0001],[2026,-0.0001],[2029,0.0002],[2032,-0.0001],[2034,0.0002],[2035,0],[2036,0.0001],[2037,-0.0001],[2038,0.0003],[2041,-0.0002],[2042,0.0001],[2044,-0.0001],[2050,0.0001],[2052,0],[2054,0],[2055,0.0002],[2056,-0.0001],[2058,-0.0001],[2059,0.0002],[2060,-0.0001],[2061,0],[2062,-2.214],[2098,-6.0957],[2103,-6.0974]]
       }
       leg_l_pv_ik_ctrl.Rotation.Z {
        type=MovieSceneScriptingFloatChannel
        default=0
        thinned_from=265
        keys [[1648,-1.1917],[1801,0],[2061,0],[2062,153.9469],[2098,-9.431],[2103,-9.4307]]
       }
       leg_l_pv_ik_ctrl.Scale.X {type=MovieSceneScriptingFloatChannel, default=1, const [1,264]}
       leg_l_pv_ik_ctrl.Scale.Y {type=MovieSceneScriptingFloatChannel, default=1, const [1,264]}
       leg_l_pv_ik_ctrl.Scale.Z {type=MovieSceneScriptingFloatChannel, default=1, const [1,264]}
       thigh_r_fk_ctrl.Location.X {type=MovieSceneScriptingFloatChannel, default=0, const [0,35]}
       thigh_r_fk_ctrl.Location.Y {type=MovieSceneScriptingFloatChannel, default=0, const [0,35]}
       thigh_r_fk_ctrl.Location.Z {type=MovieSceneScriptingFloatChannel, default=0, const [0,35]}
       thigh_r_fk_ctrl.Rotation.X {
        type=MovieSceneScriptingFloatChannel
        default=10.1334
        thinned_from=37
        keys [[1801,10.1334],[1816,10.1334],[1819,5.5482],[1825,10.1334],[1885,10.1334],[1888,62.5769],[1889,10.1334],[1984,10.1334],[1994,4.7163],[2007,10.1334],[2098,10.1334]]
       }
       thigh_r_fk_ctrl.Rotation.Y {
        type=MovieSceneScriptingFloatChannel
        default=-4.4158
        thinned_from=37
        keys [[1801,-4.4158],[1816,-1.5734],[1819,5.7726],[1825,-0.2791],[1830,-0.1411],[1832,-0.0763],[1837,-0.0031],[1885,-0.0031],[1888,37.3905],[1889,36.8454],[1891,32.3199],[1894,23.1374],[1902,12.7851],[1907,-0.0031],[1938,-0.0031],[1942,-4.4158],[1950,-4.4158],[1959,-2.8082],[1984,-4.4158],[2007,-0.0031],[2098,-0.0031]]
       }
       thigh_r_fk_ctrl.Rotation.Z {
        type=MovieSceneScriptingFloatChannel
        default=1.0959
        thinned_from=43
        keys [[1801,1.0959],[1816,-2.6313],[1819,29.6174],[1825,21.5911],[1830,6.7654],[1832,6.4553],[1837,30.2812],[1841,11.5376],[1848,-15.8217],[1883,-15.8217],[1885,-22.6073],[1888,-16.8819],[1889,-15.8217],[1935,-15.8217],[1938,-49.0664],[1942,1.0959],[1950,1.0959],[1955,29.8691],[1959,31.6047],[1964,15.396],[1972,0],[1984,1.0959],[2007,-15.8217],[2014,-5.4369],[2017,-44.4986],[2018,-16.9417],[2019,27.7172],[2020,26.4312],[2021,39.4609],[2023,2.1418],[2027,-15.8217],[2098,-15.8217]]
       }
       thigh_r_fk_ctrl.Scale.X {type=MovieSceneScriptingFloatChannel, default=1, const [1,35]}
       thigh_r_fk_ctrl.Scale.Y {type=MovieSceneScriptingFloatChannel, default=1, const [1,35]}
       thigh_r_fk_ctrl.Scale.Z {type=MovieSceneScriptingFloatChannel, default=1, const [1,35]}
       calf_r_fk_ctrl.Location.X {type=MovieSceneScriptingFloatChannel, default=0, const [0,30]}
       calf_r_fk_ctrl.Location.Y {type=MovieSceneScriptingFloatChannel, default=0, const [0,30]}
       calf_r_fk_ctrl.Location.Z {type=MovieSceneScriptingFloatChannel, default=0, const [0,30]}
       calf_r_fk_ctrl.Rotation.X {type=MovieSceneScriptingFloatChannel, default=0, const [0,30]}
       calf_r_fk_ctrl.Rotation.Y {type=MovieSceneScriptingFloatChannel, default=0, const [0,30]}
       calf_r_fk_ctrl.Rotation.Z {
        type=MovieSceneScriptingFloatChannel
        default=-0.2803
        thinned_from=36
        keys [[1801,-0.2803],[1816,-0.2803],[1819,48.5745],[1825,50.9305],[1830,-0.2803],[1837,-0.2803],[1841,20.0421],[1848,-0.2803],[1889,-0.2803],[1891,37.1745],[1894,56.3777],[1902,31.0307],[1907,-0.2803],[1950,-0.2803],[1955,28.4929],[1960,-0.2803],[1964,14.0197],[1972,0],[1984,-0.2803],[1994,88.2501],[2007,35.762],[2021,-0.2803],[2098,-0.2803]]
       }
       calf_r_fk_ctrl.Scale.X {type=MovieSceneScriptingFloatChannel, default=1, const [1,30]}
       calf_r_fk_ctrl.Scale.Y {type=MovieSceneScriptingFloatChannel, default=1, const [1,30]}
       calf_r_fk_ctrl.Scale.Z {type=MovieSceneScriptingFloatChannel, default=1, const [1,30]}
       foot_r_fk_ctrl.Location.X {type=MovieSceneScriptingFloatChannel, default=0, const [0,15]}
       foot_r_fk_ctrl.Location.Y {type=MovieSceneScriptingFloatChannel, default=0, const [0,15]}
       foot_r_fk_ctrl.Location.Z {type=MovieSceneScriptingFloatChannel, default=0, const [0,15]}
       foot_r_fk_ctrl.Rotation.X {type=MovieSceneScriptingFloatChannel, default=-2.8885, const [-2.8885,15]}
       foot_r_fk_ctrl.Rotation.Y {type=MovieSceneScriptingFloatChannel, default=-10.6402, const [-10.6402,15]}
       foot_r_fk_ctrl.Rotation.Z {type=MovieSceneScriptingFloatChannel, default=30.8604, const [30.8604,15]}
       foot_r_fk_ctrl.Scale.X {type=MovieSceneScriptingFloatChannel, default=1, const [1,15]}
       foot_r_fk_ctrl.Scale.Y {type=MovieSceneScriptingFloatChannel, default=1, const [1,15]}
       foot_r_fk_ctrl.Scale.Z {type=MovieSceneScriptingFloatChannel, default=1, const [1,15]}
       ball_r_fk_ctrl.Location.X {type=MovieSceneScriptingFloatChannel, default=0, const [0,15]}
       ball_r_fk_ctrl.Location.Y {type=MovieSceneScriptingFloatChannel, default=0, const [0,15]}
       ball_r_fk_ctrl.Location.Z {type=MovieSceneScriptingFloatChannel, default=0, const [0,15]}
       ball_r_fk_ctrl.Rotation.X {type=MovieSceneScriptingFloatChannel, default=0, const [0,15]}
       ball_r_fk_ctrl.Rotation.Y {type=MovieSceneScriptingFloatChannel, default=0, const [0,15]}
       ball_r_fk_ctrl.Rotation.Z {type=MovieSceneScriptingFloatChannel, default=0, const [0,15]}
       ball_r_fk_ctrl.Scale.X {type=MovieSceneScriptingFloatChannel, default=1, const [1,15]}
       ball_r_fk_ctrl.Scale.Y {type=MovieSceneScriptingFloatChannel, default=1, const [1,15]}
       ball_r_fk_ctrl.Scale.Z {type=MovieSceneScriptingFloatChannel, default=1, const [1,15]}
       foot_r_ik_ctrl.Location.X {
        type=MovieSceneScriptingFloatChannel
        default=0
        thinned_from=264
        keys [[1801,0],[1802,16.941],[1803,34.5117],[1804,52.5406],[1805,70.8733],[1809,144.9955],[1810,163.2399],[1811,181.1292],[1812,198.4968],[1813,215.1578],[1814,230.9319],[1815,245.6739],[1816,259.3116],[1817,271.8825],[1821,317.3895],[1822,330.3116],[1823,345.263],[1828,424.754],[1838,573.8842],[1850,655.3793],[1853,671.9486],[1864,708.9883],[1866,718.9076],[1868,732.1183],[1870,749.5869],[1872,772.2678],[1873,785.853],[1874,801.0845],[1875,818.0718],[1877,857.3149],[1879,901.3164],[1882,973.9094],[1896,1347.5029],[1900,1443.2144],[1903,1505.4414],[1906,1556.8794],[1935,1904.0986],[1984,2248.9512],[2041,2385.4297],[2042,2379.8896],[2044,2352.5273],[2046,2311.4326],[2048,2257.2881],[2050,2189.3657],[2052,2109.2808],[2054,2018.7449],[2057,1867.0938],[2061,1644.2925],[2062,1627.8691],[2098,-0.7565],[2103,-0.7542]]
       }
       foot_r_ik_ctrl.Location.Y {
        type=MovieSceneScriptingFloatChannel
        default=0
        thinned_from=264
        keys [[1801,0],[1802,11.032],[1803,22.5137],[1804,35.1211],[1805,49.3629],[1806,65.5582],[1807,83.8298],[1808,104.1121],[1809,126.1724],[1810,149.6453],[1811,174.0763],[1813,223.8478],[1814,248.2817],[1815,271.9473],[1816,294.6443],[1817,316.3121],[1818,337.0315],[1822,415.9221],[1834,670.3005],[1838,749.5125],[1849,639.6462],[1851,618.0847],[1852,604.5654],[1853,586.0698],[1854,562.9839],[1855,535.6929],[1856,504.5786],[1857,470.0227],[1858,432.4043],[1859,392.1011],[1860,349.4893],[1861,304.9443],[1862,258.8411],[1863,211.5542],[1864,163.4592],[1866,66.355],[1867,18.1057],[1868,-29.4312],[1869,-75.8684],[1870,-120.8163],[1871,-163.8816],[1872,-204.6689],[1873,-242.7803],[1874,-277.8157],[1875,-309.374],[1876,-337.0533],[1877,-362.2245],[1878,-386.5181],[1879,-409.9213],[1880,-432.4221],[1882,-474.8278],[1884,-513.8594],[1886,-549.5114],[1888,-581.7783],[1890,-610.655],[1892,-636.1376],[1894,-658.2234],[1896,-676.912],[1898,-692.2048],[1900,-704.1057],[1902,-712.6201],[1904,-717.7559],[1906,-719.5212],[1941,-667.2405],[1992,-532.3125],[2002,-495.6538],[2013,-443.4629],[2032,-335.0337],[2040,-286.145],[2041,-278.7944],[2042,-263.2588],[2043,-238.9893],[2044,-207.166],[2045,-169.4585],[2046,-127.1592],[2047,-80.7363],[2048,-29.9253],[2049,25.3975],[2050,84.7734],[2051,147.7446],[2052,213.8521],[2053,282.6367],[2054,353.6401],[2055,426.4038],[2060,800.5913],[2061,874.2993],[2062,1329.8405],[2098,24.1319],[2103,24.132]]
       }
       foot_r_ik_ctrl.Location.Z {
        type=MovieSceneScriptingFloatChannel
        default=0
        thinned_from=264
        keys [[1801,0],[1802,114.834],[1803,232.873],[1804,353.8594],[1805,477.5527],[1806,603.7109],[1807,732.1055],[1808,862.5273],[1810,1128.7168],[1813,1539.2402],[1824,3088.5352],[1828,3624.6719],[1831,4004.0947],[1834,4357.6396],[1837,4679.5244],[1840,4965.2856],[1842,5132.5605],[1844,5278.9727],[1846,5402.8799],[1848,5502.6387],[1850,5576.6011],[1893,6520.3916],[1906,6637.7095],[2042,6418.8359],[2058,6710.4248],[2061,6730.958],[2062,5393.188],[2098,-101.3431],[2103,-101.3428]]
       }
       foot_r_ik_ctrl.Rotation.X {
        type=MovieSceneScriptingFloatChannel
        default=0
        thinned_from=264
        keys [[1801,0],[1802,0.0613],[1803,0.2308],[1804,0.4848],[1805,0.797],[1806,1.1405],[1807,1.489],[1808,1.8183],[1809,2.1071],[1810,2.3385],[1811,2.5001],[1812,2.5849],[1813,-357.409],[1820,-358.9225],[1821,0.7528],[1822,0.4013],[1823,0],[1824,-0.6119],[1825,-1.5587],[1826,-2.8178],[1827,-4.3653],[1828,-6.1755],[1829,-8.2212],[1830,-10.474],[1831,-12.905],[1832,344.5148],[1852,284.49],[1855,273.2982],[1868,214.8127],[1871,202.7805],[1873,195.8062],[1875,189.8993],[1877,185.2564],[1878,183.4701],[1879,182.0732],[1895,170.8593],[1904,167.64],[1926,170.6744],[1981,190.4786],[1991,191.1617],[1994,189.7767],[2004,180.6994],[2007,179.0847],[2040,178.8531],[2041,180.6054],[2042,192.919],[2043,213.1892],[2044,236.8699],[2045,259.4147],[2046,276.2764],[2047,282.9075],[2050,281.8029],[2054,277.4879],[2061,265.386],[2062,193.9569],[2098,13.898],[2103,13.8977]]
       }
       foot_r_ik_ctrl.Rotation.Y {
        type=MovieSceneScriptingFloatChannel
        default=0
        thinned_from=264
        keys [[1801,0],[1802,0.0076],[1803,0.031],[1804,0.0728],[1805,0.1364],[1806,0.2253],[1807,0.3409],[1808,0.4827],[1809,0.6473],[1810,0.8289],[1812,1.2078],[1813,1.3839],[1814,1.5355],[1815,1.6509],[1816,1.7186],[1817,1.7277],[1818,1.6676],[1819,1.5281],[1820,1.2993],[1821,0.9724],[1822,0.5401],[1823,0],[1824,-0.5808],[1825,-1.1356],[1826,-1.662],[1827,-2.1578],[1828,-2.6218],[1829,-3.054],[1830,-3.4567],[1831,-3.8341],[1832,-4.1926],[1834,-4.8883],[1835,-5.2462],[1836,-5.6256],[1837,-6.0372],[1838,-6.4906],[1839,-7.021],[1840,-7.6419],[1841,-8.3303],[1842,-9.0629],[1844,-10.568],[1845,-11.294],[1846,-11.9714],[1847,-12.5769],[1848,-13.0874],[1849,-13.4796],[1850,-13.7304],[1851,-13.8165],[1852,-13.7513],[1853,-13.5691],[1854,-13.2786],[1855,-12.8882],[1856,-12.4062],[1857,-11.8411],[1858,-11.2014],[1859,-10.4955],[1860,-9.7318],[1861,-8.9188],[1862,-8.0651],[1863,-7.179],[1864,-6.2691],[1867,-3.4828],[1868,-2.564],[1869,-1.6645],[1870,-0.7932],[1871,0.0412],[1872,0.83],[1873,1.5643],[1874,2.2354],[1875,2.8344],[1876,3.3525],[1877,3.781],[1878,4.1109],[1879,4.3335],[1880,4.4401],[1881,4.4771],[1882,4.4978],[1883,4.503],[1884,4.4934],[1885,4.4699],[1886,4.4332],[1887,4.3843],[1888,4.3238],[1889,4.2527],[1890,4.1718],[1891,4.0818],[1893,3.8783],[1895,3.6489],[1898,3.2714],[1903,2.6125],[1905,2.3595],[1906,2.2389],[1907,2.1235],[1909,1.9011],[1911,1.67],[1913,1.4304],[1915,1.184],[1917,0.9321],[1919,0.6763],[1924,0.0291],[1925,-0.1002],[1926,-0.2289],[1927,-0.3569],[1928,-0.484],[1929,-0.6101],[1931,-0.8582],[1933,-1.0997],[1935,-1.333],[1937,-1.5566],[1939,-1.7689],[1941,-1.9684],[1943,-2.1534],[1945,-2.3225],[1958,-3.3195],[1964,-3.7382],[1969,-4.0488],[1974,-4.3132],[1978,-4.4842],[1982,-4.6131],[1985,-4.6788],[1988,-4.7155],[1990,-4.7227],[1991,-4.6602],[1992,-4.4822],[1993,-4.2026],[1994,-3.8355],[1995,-3.3949],[1996,-2.8947],[1997,-2.349],[1998,-1.7717],[1999,-1.177],[2000,-0.5787],[2001,0.009],[2002,0.5722],[2003,1.0969],[2004,1.5691],[2005,1.9748],[2006,2.2999],[2007,2.5304],[2008,2.6524],[2009,2.6683],[2040,2.6683],[2041,2.7203],[2042,3.0858],[2043,3.6883],[2044,4.394],[2045,5.0716],[2046,5.5923],[2047,5.8288],[2054,6.1768],[2057,6.2885],[2060,6.3563],[2061,6.3667],[2062,24.1219],[2098,-13.627],[2103,-13.6255]]
       }
       foot_r_ik_ctrl.Rotation.Z {
        type=MovieSceneScriptingFloatChannel
        default=0
        thinned_from=264
        keys [[1801,0],[1802,-0.1897],[1803,-0.7233],[1804,-1.5475],[1805,-2.6097],[1806,-3.8569],[1807,-5.2358],[1808,-6.6923],[1809,-8.1718],[1810,-9.6193],[1811,-10.9789],[1812,-12.1946],[1813,-13.2098],[1814,-13.968],[1815,-14.4127],[1816,-14.4875],[1817,-14.1364],[1818,-13.3037],[1819,-11.9345],[1820,-9.9739],[1821,-7.3677],[1822,-4.0615],[1823,0],[1824,4.9055],[1825,10.5977],[1826,16.9137],[1827,23.6906],[1828,30.7651],[1830,45.1534],[1831,52.1398],[1832,58.7695],[1833,64.8783],[1834,70.3024],[1835,74.8781],[1836,78.4414],[1837,80.8284],[1838,81.8752],[1851,86.0911],[1852,86.511],[1853,87.1285],[1854,87.933],[1856,90.0603],[1858,92.8074],[1860,96.0888],[1863,101.8262],[1874,126.9277],[1875,-230.784],[1892,-197.9318],[1900,-185.5294],[1906,-178.5698],[1941,-151.727],[1971,-136.6822],[1985,-133.1749],[2040,-133.0275],[2041,-133.2502],[2042,-134.8156],[2043,-137.3934],[2045,-143.2766],[2046,-145.4242],[2047,-146.2692],[2053,-145.0821],[2061,-141.2392],[2062,-136.1738],[2098,-11.3627],[2103,-11.3626]]
       }
       foot_r_ik_ctrl.Scale.X {type=MovieSceneScriptingFloatChannel, default=1, const [1,263]}
       foot_r_ik_ctrl.Scale.Y {type=MovieSceneScriptingFloatChannel, default=1, const [1,263]}
       foot_r_ik_ctrl.Scale.Z {type=MovieSceneScriptingFloatChannel, default=1, const [1,263]}
       ball_r_ik_ctrl.Location.X {type=MovieSceneScriptingFloatChannel, default=0, thinned_from=263, keys [[1801,0],[2098,0]]}
       ball_r_ik_ctrl.Location.Y {type=MovieSceneScriptingFloatChannel, default=0, thinned_from=263, keys [[1801,0],[2098,0]]}
       ball_r_ik_ctrl.Location.Z {type=MovieSceneScriptingFloatChannel, default=0, thinned_from=264, keys [[1801,0],[2061,0],[2062,0.0002],[2103,0]]}
       ball_r_ik_ctrl.Rotation.X {type=MovieSceneScriptingFloatChannel, default=0, thinned_from=263, keys [[1801,0],[2098,0]]}
       ball_r_ik_ctrl.Rotation.Y {type=MovieSceneScriptingFloatChannel, default=0, thinned_from=263, keys [[1801,0],[2098,0]]}
       ball_r_ik_ctrl.Rotation.Z {type=MovieSceneScriptingFloatChannel, default=0, thinned_from=263, keys [[1801,0],[2098,0]]}
       ball_r_ik_ctrl.Scale.X {type=MovieSceneScriptingFloatChannel, default=1, const [1,263]}
       ball_r_ik_ctrl.Scale.Y {type=MovieSceneScriptingFloatChannel, default=1, const [1,263]}
       ball_r_ik_ctrl.Scale.Z {type=MovieSceneScriptingFloatChannel, default=1, const [1,263]}
       foot_bk1_r_ctrl.Location.X {type=MovieSceneScriptingFloatChannel, default=0, const [0,263]}
       foot_bk1_r_ctrl.Location.Y {type=MovieSceneScriptingFloatChannel, default=0, const [0,263]}
       foot_bk1_r_ctrl.Location.Z {type=MovieSceneScriptingFloatChannel, default=0, const [0,263]}
       foot_bk1_r_ctrl.Rotation.X {type=MovieSceneScriptingFloatChannel, default=0, const [0,263]}
       foot_bk1_r_ctrl.Rotation.Y {type=MovieSceneScriptingFloatChannel, default=0, const [0,263]}
       foot_bk1_r_ctrl.Rotation.Z {type=MovieSceneScriptingFloatChannel, default=0, const [0,263]}
       foot_bk1_r_ctrl.Scale.X {type=MovieSceneScriptingFloatChannel, default=1, const [1,263]}
       foot_bk1_r_ctrl.Scale.Y {type=MovieSceneScriptingFloatChannel, default=1, const [1,263]}
       foot_bk1_r_ctrl.Scale.Z {type=MovieSceneScriptingFloatChannel, default=1, const [1,263]}
       foot_roll_r_ctrl.Location.X {type=MovieSceneScriptingFloatChannel, default=0, const [0,263]}
       foot_roll_r_ctrl.Location.Y {type=MovieSceneScriptingFloatChannel, default=0, const [0,263]}
       foot_roll_r_ctrl.Location.Z {type=MovieSceneScriptingFloatChannel, default=0, const [0,263]}
       foot_roll_r_ctrl.Rotation.X {type=MovieSceneScriptingFloatChannel, default=0, const [0,263]}
       foot_roll_r_ctrl.Rotation.Y {type=MovieSceneScriptingFloatChannel, default=0, const [0,263]}
       foot_roll_r_ctrl.Rotation.Z {type=MovieSceneScriptingFloatChannel, default=0, const [0,263]}
       foot_roll_r_ctrl.Scale.X {type=MovieSceneScriptingFloatChannel, default=1, const [1,263]}
       foot_roll_r_ctrl.Scale.Y {type=MovieSceneScriptingFloatChannel, default=1, const [1,263]}
       foot_roll_r_ctrl.Scale.Z {type=MovieSceneScriptingFloatChannel, default=1, const [1,263]}
       roll_blend_r_ctrl {type=MovieSceneScriptingFloatChannel, default=1, const [1,263]}
       tip_r_ctrl.Location.X {type=MovieSceneScriptingFloatChannel, default=0, const [0,263]}
       tip_r_ctrl.Location.Y {type=MovieSceneScriptingFloatChannel, default=0, const [0,263]}
       tip_r_ctrl.Location.Z {type=MovieSceneScriptingFloatChannel, default=0, const [0,263]}
       tip_r_ctrl.Rotation.X {type=MovieSceneScriptingFloatChannel, default=0, const [0,263]}
       tip_r_ctrl.Rotation.Y {type=MovieSceneScriptingFloatChannel, default=0, const [0,263]}
       tip_r_ctrl.Rotation.Z {type=MovieSceneScriptingFloatChannel, default=0, const [0,263]}
       tip_r_ctrl.Scale.X {type=MovieSceneScriptingFloatChannel, default=1, const [1,263]}
       tip_r_ctrl.Scale.Y {type=MovieSceneScriptingFloatChannel, default=1, const [1,263]}
       tip_r_ctrl.Scale.Z {type=MovieSceneScriptingFloatChannel, default=1, const [1,263]}
       heel_r_ctrl.Location.X {type=MovieSceneScriptingFloatChannel, default=0, thinned_from=263, keys [[1801,0],[2098,0]]}
       heel_r_ctrl.Location.Y {type=MovieSceneScriptingFloatChannel, default=0, thinned_from=263, keys [[1801,0],[2098,0]]}
       heel_r_ctrl.Location.Z {type=MovieSceneScriptingFloatChannel, default=0, thinned_from=263, keys [[1801,0],[2098,0]]}
       heel_r_ctrl.Rotation.X {type=MovieSceneScriptingFloatChannel, default=0, const [0,263]}
       heel_r_ctrl.Rotation.Y {type=MovieSceneScriptingFloatChannel, default=0, const [0,263]}
       heel_r_ctrl.Rotation.Z {type=MovieSceneScriptingFloatChannel, default=0, const [0,263]}
       heel_r_ctrl.Scale.X {type=MovieSceneScriptingFloatChannel, default=1, const [1,263]}
       heel_r_ctrl.Scale.Y {type=MovieSceneScriptingFloatChannel, default=1, const [1,263]}
       heel_r_ctrl.Scale.Z {type=MovieSceneScriptingFloatChannel, default=1, const [1,263]}
       leg_r_softness {type=MovieSceneScriptingFloatChannel, default=0, const [0,263]}
       leg_r_pv_ik_ctrl.Location.X {
        type=MovieSceneScriptingFloatChannel
        default=0
        thinned_from=264
        keys [[1801,0],[1959,0],[1960,0.0005],[1961,0],[2061,0],[2062,1604.3867],[2098,5.681],[2103,5.682]]
       }
       leg_r_pv_ik_ctrl.Location.Y {
        type=MovieSceneScriptingFloatChannel
        default=0
        thinned_from=264
        keys [[1801,0],[1837,0],[1838,0.0002],[1839,0],[1840,0.0002],[1841,0],[1845,0],[1846,-0.0002],[1847,0],[1849,0],[1850,-0.0002],[1851,-0.0002],[1852,0],[1853,0],[1854,-0.0002],[1855,0],[1856,0],[1857,0.0002],[1858,0],[1861,0],[1862,0.0002],[1863,0],[1866,0],[1867,0.0002],[1868,0],[1876,0],[1877,0.0001],[2061,0],[2062,1303.2799],[2098,10.648],[2103,10.6479]]
       }
       leg_r_pv_ik_ctrl.Location.Z {
        type=MovieSceneScriptingFloatChannel
        default=0.0002
        thinned_from=264
        keys [[1801,0.0002],[1802,-0.0005],[1803,0.0007],[1804,0],[1805,0],[1806,0.0002],[1807,-0.0005],[1808,-0.0007],[1809,0.0005],[1810,0.001],[1811,0.0007],[1812,0.0002],[1813,0.0002],[1814,-0.0007],[1815,0.0002],[1816,-0.0002],[1817,0.0005],[1818,-0.001],[1819,0.001],[1820,-0.0007],[1821,0.001],[1822,-0.001],[1823,0.0012],[1824,0.0002],[1825,-0.0005],[1826,-0.0002],[1827,-0.0002],[1828,-0.0005],[1830,0],[1831,0.0005],[1833,-0.0005],[1834,0.0002],[1835,0.0002],[1836,-0.0002],[1837,0.0002],[1838,0.0002],[1839,0],[1853,0],[1855,0.0001],[1857,-0.0001],[1859,0.0001],[1862,-0.0003],[1863,0],[1867,-0.0001],[1869,0.0001],[1871,-0.0001],[1872,0.0002],[1874,0.0001],[1875,-0.0002],[1876,0],[1877,-0.0002],[1878,-0.0002],[1879,0.0002],[1880,0],[1882,0],[1883,-0.0002],[1884,-0.0002],[1885,0.0005],[1886,-0.0002],[1887,-0.0002],[1888,0],[1889,0],[1890,0.0002],[1891,0],[1893,0],[1894,0.0005],[1895,0],[1898,0],[1899,0.0005],[1900,0],[1911,0],[1912,0.0005],[1913,0],[1928,0],[1929,0.0005],[1930,0],[1939,0],[1940,-0.0005],[1941,0],[1942,0],[1943,-0.0005],[1944,0],[1946,0],[1947,-0.0005],[1948,0],[1949,-0.0005],[1950,0],[1951,0],[1952,0.0005],[1953,0],[1959,0],[1960,-0.0005],[1961,-0.0005],[1962,0],[1963,0],[1964,-0.0005],[1966,0.0005],[1967,0],[1980,0],[1981,0.0005],[1982,0],[1983,0],[1984,-0.0005],[1985,0],[1989,0],[1990,0.0005],[1991,0],[1995,0],[1996,-0.0005],[1997,0],[1998,-0.0005],[1999,0],[2000,0],[2001,0.0005],[2002,0],[2004,0],[2005,0.0005],[2006,0],[2021,0],[2022,0.0005],[2023,0],[2039,0],[2040,0.0005],[2041,0],[2047,0],[2048,-0.0005],[2049,0],[2054,0],[2055,-0.0002],[2056,0.0002],[2058,0.0001],[2059,-0.0001],[2060,0],[2061,-0.0001],[2062,5304.5386],[2098,-90.3135],[2103,-90.313]]
       }
       leg_r_pv_ik_ctrl.Rotation.X {
        type=MovieSceneScriptingFloatChannel
        default=0
        thinned_from=264
        keys [[1801,0],[1808,0],[1824,0],[1825,0.0001],[1826,0],[1827,0.0001],[1829,-0.0001],[1830,0.0001],[1833,-0.0001],[1835,0.0001],[1836,-0.0001],[1838,0.0002],[1839,0],[1840,0.0001],[1842,-0.0002],[1844,0.0001],[1846,-0.0001],[1847,0],[1848,-0.0001],[1851,-0.0002],[1852,0.0001],[1854,-0.0002],[1856,0.0001],[1859,-0.0002],[1860,0.0001],[1871,0.0001],[1876,0],[1878,0.0001],[1880,-0.0001],[1887,0],[1888,-0.0001],[1902,-0.0001],[1907,0.0001],[1912,-0.0001],[1916,0.0001],[1917,-0.0002],[1919,0],[1920,-0.0001],[1921,0.0001],[1933,0.0001],[1936,-0.0001],[1937,0.0001],[1939,0],[1940,0.0001],[1944,-0.0001],[1946,0.0001],[1948,0],[1949,0.0003],[1952,-0.0001],[1957,0.0001],[1958,0],[1961,0.0003],[1962,-0.0001],[1964,0.0002],[1968,-0.0002],[1973,0.0002],[1975,-0.0002],[1977,0.0001],[1978,-0.0001],[1982,0.0001],[1984,0.0001],[1985,0.0002],[1986,0],[1987,0.0002],[1990,-0.0002],[1992,0.0001],[1993,-0.0001],[1996,0.0002],[1997,0.0001],[1998,0.0003],[2000,-0.0001],[2001,-0.0002],[2002,0.0002],[2003,0],[2005,-0.0001],[2009,0.0002],[2011,0],[2015,0.0002],[2016,-0.0001],[2017,-0.0002],[2018,0.0001],[2019,-0.0001],[2020,0.0002],[2021,-0.0002],[2022,-0.0003],[2023,0.0001],[2024,-0.0001],[2025,0.0001],[2026,0],[2027,0.0001],[2029,0],[2030,0.0001],[2033,-0.0002],[2034,0],[2036,-0.0001],[2038,0],[2040,-0.0002],[2041,0],[2042,-0.0001],[2049,0.0001],[2053,0],[2055,0.0002],[2056,-0.0001],[2058,0],[2059,-0.0002],[2060,0.0002],[2061,0.0002],[2062,162.7253],[2098,-16.6824],[2103,-16.6826]]
       }
       leg_r_pv_ik_ctrl.Rotation.Y {
        type=MovieSceneScriptingFloatChannel
        default=-0.0001
        thinned_from=264
        keys [[1801,-0.0001],[1808,0.0001],[1821,-0.0001],[1822,0.0001],[1834,0],[1835,0.0001],[1836,-0.0001],[1837,0.0002],[1838,0],[1839,0.0001],[1840,-0.0002],[1841,0.0001],[1845,0.0002],[1846,-0.0001],[1848,0.0001],[1849,-0.0001],[1853,0],[1854,0.0001],[1856,0],[1862,0.0001],[1863,-0.0001],[1865,0],[1866,-0.0001],[1867,0.0001],[1880,0],[1900,0],[1901,0.0002],[1903,-0.0001],[1908,0],[1909,0.0002],[1911,0.0001],[1913,0.0001],[1914,-0.0002],[1916,0],[1917,0.0002],[1922,-0.0001],[1923,0.0001],[1924,-0.0001],[1925,0.0001],[1926,-0.0001],[1927,0.0002],[1931,-0.0001],[1933,0],[1934,-0.0001],[1936,0.0001],[1937,-0.0001],[1939,0.0002],[1940,-0.0002],[1942,0.0002],[1943,-0.0001],[1944,0.0002],[1945,-0.0001],[1946,0.0001],[1948,-0.0001],[1951,0.0001],[1952,-0.0001],[1953,0],[1954,-0.0001],[1955,0.0001],[1961,-0.0001],[1962,0.0001],[1964,-0.0002],[1965,0],[1966,-0.0001],[1967,0.0001],[1968,0],[1969,0.0003],[1970,0.0003],[1971,-0.0002],[1972,0.0001],[1974,-0.0001],[1976,0.0002],[1977,-0.0002],[1979,0.0002],[1982,0],[1983,-0.0002],[1985,0.0001],[1986,-0.0001],[1993,-0.0001],[1995,0.0001],[1996,-0.0002],[1997,0.0001],[2000,-0.0002],[2002,-0.0001],[2003,-0.0002],[2005,-0.0001],[2006,0.0001],[2007,-0.0002],[2008,-0.0003],[2009,0],[2010,-0.0003],[2011,0],[2012,-0.0001],[2014,0],[2017,-0.0002],[2019,0.0001],[2022,0],[2023,-0.0002],[2026,0],[2028,0],[2029,-0.0002],[2032,0.0001],[2034,-0.0002],[2035,0],[2036,-0.0001],[2037,0.0001],[2038,-0.0002],[2041,0.0002],[2042,-0.0001],[2043,0.0001],[2050,-0.0001],[2058,0.0001],[2059,-0.0002],[2060,0],[2061,0],[2062,13.0023],[2098,-2.5237],[2103,-2.5222]]
       }
       leg_r_pv_ik_ctrl.Rotation.Z {type=MovieSceneScriptingFloatChannel, default=0, thinned_from=264, keys [[1801,0],[2061,0],[2062,-139.7382],[2098,-7.8978],[2103,-7.8977]]}
       leg_r_pv_ik_ctrl.Scale.X {type=MovieSceneScriptingFloatChannel, default=1, const [1,263]}
       leg_r_pv_ik_ctrl.Scale.Y {type=MovieSceneScriptingFloatChannel, default=1, const [1,263]}
       leg_r_pv_ik_ctrl.Scale.Z {type=MovieSceneScriptingFloatChannel, default=1, const [1,263]}
       upperarm_l_fk_ctrl.Location.X {type=MovieSceneScriptingFloatChannel, default=0, const [0,31]}
       upperarm_l_fk_ctrl.Location.Y {type=MovieSceneScriptingFloatChannel, default=0, const [0,31]}
       upperarm_l_fk_ctrl.Location.Z {type=MovieSceneScriptingFloatChannel, default=0, const [0,31]}
       upperarm_l_fk_ctrl.Rotation.X {
        type=MovieSceneScriptingFloatChannel
        default=20.8192
        thinned_from=36
        keys [[1801,20.8192],[1816,20.6964],[1825,14.8902],[1834,-3.4706],[1835,-3.0383],[1836,-2.0114],[1840,1.9056],[1844,11.9673],[1847,-25.9141],[1849,-30.5331],[1856,-24.203],[1862,-35.1363],[1865,-21.3276],[1868,-15.0293],[1874,1.68],[1882,-3.4706],[1884,14.2229],[1886,-65.0285],[1893,-3.4706],[1922,-3.4706],[1931,-29.3952],[1933,-30.5331],[1935,-26.9723],[1938,-0.0305],[1942,20.8192],[1984,20.8192],[1995,4.7766],[2007,-3.4706],[2016,-3.642],[2017,-8.2051],[2018,52.387],[2021,-23.0045],[2035,-23.0045],[2040,-3.4706],[2098,-3.4706]]
       }
       upperarm_l_fk_ctrl.Rotation.Y {
        type=MovieSceneScriptingFloatChannel
        default=-10.4918
        thinned_from=35
        keys [[1801,-10.4918],[1816,-10.436],[1825,60.5121],[1832,0.5339],[1835,39.0354],[1838,49.9578],[1840,12.651],[1844,8.4755],[1846,27.2565],[1849,52.318],[1851,68.745],[1856,57.8374],[1862,68.7987],[1865,36.1848],[1868,23.1668],[1874,-6.4068],[1882,0.5339],[1886,119.2895],[1893,0.5339],[1922,0.5339],[1931,50.1406],[1933,52.318],[1935,45.5043],[1938,12.6815],[1942,-10.4918],[1984,-10.4918],[1995,-3.1716],[2007,0.5339],[2016,0.5339],[2017,37.8876],[2018,38.3419],[2021,0.5339],[2098,0.5339]]
       }
       upperarm_l_fk_ctrl.Rotation.Z {
        type=MovieSceneScriptingFloatChannel
        default=-2.7668
        thinned_from=32
        keys [[1801,-2.7668],[1984,-2.7668],[1995,-53.1916],[2007,-2.7668],[2016,-2.9526],[2018,75.7425],[2021,-23.9461],[2035,-23.9461],[2040,-2.7668],[2098,-2.7668]]
       }
       upperarm_l_fk_ctrl.Scale.X {type=MovieSceneScriptingFloatChannel, default=1, const [1,30]}
       upperarm_l_fk_ctrl.Scale.Y {type=MovieSceneScriptingFloatChannel, default=1, const [1,30]}
       upperarm_l_fk_ctrl.Scale.Z {type=MovieSceneScriptingFloatChannel, default=1, const [1,30]}
       lowerarm_l_fk_ctrl.Location.X {type=MovieSceneScriptingFloatChannel, default=0, const [0,23]}
       lowerarm_l_fk_ctrl.Location.Y {type=MovieSceneScriptingFloatChannel, default=0, const [0,23]}
       lowerarm_l_fk_ctrl.Location.Z {type=MovieSceneScriptingFloatChannel, default=0, const [0,23]}
       lowerarm_l_fk_ctrl.Rotation.X {type=MovieSceneScriptingFloatChannel, default=0, const [0,23]}
       lowerarm_l_fk_ctrl.Rotation.Y {type=MovieSceneScriptingFloatChannel, default=0, const [0,23]}
       lowerarm_l_fk_ctrl.Rotation.Z {
        type=MovieSceneScriptingFloatChannel
        default=-6.3418
        thinned_from=25
        keys [[1801,-6.3418],[1816,-6.08],[1825,42.2469],[1830,45.4376],[1849,27.1022],[1865,12.9355],[1883,45.4376],[1922,45.4376],[1931,27.8732],[1933,27.1022],[1935,29.5148],[1938,45.4376],[1942,-6.3418],[1984,-6.3418],[1995,28.0357],[1998,10.4432],[2007,45.4376],[2098,45.4376]]
       }
       lowerarm_l_fk_ctrl.Scale.X {type=MovieSceneScriptingFloatChannel, default=1, const [1,22]}
       lowerarm_l_fk_ctrl.Scale.Y {type=MovieSceneScriptingFloatChannel, default=1, const [1,22]}
       lowerarm_l_fk_ctrl.Scale.Z {type=MovieSceneScriptingFloatChannel, default=1, const [1,22]}
       hand_l_fk_ctrl.Location.X {type=MovieSceneScriptingFloatChannel, default=0, const [0,22]}
       hand_l_fk_ctrl.Location.Y {type=MovieSceneScriptingFloatChannel, default=0, const [0,22]}
       hand_l_fk_ctrl.Location.Z {type=MovieSceneScriptingFloatChannel, default=0, const [0,22]}
       hand_l_fk_ctrl.Rotation.X {type=MovieSceneScriptingFloatChannel, default=-58.3895, const [-58.3895,22]}
       hand_l_fk_ctrl.Rotation.Y {type=MovieSceneScriptingFloatChannel, default=11.0965, const [11.0965,22]}
       hand_l_fk_ctrl.Rotation.Z {type=MovieSceneScriptingFloatChannel, default=-8.5917, const [-8.5917,22]}
       hand_l_fk_ctrl.Scale.X {type=MovieSceneScriptingFloatChannel, default=1, const [1,21]}
       hand_l_fk_ctrl.Scale.Y {type=MovieSceneScriptingFloatChannel, default=1, const [1,21]}
       hand_l_fk_ctrl.Scale.Z {type=MovieSceneScriptingFloatChannel, default=1, const [1,21]}
       thumb_01_l_ctrl.Location.X {type=MovieSceneScriptingFloatChannel, default=0, const [0,15]}
       thumb_01_l_ctrl.Location.Y {type=MovieSceneScriptingFloatChannel, default=0, const [0,15]}
       thumb_01_l_ctrl.Location.Z {type=MovieSceneScriptingFloatChannel, default=0, const [0,15]}
       thumb_01_l_ctrl.Rotation.X {type=MovieSceneScriptingFloatChannel, default=-1.5583, const [-1.5583,15]}
       thumb_01_l_ctrl.Rotation.Y {type=MovieSceneScriptingFloatChannel, default=0.8707, const [0.8707,15]}
       thumb_01_l_ctrl.Rotation.Z {type=MovieSceneScriptingFloatChannel, default=-13.9227, const [-13.9227,15]}
       thumb_01_l_ctrl.Scale.X {type=MovieSceneScriptingFloatChannel, default=1, const [1,15]}
       thumb_01_l_ctrl.Scale.Y {type=MovieSceneScriptingFloatChannel, default=1, const [1,15]}
       thumb_01_l_ctrl.Scale.Z {type=MovieSceneScriptingFloatChannel, default=1, const [1,15]}
       thumb_02_l_ctrl.Location.X {type=MovieSceneScriptingFloatChannel, default=0, const [0,15]}
       thumb_02_l_ctrl.Location.Y {type=MovieSceneScriptingFloatChannel, default=0, const [0,15]}
       thumb_02_l_ctrl.Location.Z {type=MovieSceneScriptingFloatChannel, default=0, const [0,15]}
       thumb_02_l_ctrl.Rotation.X {type=MovieSceneScriptingFloatChannel, default=0, const [0,15]}
       thumb_02_l_ctrl.Rotation.Y {type=MovieSceneScriptingFloatChannel, default=0, const [0,15]}
       thumb_02_l_ctrl.Rotation.Z {type=MovieSceneScriptingFloatChannel, default=0, const [0,15]}
       thumb_02_l_ctrl.Scale.X {type=MovieSceneScriptingFloatChannel, default=1, const [1,15]}
       thumb_02_l_ctrl.Scale.Y {type=MovieSceneScriptingFloatChannel, default=1, const [1,15]}
       thumb_02_l_ctrl.Scale.Z {type=MovieSceneScriptingFloatChannel, default=1, const [1,15]}
       thumb_03_l_ctrl.Location.X {type=MovieSceneScriptingFloatChannel, default=0, const [0,15]}
       thumb_03_l_ctrl.Location.Y {type=MovieSceneScriptingFloatChannel, default=0, const [0,15]}
       thumb_03_l_ctrl.Location.Z {type=MovieSceneScriptingFloatChannel, default=0, const [0,15]}
       thumb_03_l_ctrl.Rotation.X {type=MovieSceneScriptingFloatChannel, default=8.1645, const [8.1645,15]}
       thumb_03_l_ctrl.Rotation.Y {type=MovieSceneScriptingFloatChannel, default=-9.9357, const [-9.9357,15]}
       thumb_03_l_ctrl.Rotation.Z {type=MovieSceneScriptingFloatChannel, default=-50.7089, const [-50.7089,15]}
       thumb_03_l_ctrl.Scale.X {type=MovieSceneScriptingFloatChannel, default=1, const [1,15]}
       thumb_03_l_ctrl.Scale.Y {type=MovieSceneScriptingFloatChannel, default=1, const [1,15]}
       thumb_03_l_ctrl.Scale.Z {type=MovieSceneScriptingFloatChannel, default=1, const [1,15]}
       index_metacarpal_l_ctrl.Location.X {type=MovieSceneScriptingFloatChannel, default=0, const [0,15]}
       index_metacarpal_l_ctrl.Location.Y {type=MovieSceneScriptingFloatChannel, default=0, const [0,15]}
       index_metacarpal_l_ctrl.Location.Z {type=MovieSceneScriptingFloatChannel, default=0, const [0,15]}
       index_metacarpal_l_ctrl.Rotation.X {type=MovieSceneScriptingFloatChannel, default=0, const [0,15]}
       index_metacarpal_l_ctrl.Rotation.Y {type=MovieSceneScriptingFloatChannel, default=3.7815, const [3.7815,15]}
       index_metacarpal_l_ctrl.Rotation.Z {type=MovieSceneScriptingFloatChannel, default=-3.094, const [-3.094,15]}
       index_metacarpal_l_ctrl.Scale.X {type=MovieSceneScriptingFloatChannel, default=1, const [1,15]}
       index_metacarpal_l_ctrl.Scale.Y {type=MovieSceneScriptingFloatChannel, default=1, const [1,15]}
       index_metacarpal_l_ctrl.Scale.Z {type=MovieSceneScriptingFloatChannel, default=1, const [1,15]}
       index_01_l_ctrl.Location.X {type=MovieSceneScriptingFloatChannel, default=0, const [0,15]}
       index_01_l_ctrl.Location.Y {type=MovieSceneScriptingFloatChannel, default=0, const [0,15]}
       index_01_l_ctrl.Location.Z {type=MovieSceneScriptingFloatChannel, default=0, const [0,15]}
       index_01_l_ctrl.Rotation.X {type=MovieSceneScriptingFloatChannel, default=6.2821, const [6.2821,15]}
       index_01_l_ctrl.Rotation.Y {type=MovieSceneScriptingFloatChannel, default=-2.5276, const [-2.5276,15]}
       index_01_l_ctrl.Rotation.Z {type=MovieSceneScriptingFloatChannel, default=-65.1155, const [-65.1155,15]}
       index_01_l_ctrl.Scale.X {type=MovieSceneScriptingFloatChannel, default=1, const [1,15]}
       index_01_l_ctrl.Scale.Y {type=MovieSceneScriptingFloatChannel, default=1, const [1,15]}
       index_01_l_ctrl.Scale.Z {type=MovieSceneScriptingFloatChannel, default=1, const [1,15]}
       index_02_l_ctrl.Location.X {type=MovieSceneScriptingFloatChannel, default=0, const [0,15]}
       index_02_l_ctrl.Location.Y {type=MovieSceneScriptingFloatChannel, default=0, const [0,15]}
       index_02_l_ctrl.Location.Z {type=MovieSceneScriptingFloatChannel, default=0, const [0,15]}
       index_02_l_ctrl.Rotation.X {type=MovieSceneScriptingFloatChannel, default=0, const [0,15]}
       index_02_l_ctrl.Rotation.Y {type=MovieSceneScriptingFloatChannel, default=0, const [0,15]}
       index_02_l_ctrl.Rotation.Z {type=MovieSceneScriptingFloatChannel, default=-60.8298, const [-60.8298,15]}
       index_02_l_ctrl.Scale.X {type=MovieSceneScriptingFloatChannel, default=1, const [1,15]}
       index_02_l_ctrl.Scale.Y {type=MovieSceneScriptingFloatChannel, default=1, const [1,15]}
       index_02_l_ctrl.Scale.Z {type=MovieSceneScriptingFloatChannel, default=1, const [1,15]}
       index_03_l_ctrl.Location.X {type=MovieSceneScriptingFloatChannel, default=0, const [0,15]}
       index_03_l_ctrl.Location.Y {type=MovieSceneScriptingFloatChannel, default=0, const [0,15]}
       index_03_l_ctrl.Location.Z {type=MovieSceneScriptingFloatChannel, default=0, const [0,15]}
       index_03_l_ctrl.Rotation.X {type=MovieSceneScriptingFloatChannel, default=0.0236, const [0.0236,15]}
       index_03_l_ctrl.Rotation.Y {type=MovieSceneScriptingFloatChannel, default=-0.109, const [-0.109,15]}
       index_03_l_ctrl.Rotation.Z {type=MovieSceneScriptingFloatChannel, default=-78.806, const [-78.806,15]}
       index_03_l_ctrl.Scale.X {type=MovieSceneScriptingFloatChannel, default=1, const [1,15]}
       index_03_l_ctrl.Scale.Y {type=MovieSceneScriptingFloatChannel, default=1, const [1,15]}
       index_03_l_ctrl.Scale.Z {type=MovieSceneScriptingFloatChannel, default=1, const [1,15]}
       middle_metacarpal_l_ctrl.Location.X {type=MovieSceneScriptingFloatChannel, default=0, const [0,15]}
       middle_metacarpal_l_ctrl.Location.Y {type=MovieSceneScriptingFloatChannel, default=0, const [0,15]}
       middle_metacarpal_l_ctrl.Location.Z {type=MovieSceneScriptingFloatChannel, default=0, const [0,15]}
       middle_metacarpal_l_ctrl.Rotation.X {type=MovieSceneScriptingFloatChannel, default=-7.29, const [-7.29,15]}
       middle_metacarpal_l_ctrl.Rotation.Y {type=MovieSceneScriptingFloatChannel, default=1.3582, const [1.3582,15]}
       middle_metacarpal_l_ctrl.Rotation.Z {type=MovieSceneScriptingFloatChannel, default=-2.7048, const [-2.7048,15]}
       middle_metacarpal_l_ctrl.Scale.X {type=MovieSceneScriptingFloatChannel, default=1, const [1,15]}
       middle_metacarpal_l_ctrl.Scale.Y {type=MovieSceneScriptingFloatChannel, default=1, const [1,15]}
       middle_metacarpal_l_ctrl.Scale.Z {type=MovieSceneScriptingFloatChannel, default=1, const [1,15]}
       middle_01_l_ctrl.Location.X {type=MovieSceneScriptingFloatChannel, default=0, const [0,15]}
       middle_01_l_ctrl.Location.Y {type=MovieSceneScriptingFloatChannel, default=0, const [0,15]}
       middle_01_l_ctrl.Location.Z {type=MovieSceneScriptingFloatChannel, default=0, const [0,15]}
       middle_01_l_ctrl.Rotation.X {type=MovieSceneScriptingFloatChannel, default=0, const [0,15]}
       middle_01_l_ctrl.Rotation.Y {type=MovieSceneScriptingFloatChannel, default=1.3494, const [1.3494,15]}
       middle_01_l_ctrl.Rotation.Z {type=MovieSceneScriptingFloatChannel, default=-74.7137, const [-74.7137,15]}
       middle_01_l_ctrl.Scale.X {type=MovieSceneScriptingFloatChannel, default=1, const [1,15]}
       middle_01_l_ctrl.Scale.Y {type=MovieSceneScriptingFloatChannel, default=1, const [1,15]}
       middle_01_l_ctrl.Scale.Z {type=MovieSceneScriptingFloatChannel, default=1, const [1,15]}
       middle_02_l_ctrl.Location.X {type=MovieSceneScriptingFloatChannel, default=-0.0452, const [-0.0452,15]}
       middle_02_l_ctrl.Location.Y {type=MovieSceneScriptingFloatChannel, default=0.0507, const [0.0507,15]}
       middle_02_l_ctrl.Location.Z {type=MovieSceneScriptingFloatChannel, default=0.1146, const [0.1146,15]}
       middle_02_l_ctrl.Rotation.X {type=MovieSceneScriptingFloatChannel, default=-3.3929, const [-3.3929,15]}
       middle_02_l_ctrl.Rotation.Y {type=MovieSceneScriptingFloatChannel, default=-1.8646, const [-1.8646,15]}
       middle_02_l_ctrl.Rotation.Z {type=MovieSceneScriptingFloatChannel, default=-58.0319, const [-58.0319,15]}
       middle_02_l_ctrl.Scale.X {type=MovieSceneScriptingFloatChannel, default=1, const [1,15]}
       middle_02_l_ctrl.Scale.Y {type=MovieSceneScriptingFloatChannel, default=1, const [1,15]}
       middle_02_l_ctrl.Scale.Z {type=MovieSceneScriptingFloatChannel, default=1, const [1,15]}
       middle_03_l_ctrl.Location.X {type=MovieSceneScriptingFloatChannel, default=0.0206, const [0.0206,15]}
       middle_03_l_ctrl.Location.Y {type=MovieSceneScriptingFloatChannel, default=-0.0583, const [-0.0583,15]}
       middle_03_l_ctrl.Location.Z {type=MovieSceneScriptingFloatChannel, default=0.1324, const [0.1324,15]}
       middle_03_l_ctrl.Rotation.X {type=MovieSceneScriptingFloatChannel, default=-1.8232, const [-1.8232,15]}
       middle_03_l_ctrl.Rotation.Y {type=MovieSceneScriptingFloatChannel, default=1.4852, const [1.4852,15]}
       middle_03_l_ctrl.Rotation.Z {type=MovieSceneScriptingFloatChannel, default=-64.6207, const [-64.6207,15]}
       middle_03_l_ctrl.Scale.X {type=MovieSceneScriptingFloatChannel, default=1, const [1,15]}
       middle_03_l_ctrl.Scale.Y {type=MovieSceneScriptingFloatChannel, default=1, const [1,15]}
       middle_03_l_ctrl.Scale.Z {type=MovieSceneScriptingFloatChannel, default=1, const [1,15]}
       ring_metacarpal_l_ctrl.Location.X {type=MovieSceneScriptingFloatChannel, default=0, const [0,15]}
       ring_metacarpal_l_ctrl.Location.Y {type=MovieSceneScriptingFloatChannel, default=0, const [0,15]}
       ring_metacarpal_l_ctrl.Location.Z {type=MovieSceneScriptingFloatChannel, default=0, const [0,15]}
       ring_metacarpal_l_ctrl.Rotation.X {type=MovieSceneScriptingFloatChannel, default=-2.4847, const [-2.4847,15]}
       ring_metacarpal_l_ctrl.Rotation.Y {type=MovieSceneScriptingFloatChannel, default=-0.8637, const [-0.8637,15]}
       ring_metacarpal_l_ctrl.Rotation.Z {type=MovieSceneScriptingFloatChannel, default=-5.6698, const [-5.6698,15]}
       ring_metacarpal_l_ctrl.Scale.X {type=MovieSceneScriptingFloatChannel, default=1, const [1,15]}
       ring_metacarpal_l_ctrl.Scale.Y {type=MovieSceneScriptingFloatChannel, default=1, const [1,15]}
       ring_metacarpal_l_ctrl.Scale.Z {type=MovieSceneScriptingFloatChannel, default=1, const [1,15]}
       ring_01_l_ctrl.Location.X {type=MovieSceneScriptingFloatChannel, default=0, const [0,15]}
       ring_01_l_ctrl.Location.Y {type=MovieSceneScriptingFloatChannel, default=0, const [0,15]}
       ring_01_l_ctrl.Location.Z {type=MovieSceneScriptingFloatChannel, default=0, const [0,15]}
       ring_01_l_ctrl.Rotation.X {type=MovieSceneScriptingFloatChannel, default=-4.509, const [-4.509,15]}
       ring_01_l_ctrl.Rotation.Y {type=MovieSceneScriptingFloatChannel, default=-1.201, const [-1.201,15]}
       ring_01_l_ctrl.Rotation.Z {type=MovieSceneScriptingFloatChannel, default=-79.0231, const [-79.0231,15]}
       ring_01_l_ctrl.Scale.X {type=MovieSceneScriptingFloatChannel, default=1, const [1,15]}
       ring_01_l_ctrl.Scale.Y {type=MovieSceneScriptingFloatChannel, default=1, const [1,15]}
       ring_01_l_ctrl.Scale.Z {type=MovieSceneScriptingFloatChannel, default=1, const [1,15]}
       ring_02_l_ctrl.Location.X {type=MovieSceneScriptingFloatChannel, default=0, const [0,15]}
       ring_02_l_ctrl.Location.Y {type=MovieSceneScriptingFloatChannel, default=0, const [0,15]}
       ring_02_l_ctrl.Location.Z {type=MovieSceneScriptingFloatChannel, default=0, const [0,15]}
       ring_02_l_ctrl.Rotation.X {type=MovieSceneScriptingFloatChannel, default=0, const [0,15]}
       ring_02_l_ctrl.Rotation.Y {type=MovieSceneScriptingFloatChannel, default=0, const [0,15]}
       ring_02_l_ctrl.Rotation.Z {type=MovieSceneScriptingFloatChannel, default=-66.6687, const [-66.6687,15]}
       ring_02_l_ctrl.Scale.X {type=MovieSceneScriptingFloatChannel, default=1, const [1,15]}
       ring_02_l_ctrl.Scale.Y {type=MovieSceneScriptingFloatChannel, default=1, const [1,15]}
       ring_02_l_ctrl.Scale.Z {type=MovieSceneScriptingFloatChannel, default=1, const [1,15]}
       ring_03_l_ctrl.Location.X {type=MovieSceneScriptingFloatChannel, default=0, const [0,15]}
       ring_03_l_ctrl.Location.Y {type=MovieSceneScriptingFloatChannel, default=0, const [0,15]}
       ring_03_l_ctrl.Location.Z {type=MovieSceneScriptingFloatChannel, default=0, const [0,15]}
       ring_03_l_ctrl.Rotation.X {type=MovieSceneScriptingFloatChannel, default=0, const [0,15]}
       ring_03_l_ctrl.Rotation.Y {type=MovieSceneScriptingFloatChannel, default=0, const [0,15]}
       ring_03_l_ctrl.Rotation.Z {type=MovieSceneScriptingFloatChannel, default=-73.2004, const [-73.2004,15]}
       ring_03_l_ctrl.Scale.X {type=MovieSceneScriptingFloatChannel, default=1, const [1,15]}
       ring_03_l_ctrl.Scale.Y {type=MovieSceneScriptingFloatChannel, default=1, const [1,15]}
       ring_03_l_ctrl.Scale.Z {type=MovieSceneScriptingFloatChannel, default=1, const [1,15]}
       pinky_metacarpal_l_ctrl.Location.X {type=MovieSceneScriptingFloatChannel, default=0, const [0,15]}
       pinky_metacarpal_l_ctrl.Location.Y {type=MovieSceneScriptingFloatChannel, default=0, const [0,15]}
       pinky_metacarpal_l_ctrl.Location.Z {type=MovieSceneScriptingFloatChannel, default=0, const [0,15]}
       pinky_metacarpal_l_ctrl.Rotation.X {type=MovieSceneScriptingFloatChannel, default=-5.9588, const [-5.9588,15]}
       pinky_metacarpal_l_ctrl.Rotation.Y {type=MovieSceneScriptingFloatChannel, default=-5.2205, const [-5.2205,15]}
       pinky_metacarpal_l_ctrl.Rotation.Z {type=MovieSceneScriptingFloatChannel, default=-11.0008, const [-11.0008,15]}
       pinky_metacarpal_l_ctrl.Scale.X {type=MovieSceneScriptingFloatChannel, default=1, const [1,15]}
       pinky_metacarpal_l_ctrl.Scale.Y {type=MovieSceneScriptingFloatChannel, default=1, const [1,15]}
       pinky_metacarpal_l_ctrl.Scale.Z {type=MovieSceneScriptingFloatChannel, default=1, const [1,15]}
       pinky_01_l_ctrl.Location.X {type=MovieSceneScriptingFloatChannel, default=0, const [0,15]}
       pinky_01_l_ctrl.Location.Y {type=MovieSceneScriptingFloatChannel, default=0, const [0,15]}
       pinky_01_l_ctrl.Location.Z {type=MovieSceneScriptingFloatChannel, default=0, const [0,15]}
       pinky_01_l_ctrl.Rotation.X {type=MovieSceneScriptingFloatChannel, default=-11.7646, const [-11.7646,15]}
       pinky_01_l_ctrl.Rotation.Y {type=MovieSceneScriptingFloatChannel, default=-1.0914, const [-1.0914,15]}
       pinky_01_l_ctrl.Rotation.Z {type=MovieSceneScriptingFloatChannel, default=-84.6646, const [-84.6646,15]}
       pinky_01_l_ctrl.Scale.X {type=MovieSceneScriptingFloatChannel, default=1, const [1,15]}
       pinky_01_l_ctrl.Scale.Y {type=MovieSceneScriptingFloatChannel, default=1, const [1,15]}
       pinky_01_l_ctrl.Scale.Z {type=MovieSceneScriptingFloatChannel, default=1, const [1,15]}
       pinky_02_l_ctrl.Location.X {type=MovieSceneScriptingFloatChannel, default=0, const [0,15]}
       pinky_02_l_ctrl.Location.Y {type=MovieSceneScriptingFloatChannel, default=0, const [0,15]}
       pinky_02_l_ctrl.Location.Z {type=MovieSceneScriptingFloatChannel, default=0, const [0,15]}
       pinky_02_l_ctrl.Rotation.X {type=MovieSceneScriptingFloatChannel, default=0, const [0,15]}
       pinky_02_l_ctrl.Rotation.Y {type=MovieSceneScriptingFloatChannel, default=0, const [0,15]}
       pinky_02_l_ctrl.Rotation.Z {type=MovieSceneScriptingFloatChannel, default=-63.0936, const [-63.0936,15]}
       pinky_02_l_ctrl.Scale.X {type=MovieSceneScriptingFloatChannel, default=1, const [1,15]}
       pinky_02_l_ctrl.Scale.Y {type=MovieSceneScriptingFloatChannel, default=1, const [1,15]}
       pinky_02_l_ctrl.Scale.Z {type=MovieSceneScriptingFloatChannel, default=1, const [1,15]}
       pinky_03_l_ctrl.Location.X {type=MovieSceneScriptingFloatChannel, default=0, const [0,15]}
       pinky_03_l_ctrl.Location.Y {type=MovieSceneScriptingFloatChannel, default=0, const [0,15]}
       pinky_03_l_ctrl.Location.Z {type=MovieSceneScriptingFloatChannel, default=0, const [0,15]}
       pinky_03_l_ctrl.Rotation.X {type=MovieSceneScriptingFloatChannel, default=0, const [0,15]}
       pinky_03_l_ctrl.Rotation.Y {type=MovieSceneScriptingFloatChannel, default=0, const [0,15]}
       pinky_03_l_ctrl.Rotation.Z {type=MovieSceneScriptingFloatChannel, default=-117.2954, const [-117.2954,15]}
       pinky_03_l_ctrl.Scale.X {type=MovieSceneScriptingFloatChannel, default=1, const [1,15]}
       pinky_03_l_ctrl.Scale.Y {type=MovieSceneScriptingFloatChannel, default=1, const [1,15]}
       pinky_03_l_ctrl.Scale.Z {type=MovieSceneScriptingFloatChannel, default=1, const [1,15]}
       hand_l_ik_ctrl.Location.X {type=MovieSceneScriptingFloatChannel, default=0, thinned_from=264, keys [[1801,0],[2061,0],[2062,-4341.749],[2098,79.1583],[2103,79.1583]]}
       hand_l_ik_ctrl.Location.Y {type=MovieSceneScriptingFloatChannel, default=0, thinned_from=264, keys [[1801,0],[2061,0],[2062,-212.4133],[2098,33.715],[2103,33.715]]}
       hand_l_ik_ctrl.Location.Z {type=MovieSceneScriptingFloatChannel, default=0, thinned_from=264, keys [[1801,0],[2061,0],[2062,3605.5813],[2098,-24.0609],[2103,-24.0608]]}
       hand_l_ik_ctrl.Rotation.X {type=MovieSceneScriptingFloatChannel, default=0, thinned_from=264, keys [[1801,0],[2061,0],[2062,96.7123],[2098,-72.4496],[2103,-72.4496]]}
       hand_l_ik_ctrl.Rotation.Y {type=MovieSceneScriptingFloatChannel, default=0, thinned_from=264, keys [[1801,0],[2061,0],[2062,54.836],[2098,52.3244],[2103,52.3244]]}
       hand_l_ik_ctrl.Rotation.Z {type=MovieSceneScriptingFloatChannel, default=0, thinned_from=264, keys [[1801,0],[2061,0],[2062,-34.9447],[2098,-19.8578],[2103,-19.8578]]}
       hand_l_ik_ctrl.Scale.X {type=MovieSceneScriptingFloatChannel, default=1, const [1,263]}
       hand_l_ik_ctrl.Scale.Y {type=MovieSceneScriptingFloatChannel, default=1, const [1,263]}
       hand_l_ik_ctrl.Scale.Z {type=MovieSceneScriptingFloatChannel, default=1, const [1,263]}
       Softness {type=MovieSceneScriptingFloatChannel, default=0, const [0,263]}
       hand_l_ikorient_ctrl.Rotation.Roll {type=MovieSceneScriptingFloatChannel, default=0, const [0,263]}
       hand_l_ikorient_ctrl.Rotation.Pitch {type=MovieSceneScriptingFloatChannel, default=0, const [0,263]}
       hand_l_ikorient_ctrl.Rotation.Yaw {type=MovieSceneScriptingFloatChannel, default=0, const [0,263]}
       arm_l_pv_ik_ctrl.Location.X {type=MovieSceneScriptingFloatChannel, default=0, thinned_from=264, keys [[1801,0],[2061,0],[2062,-1726.6678],[2098,-2.1938],[2103,-2.1938]]}
       arm_l_pv_ik_ctrl.Location.Y {type=MovieSceneScriptingFloatChannel, default=0, thinned_from=264, keys [[1801,0],[2061,0],[2062,1323.1769],[2098,1.2592],[2103,1.2593]]}
       arm_l_pv_ik_ctrl.Location.Z {type=MovieSceneScriptingFloatChannel, default=0, thinned_from=264, keys [[1801,0],[2061,0],[2062,5211.0234],[2098,-99.4732],[2103,-99.4731]]}
       arm_l_pv_ik_ctrl.Rotation.X {type=MovieSceneScriptingFloatChannel, default=0, thinned_from=264, keys [[1801,0],[2061,0],[2062,179.1403],[2098,1.2845],[2103,1.2845]]}
       arm_l_pv_ik_ctrl.Rotation.Y {type=MovieSceneScriptingFloatChannel, default=0, thinned_from=264, keys [[1801,0],[2061,0],[2062,-2.6505],[2098,0.8359],[2103,0.8359]]}
       arm_l_pv_ik_ctrl.Rotation.Z {type=MovieSceneScriptingFloatChannel, default=0, thinned_from=264, keys [[1801,0],[2061,0],[2062,151.654],[2098,-4.1794],[2103,-4.1794]]}
       arm_l_pv_ik_ctrl.Scale.X {type=MovieSceneScriptingFloatChannel, default=1, const [1,263]}
       arm_l_pv_ik_ctrl.Scale.Y {type=MovieSceneScriptingFloatChannel, default=1, const [1,263]}
       arm_l_pv_ik_ctrl.Scale.Z {type=MovieSceneScriptingFloatChannel, default=1, const [1,263]}
       upperarm_r_fk_ctrl.Location.X {type=MovieSceneScriptingFloatChannel, default=0, const [0,34]}
       upperarm_r_fk_ctrl.Location.Y {type=MovieSceneScriptingFloatChannel, default=0, const [0,34]}
       upperarm_r_fk_ctrl.Location.Z {type=MovieSceneScriptingFloatChannel, default=0, const [0,34]}
       upperarm_r_fk_ctrl.Rotation.X {
        type=MovieSceneScriptingFloatChannel
        default=18.1075
        thinned_from=37
        keys [[1801,18.1075],[1816,18.1075],[1825,1.4492],[1834,18.1075],[1835,20.2052],[1836,22.6099],[1840,27.7919],[1844,39.9658],[1847,5.8126],[1849,0.8428],[1862,-15.4402],[1865,0.9364],[1868,7.1539],[1874,21.5181],[1882,18.0273],[1883,18.1075],[1922,18.1075],[1931,1.5688],[1933,0.8428],[1935,3.1145],[1938,29.0472],[1942,18.1075],[2007,18.1075],[2017,-38.962],[2018,-45.6529],[2019,25.917],[2020,25.917],[2021,16.3656],[2035,16.3656],[2040,17.97],[2042,18.1075],[2098,18.1075]]
       }
       upperarm_r_fk_ctrl.Rotation.Y {
        type=MovieSceneScriptingFloatChannel
        default=-12.7105
        thinned_from=36
        keys [[1801,-12.7105],[1816,-12.607],[1825,28.3251],[1832,7.7548],[1835,45.3629],[1838,58.1486],[1840,17.3764],[1844,7.0367],[1846,25.2769],[1849,35.9945],[1862,32.181],[1865,18.7695],[1868,14.504],[1874,5.6974],[1882,7.8027],[1883,7.7548],[1922,7.7548],[1931,34.8071],[1933,35.9945],[1935,32.2788],[1938,45.5439],[1942,-12.7105],[1984,-12.7105],[1995,-3.5815],[2007,7.7548],[2017,34.3133],[2020,34.2469],[2021,9.0348],[2035,9.0348],[2040,7.8558],[2042,7.7548],[2098,7.7548]]
       }
       upperarm_r_fk_ctrl.Rotation.Z {
        type=MovieSceneScriptingFloatChannel
        default=-14.6012
        thinned_from=37
        keys [[1801,-14.6012],[1816,-14.6732],[1825,24.4431],[1834,-28.8297],[1883,-28.8297],[1885,-54.6818],[1889,-28.8297],[1938,-28.8297],[1942,-14.6012],[1984,-14.6012],[1995,-71.373],[2007,-28.8297],[2017,-88.6391],[2018,-58.9322],[2019,138.371],[2020,138.371],[2021,156.9266],[2023,179.8303],[2035,156.9266],[2040,-14.1647],[2042,-28.8297],[2098,-28.8297]]
       }
       upperarm_r_fk_ctrl.Scale.X {type=MovieSceneScriptingFloatChannel, default=1, const [1,34]}
       upperarm_r_fk_ctrl.Scale.Y {type=MovieSceneScriptingFloatChannel, default=1, const [1,34]}
       upperarm_r_fk_ctrl.Scale.Z {type=MovieSceneScriptingFloatChannel, default=1, const [1,34]}
       lowerarm_r_fk_ctrl.Location.X {type=MovieSceneScriptingFloatChannel, default=0, const [0,27]}
       lowerarm_r_fk_ctrl.Location.Y {type=MovieSceneScriptingFloatChannel, default=0, const [0,27]}
       lowerarm_r_fk_ctrl.Location.Z {type=MovieSceneScriptingFloatChannel, default=0, const [0,27]}
       lowerarm_r_fk_ctrl.Rotation.X {type=MovieSceneScriptingFloatChannel, default=0, const [0,27]}
       lowerarm_r_fk_ctrl.Rotation.Y {type=MovieSceneScriptingFloatChannel, default=0, const [0,27]}
       lowerarm_r_fk_ctrl.Rotation.Z {
        type=MovieSceneScriptingFloatChannel
        default=-7.3516
        thinned_from=31
        keys [[1801,-7.3516],[1816,-7.1871],[1825,23.1788],[1830,25.1837],[1938,25.1837],[1942,-7.3516],[1984,-7.3516],[1995,7.1615],[1998,-14.0715],[2007,25.1837],[2012,68.2397],[2016,25.1837],[2018,-13.9896],[2019,-6.6244],[2021,-6.6333],[2035,19.7332],[2038,73.1231],[2040,59.1064],[2043,30.976],[2045,25.1837],[2098,25.1837]]
       }
       lowerarm_r_fk_ctrl.Scale.X {type=MovieSceneScriptingFloatChannel, default=1, const [1,26]}
       lowerarm_r_fk_ctrl.Scale.Y {type=MovieSceneScriptingFloatChannel, default=1, const [1,26]}
       lowerarm_r_fk_ctrl.Scale.Z {type=MovieSceneScriptingFloatChannel, default=1, const [1,26]}
       hand_r_fk_ctrl.Location.X {type=MovieSceneScriptingFloatChannel, default=0, const [0,22]}
       hand_r_fk_ctrl.Location.Y {type=MovieSceneScriptingFloatChannel, default=0, const [0,22]}
       hand_r_fk_ctrl.Location.Z {type=MovieSceneScriptingFloatChannel, default=0, const [0,22]}
       hand_r_fk_ctrl.Rotation.X {type=MovieSceneScriptingFloatChannel, default=-9.9207, const [-9.9207,22]}
       hand_r_fk_ctrl.Rotation.Y {type=MovieSceneScriptingFloatChannel, default=11.8409, const [11.8409,22]}
       hand_r_fk_ctrl.Rotation.Z {type=MovieSceneScriptingFloatChannel, default=-9.9831, const [-9.9831,22]}
       hand_r_fk_ctrl.Scale.X {type=MovieSceneScriptingFloatChannel, default=1, const [1,21]}
       hand_r_fk_ctrl.Scale.Y {type=MovieSceneScriptingFloatChannel, default=1, const [1,21]}
       hand_r_fk_ctrl.Scale.Z {type=MovieSceneScriptingFloatChannel, default=1, const [1,21]}
       thumb_01_r_ctrl.Location.X {type=MovieSceneScriptingFloatChannel, default=-0.0001, const [-0.0001,15]}
       thumb_01_r_ctrl.Location.Y {type=MovieSceneScriptingFloatChannel, default=0.0001, const [0.0001,15]}
       thumb_01_r_ctrl.Location.Z {type=MovieSceneScriptingFloatChannel, default=0.0001, const [0.0001,15]}
       thumb_01_r_ctrl.Rotation.X {type=MovieSceneScriptingFloatChannel, default=-1.5583, const [-1.5583,15]}
       thumb_01_r_ctrl.Rotation.Y {type=MovieSceneScriptingFloatChannel, default=0.8707, const [0.8707,15]}
       thumb_01_r_ctrl.Rotation.Z {type=MovieSceneScriptingFloatChannel, default=-13.9227, const [-13.9227,15]}
       thumb_01_r_ctrl.Scale.X {type=MovieSceneScriptingFloatChannel, default=1, const [1,15]}
       thumb_01_r_ctrl.Scale.Y {type=MovieSceneScriptingFloatChannel, default=1, const [1,15]}
       thumb_01_r_ctrl.Scale.Z {type=MovieSceneScriptingFloatChannel, default=1, const [1,15]}
       thumb_02_r_ctrl.Location.X {type=MovieSceneScriptingFloatChannel, default=0, const [0,15]}
       thumb_02_r_ctrl.Location.Y {type=MovieSceneScriptingFloatChannel, default=0, const [0,15]}
       thumb_02_r_ctrl.Location.Z {type=MovieSceneScriptingFloatChannel, default=0, const [0,15]}
       thumb_02_r_ctrl.Rotation.X {type=MovieSceneScriptingFloatChannel, default=0, const [0,15]}
       thumb_02_r_ctrl.Rotation.Y {type=MovieSceneScriptingFloatChannel, default=0, const [0,15]}
       thumb_02_r_ctrl.Rotation.Z {type=MovieSceneScriptingFloatChannel, default=0, const [0,15]}
       thumb_02_r_ctrl.Scale.X {type=MovieSceneScriptingFloatChannel, default=1, const [1,15]}
       thumb_02_r_ctrl.Scale.Y {type=MovieSceneScriptingFloatChannel, default=1, const [1,15]}
       thumb_02_r_ctrl.Scale.Z {type=MovieSceneScriptingFloatChannel, default=1, const [1,15]}
       thumb_03_r_ctrl.Location.X {type=MovieSceneScriptingFloatChannel, default=0, const [0,15]}
       thumb_03_r_ctrl.Location.Y {type=MovieSceneScriptingFloatChannel, default=0, const [0,15]}
       thumb_03_r_ctrl.Location.Z {type=MovieSceneScriptingFloatChannel, default=0, const [0,15]}
       thumb_03_r_ctrl.Rotation.X {type=MovieSceneScriptingFloatChannel, default=8.1645, const [8.1645,15]}
       thumb_03_r_ctrl.Rotation.Y {type=MovieSceneScriptingFloatChannel, default=-9.9357, const [-9.9357,15]}
       thumb_03_r_ctrl.Rotation.Z {type=MovieSceneScriptingFloatChannel, default=-50.7089, const [-50.7089,15]}
       thumb_03_r_ctrl.Scale.X {type=MovieSceneScriptingFloatChannel, default=1, const [1,15]}
       thumb_03_r_ctrl.Scale.Y {type=MovieSceneScriptingFloatChannel, default=1, const [1,15]}
       thumb_03_r_ctrl.Scale.Z {type=MovieSceneScriptingFloatChannel, default=1, const [1,15]}
       index_metacarpal_r_ctrl.Location.X {type=MovieSceneScriptingFloatChannel, default=0.0001, const [0.0001,15]}
       index_metacarpal_r_ctrl.Location.Y {type=MovieSceneScriptingFloatChannel, default=0.0002, const [0.0002,15]}
       index_metacarpal_r_ctrl.Location.Z {type=MovieSceneScriptingFloatChannel, default=0, const [0,15]}
       index_metacarpal_r_ctrl.Rotation.X {type=MovieSceneScriptingFloatChannel, default=0, const [0,15]}
       index_metacarpal_r_ctrl.Rotation.Y {type=MovieSceneScriptingFloatChannel, default=3.7815, const [3.7815,15]}
       index_metacarpal_r_ctrl.Rotation.Z {type=MovieSceneScriptingFloatChannel, default=-3.094, const [-3.094,15]}
       index_metacarpal_r_ctrl.Scale.X {type=MovieSceneScriptingFloatChannel, default=1, const [1,15]}
       index_metacarpal_r_ctrl.Scale.Y {type=MovieSceneScriptingFloatChannel, default=1, const [1,15]}
       index_metacarpal_r_ctrl.Scale.Z {type=MovieSceneScriptingFloatChannel, default=1, const [1,15]}
       index_01_r_ctrl.Location.X {type=MovieSceneScriptingFloatChannel, default=0, const [0,15]}
       index_01_r_ctrl.Location.Y {type=MovieSceneScriptingFloatChannel, default=0, const [0,15]}
       index_01_r_ctrl.Location.Z {type=MovieSceneScriptingFloatChannel, default=0, const [0,15]}
       index_01_r_ctrl.Rotation.X {type=MovieSceneScriptingFloatChannel, default=6.2821, const [6.2821,15]}
       index_01_r_ctrl.Rotation.Y {type=MovieSceneScriptingFloatChannel, default=-2.5276, const [-2.5276,15]}
       index_01_r_ctrl.Rotation.Z {type=MovieSceneScriptingFloatChannel, default=-65.1155, const [-65.1155,15]}
       index_01_r_ctrl.Scale.X {type=MovieSceneScriptingFloatChannel, default=1, const [1,15]}
       index_01_r_ctrl.Scale.Y {type=MovieSceneScriptingFloatChannel, default=1, const [1,15]}
       index_01_r_ctrl.Scale.Z {type=MovieSceneScriptingFloatChannel, default=1, const [1,15]}
       index_02_r_ctrl.Location.X {type=MovieSceneScriptingFloatChannel, default=-0.0001, const [-0.0001,15]}
       index_02_r_ctrl.Location.Y {type=MovieSceneScriptingFloatChannel, default=-0.0001, const [-0.0001,15]}
       index_02_r_ctrl.Location.Z {type=MovieSceneScriptingFloatChannel, default=-0.0001, const [-0.0001,15]}
       index_02_r_ctrl.Rotation.X {type=MovieSceneScriptingFloatChannel, default=0, const [0,15]}
       index_02_r_ctrl.Rotation.Y {type=MovieSceneScriptingFloatChannel, default=0, const [0,15]}
       index_02_r_ctrl.Rotation.Z {type=MovieSceneScriptingFloatChannel, default=-60.8298, const [-60.8298,15]}
       index_02_r_ctrl.Scale.X {type=MovieSceneScriptingFloatChannel, default=1, const [1,15]}
       index_02_r_ctrl.Scale.Y {type=MovieSceneScriptingFloatChannel, default=1, const [1,15]}
       index_02_r_ctrl.Scale.Z {type=MovieSceneScriptingFloatChannel, default=1, const [1,15]}
       index_03_r_ctrl.Location.X {type=MovieSceneScriptingFloatChannel, default=-0.0001, const [-0.0001,15]}
       index_03_r_ctrl.Location.Y {type=MovieSceneScriptingFloatChannel, default=0, const [0,15]}
       index_03_r_ctrl.Location.Z {type=MovieSceneScriptingFloatChannel, default=-0.0001, const [-0.0001,15]}
       index_03_r_ctrl.Rotation.X {type=MovieSceneScriptingFloatChannel, default=0, const [0,15]}
       index_03_r_ctrl.Rotation.Y {type=MovieSceneScriptingFloatChannel, default=0, const [0,15]}
       index_03_r_ctrl.Rotation.Z {type=MovieSceneScriptingFloatChannel, default=-78.806, const [-78.806,15]}
       index_03_r_ctrl.Scale.X {type=MovieSceneScriptingFloatChannel, default=1, const [1,15]}
       index_03_r_ctrl.Scale.Y {type=MovieSceneScriptingFloatChannel, default=1, const [1,15]}
       index_03_r_ctrl.Scale.Z {type=MovieSceneScriptingFloatChannel, default=1, const [1,15]}
       middle_metacarpal_r_ctrl.Location.X {type=MovieSceneScriptingFloatChannel, default=0, const [0,15]}
       middle_metacarpal_r_ctrl.Location.Y {type=MovieSceneScriptingFloatChannel, default=0.0001, const [0.0001,15]}
       middle_metacarpal_r_ctrl.Location.Z {type=MovieSceneScriptingFloatChannel, default=0, const [0,15]}
       middle_metacarpal_r_ctrl.Rotation.X {type=MovieSceneScriptingFloatChannel, default=-7.29, const [-7.29,15]}
       middle_metacarpal_r_ctrl.Rotation.Y {type=MovieSceneScriptingFloatChannel, default=1.3582, const [1.3582,15]}
       middle_metacarpal_r_ctrl.Rotation.Z {type=MovieSceneScriptingFloatChannel, default=-2.7048, const [-2.7048,15]}
       middle_metacarpal_r_ctrl.Scale.X {type=MovieSceneScriptingFloatChannel, default=1, const [1,15]}
       middle_metacarpal_r_ctrl.Scale.Y {type=MovieSceneScriptingFloatChannel, default=1, const [1,15]}
       middle_metacarpal_r_ctrl.Scale.Z {type=MovieSceneScriptingFloatChannel, default=1, const [1,15]}
       middle_01_r_ctrl.Location.X {type=MovieSceneScriptingFloatChannel, default=0.0002, const [0.0002,15]}
       middle_01_r_ctrl.Location.Y {type=MovieSceneScriptingFloatChannel, default=0, const [0,15]}
       middle_01_r_ctrl.Location.Z {type=MovieSceneScriptingFloatChannel, default=0, const [0,15]}
       middle_01_r_ctrl.Rotation.X {type=MovieSceneScriptingFloatChannel, default=0, const [0,15]}
       middle_01_r_ctrl.Rotation.Y {type=MovieSceneScriptingFloatChannel, default=1.3494, const [1.3494,15]}
       middle_01_r_ctrl.Rotation.Z {type=MovieSceneScriptingFloatChannel, default=-74.7137, const [-74.7137,15]}
       middle_01_r_ctrl.Scale.X {type=MovieSceneScriptingFloatChannel, default=1, const [1,15]}
       middle_01_r_ctrl.Scale.Y {type=MovieSceneScriptingFloatChannel, default=1, const [1,15]}
       middle_01_r_ctrl.Scale.Z {type=MovieSceneScriptingFloatChannel, default=1, const [1,15]}
       middle_02_r_ctrl.Location.X {type=MovieSceneScriptingFloatChannel, default=0.0008, const [0.0008,15]}
       middle_02_r_ctrl.Location.Y {type=MovieSceneScriptingFloatChannel, default=0.0009, const [0.0009,15]}
       middle_02_r_ctrl.Location.Z {type=MovieSceneScriptingFloatChannel, default=0.0476, const [0.0476,15]}
       middle_02_r_ctrl.Rotation.X {type=MovieSceneScriptingFloatChannel, default=0.0065, const [0.0065,15]}
       middle_02_r_ctrl.Rotation.Y {type=MovieSceneScriptingFloatChannel, default=-0.0119, const [-0.0119,15]}
       middle_02_r_ctrl.Rotation.Z {type=MovieSceneScriptingFloatChannel, default=-58.0462, const [-58.0462,15]}
       middle_02_r_ctrl.Scale.X {type=MovieSceneScriptingFloatChannel, default=1, const [1,15]}
       middle_02_r_ctrl.Scale.Y {type=MovieSceneScriptingFloatChannel, default=1, const [1,15]}
       middle_02_r_ctrl.Scale.Z {type=MovieSceneScriptingFloatChannel, default=1, const [1,15]}
       middle_03_r_ctrl.Location.X {type=MovieSceneScriptingFloatChannel, default=0.0005, const [0.0005,15]}
       middle_03_r_ctrl.Location.Y {type=MovieSceneScriptingFloatChannel, default=0, const [0,15]}
       middle_03_r_ctrl.Location.Z {type=MovieSceneScriptingFloatChannel, default=0.0061, const [0.0061,15]}
       middle_03_r_ctrl.Rotation.X {type=MovieSceneScriptingFloatChannel, default=-0.007, const [-0.007,15]}
       middle_03_r_ctrl.Rotation.Y {type=MovieSceneScriptingFloatChannel, default=-0.0086, const [-0.0086,15]}
       middle_03_r_ctrl.Rotation.Z {type=MovieSceneScriptingFloatChannel, default=-64.5776, const [-64.5776,15]}
       middle_03_r_ctrl.Scale.X {type=MovieSceneScriptingFloatChannel, default=1, const [1,15]}
       middle_03_r_ctrl.Scale.Y {type=MovieSceneScriptingFloatChannel, default=1, const [1,15]}
       middle_03_r_ctrl.Scale.Z {type=MovieSceneScriptingFloatChannel, default=1, const [1,15]}
       ring_metacarpal_r_ctrl.Location.X {type=MovieSceneScriptingFloatChannel, default=0.0001, const [0.0001,15]}
       ring_metacarpal_r_ctrl.Location.Y {type=MovieSceneScriptingFloatChannel, default=0.0002, const [0.0002,15]}
       ring_metacarpal_r_ctrl.Location.Z {type=MovieSceneScriptingFloatChannel, default=0, const [0,15]}
       ring_metacarpal_r_ctrl.Rotation.X {type=MovieSceneScriptingFloatChannel, default=-2.4847, const [-2.4847,15]}
       ring_metacarpal_r_ctrl.Rotation.Y {type=MovieSceneScriptingFloatChannel, default=-0.8637, const [-0.8637,15]}
       ring_metacarpal_r_ctrl.Rotation.Z {type=MovieSceneScriptingFloatChannel, default=-5.6698, const [-5.6698,15]}
       ring_metacarpal_r_ctrl.Scale.X {type=MovieSceneScriptingFloatChannel, default=1, const [1,15]}
       ring_metacarpal_r_ctrl.Scale.Y {type=MovieSceneScriptingFloatChannel, default=1, const [1,15]}
       ring_metacarpal_r_ctrl.Scale.Z {type=MovieSceneScriptingFloatChannel, default=1, const [1,15]}
       ring_01_r_ctrl.Location.X {type=MovieSceneScriptingFloatChannel, default=-0.0002, const [-0.0002,15]}
       ring_01_r_ctrl.Location.Y {type=MovieSceneScriptingFloatChannel, default=0, const [0,15]}
       ring_01_r_ctrl.Location.Z {type=MovieSceneScriptingFloatChannel, default=0, const [0,15]}
       ring_01_r_ctrl.Rotation.X {type=MovieSceneScriptingFloatChannel, default=-4.509, const [-4.509,15]}
       ring_01_r_ctrl.Rotation.Y {type=MovieSceneScriptingFloatChannel, default=-1.201, const [-1.201,15]}
       ring_01_r_ctrl.Rotation.Z {type=MovieSceneScriptingFloatChannel, default=-79.0231, const [-79.0231,15]}
       ring_01_r_ctrl.Scale.X {type=MovieSceneScriptingFloatChannel, default=1, const [1,15]}
       ring_01_r_ctrl.Scale.Y {type=MovieSceneScriptingFloatChannel, default=1, const [1,15]}
       ring_01_r_ctrl.Scale.Z {type=MovieSceneScriptingFloatChannel, default=1, const [1,15]}
       ring_02_r_ctrl.Location.X {type=MovieSceneScriptingFloatChannel, default=0.0003, const [0.0003,15]}
       ring_02_r_ctrl.Location.Y {type=MovieSceneScriptingFloatChannel, default=0.0001, const [0.0001,15]}
       ring_02_r_ctrl.Location.Z {type=MovieSceneScriptingFloatChannel, default=0, const [0,15]}
       ring_02_r_ctrl.Rotation.X {type=MovieSceneScriptingFloatChannel, default=0, const [0,15]}
       ring_02_r_ctrl.Rotation.Y {type=MovieSceneScriptingFloatChannel, default=0, const [0,15]}
       ring_02_r_ctrl.Rotation.Z {type=MovieSceneScriptingFloatChannel, default=-66.6687, const [-66.6687,15]}
       ring_02_r_ctrl.Scale.X {type=MovieSceneScriptingFloatChannel, default=1, const [1,15]}
       ring_02_r_ctrl.Scale.Y {type=MovieSceneScriptingFloatChannel, default=1, const [1,15]}
       ring_02_r_ctrl.Scale.Z {type=MovieSceneScriptingFloatChannel, default=1, const [1,15]}
       ring_03_r_ctrl.Location.X {type=MovieSceneScriptingFloatChannel, default=0, const [0,15]}
       ring_03_r_ctrl.Location.Y {type=MovieSceneScriptingFloatChannel, default=-0.0001, const [-0.0001,15]}
       ring_03_r_ctrl.Location.Z {type=MovieSceneScriptingFloatChannel, default=0, const [0,15]}
       ring_03_r_ctrl.Rotation.X {type=MovieSceneScriptingFloatChannel, default=0, const [0,15]}
       ring_03_r_ctrl.Rotation.Y {type=MovieSceneScriptingFloatChannel, default=0, const [0,15]}
       ring_03_r_ctrl.Rotation.Z {type=MovieSceneScriptingFloatChannel, default=-73.2004, const [-73.2004,15]}
       ring_03_r_ctrl.Scale.X {type=MovieSceneScriptingFloatChannel, default=1, const [1,15]}
       ring_03_r_ctrl.Scale.Y {type=MovieSceneScriptingFloatChannel, default=1, const [1,15]}
       ring_03_r_ctrl.Scale.Z {type=MovieSceneScriptingFloatChannel, default=1, const [1,15]}
       pinky_metacarpal_r_ctrl.Location.X {type=MovieSceneScriptingFloatChannel, default=-0.0001, const [-0.0001,15]}
       pinky_metacarpal_r_ctrl.Location.Y {type=MovieSceneScriptingFloatChannel, default=0, const [0,15]}
       pinky_metacarpal_r_ctrl.Location.Z {type=MovieSceneScriptingFloatChannel, default=0, const [0,15]}
       pinky_metacarpal_r_ctrl.Rotation.X {type=MovieSceneScriptingFloatChannel, default=-5.9588, const [-5.9588,15]}
       pinky_metacarpal_r_ctrl.Rotation.Y {type=MovieSceneScriptingFloatChannel, default=-5.2205, const [-5.2205,15]}
       pinky_metacarpal_r_ctrl.Rotation.Z {type=MovieSceneScriptingFloatChannel, default=-11.0008, const [-11.0008,15]}
       pinky_metacarpal_r_ctrl.Scale.X {type=MovieSceneScriptingFloatChannel, default=1, const [1,15]}
       pinky_metacarpal_r_ctrl.Scale.Y {type=MovieSceneScriptingFloatChannel, default=1, const [1,15]}
       pinky_metacarpal_r_ctrl.Scale.Z {type=MovieSceneScriptingFloatChannel, default=1, const [1,15]}
       pinky_01_r_ctrl.Location.X {type=MovieSceneScriptingFloatChannel, default=0.0001, const [0.0001,15]}
       pinky_01_r_ctrl.Location.Y {type=MovieSceneScriptingFloatChannel, default=0.0002, const [0.0002,15]}
       pinky_01_r_ctrl.Location.Z {type=MovieSceneScriptingFloatChannel, default=0, const [0,15]}
       pinky_01_r_ctrl.Rotation.X {type=MovieSceneScriptingFloatChannel, default=-11.7646, const [-11.7646,15]}
       pinky_01_r_ctrl.Rotation.Y {type=MovieSceneScriptingFloatChannel, default=-1.0914, const [-1.0914,15]}
       pinky_01_r_ctrl.Rotation.Z {type=MovieSceneScriptingFloatChannel, default=-84.6646, const [-84.6646,15]}
       pinky_01_r_ctrl.Scale.X {type=MovieSceneScriptingFloatChannel, default=1, const [1,15]}
       pinky_01_r_ctrl.Scale.Y {type=MovieSceneScriptingFloatChannel, default=1, const [1,15]}
       pinky_01_r_ctrl.Scale.Z {type=MovieSceneScriptingFloatChannel, default=1, const [1,15]}
       pinky_02_r_ctrl.Location.X {type=MovieSceneScriptingFloatChannel, default=-0.0001, const [-0.0001,15]}
       pinky_02_r_ctrl.Location.Y {type=MovieSceneScriptingFloatChannel, default=0.0001, const [0.0001,15]}
       pinky_02_r_ctrl.Location.Z {type=MovieSceneScriptingFloatChannel, default=0, const [0,15]}
       pinky_02_r_ctrl.Rotation.X {type=MovieSceneScriptingFloatChannel, default=0, const [0,15]}
       pinky_02_r_ctrl.Rotation.Y {type=MovieSceneScriptingFloatChannel, default=0, const [0,15]}
       pinky_02_r_ctrl.Rotation.Z {type=MovieSceneScriptingFloatChannel, default=-63.0936, const [-63.0936,15]}
       pinky_02_r_ctrl.Scale.X {type=MovieSceneScriptingFloatChannel, default=1, const [1,15]}
       pinky_02_r_ctrl.Scale.Y {type=MovieSceneScriptingFloatChannel, default=1, const [1,15]}
       pinky_02_r_ctrl.Scale.Z {type=MovieSceneScriptingFloatChannel, default=1, const [1,15]}
       pinky_03_r_ctrl.Location.X {type=MovieSceneScriptingFloatChannel, default=0.0002, const [0.0002,15]}
       pinky_03_r_ctrl.Location.Y {type=MovieSceneScriptingFloatChannel, default=-0.0002, const [-0.0002,15]}
       pinky_03_r_ctrl.Location.Z {type=MovieSceneScriptingFloatChannel, default=0, const [0,15]}
       pinky_03_r_ctrl.Rotation.X {type=MovieSceneScriptingFloatChannel, default=0, const [0,15]}
       pinky_03_r_ctrl.Rotation.Y {type=MovieSceneScriptingFloatChannel, default=0, const [0,15]}
       pinky_03_r_ctrl.Rotation.Z {type=MovieSceneScriptingFloatChannel, default=-117.2954, const [-117.2954,15]}
       pinky_03_r_ctrl.Scale.X {type=MovieSceneScriptingFloatChannel, default=1, const [1,15]}
       pinky_03_r_ctrl.Scale.Y {type=MovieSceneScriptingFloatChannel, default=1, const [1,15]}
       pinky_03_r_ctrl.Scale.Z {type=MovieSceneScriptingFloatChannel, default=1, const [1,15]}
       hand_r_ik_ctrl.Location.X {type=MovieSceneScriptingFloatChannel, default=0, thinned_from=264, keys [[1801,0],[2061,0],[2062,-3376.427],[2098,76.7171],[2103,76.717]]}
       hand_r_ik_ctrl.Location.Y {type=MovieSceneScriptingFloatChannel, default=0, thinned_from=264, keys [[1801,0],[2061,0],[2062,-3322.4141],[2098,36.3385],[2103,36.3385]]}
       hand_r_ik_ctrl.Location.Z {type=MovieSceneScriptingFloatChannel, default=0, thinned_from=264, keys [[1801,0],[2061,0],[2062,3018.4719],[2098,-55.8278],[2103,-55.8279]]}
       hand_r_ik_ctrl.Rotation.X {type=MovieSceneScriptingFloatChannel, default=0, thinned_from=264, keys [[1801,0],[2061,0],[2062,-31.9522],[2098,6.6283],[2103,6.6283]]}
       hand_r_ik_ctrl.Rotation.Y {type=MovieSceneScriptingFloatChannel, default=0, thinned_from=264, keys [[1801,0],[2061,0],[2062,48.146],[2098,7.5741],[2103,7.5741]]}
       hand_r_ik_ctrl.Rotation.Z {type=MovieSceneScriptingFloatChannel, default=0, thinned_from=264, keys [[1801,0],[2061,0],[2062,164.8866],[2098,3.195],[2103,3.195]]}
       hand_r_ik_ctrl.Scale.X {type=MovieSceneScriptingFloatChannel, default=1, const [1,263]}
       hand_r_ik_ctrl.Scale.Y {type=MovieSceneScriptingFloatChannel, default=1, const [1,263]}
       hand_r_ik_ctrl.Scale.Z {type=MovieSceneScriptingFloatChannel, default=1, const [1,263]}
       Softness_2 {type=MovieSceneScriptingFloatChannel, default=0, const [0,263]}
       hand_r_ikorient_ctrl.Rotation.Roll {type=MovieSceneScriptingFloatChannel, default=0, const [0,263]}
       hand_r_ikorient_ctrl.Rotation.Pitch {type=MovieSceneScriptingFloatChannel, default=0, const [0,263]}
       hand_r_ikorient_ctrl.Rotation.Yaw {type=MovieSceneScriptingFloatChannel, default=0, const [0,263]}
       arm_r_pv_ik_ctrl.Location.X {type=MovieSceneScriptingFloatChannel, default=0, thinned_from=264, keys [[1801,0],[2061,0],[2062,1616.2886],[2098,4.523],[2103,4.5229]]}
       arm_r_pv_ik_ctrl.Location.Y {type=MovieSceneScriptingFloatChannel, default=0, thinned_from=264, keys [[1801,0],[2061,0],[2062,1275.6674],[2098,-15.169],[2103,-15.1691]]}
       arm_r_pv_ik_ctrl.Location.Z {type=MovieSceneScriptingFloatChannel, default=0, thinned_from=264, keys [[1801,0],[2061,0],[2062,5197.8921],[2098,-84.3262],[2103,-84.3262]]}
       arm_r_pv_ik_ctrl.Rotation.X {
        type=MovieSceneScriptingFloatChannel
        default=0
        thinned_from=264
        keys [[1801,0],[2061,0],[2062,-147.3169],[2098,-325.0604],[2103,-325.0604]]
       }
       arm_r_pv_ik_ctrl.Rotation.Y {type=MovieSceneScriptingFloatChannel, default=0, thinned_from=264, keys [[1801,0],[2061,0],[2062,1.5987],[2098,0.0731],[2103,0.0731]]}
       arm_r_pv_ik_ctrl.Rotation.Z {type=MovieSceneScriptingFloatChannel, default=0, thinned_from=264, keys [[1801,0],[2061,0],[2062,-155.3544],[2098,7.915],[2103,7.915]]}
       arm_r_pv_ik_ctrl.Scale.X {type=MovieSceneScriptingFloatChannel, default=1, const [1,263]}
       arm_r_pv_ik_ctrl.Scale.Y {type=MovieSceneScriptingFloatChannel, default=1, const [1,263]}
       arm_r_pv_ik_ctrl.Scale.Z {type=MovieSceneScriptingFloatChannel, default=1, const [1,263]}
       body_ctrl.Location.X {
        type=MovieSceneScriptingFloatChannel
        default=-0.0001
        thinned_from=265
        keys [[1800,-1014.6091],[1801,64.6181],[1802,60.4702],[1803,48.5004],[1804,29.4199],[1805,3.9397],[1806,-27.2291],[1807,-63.3755],[1808,-103.7883],[1809,-147.7565],[1810,-194.5689],[1811,-243.5146],[1812,-293.8824],[1813,-344.9613],[1814,-401.8115],[1815,-468.7392],[1816,-543.9026],[1817,-625.4598],[1818,-711.5688],[1821,-978.7888],[1822,-1064.6868],[1823,-1145.9274],[1824,-1220.6685],[1825,-1287.0684],[1826,-1343.2852],[1827,-1387.4771],[1828,-1417.802],[1830,-1460.3765],[1832,-1494.184],[1837,-1541.179],[1848,-1576.5333],[1849,-1735.9294],[1850,-2029.6831],[1851,-2253.5022],[1852,-2393.9734],[1853,-2470.4045],[1854,-2502.1038],[1855,-2508.3796],[1856,-2458.8945],[1857,-2331.8274],[1858,-2159.2593],[1859,-1973.2721],[1860,-1805.9473],[1861,-1617.4164],[1862,-1373.1763],[1863,-1113.3566],[1864,-878.0868],[1865,-707.4966],[1866,-641.7155],[1867,-701.6348],[1868,-841.5766],[1869,-1001.8165],[1870,-1122.6304],[1871,-1198.8322],[1872,-1265.4305],[1873,-1322.9928],[1874,-1372.0867],[1875,-1413.2794],[1876,-1447.1387],[1877,-1474.2318],[1878,-1495.1263],[1879,-1510.3899],[1880,-1520.5898],[1881,-1526.2936],[1882,-1528.0687],[1883,-1519.3226],[1884,-1495.4863],[1885,-1460.1624],[1886,-1416.9535],[1888,-1321.2924],[1889,-1276.0455],[1890,-1236.2583],[1891,-1201.8146],[1892,-1172.3384],[1893,-1147.4537],[1894,-1126.7848],[1895,-1109.9553],[1896,-1096.5897],[1897,-1086.3118],[1898,-1078.7456],[1900,-1070.2448],[1902,-1068.0795],[1904,-1071.9091],[1908,-1088.9446],[1914,-1131.2145],[1923,-1220.1511],[1931,-1313.8014],[1932,-1336.7155],[1933,-1377.5201],[1934,-1430.2517],[1935,-1488.9469],[1936,-1604.9487],[1937,-1765.9899],[1939,-2126.0437],[1940,-2282.9026],[1942,-2582.9541],[1943,-2718.2095],[1944,-2815.229],[1945,-2852.7209],[1946,-2834.9138],[1947,-2784.9978],[1948,-2710.1255],[1949,-2617.4487],[1952,-2304.1128],[1953,-2211.7393],[1954,-2137.3218],[1955,-2088.0115],[1956,-2070.9604],[1957,-2168.2253],[1958,-2396.1836],[1959,-2663.4668],[1960,-2878.7073],[1961,-3026.2485],[1962,-3106.3474],[1963,-3051.5051],[1964,-2851.5142],[1965,-2568.7058],[1967,-1919.8741],[1968,-1610.917],[1969,-1313.2244],[1970,-1038.0281],[1971,-796.5604],[1972,-600.0531],[1973,-459.738],[1974,-361.2977],[1975,-281.6797],[1976,-219.0131],[1977,-171.427],[1978,-137.0503],[1979,-114.0122],[1980,-100.4417],[1981,-94.468],[1982,-94.2203],[1983,-97.3664],[1984,-102.5835],[1985,-109.5149],[1986,-117.804],[1987,-127.0946],[1988,-137.0302],[1989,-147.2547],[1990,-160.5823],[1991,-179.0276],[1992,-201.0355],[1993,-225.0507],[1994,-249.5181],[1995,-272.8828],[1996,-293.5907],[1997,-310.0882],[1998,-320.8226],[2000,-334.9644],[2001,-341.2708],[2002,-346.7422],[2003,-351.1353],[2004,-354.2076],[2005,-355.7164],[2006,-351.8325],[2007,-340.044],[2008,-322.2645],[2009,-300.4268],[2011,-251.6239],[2012,-228.182],[2013,-207.7288],[2014,-192.026],[2015,-182.8354],[2016,-181.9185],[2017,-191.0373],[2018,-590.4442],[2019,-1195.6017],[2020,-1634.8564],[2021,-1846.0989],[2022,-2030.8135],[2023,-2188.1826],[2024,-2317.3889],[2025,-2417.615],[2026,-2488.0437],[2027,-2537.8889],[2028,-2577.4009],[2029,-2607.3152],[2030,-2628.3711],[2031,-2641.3083],[2032,-2646.8655],[2033,-2645.78],[2035,-2626.6345],[2037,-2589.7759],[2041,-2486.5374],[2044,-2391.6897],[2047,-2278.1006],[2058,-1805.1935],[2061,-1689.1779],[2062,-1654.1221],[2085,-1080.6288],[2098,-1014.6091]]
       }
       body_ctrl.Location.Y {
        type=MovieSceneScriptingFloatChannel
        default=0
        thinned_from=265
        keys [[1800,1319.1586],[1801,74.9591],[1802,87.8922],[1803,123.6484],[1804,177.6632],[1805,245.3719],[1806,322.2098],[1808,485.0149],[1809,561.8528],[1810,629.5615],[1811,683.5762],[1812,719.3325],[1813,732.2656],[1814,714.0065],[1815,662.4776],[1816,582.5519],[1817,479.1023],[1818,357.0018],[1819,221.1231],[1820,76.3393],[1821,-72.4768],[1822,-220.4523],[1823,-362.7142],[1824,-494.3896],[1825,-610.6058],[1826,-706.4896],[1827,-777.1683],[1828,-817.7689],[1829,-841.0663],[1830,-860.5248],[1831,-876.1818],[1832,-888.2257],[1833,-893.0433],[1834,-891.6086],[1836,-883.3595],[1837,-877.4775],[1838,-929.9591],[1839,-999.0582],[1841,-1064.4336],[1842,-1095.0697],[1843,-1123.2441],[1844,-1148.136],[1845,-1168.9249],[1846,-1184.79],[1847,-1194.9109],[1848,-1198.4669],[1849,-1164.2474],[1850,-1078.6986],[1851,-967.4852],[1852,-753.9434],[1853,-407.7791],[1854,-24.2109],[1855,301.5425],[1856,563.1143],[1857,803.4776],[1858,1001.4238],[1859,1135.7445],[1860,1185.2311],[1861,1168.6956],[1862,1115.9248],[1863,1022.1728],[1864,882.6935],[1866,502.7034],[1867,355.7047],[1868,236.1021],[1869,128.2532],[1870,16.5151],[1871,-92.3103],[1872,-182.1871],[1873,-254.9065],[1874,-312.2599],[1875,-356.0386],[1876,-388.0341],[1877,-410.0375],[1878,-423.8403],[1879,-431.2337],[1880,-434.0092],[1881,-433.9581],[1888,-426.3342],[1894,-412.1805],[1897,-402.0284],[1906,-363.4843],[1907,-358.6352],[1908,-352.6557],[1910,-338.8527],[1911,-333.3664],[1912,-330.3819],[1932,-303.2744],[1933,-301.4374],[1934,-259.2678],[1935,-168.7527],[1936,-36.3986],[1937,136.1406],[1938,336.9888],[1939,538.7642],[1940,714.0836],[1941,865.9808],[1943,1163.9634],[1946,1598.9376],[1955,2859.8113],[1956,3022.0195],[1957,3204.6846],[1958,3370.1997],[1959,3480.9597],[1960,3499.3589],[1961,3360.0869],[1962,3115.4009],[1963,2904.0466],[1964,2766.0637],[1965,2648.6726],[1966,2551.3921],[1967,2473.7405],[1968,2415.2368],[1969,2375.4001],[1970,2353.7498],[1971,2349.8059],[1972,2363.0881],[1973,2393.1169],[1974,2442.4309],[1975,2508.0776],[1977,2653.502],[1978,2715.8484],[1979,2759.6631],[1980,2776.2312],[1981,2738.8547],[1982,2656.5632],[1983,2560.3611],[1984,2452.01],[1985,2332.522],[1986,2202.9087],[1987,2064.1833],[1988,1917.3578],[1989,1763.4451],[1990,1603.4357],[1991,1438.3148],[1992,1269.0895],[1994,922.3622],[1996,570.6549],[1997,403.4374],[1998,260.4289],[1999,156.8277],[2000,107.8343],[2003,58.1824],[2004,41.7887],[2005,25.5121],[2006,-1.1647],[2007,-46.1252],[2008,-105.4898],[2009,-175.3904],[2010,-251.8734],[2011,-330.9855],[2012,-408.7731],[2013,-481.283],[2014,-544.5614],[2015,-594.655],[2016,-627.6105],[2017,-639.4741],[2018,-246.0968],[2019,322.5253],[2020,777.915],[2021,1010.9274],[2022,1194.3234],[2023,1336.1926],[2024,1444.6244],[2025,1527.7085],[2026,1593.5344],[2027,1650.1917],[2028,1700.1775],[2029,1740.973],[2030,1773.1392],[2031,1797.2338],[2032,1813.8175],[2033,1823.4512],[2034,1826.6953],[2035,1824.1108],[2036,1816.2584],[2037,1803.6989],[2038,1786.9927],[2040,1743.3842],[2046,1592.8623],[2051,1485.9492],[2056,1395.4324],[2061,1319.8806],[2062,1306.4402],[2085,1275.5778],[2098,1319.1586]]
       }
       body_ctrl.Location.Z {
        type=MovieSceneScriptingFloatChannel
        default=0
        thinned_from=258
        keys [[1800,5556.0542],[1801,-159.6539],[1802,-114.9237],[1803,12.4184],[1804,212.0995],[1805,473.8469],[1806,787.3877],[1807,1142.4495],[1808,1528.7589],[1809,1936.0437],[1811,2772.4473],[1812,3181.021],[1816,4724.1938],[1817,5089.9487],[1818,5425.8848],[1819,5719.856],[1820,5959.7153],[1821,6133.3169],[1822,6260.7656],[1823,6370.3545],[1824,6462.2197],[1825,6536.498],[1826,6593.3257],[1827,6632.8384],[1835,6723.6816],[1845,6661.9082],[1847,6675.0562],[1850,6745.9321],[1858,7039.8359],[1861,7106.7134],[1864,7111.7197],[1867,7062.1899],[1877,6692.0664],[1879,6643.8521],[1881,6625.7378],[1882,6660.3784],[1883,6763.5493],[1884,6933.6113],[1885,7082.2632],[1887,7183.0884],[1895,7257.3286],[1913,7210.3096],[1916,7143.6406],[1923,6905.5908],[1926,6847.5732],[1929,6855.8081],[1934,6971.3989],[1942,7227.4473],[1956,7518.1274],[1957,7486.7246],[1958,7386.0063],[1959,7259.7949],[1960,7151.9136],[1961,7106.1865],[1962,7119.4053],[1965,7220.2656],[1966,7233.4849],[1968,7212.752],[1970,7147.0308],[1972,7030.791],[1973,6951.9971],[1974,6858.499],[1975,6749.6055],[1976,6624.625],[1977,6482.8657],[1978,6244.6582],[1979,5942.3589],[1980,5674.5483],[1981,5428.5913],[1982,5205.0078],[1983,5004.3174],[1984,4827.04],[1985,4673.6948],[1986,4544.8022],[1987,4440.8813],[1988,4362.4526],[1989,4310.0347],[1990,4284.146],[1991,4285.3057],[1992,4314.0342],[1993,4370.8511],[1994,4456.2778],[1995,4570.834],[1996,4715.04],[1997,4889.4165],[1999,5305.2056],[2000,5496.1611],[2001,5667.2866],[2002,5818.5181],[2003,5949.7925],[2004,6061.0454],[2005,6152.2134],[2006,6223.2334],[2007,6274.041],[2008,6304.5728],[2009,6314.7642],[2010,6306.7368],[2015,6177.8726],[2016,6169.8452],[2017,6182.998],[2018,5917.9673],[2019,5394.1289],[2020,5018.3647],[2021,4862.542],[2022,4726.3184],[2023,4608.2622],[2024,4506.9424],[2025,4420.9272],[2026,4348.7856],[2027,4289.0859],[2028,4240.3965],[2034,4031.49],[2036,3981.4104],[2038,3952.905],[2039,3949.114],[2040,3953.5623],[2041,3967.198],[2042,3994.3538],[2043,4037.2263],[2044,4093.5593],[2045,4161.0962],[2047,4320.7583],[2052,4758.0547],[2054,4900.8506],[2071,5597.7881],[2085,5542.4546],[2098,5556.0542]]
       }
       body_ctrl.Rotation.X {
        type=MovieSceneScriptingFloatChannel
        default=342.832
        thinned_from=265
        keys [[1800,-182.52],[1801,342.832],[1813,342.832],[1814,342.3033],[1815,340.6895],[1816,337.9496],[1817,334.0426],[1818,328.9274],[1819,322.5628],[1820,314.9077],[1821,305.9211],[1822,295.5619],[1823,283.7889],[1824,270.5611],[1826,240.8065],[1827,227.0286],[1828,215.0191],[1829,205.2938],[1830,198.3682],[1831,194.0019],[1832,191.2298],[1834,187.4834],[1837,183.2117],[1838,177.7064],[1840,157.2528],[1841,152.7396],[1842,153.1905],[1843,154.4532],[1844,156.3922],[1845,158.8724],[1847,164.915],[1848,168.2069],[1849,181.0719],[1850,192.2798],[1851,192.0726],[1852,191.4066],[1853,190.117],[1855,186.1865],[1859,176.558],[1860,174.5404],[1861,172.9384],[1862,171.8819],[1863,171.5009],[1869,173.8306],[1881,174.5913],[1883,175.1395],[1884,178.3017],[1885,184.9125],[1887,200.4089],[1888,206.3471],[1889,208.7197],[1890,208.3476],[1891,207.2944],[1892,205.6553],[1894,200.9986],[1900,182.8375],[1901,180.1964],[1903,175.8972],[1904,174.1338],[1905,172.8865],[1906,172.4134],[1912,173.5218],[1913,174.3602],[1914,176.1607],[1915,178.6827],[1917,184.9296],[1918,188.9531],[1919,191.3348],[1920,190.9309],[1921,189.8065],[1922,188.0925],[1924,183.4202],[1928,172.7655],[1929,170.5931],[1930,168.8792],[1931,167.7548],[1932,167.3509],[1933,173.3905],[1934,188.4896],[1935,208.1185],[1936,230.7541],[1937,249.1864],[1938,260.8145],[1939,270.3476],[1940,275.6909],[1942,280.5413],[1944,283.8938],[1950,285.2385],[1951,278.2863],[1953,247.6962],[1954,240.7439],[1955,248.5132],[1957,282.698],[1958,290.4673],[1960,240.3754],[1961,243.2389],[1963,256.5104],[1964,260.4946],[1969,267.1977],[1972,269.3008],[1973,269.4607],[1974,263.6165],[1975,248.707],[1976,228.6669],[1977,207.4308],[1978,188.9333],[1979,177.109],[1989,105.3482],[1990,90.6544],[1991,66.1791],[1992,39.8024],[1993,15.1608],[1994,-9.128],[1995,-32.7109],[1996,-55.2351],[1997,-76.3477],[1998,-95.6957],[1999,-112.9262],[2000,-127.6864],[2001,-139.6234],[2002,-149.6893],[2003,-157.7049],[2004,-161.6197],[2009,-167.0328],[2012,-169.0523],[2014,-169.4932],[2015,-166.5968],[2016,-159.3557],[2018,-140.3807],[2019,-148.313],[2021,-145.6242],[2022,-145.9226],[2023,-146.7388],[2025,-149.4507],[2028,-154.4403],[2035,-163.8746],[2037,-167.8107],[2041,-176.56],[2042,-178.3399],[2043,-179.7543],[2044,-180.7022],[2062,-182.005],[2098,-182.52]]
       }
       body_ctrl.Rotation.Y {
        type=MovieSceneScriptingFloatChannel
        default=1.885
        thinned_from=265
        keys [[1800,-0.963],[1801,1.885],[1813,1.885],[1814,2.1821],[1815,2.9481],[1816,3.9956],[1817,5.1369],[1818,6.1843],[1819,6.9504],[1820,7.2474],[1821,6.449],[1822,4.3439],[1823,1.3678],[1825,-5.4554],[1826,-8.4315],[1827,-10.5365],[1828,-11.335],[1830,-11.3045],[1832,-11.1151],[1835,-10.9439],[1837,-10.8989],[1838,-11.0201],[1839,-11.3206],[1841,-12.0817],[1842,-12.3533],[1843,-12.4263],[1844,-12.2063],[1845,-11.5987],[1846,-10.509],[1847,-8.8429],[1848,-6.5057],[1849,0.0644],[1850,9.9397],[1851,16.4803],[1852,19.1084],[1853,21.5445],[1854,23.5967],[1856,27.1557],[1857,28.8722],[1858,30.513],[1859,32.053],[1860,33.4669],[1861,34.7295],[1862,35.8155],[1863,36.6999],[1865,38.1673],[1866,38.746],[1867,39.1387],[1868,39.2834],[1869,38.8759],[1870,37.8015],[1871,36.2826],[1872,34.5414],[1873,30.918],[1874,24.7765],[1875,17.875],[1876,11.9719],[1877,8.8255],[1881,4.0975],[1882,0.7933],[1883,-5.6548],[1884,-13.5957],[1885,-21.5844],[1886,-28.6841],[1887,-34.0056],[1888,-36.66],[1893,-41.779],[1895,-43.6015],[1897,-45.1313],[1898,-45.7522],[1899,-46.259],[1900,-46.6379],[1901,-46.8753],[1902,-46.9574],[1903,-45.3359],[1904,-41.0608],[1905,-35.0168],[1906,-28.0882],[1907,-22.3135],[1929,-22.5824],[1930,-22.158],[1931,-20.9586],[1932,-19.0949],[1933,-16.6776],[1934,-13.8175],[1935,-10.6252],[1936,-7.2116],[1938,-0.1628],[1939,3.2509],[1940,6.4432],[1941,9.3033],[1942,11.7206],[1943,13.5843],[1944,14.7837],[1945,15.2081],[1950,15.2019],[1951,16.8454],[1952,21.1784],[1953,27.3044],[1954,34.3269],[1955,41.9154],[1956,49.8298],[1957,57.3096],[1958,63.5941],[1959,67.9226],[1960,69.5345],[1961,58.0049],[1962,45.8822],[1965,42.3231],[1966,41.0669],[1967,39.682],[1968,38.1842],[1969,36.5899],[1971,33.1757],[1974,27.7322],[1975,24.4846],[1976,19.1923],[1977,13.0154],[1978,7.1141],[1979,2.6485],[1980,0.7787],[1981,0.5937],[1982,0.41],[1983,0.2288],[1984,0.0515],[1985,-0.1207],[1986,-0.2865],[1987,-0.4445],[1988,-0.5937],[1989,-0.7325],[1990,-0.8598],[1991,-0.9743],[1992,-1.0747],[1993,-1.1598],[1994,-1.2281],[1995,-1.2785],[1996,-1.3097],[1997,-1.3204],[1998,-1.1788],[1999,-0.8053],[2000,-0.2772],[2002,0.9334],[2003,1.4615],[2004,1.835],[2005,1.9766],[2006,1.6033],[2007,0.5711],[2008,-0.9882],[2009,-2.9428],[2010,-5.161],[2012,-9.8609],[2013,-12.079],[2014,-14.0336],[2015,-15.5929],[2016,-16.6252],[2017,-16.9985],[2018,26.7301],[2019,23.7091],[2021,47.9946],[2022,45.8753],[2023,40.2864],[2024,32.3816],[2026,14.2384],[2027,6.3071],[2028,0.6742],[2029,-1.5068],[2032,-1.7139],[2033,-1.7784],[2034,-1.839],[2035,-1.8946],[2036,-1.9444],[2037,-1.9874],[2038,-2.0225],[2039,-2.0487],[2040,-2.0652],[2041,-2.0709],[2049,-2.0482],[2062,-1.9665],[2085,-1.1309],[2098,-0.963]]
       }
       body_ctrl.Rotation.Z {
        type=MovieSceneScriptingFloatChannel
        default=295.1217
        thinned_from=265
        keys [[1800,-160.3018],[1801,295.1217],[1813,295.1217],[1814,294.5154],[1815,292.6537],[1816,289.4719],[1817,284.9055],[1818,278.8899],[1819,271.3607],[1820,262.2534],[1823,233.2225],[1826,205.8863],[1828,188.7877],[1830,172.7185],[1831,166.394],[1832,162.3734],[1834,156.3916],[1835,153.886],[1836,151.922],[1837,150.6634],[1848,141.2982],[1849,137.1046],[1850,127.8977],[1851,116.1841],[1852,103.3406],[1853,89.2139],[1854,74.0159],[1855,57.9585],[1856,41.2536],[1857,24.1129],[1859,-10.6282],[1860,-27.8049],[1861,-44.5699],[1862,-60.7114],[1863,-76.0174],[1865,-105.443],[1866,-119.7622],[1867,-133.6092],[1868,-146.8265],[1869,-159.2566],[1870,-170.7423],[1871,-180.6166],[1872,-188.4734],[1873,-194.5465],[1874,-199.0695],[1875,-202.2761],[1876,-204.4001],[1877,-205.6751],[1884,-208.7049],[1885,-208.856],[1886,-200.9049],[1887,-185.7574],[1888,-176.4718],[1891,-170.057],[1896,-157.9518],[1898,-153.9723],[1899,-152.4134],[1900,-151.2238],[1901,-150.4649],[1902,-150.1983],[1903,-150.436],[1905,-151.9485],[1907,-154.1271],[1909,-157.2072],[1913,-165.4995],[1924,-194.7674],[1927,-201.5265],[1932,-209.4761],[1933,-210.1251],[1934,-199.3422],[1935,-184.6996],[1936,-177.0829],[1937,-170.0837],[1938,-164.3198],[1939,-160.4085],[1940,-158.9675],[1941,-159.8297],[1942,-162.2439],[1943,-165.9515],[1944,-170.6937],[1945,-176.2119],[1947,-188.5417],[1948,-199.0059],[1949,-211.8003],[1950,-218.0397],[1951,-216.9967],[1952,-214.247],[1953,-210.3595],[1954,-205.9032],[1955,-174.2887],[1956,-137.6357],[1960,-79.4512],[1961,-46.4184],[1962,7.6128],[1963,58.5569],[1964,82.3282],[1967,88.7977],[1968,91.2418],[1969,94.1935],[1970,97.5513],[1971,101.2136],[1973,109.0458],[1974,117.4497],[1975,132.318],[1976,150.0864],[1977,167.1909],[1978,180.0673],[1979,185.1514],[1980,184.5593],[1981,182.8931],[1982,180.3179],[1983,176.9993],[1984,173.1022],[1988,155.0352],[1989,150.7251],[1990,146.828],[1991,143.5093],[1992,140.9342],[1993,139.268],[1994,138.6758],[1995,150.9215],[1996,174.5904],[1997,190.0802],[1999,199.6095],[2000,203.994],[2001,207.9224],[2002,211.2425],[2003,213.8022],[2004,215.4496],[2005,216.0325],[2006,214.8184],[2007,211.4619],[2008,206.3915],[2009,200.0356],[2010,192.8227],[2012,177.54],[2013,170.3271],[2014,163.9712],[2015,158.9007],[2016,155.5443],[2017,154.3302],[2018,204.2867],[2019,209.7911],[2020,192.9734],[2021,169.2728],[2022,156.0167],[2023,145.8197],[2024,141.741],[2030,143.3993],[2038,147.2022],[2062,147.4438],[2085,190.3856],[2098,199.6981]]
       }
       body_ctrl.Scale.X {type=MovieSceneScriptingFloatChannel, default=1, const [1,36]}
       body_ctrl.Scale.Y {type=MovieSceneScriptingFloatChannel, default=1, const [1,36]}
       body_ctrl.Scale.Z {type=MovieSceneScriptingFloatChannel, default=1, const [1,36]}
       hips_ctrl.Location.X {type=MovieSceneScriptingFloatChannel, default=0, const [0,19]}
       hips_ctrl.Location.Y {type=MovieSceneScriptingFloatChannel, default=0, const [0,19]}
       hips_ctrl.Location.Z {type=MovieSceneScriptingFloatChannel, default=0, const [0,19]}
       hips_ctrl.Rotation.X {
        type=MovieSceneScriptingFloatChannel
        default=0
        thinned_from=23
        keys [[1801,0],[1816,0],[1823,25.1747],[1830,0],[1836,0],[1841,-16.6845],[1848,-0.1837],[1849,-0.0953],[1851,0],[1858,0],[1870,-8.4617],[1906,20.5128],[1922,1.0631],[1931,-0.0692],[1933,-0.0953],[1935,-0.0827],[1938,3.8026],[1942,0],[2098,0]]
       }
       hips_ctrl.Rotation.Y {
        type=MovieSceneScriptingFloatChannel
        default=0
        thinned_from=29
        keys [[1801,0],[1830,0],[1836,0.0622],[1847,0.1649],[1848,1.2608],[1849,5.5932],[1851,15.1536],[1853,-7.4552],[1856,-11.5052],[1870,7.3617],[1884,7.3617],[1890,-16.0596],[1903,-20.0511],[1915,6.5284],[1922,7.3151],[1931,5.6656],[1933,5.5932],[1935,5.8259],[1938,1.2608],[1942,0],[1984,0],[2007,13.7853],[2021,14.2807],[2035,14.5059],[2062,8.3379],[2098,0],[2103,0]]
       }
       hips_ctrl.Rotation.Z {type=MovieSceneScriptingFloatChannel, default=0, const [0,20]}
       hips_ctrl.Scale.X {type=MovieSceneScriptingFloatChannel, default=1, const [1,19]}
       hips_ctrl.Scale.Y {type=MovieSceneScriptingFloatChannel, default=1, const [1,19]}
       hips_ctrl.Scale.Z {type=MovieSceneScriptingFloatChannel, default=1, const [1,19]}
       hips_tan_ctrl.Location.X {type=MovieSceneScriptingFloatChannel, default=0, thinned_from=16, keys [[1801,0],[2062,0],[2098,0.0198],[2103,0.0196]]}
       hips_tan_ctrl.Location.Y {type=MovieSceneScriptingFloatChannel, default=0.008, thinned_from=16, keys [[1801,0.008],[2062,0.008],[2098,-0.0689],[2103,-0.0689]]}
       hips_tan_ctrl.Location.Z {type=MovieSceneScriptingFloatChannel, default=-0.0073, thinned_from=16, keys [[1801,-0.0073],[2062,-0.0073],[2098,-0.0112],[2103,-0.0112]]}
       hips_tan_ctrl.Rotation.X {type=MovieSceneScriptingFloatChannel, default=0, const [0,15]}
       hips_tan_ctrl.Rotation.Y {type=MovieSceneScriptingFloatChannel, default=0, const [0,15]}
       hips_tan_ctrl.Rotation.Z {type=MovieSceneScriptingFloatChannel, default=0, const [0,15]}
       hips_tan_ctrl.Scale.X {type=MovieSceneScriptingFloatChannel, default=1, const [1,15]}
       hips_tan_ctrl.Scale.Y {type=MovieSceneScriptingFloatChannel, default=1, const [1,15]}
       hips_tan_ctrl.Scale.Z {type=MovieSceneScriptingFloatChannel, default=1, const [1,15]}
       spine_01_ctrl.Location.X {type=MovieSceneScriptingFloatChannel, default=0, thinned_from=263, keys [[1801,0],[2061,0],[2062,-0.5779],[2098,0]]}
       spine_01_ctrl.Location.Y {
        type=MovieSceneScriptingFloatChannel
        default=0
        thinned_from=264
        keys [[1801,0],[1922,0],[1923,-0.0065],[1924,0],[2061,0],[2062,-0.0023],[2098,0],[2103,0]]
       }
       spine_01_ctrl.Location.Z {
        type=MovieSceneScriptingFloatChannel
        default=0
        thinned_from=264
        keys [[1801,0],[1922,0],[1923,0.1119],[1924,0],[2061,0],[2062,0.0011],[2098,0],[2103,0]]
       }
       spine_01_ctrl.Rotation.X {
        type=MovieSceneScriptingFloatChannel
        default=0
        thinned_from=264
        keys [[1801,0],[1922,0],[1923,-0.1059],[1924,0],[2061,0],[2062,-0.0493],[2098,2.3786],[2103,2.3786]]
       }
       spine_01_ctrl.Rotation.Y {
        type=MovieSceneScriptingFloatChannel
        default=0
        thinned_from=264
        keys [[1801,0],[1922,0],[1923,-0.0066],[1924,0],[2061,0],[2062,-0.2399],[2098,0.6127],[2103,0.6127]]
       }
       spine_01_ctrl.Rotation.Z {type=MovieSceneScriptingFloatChannel, default=0, thinned_from=264, keys [[1801,0],[2062,0],[2098,1.2735],[2103,1.2735]]}
       spine_01_ctrl.Scale.X {type=MovieSceneScriptingFloatChannel, default=1, const [1,263]}
       spine_01_ctrl.Scale.Y {type=MovieSceneScriptingFloatChannel, default=1, const [1,263]}
       spine_01_ctrl.Scale.Z {type=MovieSceneScriptingFloatChannel, default=1, const [1,263]}
       spine_02_ctrl.Location.X {
        type=MovieSceneScriptingFloatChannel
        default=0
        thinned_from=264
        keys [[1801,0],[1922,0],[1923,-0.0007],[1924,0],[2061,0],[2062,0.2607],[2098,0],[2103,0]]
       }
       spine_02_ctrl.Location.Y {
        type=MovieSceneScriptingFloatChannel
        default=0
        thinned_from=264
        keys [[1801,0],[1922,0],[1923,-0.0056],[1924,0],[2061,0],[2062,-0.0014],[2098,0],[2103,0]]
       }
       spine_02_ctrl.Location.Z {
        type=MovieSceneScriptingFloatChannel
        default=0
        thinned_from=264
        keys [[1801,0],[1922,0],[1923,0.1208],[1924,0],[2061,0],[2062,0.0685],[2098,0],[2103,0]]
       }
       spine_02_ctrl.Rotation.X {
        type=MovieSceneScriptingFloatChannel
        default=0
        thinned_from=264
        keys [[1801,0],[1922,0],[1923,0.0672],[1924,0],[2061,0],[2062,0.076],[2098,2.3786],[2103,2.3786]]
       }
       spine_02_ctrl.Rotation.Y {
        type=MovieSceneScriptingFloatChannel
        default=0
        thinned_from=264
        keys [[1801,0],[1922,0],[1923,0.0077],[1924,0],[2061,0],[2062,-3.5365],[2098,0.6127],[2103,0.6127]]
       }
       spine_02_ctrl.Rotation.Z {type=MovieSceneScriptingFloatChannel, default=0, thinned_from=264, keys [[1801,0],[2061,0],[2062,-0.003],[2098,1.2735],[2103,1.2735]]}
       spine_02_ctrl.Scale.X {type=MovieSceneScriptingFloatChannel, default=1, const [1,263]}
       spine_02_ctrl.Scale.Y {type=MovieSceneScriptingFloatChannel, default=1, const [1,263]}
       spine_02_ctrl.Scale.Z {type=MovieSceneScriptingFloatChannel, default=1, const [1,263]}
       spine_03_ctrl.Location.X {type=MovieSceneScriptingFloatChannel, default=0, thinned_from=263, keys [[1801,0],[2061,0],[2062,-0.2309],[2098,0]]}
       spine_03_ctrl.Location.Y {
        type=MovieSceneScriptingFloatChannel
        default=0
        thinned_from=264
        keys [[1801,0],[1922,0],[1923,0.0287],[1924,0],[2061,0],[2062,0.0121],[2098,0],[2103,0]]
       }
       spine_03_ctrl.Location.Z {
        type=MovieSceneScriptingFloatChannel
        default=0
        thinned_from=264
        keys [[1801,0],[1922,0],[1923,-0.2311],[1924,0],[2061,0],[2062,-0.1268],[2098,0],[2103,0]]
       }
       spine_03_ctrl.Rotation.X {
        type=MovieSceneScriptingFloatChannel
        default=0
        thinned_from=264
        keys [[1801,0],[1922,0],[1923,0.0387],[1924,0],[2061,0],[2062,-0.0266],[2098,2.3786],[2103,2.3786]]
       }
       spine_03_ctrl.Rotation.Y {
        type=MovieSceneScriptingFloatChannel
        default=0
        thinned_from=264
        keys [[1801,0],[1922,0],[1923,-0.0011],[1924,0],[2061,0],[2062,3.7764],[2098,0.6127],[2103,0.6127]]
       }
       spine_03_ctrl.Rotation.Z {type=MovieSceneScriptingFloatChannel, default=0, thinned_from=264, keys [[1801,0],[2061,0],[2062,-0.0018],[2098,1.2735],[2103,1.2735]]}
       spine_03_ctrl.Scale.X {type=MovieSceneScriptingFloatChannel, default=1, const [1,263]}
       spine_03_ctrl.Scale.Y {type=MovieSceneScriptingFloatChannel, default=1, const [1,263]}
       spine_03_ctrl.Scale.Z {type=MovieSceneScriptingFloatChannel, default=1, const [1,263]}
       spine_04_ctrl.Location.X {type=MovieSceneScriptingFloatChannel, default=0, const [0,27]}
       spine_04_ctrl.Location.Y {
        type=MovieSceneScriptingFloatChannel
        default=0
        thinned_from=27
        keys [[1801,0],[1825,0],[1830,-0.1876],[1834,-0.1845],[1836,0],[1840,-0.0876],[1848,-0.0823],[1849,-0.0605],[1851,0],[1922,0],[1931,-0.058],[1933,-0.0605],[1935,-0.0526],[1938,0],[2098,0]]
       }
       spine_04_ctrl.Location.Z {
        type=MovieSceneScriptingFloatChannel
        default=0
        thinned_from=27
        keys [[1801,0],[1825,0],[1830,-0.0503],[1834,-0.026],[1836,0],[1840,-0.0118],[1848,-0.015],[1849,-0.0111],[1851,0],[1922,0],[1931,-0.0107],[1933,-0.0111],[1935,-0.0097],[1938,0],[2098,0]]
       }
       spine_04_ctrl.Rotation.X {
        type=MovieSceneScriptingFloatChannel
        default=2.3786
        thinned_from=27
        keys [[1801,2.3786],[1825,2.3786],[1830,16.3913],[1834,17.0839],[1836,2.3786],[1840,13.137],[1848,17.5674],[1849,13.6296],[1851,2.3786],[1922,2.3786],[1931,13.1565],[1933,13.6296],[1935,12.1492],[1938,2.3786],[2098,2.3786]]
       }
       spine_04_ctrl.Rotation.Y {
        type=MovieSceneScriptingFloatChannel
        default=0.6127
        thinned_from=28
        keys [[1801,0.6127],[1883,0.6127],[1885,12.0414],[1889,0.6127],[2007,0.6127],[2016,0.4753],[2018,-3.2009],[2021,-15.0464],[2035,-15.0464],[2040,0.6127],[2098,0.6127]]
       }
       spine_04_ctrl.Rotation.Z {
        type=MovieSceneScriptingFloatChannel
        default=1.2735
        thinned_from=28
        keys [[1801,1.2735],[1883,1.2735],[1885,15.5098],[1889,1.2735],[2016,1.2735],[2017,27.4221],[2018,1.2735],[2098,1.2735]]
       }
       spine_04_ctrl.Scale.X {type=MovieSceneScriptingFloatChannel, default=1, const [1,26]}
       spine_04_ctrl.Scale.Y {type=MovieSceneScriptingFloatChannel, default=1, const [1,26]}
       spine_04_ctrl.Scale.Z {type=MovieSceneScriptingFloatChannel, default=1, const [1,26]}
       head_ctrl.Location.X {type=MovieSceneScriptingFloatChannel, default=0, const [0,22]}
       head_ctrl.Location.Y {type=MovieSceneScriptingFloatChannel, default=0, const [0,22]}
       head_ctrl.Location.Z {type=MovieSceneScriptingFloatChannel, default=0, const [0,22]}
       head_ctrl.Rotation.X {
        type=MovieSceneScriptingFloatChannel
        default=-34.6504
        thinned_from=31
        keys [[1801,-34.6504],[1816,-34.6504],[1820,-88.9655],[1825,-56.3321],[1830,-13.3768],[1834,0.1102],[1836,5.2711],[1840,23.356],[1844,17.5229],[1849,5.2687],[1852,1.4724],[1865,30.1586],[1902,29.931],[1907,-4.3265],[1912,-11.2792],[1922,-11.4629],[1931,4.5651],[1933,5.2687],[1935,3.0603],[1938,-44.01],[1942,-22.4131],[1984,-49.6574],[1992,79.3163],[1998,28.0368],[2009,-6.8392],[2021,-6.4157],[2035,-5.2728],[2042,24.7708],[2103,24.7706]]
       }
       head_ctrl.Rotation.Y {type=MovieSceneScriptingFloatChannel, default=-1.901, const [-1.901,22]}
       head_ctrl.Rotation.Z {
        type=MovieSceneScriptingFloatChannel
        default=-14.7722
        thinned_from=26
        keys [[1801,-14.7722],[1825,-14.7722],[1830,-30.2792],[1834,-33.0973],[1836,-34.1937],[1840,-36.0896],[1848,-38.4094],[1849,-38.3871],[1892,-9.4043],[1902,26.8223],[1906,28.0239],[1912,27.7754],[1922,-11.3948],[1931,-37.7043],[1933,-38.3871],[1935,-35.2798],[1938,22.0291],[1942,49.591],[1944,-35.7677],[1948,58.5378],[1984,-14.7722],[2098,-14.7722]]
       }
       head_ctrl.Scale.X {type=MovieSceneScriptingFloatChannel, default=1, const [1,22]}
       head_ctrl.Scale.Y {type=MovieSceneScriptingFloatChannel, default=1, const [1,22]}
       head_ctrl.Scale.Z {type=MovieSceneScriptingFloatChannel, default=1, const [1,22]}
       clavicle_l_ctrl.Location.X {type=MovieSceneScriptingFloatChannel, default=0, const [0,30]}
       clavicle_l_ctrl.Location.Y {type=MovieSceneScriptingFloatChannel, default=0, const [0,30]}
       clavicle_l_ctrl.Location.Z {type=MovieSceneScriptingFloatChannel, default=0, const [0,30]}
       clavicle_l_ctrl.Rotation.X {type=MovieSceneScriptingFloatChannel, default=-5.9146, const [-5.9146,30]}
       clavicle_l_ctrl.Rotation.Y {
        type=MovieSceneScriptingFloatChannel
        default=7.2579
        thinned_from=32
        keys [[1801,7.2579],[1832,7.2579],[1835,21.5134],[1838,23.7556],[1841,1.7415],[1849,7.2579],[1851,21.6637],[1862,17.5095],[1865,10.5252],[1868,9.5401],[1874,6.5255],[1882,7.2579],[1886,37.9487],[1889,7.2579],[1935,7.2579],[1938,28.1078],[1942,7.2579],[2016,7.2579],[2017,22.1523],[2018,7.2579],[2098,7.2579]]
       }
       clavicle_l_ctrl.Rotation.Z {
        type=MovieSceneScriptingFloatChannel
        default=-10.9571
        thinned_from=31
        keys [[1801,-10.9571],[1832,-10.9571],[1835,-19.1005],[1838,-18.9896],[1841,-25.0338],[1846,-17.5964],[1849,-10.9571],[1935,-10.9571],[1938,-17.5964],[1942,-10.9571],[2098,-10.9571]]
       }
       clavicle_l_ctrl.Scale.X {type=MovieSceneScriptingFloatChannel, default=1, const [1,29]}
       clavicle_l_ctrl.Scale.Y {type=MovieSceneScriptingFloatChannel, default=1, const [1,29]}
       clavicle_l_ctrl.Scale.Z {type=MovieSceneScriptingFloatChannel, default=1, const [1,29]}
       clavicle_r_ctrl.Location.X {type=MovieSceneScriptingFloatChannel, default=0, const [0,31]}
       clavicle_r_ctrl.Location.Y {type=MovieSceneScriptingFloatChannel, default=0, const [0,31]}
       clavicle_r_ctrl.Location.Z {type=MovieSceneScriptingFloatChannel, default=0, const [0,31]}
       clavicle_r_ctrl.Rotation.X {type=MovieSceneScriptingFloatChannel, default=-5.4053, const [-5.4053,31]}
       clavicle_r_ctrl.Rotation.Y {
        type=MovieSceneScriptingFloatChannel
        default=-0.8804
        thinned_from=33
        keys [[1801,-0.8804],[1832,-0.8804],[1835,13.3751],[1838,15.6172],[1841,-6.3969],[1849,-0.8804],[1935,-0.8804],[1938,19.9694],[1942,-0.8804],[2007,-0.8804],[2016,7.5009],[2017,24.313],[2018,16.2879],[2021,16.5351],[2035,16.5351],[2040,-0.8804],[2098,-0.8804]]
       }
       clavicle_r_ctrl.Rotation.Z {
        type=MovieSceneScriptingFloatChannel
        default=-15.4541
        thinned_from=33
        keys [[1801,-15.4541],[1832,-15.4541],[1835,-23.5975],[1838,-23.4866],[1841,-29.5308],[1846,-22.0934],[1849,-15.4541],[1935,-15.4541],[1938,-22.0934],[1942,-15.4541],[2007,-15.4541],[2016,-59.2277],[2018,-37.2966],[2021,10.8931],[2035,10.8931],[2040,-15.4541],[2098,-15.4541]]
       }
       clavicle_r_ctrl.Scale.X {type=MovieSceneScriptingFloatChannel, default=1, const [1,30]}
       clavicle_r_ctrl.Scale.Y {type=MovieSceneScriptingFloatChannel, default=1, const [1,30]}
       clavicle_r_ctrl.Scale.Z {type=MovieSceneScriptingFloatChannel, default=1, const [1,30]}
       neck_01_ctrl.Location.X {type=MovieSceneScriptingFloatChannel, default=0, const [0,25]}
       neck_01_ctrl.Location.Y {type=MovieSceneScriptingFloatChannel, default=0, const [0,25]}
       neck_01_ctrl.Location.Z {type=MovieSceneScriptingFloatChannel, default=0, const [0,25]}
       neck_01_ctrl.Rotation.X {
        type=MovieSceneScriptingFloatChannel
        default=-27.3356
        thinned_from=25
        keys [[1801,-27.3356],[1816,-27.3356],[1825,-3.8639],[1836,-3.8639],[1840,12.0019],[1849,10.2077],[1883,-3.8639],[1922,-3.8639],[1931,9.616],[1933,10.2077],[1935,8.3562],[1938,-3.8639],[1942,-27.3356],[1984,-27.3356],[2007,-3.8639],[2098,-3.8639]]
       }
       neck_01_ctrl.Rotation.Y {
        type=MovieSceneScriptingFloatChannel
        default=0
        thinned_from=26
        keys [[1801,0],[2007,0],[2016,0.2019],[2018,5.605],[2021,23.0151],[2035,23.0151],[2040,0],[2098,0]]
       }
       neck_01_ctrl.Rotation.Z {type=MovieSceneScriptingFloatChannel, default=0, const [0,25]}
       neck_01_ctrl.Scale.X {type=MovieSceneScriptingFloatChannel, default=1, const [1,24]}
       neck_01_ctrl.Scale.Y {type=MovieSceneScriptingFloatChannel, default=1, const [1,24]}
       neck_01_ctrl.Scale.Z {type=MovieSceneScriptingFloatChannel, default=1, const [1,24]}
       neck_02_ctrl.Location.X {type=MovieSceneScriptingFloatChannel, default=0, const [0,22]}
       neck_02_ctrl.Location.Y {type=MovieSceneScriptingFloatChannel, default=0, const [0,22]}
       neck_02_ctrl.Location.Z {type=MovieSceneScriptingFloatChannel, default=0, const [0,22]}
       neck_02_ctrl.Rotation.X {
        type=MovieSceneScriptingFloatChannel
        default=0.6608
        thinned_from=22
        keys [[1801,0.6608],[1816,0.6608],[1825,-34.6351],[1830,0.6608],[1836,0.6608],[1840,16.5266],[1849,16.1264],[1922,0.7463],[1931,15.4797],[1933,16.1264],[1935,14.0915],[1938,0.6608],[2098,0.6608]]
       }
       neck_02_ctrl.Rotation.Y {type=MovieSceneScriptingFloatChannel, default=0, const [0,22]}
       neck_02_ctrl.Rotation.Z {type=MovieSceneScriptingFloatChannel, default=0, const [0,22]}
       neck_02_ctrl.Scale.X {type=MovieSceneScriptingFloatChannel, default=1, const [1,21]}
       neck_02_ctrl.Scale.Y {type=MovieSceneScriptingFloatChannel, default=1, const [1,21]}
       neck_02_ctrl.Scale.Z {type=MovieSceneScriptingFloatChannel, default=1, const [1,21]}
       head_ik_ctrl.Location.X {type=MovieSceneScriptingFloatChannel, default=0, thinned_from=264, keys [[1801,0],[2061,0],[2062,0.0112],[2098,0.0419],[2103,0.0419]]}
       head_ik_ctrl.Location.Y {
        type=MovieSceneScriptingFloatChannel
        default=0
        thinned_from=264
        keys [[1801,0],[1922,0],[1923,-0.001],[1924,0],[2061,0],[2062,-0.5839],[2098,-0.5565],[2103,-0.5565]]
       }
       head_ik_ctrl.Location.Z {
        type=MovieSceneScriptingFloatChannel
        default=0
        thinned_from=264
        keys [[1801,0],[1922,0],[1923,-0.005],[1924,0],[2061,0],[2062,0.1703],[2098,0.2421],[2103,0.2421]]
       }
       head_ik_ctrl.Rotation.X {
        type=MovieSceneScriptingFloatChannel
        default=0
        thinned_from=264
        keys [[1801,0],[1922,0],[1923,0.4736],[1924,0],[2061,0],[2062,24.7708],[2103,24.7706]]
       }
       head_ik_ctrl.Rotation.Y {type=MovieSceneScriptingFloatChannel, default=0, thinned_from=263, keys [[1801,0],[2061,0],[2062,-1.901],[2098,-1.901]]}
       head_ik_ctrl.Rotation.Z {type=MovieSceneScriptingFloatChannel, default=0, thinned_from=263, keys [[1801,0],[2061,0],[2062,-14.7722],[2098,-14.7722]]}
       head_ik_ctrl.Scale.X {type=MovieSceneScriptingFloatChannel, default=1, const [1,263]}
       head_ik_ctrl.Scale.Y {type=MovieSceneScriptingFloatChannel, default=1, const [1,263]}
       head_ik_ctrl.Scale.Z {type=MovieSceneScriptingFloatChannel, default=1, const [1,263]}
       chest_ctrl.Location.X {
        type=MovieSceneScriptingFloatChannel
        default=0
        thinned_from=33
        keys [[1801,0],[1848,0],[1851,-2.8753],[1854,1.2895],[1861,1.1004],[1869,-1.8015],[1883,-1.7698],[1885,-2.9123],[1889,1.7115],[1893,3.7521],[1904,2.5421],[1916,-1.922],[1922,-1.5511],[1931,0],[2007,0],[2016,-0.108],[2021,0],[2062,0],[2098,-0.215],[2103,-0.215]]
       }
       chest_ctrl.Location.Y {
        type=MovieSceneScriptingFloatChannel
        default=0
        thinned_from=34
        keys [[1801,0],[1836,0],[1841,3.104],[1848,2.75],[1851,0],[1988,0],[1994,8.5918],[2004,3.2118],[2007,0],[2016,-0.108],[2017,4.1249],[2021,0],[2062,0],[2098,0.9134],[2103,0.9134]]
       }
       chest_ctrl.Location.Z {
        type=MovieSceneScriptingFloatChannel
        default=0
        thinned_from=32
        keys [[1801,0],[2007,0],[2016,-0.108],[2021,-0.0002],[2035,-0.0002],[2040,0],[2062,0],[2098,0.0584],[2103,0.0584]]
       }
       chest_ctrl.Rotation.X {
        type=MovieSceneScriptingFloatChannel
        default=0
        thinned_from=37
        keys [[1801,0],[1816,0],[1821,-4.9542],[1828,-2.3471],[1830,1.463],[1834,11.0723],[1836,0],[1840,10.7584],[1848,15.1888],[1849,11.2509],[1851,0],[1883,0],[1885,14.3423],[1889,0],[1922,0],[1931,10.7779],[1933,11.2509],[1935,9.7706],[1938,0],[1988,0],[1993,39.3234],[1997,46.5649],[2007,0],[2011,-5.7877],[2016,0],[2017,17.2664],[2021,0],[2062,0],[2098,7.1753],[2103,7.1753]]
       }
       chest_ctrl.Rotation.Y {
        type=MovieSceneScriptingFloatChannel
        default=0
        thinned_from=31
        keys [[1801,0],[1883,0],[1885,7.9174],[1889,0],[2062,0],[2098,1.6761],[2103,1.6761]]
       }
       chest_ctrl.Rotation.Z {
        type=MovieSceneScriptingFloatChannel
        default=0
        thinned_from=33
        keys [[1801,0],[2007,0],[2016,-0.4956],[2018,-13.759],[2021,-56.4967],[2035,-56.4967],[2040,0],[2062,0],[2098,3.892],[2103,3.892]]
       }
       chest_ctrl.Scale.X {type=MovieSceneScriptingFloatChannel, default=1, const [1,30]}
       chest_ctrl.Scale.Y {type=MovieSceneScriptingFloatChannel, default=1, const [1,30]}
       chest_ctrl.Scale.Z {type=MovieSceneScriptingFloatChannel, default=1, const [1,30]}
       chest_tan_ctrl.Location.X {type=MovieSceneScriptingFloatChannel, default=0, thinned_from=16, keys [[1801,0],[2062,0],[2098,-0.0662],[2103,-0.0662]]}
       chest_tan_ctrl.Location.Y {type=MovieSceneScriptingFloatChannel, default=-0.0303, thinned_from=16, keys [[1801,-0.0303],[2062,-0.0303],[2098,0.2903],[2103,0.2903]]}
       chest_tan_ctrl.Location.Z {type=MovieSceneScriptingFloatChannel, default=0.265, thinned_from=16, keys [[1801,0.265],[2062,0.265],[2098,0.3033],[2103,0.3033]]}
       chest_tan_ctrl.Rotation.X {type=MovieSceneScriptingFloatChannel, default=0, const [0,15]}
       chest_tan_ctrl.Rotation.Y {type=MovieSceneScriptingFloatChannel, default=0, const [0,15]}
       chest_tan_ctrl.Rotation.Z {type=MovieSceneScriptingFloatChannel, default=0, const [0,15]}
       chest_tan_ctrl.Scale.X {type=MovieSceneScriptingFloatChannel, default=1, const [1,15]}
       chest_tan_ctrl.Scale.Y {type=MovieSceneScriptingFloatChannel, default=1, const [1,15]}
       chest_tan_ctrl.Scale.Z {type=MovieSceneScriptingFloatChannel, default=1, const [1,15]}
       foot_l_ik_ctrl.Space {
        type=MovieSceneScriptingSpaceChannel
        const [
         "<Struct 'MovieSceneControlRigSpaceBaseKey' (0x85268f8af00) {space_type: Parent, control_rig_element: {type: Null, name: \"leg_l_controls_space\"}}>"
         1
        ]
       }
       leg_l_pv_ik_ctrl.Space {
        type=MovieSceneScriptingSpaceChannel
        const [
         "<Struct 'MovieSceneControlRigSpaceBaseKey' (0x85268f8aec0) {space_type: Parent, control_rig_element: {type: Null, name: \"leg_l_pv_footFollow\"}}>"
         1
        ]
       }
       foot_r_ik_ctrl.Space {
        type=MovieSceneScriptingSpaceChannel
        const [
         "<Struct 'MovieSceneControlRigSpaceBaseKey' (0x85268f8ad40) {space_type: Parent, control_rig_element: {type: Null, name: \"leg_r_controls_space\"}}>"
         1
        ]
       }
       leg_r_pv_ik_ctrl.Space {
        type=MovieSceneScriptingSpaceChannel
        const [
         "<Struct 'MovieSceneControlRigSpaceBaseKey' (0x85268f8a960) {space_type: Parent, control_rig_element: {type: Null, name: \"leg_r_pv_footFollow\"}}>"
         1
        ]
       }
       body_ctrl.Space {
        type=MovieSceneScriptingSpaceChannel
        const [
         "<Struct 'MovieSceneControlRigSpaceBaseKey' (0x85268f8a8a0) {space_type: Parent, control_rig_element: {type: Control, name: \"body_offset_ctrl\"}}>"
         1
        ]
       }
       ParentConstraint_1 {type=MovieSceneScriptingConstraintChannel, const [true,1]}
      ]
     }]
   }
   {type=MovieSceneSpawnTrack, sections [{type=MovieSceneSpawnSection, channels [None {type=MovieSceneScriptingBoolChannel, default=true}]}]}
   {
    type=MovieScene3DTransformTrack
    sections [
     {
      type=MovieScene3DTransformSection
      end_frame=3770
      channels [
       Location.X {type=MovieSceneScriptingDoubleChannel, default=495.3745, const [495.3745,1]}
       Location.Y {type=MovieSceneScriptingDoubleChannel, default=-613.4579, const [-613.4579,1]}
       Location.Z {type=MovieSceneScriptingDoubleChannel, default=32880.7305, const [32880.7305,1]}
       Rotation.X {type=MovieSceneScriptingDoubleChannel, default=0, const [178.8983,1]}
       Rotation.Y {type=MovieSceneScriptingDoubleChannel, default=0, const [0.7977,1]}
       Rotation.Z {type=MovieSceneScriptingDoubleChannel, default=0, const [51.3411,1]}
       Scale.X {type=MovieSceneScriptingDoubleChannel, default=4, const [4,1]}
       Scale.Y {type=MovieSceneScriptingDoubleChannel, default=4, const [4,1]}
       Scale.Z {type=MovieSceneScriptingDoubleChannel, default=4, const [4,1]}
      ]
     }
     {
      type=MovieScene3DTransformSection
      channels [
       Location.X {type=MovieSceneScriptingDoubleChannel, default=0, const [14436,2]}
       Location.Y {type=MovieSceneScriptingDoubleChannel, default=0, const [76088,2]}
       Location.Z {type=MovieSceneScriptingDoubleChannel, default=0, const [3072,2]}
       Rotation.X {type=MovieSceneScriptingDoubleChannel, default=0, const [0,2]}
       Rotation.Y {type=MovieSceneScriptingDoubleChannel, default=0, keys [[1796,0],[2098,0]]}
       Rotation.Z {type=MovieSceneScriptingDoubleChannel, default=0, const [0,2]}
       Scale.X {type=MovieSceneScriptingDoubleChannel, default=0, const [0,2]}
       Scale.Y {type=MovieSceneScriptingDoubleChannel, default=0, const [0,2]}
       Scale.Z {type=MovieSceneScriptingDoubleChannel, default=0, const [0,2]}
      ]
     }
    ]
   }
  ]
 }
 Cube {
  id=#3
  possessed_class=FortStaticMeshActor
  tracks [
   {type=MovieSceneSpawnTrack, sections [{type=MovieSceneSpawnSection, channels [None {type=MovieSceneScriptingBoolChannel, default=true}]}]}
   {
    type=MovieScene3DTransformTrack
    sections [{
      type=MovieScene3DTransformSection
      channels [
       Location.X {
        type=MovieSceneScriptingDoubleChannel
        default=0
        thinned_from=299
        keys [[1800,14488.2939],[1801,15324.5162],[1802,15353.4201],[1803,15431.9837],[1805,15689.203],[1808,16141.934],[1809,16261.7976],[1810,16345.7679],[1811,16381.6233],[1812,16357.1417],[1813,16260.1015],[1814,16051.3642],[1815,15713.3595],[1816,15266.2172],[1817,14730.0672],[1818,14125.0399],[1819,13471.2648],[1822,11417.9381],[1823,10767.8193],[1824,10167.4553],[1825,9636.6651],[1826,9195.2683],[1827,8863.0844],[1828,8659.9329],[1829,8530.8393],[1830,8419.3566],[1832,8247.8211],[1833,8203.4499],[1837,8163.4768],[1838,7991.6348],[1839,7767.9672],[1842,7444.7598],[1844,7263.3112],[1845,7190.3919],[1846,7132.7207],[1847,7092.8292],[1848,7073.2507],[1849,6781.2759],[1850,6313.8205],[1851,6101.1543],[1852,6416.283],[1853,7305.6264],[1854,8423.5447],[1855,9424.3965],[1856,10364.1189],[1858,12480.1342],[1859,13363.6759],[1860,13935.7392],[1862,14799.9019],[1863,15156.181],[1864,15308.4441],[1865,15141.5978],[1866,14712.7673],[1867,14104.4528],[1869,12645.5033],[1870,11995.5592],[1871,11466.2469],[1872,11020.1589],[1873,10650.2536],[1874,10349.4877],[1875,10110.8197],[1876,9927.2067],[1877,9791.6068],[1878,9696.9775],[1879,9636.2762],[1880,9602.4607],[1882,9586.572],[1883,9608.9618],[1884,9667.3637],[1885,9755.0084],[1890,10335.3161],[1892,10509.3714],[1895,10690.3001],[1900,10850.3919],[1911,10931.9094],[1931,10516.0114],[1932,10463.0582],[1933,10366.1593],[1934,10365.3757],[1937,10759.1257],[1939,11115.306],[1940,11272.0716],[1942,11485.3281],[1943,11634.2717],[1944,11883.9257],[1945,12285.6866],[1946,12827.0468],[1947,13450.1483],[1948,14136.2106],[1953,17890.8959],[1954,18596.0029],[1956,19937.7547],[1957,20490.3759],[1958,20773.41],[1959,20833.9078],[1960,20690.9191],[1961,20109.9593],[1962,19255.0336],[1963,18719.6434],[1964,18641.6566],[1965,18712.7872],[1974,20203.9778],[1977,20866.6385],[1978,21034.8516],[1979,21126.0619],[1980,21100.8761],[1981,20855.917],[1982,20429.4365],[1983,19947.1919],[1984,19417.6097],[1986,18237.9248],[1988,16932.3889],[1990,15537.1923],[1996,11098.2552],[1997,10399.4667],[1998,9795.5278],[1999,9346.1153],[2000,9110.9082],[2003,8809.1242],[2005,8640.7961],[2006,8521.5079],[2007,8331.4232],[2008,8084.4551],[2009,7794.453],[2012,6835.9237],[2013,6542.9138],[2014,6291.1803],[2015,6097.5386],[2016,5978.8041],[2017,5951.4711],[2018,8097.9236],[2019,11263.5782],[2020,13713.4837],[2021,14939.4389],[2022,15938.6489],[2023,16738.7128],[2024,17367.2276],[2025,17851.7928],[2026,18220.007],[2027,18517.4246],[2028,18771.4577],[2029,18974.5497],[2030,19130.5668],[2031,19243.2162],[2033,19351.3627],[2035,19324.8871],[2037,19189.6888],[2040,18839.6903],[2057,16240.0705],[2063,15513.2248],[2067,15153.8167],[2071,14922.2939],[2083,14640.1777],[2097,14638.5363],[2098,14488.2939]]
       }
       Location.Y {
        type=MovieSceneScriptingDoubleChannel
        default=0
        thinned_from=299
        keys [[1800,68939.3547],[1801,75487.107],[1804,75154.701],[1812,73063.8865],[1815,72765.9003],[1828,73710.3986],[1837,73480.3187],[1848,74168.8294],[1849,73587.7636],[1850,72459.3231],[1852,70516.9248],[1854,68363.9011],[1855,67534.3524],[1856,67039.1345],[1857,66838.9799],[1858,66886.5629],[1859,67134.5549],[1860,67535.629],[1861,68167.2244],[1862,69062.6144],[1865,72197.2293],[1866,72875.7664],[1867,73053.7194],[1870,72577.0807],[1883,72439.4749],[1897,73762.4909],[1932,72702.4424],[1934,72305.6356],[1935,71899.191],[1936,71209.4013],[1940,67213.8402],[1943,64661.9461],[1944,63961.2459],[1945,63451.7813],[1947,62908.5469],[1950,62733.9253],[1954,62810.8164],[1955,62730.5791],[1956,62515.8977],[1957,61879.6586],[1959,59733.3329],[1960,58992.3068],[1961,58754.3226],[1962,58910.4841],[1963,59487.203],[1964,60484.1941],[1965,61762.2968],[1968,65888.7688],[1970,68322.7273],[1971,69353.1625],[1972,70202.4844],[1973,70827.4835],[1976,71992.1777],[1981,72993.3296],[1986,73192.9539],[1990,72811.9316],[1994,71749.7764],[1999,69970.1567],[2005,69043.5757],[2017,68387.1222],[2018,67658.9542],[2019,66504.731],[2020,65743.3897],[2024,64488.2658],[2027,64064.3871],[2047,64850.6284],[2068,67056.9281],[2097,66982.2307],[2098,68939.3547]]
       }
       Location.Z {
        type=MovieSceneScriptingDoubleChannel
        default=0
        thinned_from=299
        keys [[1800,14084.0728],[1801,39287.2953],[1802,39107.1686],[1803,38594.4767],[1804,37790.6912],[1805,36737.282],[1806,35475.7205],[1807,34047.4761],[1808,32494.0214],[1812,25857.504],[1816,19682.1925],[1817,18222.7288],[1818,16883.6763],[1819,15713.3337],[1820,14760.0033],[1821,14071.9843],[1822,13568.6077],[1823,13136.4618],[1824,12774.7303],[1825,12482.595],[1826,12259.2395],[1827,12103.8492],[1828,12015.6061],[1833,11789.2145],[1836,11730.2701],[1843,11916.0083],[1845,11912.4183],[1846,11877.931],[1847,11815.3055],[1848,11719.7898],[1849,11583.5973],[1851,11229.9098],[1856,10162.8698],[1858,9799.5396],[1859,9650.0141],[1860,9527.5145],[1861,9437.1667],[1862,9382.9612],[1863,9362.0256],[1864,9368.1349],[1866,9447.0695],[1868,9584.4669],[1869,9680.4618],[1876,10604.4138],[1878,10802.0486],[1879,10867.8572],[1880,10905.8453],[1881,10911.6566],[1882,10762.8138],[1883,10343.2544],[1884,9659.9119],[1885,9065.8206],[1886,8845.6958],[1887,8667.5116],[1888,8572.7405],[1892,8430.7075],[1895,8383.2603],[1912,8528.75],[1913,8562.3],[1914,8626.4045],[1915,8716.1815],[1917,8953.2351],[1922,9653.521],[1924,9874.2099],[1925,9952.3008],[1926,10002.3777],[1927,10019.565],[1928,10005.4732],[1929,9966.7898],[1931,9826.5811],[1933,9618.5217],[1935,9354.0981],[1940,8572.1187],[1942,8298.2687],[1956,6895.0127],[1957,6988.4408],[1958,7352.5503],[1959,7820.3433],[1960,8224.8622],[1961,8397.0785],[1962,8345.6887],[1963,8225.7198],[1964,8083.4903],[1965,7968.6348],[1966,7927.7807],[1967,7958.3968],[1968,8028.6235],[1969,8141.5177],[1970,8299.2553],[1971,8504.0133],[1972,8757.9811],[1973,9063.3638],[1974,9423.5666],[1975,9842.928],[1976,10325.0437],[1977,10873.5054],[1978,11807.7154],[1979,12999.0964],[1980,14054.2419],[1981,15025.8799],[1982,15911.3105],[1983,16706.1507],[1984,17408.2636],[1985,18015.5225],[1986,18525.7963],[1987,18936.9616],[1988,19246.8909],[1989,19453.4662],[1990,19552.5083],[1991,19540.2141],[1992,19415.0283],[1993,19175.3993],[1994,18819.7682],[1995,18346.5838],[1996,17754.3428],[1997,17040.8181],[1999,15341.3069],[2000,14555.4558],[2001,13848.2698],[2002,13222.9873],[2003,12680.3665],[2004,12221.1708],[2005,11846.1589],[2006,11557.098],[2007,11355.4859],[2008,11238.9485],[2009,11204.7857],[2010,11244.1114],[2011,11334.3578],[2014,11703.8337],[2015,11790.9868],[2016,11825.6724],[2017,11773.4804],[2018,12780.8133],[2019,14798.1516],[2020,16241.3056],[2021,16834.7395],[2022,17355.0916],[2023,17807.5104],[2024,18197.1423],[2025,18529.1357],[2026,18808.637],[2027,19040.236],[2029,19380.4873],[2034,20051.1378],[2037,20325.5732],[2039,20390.6019],[2041,20325.1007],[2042,20220.1642],[2043,20052.4624],[2044,19831.0056],[2046,19262.8703],[2052,17205.3642],[2053,16905.5552],[2054,16642.09],[2055,16423.9773],[2058,15918.0375],[2061,15544.132],[2064,15286.2352],[2068,15095.1312],[2073,15048.4767],[2093,15635.4281],[2097,15648.0906],[2098,14084.0728]]
       }
       Rotation.X {
        type=MovieSceneScriptingDoubleChannel
        default=0
        thinned_from=299
        keys [[1800,-1.2139],[1801,163.0866],[1813,163.0866],[1814,162.572],[1815,161.0015],[1816,158.3349],[1817,154.5312],[1818,149.5476],[1819,143.3381],[1820,135.8522],[1821,127.0197],[1822,116.7782],[1823,105.0875],[1824,91.909],[1825,77.2033],[1826,62.1597],[1827,48.3372],[1828,36.2511],[1829,26.4229],[1830,19.3762],[1831,14.899],[1832,12.0506],[1833,9.9812],[1834,8.1836],[1835,6.5945],[1836,5.1508],[1837,3.7894],[1838,-1.7345],[1839,-12.1449],[1840,-22.2247],[1841,-26.7564],[1842,-26.3242],[1843,-25.0808],[1844,-23.1616],[1845,-20.7018],[1846,-17.8367],[1847,-14.7014],[1848,-11.4308],[1849,1.3361],[1850,12.3279],[1851,11.8312],[1852,10.8498],[1853,9.2365],[1854,7.1582],[1855,4.7852],[1856,2.2819],[1857,-0.1958],[1858,-2.508],[1859,-4.5341],[1860,-6.1717],[1861,-7.331],[1862,-7.9279],[1863,-7.8757],[1864,-7.3542],[1865,-6.6797],[1866,-5.9605],[1867,-5.2956],[1868,-4.7731],[1869,-4.4749],[1870,-4.4659],[1873,-4.9358],[1874,-5.0688],[1875,-5.1577],[1876,-5.2041],[1877,-5.2185],[1878,-5.2146],[1879,-5.1553],[1880,-5.0099],[1881,-4.7995],[1882,-4.5468],[1883,-4.2686],[1884,-1.1014],[1885,5.5309],[1886,13.5139],[1887,21.6217],[1888,27.7655],[1889,30.1888],[1890,29.8695],[1891,28.8717],[1892,27.2887],[1893,25.214],[1894,22.741],[1895,19.9634],[1896,16.975],[1898,10.7421],[1899,7.6862],[1900,4.7966],[1901,2.1678],[1902,-0.1058],[1903,-2.1853],[1904,-4.0805],[1905,-5.4722],[1906,-6.0664],[1907,-6.0977],[1908,-5.9841],[1909,-5.8113],[1910,-5.595],[1911,-5.3506],[1912,-5.0933],[1913,-4.2749],[1914,-2.4982],[1915,-0.0038],[1916,2.9679],[1917,6.1763],[1918,10.161],[1919,12.5007],[1920,12.0519],[1921,10.8804],[1922,9.1179],[1923,6.8959],[1924,4.3465],[1925,1.6014],[1926,-1.2074],[1927,-3.9482],[1928,-6.4889],[1929,-8.6976],[1930,-10.4476],[1931,-11.6114],[1932,-12.054],[1933,-6.0364],[1934,9.2842],[1935,29.1661],[1936,51.9021],[1937,70.4107],[1938,82.0907],[1939,91.6557],[1940,97.0155],[1942,101.8665],[1943,103.8048],[1944,105.1621],[1945,105.7472],[1948,105.9218],[1950,105.6434],[1951,98.7218],[1953,68.3426],[1954,61.5634],[1955,70.1611],[1956,87.7535],[1957,105.0814],[1958,112.9491],[1959,87.6914],[1960,61.9751],[1961,62.8739],[1962,67.9136],[1963,74.6652],[1964,79.1411],[1967,83.7002],[1969,86.275],[1970,87.3198],[1971,88.1496],[1972,88.7244],[1973,89.0041],[1974,83.399],[1975,68.8985],[1976,49.2925],[1977,28.3898],[1978,10.079],[1979,-1.6908],[1989,-74.081],[1990,-88.8634],[1991,-113.4162],[1992,-139.8543],[1993,-164.5366],[1994,-188.8407],[1995,-212.1488],[1996,-234.2222],[1997,-255.1279],[1998,-274.4286],[1999,-291.6222],[2000,-306.3559],[2001,-318.2759],[2002,-328.3327],[2003,-336.3444],[2004,-340.2584],[2011,-347.3592],[2014,-348.6329],[2015,-345.8321],[2016,-338.6581],[2017,-329.2696],[2018,-318.8855],[2019,-326.8344],[2021,-324.2014],[2025,-329.0409],[2035,-343.4091],[2042,-357.8343],[2044,-360.1968],[2098,-361.2139]]
       }
       Rotation.Y {
        type=MovieSceneScriptingDoubleChannel
        default=0
        thinned_from=299
        keys [[1800,0.583],[1801,-0.549],[1813,-0.549],[1814,-0.8488],[1815,-1.6242],[1816,-2.691],[1817,-3.867],[1818,-4.9724],[1819,-5.8296],[1820,-6.2625],[1821,-5.6388],[1822,-3.7293],[1823,-0.9628],[1825,5.4272],[1826,8.1924],[1827,10.0979],[1828,10.7124],[1830,10.3719],[1831,10.1845],[1832,10.0202],[1834,9.8194],[1835,9.7421],[1836,9.6881],[1837,9.6633],[1838,9.7761],[1839,10.0686],[1841,10.8145],[1842,11.0789],[1843,11.145],[1844,10.9184],[1845,10.3044],[1846,9.2088],[1847,7.5368],[1848,5.1942],[1849,-1.3986],[1850,-11.299],[1851,-17.8208],[1852,-20.3634],[1853,-22.6327],[1854,-24.4316],[1855,-25.8927],[1856,-27.2742],[1858,-29.8431],[1860,-32.2454],[1862,-34.4644],[1865,-37.3103],[1866,-38.1751],[1867,-38.8787],[1868,-39.3351],[1869,-39.2195],[1870,-38.4015],[1871,-37.0844],[1872,-35.487],[1873,-31.9628],[1874,-25.8878],[1875,-19.0293],[1876,-13.1526],[1877,-10.0212],[1881,-5.3146],[1882,-2.0147],[1883,4.4292],[1884,12.3662],[1885,20.353],[1886,27.5432],[1887,33.0955],[1888,35.9231],[1893,41.2778],[1895,43.2089],[1896,44.0692],[1897,44.8389],[1898,45.5028],[1899,46.0457],[1900,46.4522],[1901,46.7073],[1902,46.7957],[1903,45.1695],[1904,40.8819],[1905,34.8197],[1906,27.87],[1907,22.07],[1929,21.3918],[1930,20.9506],[1931,19.7352],[1932,17.8574],[1933,15.434],[1934,12.6987],[1935,9.7383],[1936,6.4695],[1937,3.0903],[1938,-0.3076],[1939,-3.6322],[1940,-6.7906],[1941,-9.6697],[1942,-12.1413],[1943,-14.0877],[1944,-15.3904],[1945,-15.9305],[1947,-16.1629],[1949,-16.4764],[1950,-16.5287],[1951,-18.1676],[1952,-22.4816],[1953,-28.574],[1954,-35.5497],[1956,-49.6424],[1957,-56.7727],[1958,-62.7445],[1959,-66.8182],[1960,-68.2477],[1961,-56.6074],[1962,-45.2102],[1963,-45.2376],[1964,-44.5353],[1966,-42.1647],[1967,-40.8129],[1968,-39.3503],[1969,-37.7951],[1971,-34.4596],[1974,-29.12],[1975,-25.8718],[1976,-20.4569],[1977,-14.0486],[1978,-7.9124],[1979,-3.3424],[1980,-1.4839],[1981,-1.3318],[1982,-1.1984],[1983,-1.08],[1984,-0.9729],[1985,-0.8734],[1986,-0.7784],[1989,-0.502],[1990,-0.4116],[1991,-0.3241],[1992,-0.2417],[1993,-0.1672],[1994,-0.1028],[1995,0.0378],[1996,0.3983],[1997,0.7125],[1998,0.6767],[1999,0.4098],[2000,-0.0155],[2001,-0.5267],[2002,-1.0512],[2003,-1.5162],[2004,-1.8485],[2005,-1.9746],[2006,-1.6291],[2007,-0.6764],[2008,0.7631],[2009,2.5701],[2010,4.6265],[2012,9.0149],[2013,11.1059],[2014,12.9613],[2015,14.4508],[2016,15.4418],[2017,15.8012],[2018,-26.9963],[2019,-23.847],[2020,-36.3719],[2021,-48.9747],[2022,-47.0476],[2023,-41.5636],[2024,-33.6893],[2026,-15.5449],[2027,-7.6119],[2028,-1.9768],[2029,0.2067],[2031,0.3526],[2032,0.4237],[2033,0.4922],[2034,0.5569],[2035,0.6166],[2036,0.6704],[2037,0.7169],[2038,0.7551],[2039,0.7839],[2040,0.802],[2041,0.8082],[2044,0.8046],[2047,0.7945],[2052,0.7677],[2063,0.6953],[2066,0.6797],[2069,0.6687],[2071,0.6647],[2076,0.6591],[2090,0.6275],[2094,0.6236],[2097,0.6234],[2098,0.583]]
       }
       Rotation.Z {
        type=MovieSceneScriptingDoubleChannel
        default=0
        thinned_from=299
        keys [[1800,-83.0865],[1801,-243.7783],[1813,-243.7783],[1814,-243.1738],[1815,-241.3173],[1816,-238.1455],[1817,-233.595],[1818,-227.6014],[1819,-220.0982],[1820,-211.0157],[1823,-181.9002],[1825,-163.3087],[1827,-145.6321],[1828,-137.2033],[1829,-129.0546],[1830,-121.1802],[1831,-114.8785],[1832,-110.8741],[1834,-104.9164],[1835,-102.4209],[1836,-100.465],[1837,-99.2115],[1848,-89.9125],[1849,-85.7592],[1850,-76.5578],[1851,-64.7642],[1852,-51.8038],[1853,-37.534],[1854,-22.1887],[1855,-6.0033],[1856,10.792],[1857,27.9647],[1859,62.5486],[1860,79.5401],[1861,96.0696],[1862,111.9507],[1863,126.9992],[1865,155.9743],[1866,170.1304],[1867,183.8765],[1868,197.0624],[1869,209.5379],[1870,221.1339],[1871,231.1469],[1872,239.1372],[1873,245.3629],[1874,250.0471],[1875,253.3849],[1876,255.5984],[1877,256.9163],[1885,260.4261],[1886,252.6494],[1887,237.7765],[1888,228.6602],[1897,208.5155],[1899,205.1607],[1900,203.9947],[1901,203.2506],[1902,202.9891],[1904,203.5823],[1907,206.0217],[1909,209.0954],[1913,217.3633],[1925,248.8618],[1928,255.0283],[1932,261.0135],[1933,261.6324],[1934,250.8733],[1935,236.2335],[1936,228.5684],[1937,221.5049],[1938,215.6669],[1939,211.7146],[1940,210.3056],[1941,211.2756],[1942,213.8753],[1943,217.8487],[1944,222.9393],[1945,228.8895],[1947,242.2888],[1948,253.397],[1949,266.9098],[1950,273.8815],[1951,273.558],[1952,271.5315],[1953,268.3612],[1954,264.6067],[1955,233.2601],[1956,197.0196],[1958,169.2779],[1959,155.9142],[1960,143.0777],[1961,112.9386],[1962,60.9846],[1963,11.2177],[1964,-11.7579],[1966,-13.8509],[1967,-14.876],[1968,-16.1811],[1969,-17.9903],[1970,-20.2011],[1971,-22.7111],[1972,-25.4182],[1973,-28.2205],[1974,-35.512],[1975,-49.33],[1976,-65.9996],[1977,-81.8946],[1978,-93.4847],[1979,-97.2821],[1980,-95.4653],[1981,-92.626],[1982,-88.8964],[1983,-84.444],[1984,-79.4361],[1985,-74.0402],[1988,-57.1982],[1989,-51.9241],[1990,-47.0242],[1991,-42.6028],[1992,-38.8439],[1993,-35.9319],[1994,-34.051],[1995,-44.9805],[1996,-67.3235],[1997,-81.5037],[1999,-88.5543],[2000,-91.8042],[2001,-94.691],[2002,-97.0797],[2003,-98.8355],[2004,-99.8234],[2005,-99.9089],[2006,-98.3719],[2007,-94.8827],[2008,-89.777],[2009,-83.3786],[2010,-76.1216],[2012,-60.7689],[2013,-53.5382],[2014,-47.1764],[2015,-42.1085],[2016,-38.7576],[2017,-37.5463],[2018,-88.3683],[2019,-93.7933],[2020,-77.2963],[2021,-53.7442],[2022,-40.1486],[2023,-29.6313],[2024,-25.3901],[2025,-25.3835],[2026,-25.5075],[2027,-25.7338],[2028,-26.036],[2030,-26.7987],[2036,-29.7995],[2037,-30.2308],[2038,-30.5963],[2039,-30.8787],[2040,-31.0609],[2041,-31.1254],[2072,-30.5859],[2074,-30.4384],[2076,-30.1246],[2079,-29.4206],[2088,-26.7629],[2090,-26.292],[2092,-25.9426],[2094,-25.7507],[2097,-25.725],[2098,-83.0865]]
       }
       Scale.X {type=MovieSceneScriptingDoubleChannel, default=1}
       Scale.Y {type=MovieSceneScriptingDoubleChannel, default=1}
       Scale.Z {type=MovieSceneScriptingDoubleChannel, default=1}
      ]
     }]
   }
  ]
 }
 Eye Laser Temp R {
  id=#4
  possessed_class=FortStaticMeshActor
  tracks [
   {type=MovieSceneSpawnTrack, sections [{type=MovieSceneSpawnSection, channels [None {type=MovieSceneScriptingBoolChannel, default=false}]}]}
   {
    type=MovieScene3DTransformTrack
    sections [{
      type=MovieScene3DTransformSection
      channels [
       Location.X {type=MovieSceneScriptingDoubleChannel, default=0}
       Location.Y {type=MovieSceneScriptingDoubleChannel, default=0}
       Location.Z {type=MovieSceneScriptingDoubleChannel, default=0}
       Rotation.X {type=MovieSceneScriptingDoubleChannel, default=90, const [90,2]}
       Rotation.Y {type=MovieSceneScriptingDoubleChannel, default=0, const [0,2]}
       Rotation.Z {type=MovieSceneScriptingDoubleChannel, default=0, keys [[1958,-90],[1995.5,0],[2330.5,0],[2335.5,-90]]}
       Scale.X {type=MovieSceneScriptingDoubleChannel, default=0.0357, const [0.0357,1]}
       Scale.Y {type=MovieSceneScriptingDoubleChannel, default=0.0357, const [0.0357,1]}
       Scale.Z {
        type=MovieSceneScriptingDoubleChannel
        default=2.56
        keys [[1955,5.2038],[1977,2.0912],[1995.5,3.0417],[2334.5,4],[2335.5,10.365],[2358,3.795],[2424.5,7.1925],[2444,0.8975],[2483,8.5175],[2508,1.94],[2544.5,9.4825],[2566.5,2.56]]
       }
      ]
     }]
   }
   {type=MovieScene3DAttachTrack, sections [{type=MovieScene3DAttachSection, end_frame=6088, start_frame=1918}]}
   {
    type=MovieSceneVisibilityTrack
    sections [{
      type=MovieSceneVisibilitySection
      channels [None {
        type=MovieSceneScriptingBoolChannel
        default=false
        keys [[1918,false],[1955,true],[2039.5,false],[2335.5,true],[2358.5,false],[2361,false],[2424.5,true],[2444.5,false],[2483,true],[2508.5,false],[2544.5,true],[2567,false]]
       }]
     }]
   }
  ]
 }
 CineCameraActor2 {
  id=#5
  possessed_class=CineCameraActor
  tracks [
   {
    type=MovieScene3DTransformTrack
    sections [{
      type=MovieScene3DTransformSection
      channels [
       Location.X {type=MovieSceneScriptingDoubleChannel, default=-53.8146}
       Location.Y {type=MovieSceneScriptingDoubleChannel, default=149.1452}
       Location.Z {type=MovieSceneScriptingDoubleChannel, default=102.6625}
       Rotation.X {type=MovieSceneScriptingDoubleChannel, default=41.0466}
       Rotation.Y {type=MovieSceneScriptingDoubleChannel, default=-1.2031}
       Rotation.Z {type=MovieSceneScriptingDoubleChannel, default=-84.563}
       Scale.X {type=MovieSceneScriptingDoubleChannel, default=0.25}
       Scale.Y {type=MovieSceneScriptingDoubleChannel, default=0.25}
       Scale.Z {type=MovieSceneScriptingDoubleChannel, default=0}
      ]
     }]
   }
   {type=MovieSceneSpawnTrack, sections [{type=MovieSceneSpawnSection, channels [None {type=MovieSceneScriptingBoolChannel, default=true}]}]}
   {type=MovieScene3DAttachTrack, sections [{type=MovieScene3DAttachSection, end_frame=2202, start_frame=1800}]}
  ]
 }
 CameraComponent {
  id=#6
  parent_id=#5
  possessed_class=CineCameraComponent
  tracks [
   {
    type=MovieSceneFloatTrack
    display_name=Current Focal Length
    sections [{type=MovieSceneFloatSection, channels [None {type=MovieSceneScriptingFloatChannel, default=35}]}]
   }
   {
    type=MovieSceneFloatTrack
    display_name="Manual Focus Distance (Focus Settings)"
    sections [{type=MovieSceneFloatSection, channels [None {type=MovieSceneScriptingFloatChannel, default=100000}]}]
   }
   {
    type=MovieSceneFloatTrack
    display_name=Current Aperture
    sections [{type=MovieSceneFloatSection, channels [None {type=MovieSceneScriptingFloatChannel, default=2.8}]}]
   }
  ]
 }
 Niagara {
  id=#7
  parent_id=#2
  possessed_class=NiagaraComponent
  tracks [{
    type=MovieSceneNiagaraFloatParameterTrack
    display_name=LifetimeScale
    sections [{type=MovieSceneFloatSection, channels [None {type=MovieSceneScriptingFloatChannel, default=1, const [20,1]}]}]
   }]
 }
 Eye Laser Temp L {
  id=#8
  possessed_class=FortStaticMeshActor
  tracks [
   {type=MovieSceneSpawnTrack, sections [{type=MovieSceneSpawnSection, channels [None {type=MovieSceneScriptingBoolChannel, default=false}]}]}
   {
    type=MovieScene3DTransformTrack
    sections [{
      type=MovieScene3DTransformSection
      channels [
       Location.X {type=MovieSceneScriptingDoubleChannel, default=0}
       Location.Y {type=MovieSceneScriptingDoubleChannel, default=0}
       Location.Z {type=MovieSceneScriptingDoubleChannel, default=0}
       Rotation.X {type=MovieSceneScriptingDoubleChannel, default=90, const [90,2]}
       Rotation.Y {type=MovieSceneScriptingDoubleChannel, default=0, const [0,2]}
       Rotation.Z {type=MovieSceneScriptingDoubleChannel, default=0, keys [[1958,-90],[1995.5,0],[2330.5,0],[2335.5,-90]]}
       Scale.X {type=MovieSceneScriptingDoubleChannel, default=0.0357}
       Scale.Y {type=MovieSceneScriptingDoubleChannel, default=0.0357}
       Scale.Z {
        type=MovieSceneScriptingDoubleChannel
        default=-361.2139
        keys [[1955,5.2038],[1977,2.0912],[1995.5,3.0417],[2334.5,4],[2335.5,10.365],[2358,3.795],[2424.5,7.1925],[2444,0.8975],[2483,8.5175],[2508,1.94],[2544.5,9.4825],[2566.5,2.56]]
       }
      ]
     }]
   }
   {type=MovieScene3DAttachTrack, sections [{type=MovieScene3DAttachSection, end_frame=6088, start_frame=1918}]}
   {
    type=MovieSceneVisibilityTrack
    sections [{
      type=MovieSceneVisibilitySection
      channels [None {
        type=MovieSceneScriptingBoolChannel
        default=true
        keys [[1918,false],[1955,true],[2039.5,false],[2335.5,true],[2358.5,false],[2361,true],[2424.5,true],[2444.5,false],[2483,true],[2508.5,false],[2544.5,true],[2567,false]]
       }]
     }]
   }
  ]
 }
 CAM_1_Position_CTRL {
  id=#9
  possessed_class=FortStaticMeshActor
  tracks [
   {type=MovieSceneSpawnTrack, sections [{type=MovieSceneSpawnSection, channels [None {type=MovieSceneScriptingBoolChannel, default=false}]}]}
   {
    type=MovieScene3DTransformTrack
    sections [{
      type=MovieScene3DTransformSection
      channels [
       Location.X {type=MovieSceneScriptingDoubleChannel, default=0}
       Location.Y {type=MovieSceneScriptingDoubleChannel, default=388.7891}
       Location.Z {type=MovieSceneScriptingDoubleChannel, default=0}
       Rotation.X {type=MovieSceneScriptingDoubleChannel, default=0}
       Rotation.Y {type=MovieSceneScriptingDoubleChannel, default=0}
       Rotation.Z {type=MovieSceneScriptingDoubleChannel, default=0}
       Scale.X {type=MovieSceneScriptingDoubleChannel, default=1}
       Scale.Y {type=MovieSceneScriptingDoubleChannel, default=1}
       Scale.Z {type=MovieSceneScriptingDoubleChannel, default=1}
      ]
     }]
   }
   {type=MovieScene3DAttachTrack, sections [{type=MovieScene3DAttachSection, end_frame=2202, start_frame=1800}]}
  ]
 }
]
display_rate {denominator=1, numerator=30}
tick_resolution {denominator=1, numerator=24000}
tracks [{type=MovieSceneCameraCutTrack, sections [{type=MovieSceneCameraCutSection, end_frame=2202, start_frame=1800}]}]
